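# Control Rig for the F_MED_SkullArcana female cosmetic body that procedurally poses its skirt and shoulder cloth bones to follow the limbs.
# A delta-rotation function uses swing/twist decomposition to turn thigh, calf and upperarm angles into driver weights, alongside a pelvis-height factor.
# These weights and interpolated state flags for crouch, jump, skydive, parachute and emotes blend corrective bone offsets.
# Parent constraints spread base cloth bones between neighboring body bones.
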ControlRigBlueprint /BRCosmetics/Characters/Player/Female/Medium/Bodies/F_MED_SkullArcana/Meshes/F_MED_SkullArcana_CtrlRig
graphs [
 New Function_ContainedGraph {nodes [
   n1 = RigVMFunctionEntryNode()
   RigVMFunction_MathBoolToInteger = &RigVMFunction_MathBoolToInteger(Value=@n1.onlyTwist)
   RigVMFunction_MathQuaternionSwingTwist = &RigVMFunction_MathQuaternionSwingTwist(
    TwistAxis=@n1.twistAxis
    Input=&RigUnit_GetRelativeTransformForItem(bChildInitial=False, bParentInitial=False, Child=@n1.child, Parent=@n1.parent).RelativeTransform.Rotation
   )
   n2 = RigVMDispatchNode(
    Values=((),())
    Index=@RigVMFunction_MathBoolToInteger.Result
    Values.0=@RigVMFunction_MathQuaternionSwingTwist.Swing
    Values.1=@RigVMFunction_MathQuaternionSwingTwist.Twist
   ) {template="DISPATCH_RigVMDispatch_SelectInt32(in Index,in Values,out Result)"}
   RigVMFunction_MathQuaternionToEuler = &RigVMFunction_MathQuaternionToEuler(Value=@n2.Result, RotationOrder=@n1.Rotation Order)
   > DISPATCH_RigVMDispatch_Print = RigVMDispatchNode(ScreenColor=(R=1,G=1,B=1,A=1), ScreenDuration=1.5, Enabled=@n1.debugRotation, Value=@RigVMFunction_MathQuaternionToEuler.Result) {template="DISPATCH_RigVMDispatch_Print(in Prefix,in Value,in Enabled,in ScreenDuration,in ScreenColor)"}
   Cast = RigVMDispatchNode(Value=@n1.triggerAxis) {template="DISPATCH_RigVMDispatch_CastEnumToInt(in Value,out Result)"}
   DISPATCH_RigVMDispatch_SelectInt32 = RigVMDispatchNode(
    Values=(0,0,0)
    Index=@Cast.Result
    Values.0=@RigVMFunction_MathQuaternionToEuler.Result.X
    Values.1=@RigVMFunction_MathQuaternionToEuler.Result.Y
    Values.2=@RigVMFunction_MathQuaternionToEuler.Result.Z
   ) {template="DISPATCH_RigVMDispatch_SelectInt32(in Index,in Values,out Result)"}
   Remap = &RigVMFunction_MathDoubleRemap(
    bClamp=False
    Value=@DISPATCH_RigVMDispatch_SelectInt32.Result
    TargetMinimum=@n1.outputMinMax.X
    TargetMaximum=@n1.outputMinMax.Y
    SourceMinimum=@n1.rotationMinMax.X
    SourceMaximum=@n1.rotationMinMax.Y
   )
   Absolute = &RigVMFunction_MathDoubleAbs(Value=@Remap.Result)
   Multiply = &RigVMFunction_MathDoubleMul(
    A=&RigVMFunction_MathDoubleClamp(Value=RigVMDispatchNode(True=@Absolute.Result, Condition=@n1.absoluteValue, False=@Remap.Result) {template="DISPATCH_RigVMDispatch_If(in Condition,in True,in False,out Result)"}.Result).Result
    B=@n1.premult
   )
   > RigVMFunctionReturnNode(outputQuat=@n2.Result, outputWeight=@Multiply.Result)
  ]}
 New Function_ContainedGraph {nodes [
   RigVMFunctionEntryNode()
   > RigVMFunction_Sequence = RigVMAggregateNode()
   n1 = &RigVMFunction_AlphaInterp(InterpSpeedDecreasing=10, InterpSpeedIncreasing=20, Scale=1, bClampResult=False, Value=RigVMVariableNode() {variable=Is Tacticaljump}.Value)
   &RigUnit_ModifyTransforms(
    ItemToModify="((Item=(Type=Bone,Name=\"dyn_R_SideSkirt_2\"),Transform=(Rotation=(X=0.0427,Y=-0.1223,Z=-0.0951,W=0.987),Translation=(X=0,Y=0,Z=0),Scale3D=(X=1,Y=1,Z=1))),(Item=(Type=Bone,Name=\"dyn_R_BackSkirt_1\"),Transform=(Rotation=(X=-0.0676,Y=0.0914,Z=0.178,W=0.9774),Tr ...(truncated)"
    WeightMinimum=0
    Weight=@n1.Result
    ExecutePin=@RigVMFunction_Sequence.A
   )
   AlphaInterp = &RigVMFunction_AlphaInterp(
    InterpSpeedDecreasing=10
    InterpSpeedIncreasing=10
    Scale=1
    bClampResult=false
    bMapRange=false
    Value=RigVMVariableNode() {variable=IsJump}.Value
   )
   &RigUnit_ModifyTransforms(
    ItemToModify="((Item=(Type=Bone,Name=\"dyn_C_BackSkirt_2\"),Transform=(Rotation=(X=-0.1631,Y=0.0811,Z=-0.3264,W=0.9275),Translation=(X=0,Y=0,Z=0),Scale3D=(X=1,Y=1,Z=1))),(Item=(Type=Bone,Name=\"dyn_R_SideSkirt_2\"),Transform=(Rotation=(X=-0.0883,Y=-0.1398,Z=0.0852,W=0.9826) ...(truncated)"
    WeightMinimum=0
    Weight=@AlphaInterp.Result
    ExecutePin=@RigVMFunction_Sequence.B
   )
   RigVMFunctionReturnNode(ExecuteContext=@RigVMFunction_Sequence.C)
  ]}
 New Function_ContainedGraph {nodes [
   n1 = RigVMFunctionEntryNode()
   RigVMFunction_Sequence = &RigVMFunction_Sequence(ExecuteContext=@n1.ExecuteContext)
   DeltaRotation_1 = RigVMFunctionReferenceNode(
    Rotation Order=ZYX
    child=(Type=Bone,Name="thigh_r")
    outputMinMax=(X=0,Y=1)
    parent=(Type=Null,Name="thigh_r_null")
    premult=1
    rotationMinMax=(X=0,Y=-120)
    triggerAxis=Z
    twistAxis=(X=1,Y=0,Z=0)
    ExecuteContext=@RigVMFunction_Sequence.A
   )
   > RigVMVariableNode(Value=@DeltaRotation_1.outputWeight) {variable=SPR R Leg Fwd}
   > DeltaRotation_2 = RigVMFunctionReferenceNode(
    Rotation Order=ZYX
    child=(Type=Bone,Name="thigh_l")
    outputMinMax=(X=0,Y=1)
    parent=(Type=Null,Name="thigh_l_null")
    premult=1
    rotationMinMax=(X=0,Y=-120)
    triggerAxis=Z
    twistAxis=(X=1,Y=0,Z=0)
   )
   > RigVMVariableNode(Value=@DeltaRotation_2.outputWeight) {variable=SPR L Leg Fwd}
   > DeltaRotation_3 = RigVMFunctionReferenceNode(
    Rotation Order=ZYX
    child=(Type=Bone,Name="thigh_r")
    outputMinMax=(X=0,Y=1)
    parent=(Type=Null,Name="thigh_r_null")
    premult=1
    rotationMinMax=(X=0,Y=100)
    triggerAxis=Z
    twistAxis=(X=1,Y=0,Z=0)
   )
   > RigVMVariableNode(Value=@DeltaRotation_3.outputWeight) {variable=SPR R Leg Bkd}
   > DeltaRotation_8 = RigVMFunctionReferenceNode(
    Rotation Order=ZYX
    child=(Type=Bone,Name="thigh_l")
    outputMinMax=(X=0,Y=1)
    parent=(Type=Null,Name="thigh_l_null")
    premult=1
    rotationMinMax=(X=0,Y=100)
    triggerAxis=Z
    twistAxis=(X=1,Y=0,Z=0)
   )
   > RigVMVariableNode(Value=@DeltaRotation_8.outputWeight) {variable=SPR L Leg Bkd}
   > DeltaRotation_9 = RigVMFunctionReferenceNode(
    Rotation Order=ZYX
    child=(Type=Bone,Name="thigh_r")
    outputMinMax=(X=0,Y=1)
    parent=(Type=Null,Name="thigh_r_null")
    premult=1
    rotationMinMax=(X=0,Y=-90)
    triggerAxis=Y
    twistAxis=(X=1,Y=0,Z=0)
   )
   > RigVMVariableNode(Value=@DeltaRotation_9.outputWeight) {variable=SPR R Leg Otw}
   > DeltaRotation_10 = RigVMFunctionReferenceNode(
    Rotation Order=ZYX
    child=(Type=Bone,Name="thigh_l")
    outputMinMax=(X=0,Y=1)
    parent=(Type=Null,Name="thigh_l_null")
    premult=1
    rotationMinMax=(X=0,Y=-90)
    triggerAxis=Y
    twistAxis=(X=1,Y=0,Z=0)
   )
   > RigVMVariableNode(Value=@DeltaRotation_10.outputWeight) {variable=SPR L Leg Otw}
   > DeltaRotation_11 = RigVMFunctionReferenceNode(
    Rotation Order=ZYX
    child=(Type=Bone,Name="thigh_r")
    outputMinMax=(X=0,Y=1)
    parent=(Type=Null,Name="thigh_r_null")
    premult=1
    rotationMinMax=(X=0,Y=50)
    triggerAxis=Y
    twistAxis=(X=1,Y=0,Z=0)
   )
   > RigVMVariableNode(Value=@DeltaRotation_11.outputWeight) {variable=SPR R Leg Itw}
   > DeltaRotation_18 = RigVMFunctionReferenceNode(
    Rotation Order=ZYX
    child=(Type=Bone,Name="thigh_l")
    outputMinMax=(X=0,Y=1)
    parent=(Type=Null,Name="thigh_l_null")
    premult=1
    rotationMinMax=(X=0,Y=50)
    triggerAxis=Y
    twistAxis=(X=1,Y=0,Z=0)
   )
   > RigVMVariableNode(Value=@DeltaRotation_18.outputWeight) {variable=SPR L Leg Itw}
   > DeltaRotation_19 = RigVMFunctionReferenceNode(
    Rotation Order=ZYX
    child=(Type=Bone,Name="calf_r")
    outputMinMax=(X=0,Y=1)
    parent=(Type=Null,Name="calf_r_null")
    premult=1
    rotationMinMax=(X=15,Y=100)
    triggerAxis=Z
    twistAxis=(X=1,Y=0,Z=0)
   )
   > RigVMVariableNode(Value=@DeltaRotation_19.outputWeight) {variable=SPR R calf Bkd}
   > DeltaRotation_20 = RigVMFunctionReferenceNode(
    Rotation Order=ZYX
    child=(Type=Bone,Name="calf_l")
    outputMinMax=(X=0,Y=1)
    parent=(Type=Null,Name="calf_l_null")
    premult=1
    rotationMinMax=(X=15,Y=100)
    triggerAxis=Z
    twistAxis=(X=1,Y=0,Z=0)
   )
   > RigVMVariableNode(Value=@DeltaRotation_20.outputWeight) {variable=SPR L calf Bkd}
   > DeltaRotation_21 = RigVMFunctionReferenceNode(
    Rotation Order=ZYX
    child=(Type=Bone,Name="calf_r")
    outputMinMax=(X=0,Y=1)
    parent=(Type=Null,Name="calf_r_null")
    premult=1
    rotationMinMax=(X=120,Y=140)
    triggerAxis=Z
    twistAxis=(X=1,Y=0,Z=0)
   )
   > RigVMVariableNode(Value=@DeltaRotation_21.outputWeight) {variable=SPR R calf Max Bkd}
   > DeltaRotation_22 = RigVMFunctionReferenceNode(
    Rotation Order=ZYX
    child=(Type=Bone,Name="calf_l")
    outputMinMax=(X=0,Y=1)
    parent=(Type=Null,Name="calf_l_null")
    premult=1
    rotationMinMax=(X=0,Y=130)
    triggerAxis=Z
    twistAxis=(X=1,Y=0,Z=0)
   )
   > RigVMVariableNode(Value=@DeltaRotation_22.outputWeight) {variable=SPR L calf Max Bkd}
   n2 = &RigVMFunction_AnimEvalRichCurve(
    Curve="(EditorCurveData=(Keys=((InterpMode=RCIM_Cubic,TangentMode=RCTM_Break,TangentWeightMode=RCTWM_WeightedNone,Time=0,Value=0,ArriveTangent=0,ArriveTangentWeight=0,LeaveTangent=0,LeaveTangentWeight=0),(InterpMode=RCIM_Cubic,TangentMode=RCTM_Break,TangentWeight ...(truncated)"
    SourceMinimum=0.5
    TargetMinimum=0
    Value=&RigVMFunction_MathDoubleRemap(
     SourceMaximum=30
     SourceMinimum=60
     TargetMaximum=0
     TargetMinimum=1
     bClamp=False
     Value=&RigUnit_GetRelativeTransformForItem(Child=(Type=Bone,Name="pelvis"), Parent=(Type=Bone,Name="root"), bChildInitial=False, bParentInitial=False).RelativeTransform.Translation.Z
    ).Result
   )
   > RigVMVariableNode(Value=@n2.Result) {variable=PelvisProx}
   > DeltaRotation_5 = RigVMFunctionReferenceNode(
    Rotation Order=ZYX
    child=(Type=Bone,Name="upperarm_l")
    outputMinMax=(X=0,Y=1)
    parent=(Type=Null,Name="upperarm_l_null")
    premult=1
    rotationMinMax=(X=10,Y=70)
    triggerAxis=Z
    twistAxis=(X=1,Y=0,Z=0)
   )
   > RigVMVariableNode(Value=@DeltaRotation_5.outputWeight) {variable=Upperarm_L_FW}
   > DeltaRotation_6 = RigVMFunctionReferenceNode(
    Rotation Order=ZYX
    child=(Type=Bone,Name="upperarm_r")
    outputMinMax=(X=0,Y=1)
    parent=(Type=Null,Name="upperarm_r_null")
    premult=1
    rotationMinMax=(X=10,Y=70)
    triggerAxis=Z
    twistAxis=(X=1,Y=0,Z=0)
   )
   > RigVMVariableNode(Value=@DeltaRotation_6.outputWeight) {variable=Upperarm_R_FW}
   > DeltaRotation_7 = RigVMFunctionReferenceNode(
    Rotation Order=ZYX
    child=(Type=Bone,Name="upperarm_l")
    outputMinMax=(X=0,Y=1)
    parent=(Type=Null,Name="upperarm_l_null")
    premult=1
    rotationMinMax=(X=10,Y=60)
    triggerAxis=Y
    twistAxis=(X=1,Y=0,Z=0)
   )
   > RigVMVariableNode(Value=@DeltaRotation_7.outputWeight) {variable=Upperarm_L_Dwn}
   > DeltaRotation_12 = RigVMFunctionReferenceNode(
    Rotation Order=ZYX
    child=(Type=Bone,Name="upperarm_r")
    outputMinMax=(X=0,Y=1)
    parent=(Type=Null,Name="upperarm_r_null")
    premult=1
    rotationMinMax=(X=10,Y=60)
    triggerAxis=Y
    twistAxis=(X=1,Y=0,Z=0)
   )
   > RigVMVariableNode(Value=@DeltaRotation_12.outputWeight) {variable=Upperarm_R_Dwn}
   RigVMFunctionReturnNode(ExecuteContext=@RigVMFunction_Sequence.B)
  ]}
 New Function_ContainedGraph {nodes [
   RigVMFunctionEntryNode()
   > RigVMFunction_Sequence = RigVMAggregateNode()
   &RigUnit_ModifyTransforms(
    ItemToModify=((Item=(Type=Bone,Name="thigh_r"),Transform=(Rotation=(X=0,Y=0,Z=0,W=1),Translation=(X=0,Y=0,Z=0),Scale3D=(X=1,Y=1,Z=1))),(Item=(Type=Bone,Name="thigh_l"),Transform=(Rotation=(X=0,Y=0,Z=0,W=1),Translation=(X=0,Y=0,Z=0),Scale3D=(X=1,Y=1,Z=1))))
    Weight=1
    WeightMinimum=0
    ExecutePin=@RigVMFunction_Sequence.A
   )
   > &RigUnit_ParentConstraint(
    AdvancedSettings=(InterpolationType=Average,RotationOrderForFilter=XZY)
    Child=(Type=Bone,Name="base_R_FrontSkirt")
    Filter=(TranslationFilter=(bX=false,bY=false,bZ=false),RotationFilter=(bX=True,bY=True,bZ=True),ScaleFilter=(bX=false,bY=false,bZ=false))
    Parents=((Item=(Type=Bone,Name="thigh_r"),Weight=0.8),(Item=(Type=Bone,Name="pelvis"),Weight=0.2))
    Weight=1
   )
   > &RigUnit_ParentConstraint(
    AdvancedSettings=(InterpolationType=Average,RotationOrderForFilter=XZY)
    Child=(Type=Bone,Name="base_R_SideSkirt")
    Filter=(TranslationFilter=(bX=false,bY=false,bZ=false),RotationFilter=(bX=True,bY=True,bZ=True),ScaleFilter=(bX=false,bY=false,bZ=false))
    Parents=((Item=(Type=Bone,Name="thigh_r"),Weight=0.5),(Item=(Type=Bone,Name="pelvis"),Weight=0.5))
    Weight=1
   )
   > &RigUnit_ParentConstraint(
    AdvancedSettings=(InterpolationType=Average,RotationOrderForFilter=XZY)
    Child=(Type=Bone,Name="base_R_BackSkirt")
    Filter=(TranslationFilter=(bX=false,bY=false,bZ=false),RotationFilter=(bX=True,bY=True,bZ=True),ScaleFilter=(bX=false,bY=false,bZ=false))
    Parents=((Item=(Type=Bone,Name="thigh_r"),Weight=0.1),(Item=(Type=Bone,Name="pelvis"),Weight=0.9))
    Weight=1
   )
   > &RigUnit_ParentConstraint(
    AdvancedSettings=(InterpolationType=Average,RotationOrderForFilter=XZY)
    Child=(Type=Bone,Name="base_L_FrontSkirt")
    Filter=(TranslationFilter=(bX=false,bY=false,bZ=false),RotationFilter=(bX=True,bY=True,bZ=True),ScaleFilter=(bX=false,bY=false,bZ=false))
    Parents=((Item=(Type=Bone,Name="thigh_l"),Weight=0.8),(Item=(Type=Bone,Name="pelvis"),Weight=0.2))
    Weight=1
   )
   > &RigUnit_ParentConstraint(
    AdvancedSettings=(InterpolationType=Average,RotationOrderForFilter=XZY)
    Child=(Type=Bone,Name="base_L_SideSkirt")
    Filter=(TranslationFilter=(bX=false,bY=false,bZ=false),RotationFilter=(bX=True,bY=True,bZ=True),ScaleFilter=(bX=false,bY=false,bZ=false))
    Parents=((Item=(Type=Bone,Name="thigh_l"),Weight=0.5),(Item=(Type=Bone,Name="pelvis"),Weight=0.5))
    Weight=1
   )
   > &RigUnit_ParentConstraint(
    AdvancedSettings=(InterpolationType=Average,RotationOrderForFilter=XZY)
    Child=(Type=Bone,Name="base_L_BackSkirt")
    Filter=(TranslationFilter=(bX=false,bY=false,bZ=false),RotationFilter=(bX=True,bY=True,bZ=True),ScaleFilter=(bX=false,bY=false,bZ=false))
    Parents=((Item=(Type=Bone,Name="thigh_l"),Weight=0.2),(Item=(Type=Bone,Name="pelvis"),Weight=0.8))
    Weight=1
   )
   > &RigUnit_ParentConstraint(
    AdvancedSettings=(InterpolationType=Average,RotationOrderForFilter=XZY)
    Child=(Type=Bone,Name="base_C_FrontSkirt")
    Filter=(TranslationFilter=(bX=false,bY=false,bZ=false),RotationFilter=(bX=True,bY=True,bZ=True),ScaleFilter=(bX=false,bY=false,bZ=false))
    Parents=((Item=(Type=Bone,Name="base_R_FrontSkirt"),Weight=0.5),(Item=(Type=Bone,Name="base_L_FrontSkirt"),Weight=0.5))
    Weight=1
   )
   &RigUnit_ParentConstraint(
    AdvancedSettings=(InterpolationType=Average,RotationOrderForFilter=XZY)
    Child=(Type=Bone,Name="base_R_BackSkirt_Charm")
    Parents=((Item=(Type=Bone,Name="dyn_R_BackSkirt_4"),Weight=0.6),(Item=(Type=Bone,Name="dyn_C_BackSkirt_4"),Weight=0.2),(Item=(Type=Bone,Name="dyn_R_SideSkirt_4"),Weight=0.2))
    Weight=0
    ExecutePin=@RigVMFunction_Sequence.B
   )
   RigVMFunctionReturnNode(ExecuteContext=@RigVMFunction_Sequence.C)
   &RigUnit_ParentConstraint(
    AdvancedSettings=(InterpolationType=Average,RotationOrderForFilter=XZY)
    Child=(Type=Bone,Name="base_C_BackSkirt")
    ExecutePin=()
    Filter=(TranslationFilter=(bX=false,bY=false,bZ=false),RotationFilter=(bX=True,bY=True,bZ=True),ScaleFilter=(bX=false,bY=false,bZ=false))
    Parents=((Item=(Type=Bone,Name="base_R_BackSkirt"),Weight=0.5),(Item=(Type=Bone,Name="base_L_BackSkirt"),Weight=0.5))
    Weight=1
   )
  ]}
 New Function_ContainedGraph {nodes [
   RigVMFunctionEntryNode()
   > RigVMFunction_Sequence = RigVMAggregateNode()
   n1 = RigVMVariableNode() {variable=SPR L Leg Fwd}
   &RigUnit_ModifyTransforms(
    ItemToModify="((Item=(Type=Bone,Name=\"dyn_L_FrontSkirt_1\"),Transform=(Rotation=(X=0.051,Y=0.2422,Z=0.0541,W=-0.9674),Translation=(X=-5.0016,Y=1.0812,Z=5.8115),Scale3D=(X=1,Y=1,Z=1))),(Item=(Type=Bone,Name=\"dyn_C_FrontSkirt_1\"),Transform=(Rotation=(X=0,Y=0.497,Z=0,W=-0.8 ...(truncated)"
    WeightMinimum=0
    ExecutePin=@RigVMFunction_Sequence.A
    Weight=@n1.Value
   )
   n2 = RigVMVariableNode() {variable=SPR R Leg Fwd}
   > &RigUnit_ModifyTransforms(
    ItemToModify="((Item=(Type=Bone,Name=\"dyn_C_FrontSkirt_1\"),Transform=(Rotation=(X=0,Y=-0.4032,Z=0,W=0.9151),Translation=(X=0,Y=0,Z=0),Scale3D=(X=1,Y=1,Z=1))),(Item=(Type=Bone,Name=\"dyn_R_FrontSkirt_1\"),Transform=(Rotation=(X=0,Y=-0.1134,Z=0,W=0.9935),Translation=(X=-1.9 ...(truncated)"
    WeightMinimum=0
    Weight=@n2.Value
   )
   > &RigUnit_ModifyTransforms(
    ItemToModify="((Item=(Type=Bone,Name=\"dyn_R_FrontSkirt_1\"),Transform=(Rotation=(X=-0.019,Y=0.1862,Z=0.0558,W=0.9808),Translation=(X=0,Y=0,Z=0),Scale3D=(X=1,Y=1,Z=1))),(Item=(Type=Bone,Name=\"dyn_L_FrontSkirt_1\"),Transform=(Rotation=(X=0,Y=-0.1028,Z=0,W=-0.9947),Translati ...(truncated)"
    WeightMinimum=0
    Weight=&RigVMFunction_MathDoubleMul(A=@n2.Value, B=@n1.Value).Result
   )
   n3 = RigVMVariableNode() {variable=SPR L Leg Bkd}
   > &RigUnit_ModifyTransforms(
    ItemToModify="((Item=(Type=Bone,Name=\"dyn_C_FrontSkirt_1\"),Transform=(Rotation=(X=0,Y=-0.3961,Z=0,W=0.9182),Translation=(X=0,Y=0,Z=0),Scale3D=(X=1,Y=1,Z=1))),(Item=(Type=Bone,Name=\"dyn_L_FrontSkirt_1\"),Transform=(Rotation=(X=0.0972,Y=0.6912,Z=-0.0997,W=-0.7091),Translat ...(truncated)"
    WeightMinimum=0
    Weight=@n3.Value
   )
   n4 = RigVMVariableNode() {variable=SPR R Leg Bkd}
   > &RigUnit_ModifyTransforms(
    ItemToModify="((Item=(Type=Bone,Name=\"dyn_R_SideSkirt_1\"),Transform=(Rotation=(X=0.0567,Y=0.175,Z=-0.3028,W=-0.9351),Translation=(X=0,Y=0,Z=0),Scale3D=(X=1,Y=1,Z=1))),(Item=(Type=Bone,Name=\"dyn_C_FrontSkirt_1\"),Transform=(Rotation=(X=0,Y=-0.4926,Z=0,W=0.8703),Translatio ...(truncated)"
    WeightMinimum=0
    Weight=@n4.Value
   )
   Multiply = &RigVMFunction_MathDoubleMul(B=@n3.Value, A=@n4.Value)
   > &RigUnit_ModifyTransforms(
    ItemToModify="((Item=(Type=Bone,Name=\"None\"),Transform=(Rotation=(X=0,Y=0,Z=0,W=1),Translation=(X=0,Y=0,Z=0),Scale3D=(X=1,Y=1,Z=1))),(Item=(Type=Bone,Name=\"dyn_C_BackSkirt_1\"),Transform=(Rotation=(X=0,Y=-0.6981,Z=0,W=-0.716),Translation=(X=0,Y=0,Z=0),Scale3D=(X=1,Y=1,Z= ...(truncated)"
    WeightMinimum=0
    Weight=@Multiply.Result
   )
   > &RigUnit_ModifyTransforms(
    ItemToModify="((Item=(Type=Bone,Name=\"dyn_C_FrontSkirt_1\"),Transform=(Rotation=(X=0,Y=0,Z=0,W=1),Translation=(X=0,Y=0,Z=0),Scale3D=(X=1,Y=1,Z=1))),(Item=(Type=Bone,Name=\"dyn_L_FrontSkirt_1\"),Transform=(Rotation=(X=0,Y=0,Z=-0.0182,W=-0.9998),Translation=(X=0,Y=0,Z=0),Sca ...(truncated)"
    WeightMinimum=0
    Weight=RigVMVariableNode() {variable=SPR L Leg Otw}.Value
   )
   > &RigUnit_ModifyTransforms(
    ItemToModify="((Item=(Type=Bone,Name=\"dyn_R_SideSkirt_1\"),Transform=(Rotation=(X=0,Y=-0.3947,Z=0,W=0.9188),Translation=(X=0,Y=0,Z=0),Scale3D=(X=1,Y=1,Z=1))),(Item=(Type=Bone,Name=\"dyn_C_FrontSkirt_1\"),Transform=(Rotation=(X=0,Y=0,Z=0,W=1),Translation=(X=0,Y=0,Z=0),Scale ...(truncated)"
    WeightMinimum=0
    Weight=RigVMVariableNode() {variable=SPR R Leg Otw}.Value
   )
   > &RigUnit_ModifyTransforms(
    ItemToModify="((Item=(Type=Bone,Name=\"dyn_R_SideSkirt_1\"),Transform=(Rotation=(X=0.0001,Y=0.206,Z=0.0001,W=-0.9785),Translation=(X=0,Y=0,Z=0),Scale3D=(X=1,Y=1,Z=1))),(Item=(Type=Bone,Name=\"dyn_R_FrontSkirt_1\"),Transform=(Rotation=(X=0.0553,Y=-0.1662,Z=0.3108,W=0.9342),T ...(truncated)"
    WeightMinimum=0
    Weight=RigVMVariableNode() {variable=SPR R Leg Itw}.Value
   )
   > &RigUnit_ModifyTransforms(
    ItemToModify="((Item=(Type=Bone,Name=\"dyn_C_FrontSkirt_1\"),Transform=(Rotation=(X=0,Y=0,Z=-0.196,W=0.9806),Translation=(X=0,Y=0,Z=0),Scale3D=(X=1,Y=1,Z=1))),(Item=(Type=Bone,Name=\"dyn_L_FrontSkirt_1\"),Transform=(Rotation=(X=0.0477,Y=0.1091,Z=0.3979,W=-0.9097),Translatio ...(truncated)"
    WeightMinimum=0
    Weight=RigVMVariableNode() {variable=SPR L Leg Itw}.Value
   )
   RigVMFunction_MathBoolOr = &RigVMFunction_MathBoolOr(A=RigVMVariableNode() {variable=Emote}.Value, B=RigVMVariableNode() {variable=SkydiveBool}.Value)
   AlphaInterp = &RigVMFunction_AlphaInterp(
    InterpSpeedDecreasing=5
    InterpSpeedIncreasing=10
    Scale=1
    bClampResult=false
    Value=&RigVMFunction_MathDoubleMul(
     A=RigVMDispatchNode(False=1, True=0, Condition=@RigVMFunction_MathBoolOr.Result) {template="DISPATCH_RigVMDispatch_If(in Condition,in True,in False,out Result)"}.Result
     B=RigVMVariableNode() {variable=SPR L calf Bkd}.Value
    ).Result
   )
   > &RigUnit_ModifyTransforms(
    ItemToModify="((Item=(Type=Bone,Name=\"None\"),Transform=(Rotation=(X=0,Y=0,Z=0,W=1),Translation=(X=0,Y=0,Z=0),Scale3D=(X=1,Y=1,Z=1))),(Item=(Type=Bone,Name=\"dyn_L_BackSkirt_2\"),Transform=(Rotation=(X=0.355,Y=-0.5002,Z=-0.0723,W=0.7865),Translation=(X=0,Y=0,Z=0),Scale3D=( ...(truncated)"
    WeightMinimum=0
    Weight=@AlphaInterp.Result
   )
   n5 = &RigVMFunction_MathBoolOr(A=RigVMVariableNode() {variable=Emote}.Value, B=RigVMVariableNode() {variable=SkydiveBool}.Value)
   n6 = &RigVMFunction_AlphaInterp(
    ClampMax=0.8
    InterpSpeedDecreasing=5
    InterpSpeedIncreasing=10
    Scale=1
    bClampResult=false
    Value=&RigVMFunction_MathDoubleMul(
     B=RigVMDispatchNode(False=1, True=0, Condition=@n5.Result) {template="DISPATCH_RigVMDispatch_If(in Condition,in True,in False,out Result)"}.Result
     A=RigVMVariableNode() {variable=SPR R calf Bkd}.Value
    ).Result
   )
   > &RigUnit_ModifyTransforms(
    ItemToModify="((Item=(Type=Bone,Name=\"None\"),Transform=(Rotation=(X=0,Y=0,Z=0,W=1),Translation=(X=0,Y=0,Z=0),Scale3D=(X=1,Y=1,Z=1))),(Item=(Type=Bone,Name=\"dyn_R_BackSkirt_2\"),Transform=(Rotation=(X=-0.0208,Y=-0.6224,Z=0.0893,W=0.7773),Translation=(X=0,Y=0,Z=0),Scale3D= ...(truncated)"
    WeightMinimum=0
    Weight=@n6.Result
   )
   > &RigUnit_ModifyTransforms(
    ItemToModify="((Item=(Type=Bone,Name=\"dyn_R_BackSkirt_3\"),Transform=(Rotation=(X=0,Y=-0.3365,Z=0,W=-0.9417),Translation=(X=0,Y=0,Z=0),Scale3D=(X=1,Y=1,Z=1))),(Item=(Type=Bone,Name=\"dyn_C_BackSkirt_2\"),Transform=(Rotation=(X=0,Y=0.7622,Z=0,W=0.6473),Translation=(X=0,Y=0, ...(truncated)"
    WeightMinimum=0
    Weight=&RigVMFunction_MathDoubleMul(B=@AlphaInterp.Result, A=@n6.Result).Result
   )
   n7 = RigVMVariableNode() {variable=SkydiveBool}
   n8 = &RigVMFunction_AlphaInterp(
    ClampMax=0.4
    InterpSpeedDecreasing=5
    InterpSpeedIncreasing=10
    Scale=1
    bClampResult=true
    Value=&RigVMFunction_MathDoubleMul(
     A=RigVMDispatchNode(False=0, True=1, Condition=@n7.Value) {template="DISPATCH_RigVMDispatch_If(in Condition,in True,in False,out Result)"}.Result
     B=RigVMVariableNode() {variable=SPR L calf Bkd}.Value
    ).Result
   )
   > &RigUnit_ModifyTransforms(
    ItemToModify="((Item=(Type=Bone,Name=\"dyn_C_BackSkirt_3\"),Transform=(Rotation=(X=0,Y=-0.342,Z=0,W=0.9397),Translation=(X=0,Y=0,Z=0),Scale3D=(X=1,Y=1,Z=1))),(Item=(Type=Bone,Name=\"dyn_C_BackSkirt_2\"),Transform=(Rotation=(X=0,Y=0,Z=0,W=1),Translation=(X=0,Y=0,Z=0),Scale3D ...(truncated)"
    WeightMinimum=0
    Weight=@n8.Result
   )
   n9 = RigVMVariableNode() {variable=SkydiveBool}
   n10 = &RigVMFunction_AlphaInterp(
    ClampMax=0.4
    InterpSpeedDecreasing=5
    InterpSpeedIncreasing=10
    Scale=1
    bClampResult=true
    Value=&RigVMFunction_MathDoubleMul(
     B=RigVMDispatchNode(False=0, True=1, Condition=@n9.Value) {template="DISPATCH_RigVMDispatch_If(in Condition,in True,in False,out Result)"}.Result
     A=RigVMVariableNode() {variable=SPR R calf Bkd}.Value
    ).Result
   )
   > &RigUnit_ModifyTransforms(
    ItemToModify="((Item=(Type=Bone,Name=\"dyn_C_BackSkirt_3\"),Transform=(Rotation=(X=0,Y=0.6943,Z=0,W=-0.7197),Translation=(X=0,Y=0,Z=0),Scale3D=(X=1,Y=1,Z=1))),(Item=(Type=Bone,Name=\"dyn_R_BackSkirt_3\"),Transform=(Rotation=(X=-0.0349,Y=-0.7639,Z=0.0294,W=0.6437),Translatio ...(truncated)"
    WeightMinimum=0
    Weight=@n10.Result
   )
   > &RigUnit_ModifyTransforms(
    ItemToModify="((Item=(Type=Bone,Name=\"dyn_R_BackSkirt_3\"),Transform=(Rotation=(X=0,Y=-0.3839,Z=0,W=-0.9234),Translation=(X=0,Y=0,Z=0),Scale3D=(X=1,Y=1,Z=1))),(Item=(Type=Bone,Name=\"dyn_C_BackSkirt_2\"),Transform=(Rotation=(X=0,Y=-0.3114,Z=0,W=-0.9503),Translation=(X=0,Y= ...(truncated)"
    WeightMinimum=0
    Weight=&RigVMFunction_MathDoubleMul(B=@n8.Result, A=@n10.Result).Result
   )
   n11 = &RigVMFunction_AlphaInterp(
    InterpSpeedDecreasing=10
    InterpSpeedIncreasing=10
    Value=0
    bClampResult=False
    Scale=RigVMDispatchNode(False=0, True=1, Condition=RigVMVariableNode() {variable=idle}.Value) {template="DISPATCH_RigVMDispatch_If(in Condition,in True,in False,out Result)"}.Result
   )
   &RigUnit_ModifyTransforms(
    ItemToModify=((Item=(Type=Bone,Name="None"),Transform=(Rotation=(X=0,Y=0,Z=0,W=1),Translation=(X=0,Y=0,Z=0),Scale3D=(X=1,Y=1,Z=1))))
    WeightMinimum=0
    Weight=@n11.Result
    ExecutePin=@RigVMFunction_Sequence.B
   )
   > &RigUnit_ParentConstraint(
    AdvancedSettings=(RotationOrderForFilter=XZY)
    Child=(Type=Bone,Name="base_shoulderCloth_r")
    Parents="((Item=(Type=Bone,Name=\"deform_upperarm_up_r\"),Weight=0.0029),(Item=(Type=Bone,Name=\"deform_pec_r\"),Weight=0.0974),(Item=(Type=Bone,Name=\"deform_upperarm_fr_r\"),Weight=0.1426),(Item=(Type=Bone,Name=\"upperarm_twist_01_r\"),Weight=0.1433),(Item=(Type=Bone,Nam ...(truncated)"
    Weight=1
   )
   > &RigUnit_ParentConstraint(
    AdvancedSettings=(RotationOrderForFilter=XZY)
    Child=(Type=Bone,Name="base_shoulderCloth_l")
    Parents="((Item=(Type=Bone,Name=\"deform_upperarm_up_l\"),Weight=0.0046),(Item=(Type=Bone,Name=\"deform_upperarm_fr_l\"),Weight=0.1174),(Item=(Type=Bone,Name=\"spine_05\"),Weight=0.1253),(Item=(Type=Bone,Name=\"deform_pec_l\"),Weight=0.1492),(Item=(Type=Bone,Name=\"upperarm ...(truncated)"
    Weight=1
   )
   n12 = &RigVMFunction_AlphaInterp(InterpSpeedDecreasing=20, InterpSpeedIncreasing=10, Scale=1, bClampResult=False, Value=RigVMVariableNode() {variable=Upperarm_L_FW}.Value)
   > &RigUnit_ModifyTransforms(
    ItemToModify=((Item=(Type=Bone,Name="dyn_shoulderCloth_l_01"),Transform=(Rotation=(X=-0.1807,Y=0,Z=0,W=0.9835),Translation=(X=0,Y=3.0664,Z=0.7021),Scale3D=(X=1,Y=1,Z=1))))
    WeightMinimum=0
    Weight=@n12.Result
   )
   n13 = &RigVMFunction_AlphaInterp(InterpSpeedDecreasing=20, InterpSpeedIncreasing=10, Scale=1, bClampResult=False, Value=RigVMVariableNode() {variable=Upperarm_R_FW}.Value)
   > &RigUnit_ModifyTransforms(
    ItemToModify=((Item=(Type=Bone,Name="dyn_shoulderCloth_r_01"),Transform=(Rotation=(X=-0.2588,Y=0,Z=0,W=0.9659),Translation=(X=0,Y=0,Z=0),Scale3D=(X=1,Y=1,Z=1))))
    WeightMinimum=0
    Weight=@n13.Result
   )
   RigVMFunctionReturnNode(ExecuteContext=@RigVMFunction_Sequence.C)
  ]}
 New Function_ContainedGraph {nodes [
   RigVMFunctionEntryNode()
   > RigVMFunction_Sequence_1 = RigVMAggregateNode()
   AlphaInterp = &RigVMFunction_AlphaInterp(InterpSpeedDecreasing=10, InterpSpeedIncreasing=10, Scale=1, bClampResult=False, Value=RigVMVariableNode() {variable=Is Crouching}.Value)
   &RigUnit_ModifyTransforms(
    ItemToModify="((Item=(Type=Bone,Name=\"dyn_C_FrontSkirt_1\"),Transform=(Rotation=(X=-0.1265,Y=-0.2335,Z=-0.0517,W=-0.9627),Translation=(X=0,Y=0,Z=0),Scale3D=(X=1,Y=1,Z=1))),(Item=(Type=Bone,Name=\"dyn_R_FrontSkirt_1\"),Transform=(Rotation=(X=-0.1332,Y=-0.0709,Z=-0.0302,W=-0 ...(truncated)"
    WeightMinimum=0
    ExecutePin=@RigVMFunction_Sequence_1.A
    Weight=@AlphaInterp.Result
   )
   > &RigUnit_ModifyTransforms(
    ItemToModify="((Item=(Type=Bone,Name=\"dyn_R_SideSkirt_4\"),Transform=(Rotation=(X=0.1925,Y=0.1285,Z=0.3021,W=-0.9247),Translation=(X=0,Y=0,Z=0),Scale3D=(X=1,Y=1,Z=1))),(Item=(Type=Bone,Name=\"dyn_R_BackSkirt_2\"),Transform=(Rotation=(X=-0.0838,Y=0.3216,Z=-0.0671,W=0.9408), ...(truncated)"
    WeightMinimum=0
    Weight=@AlphaInterp.Result
   )
   Multiply = &RigVMFunction_MathDoubleMul(A=RigVMVariableNode() {variable=Is Crouching}.Value, B=RigVMVariableNode() {variable=PelvisProx}.Value)
   n1 = &RigVMFunction_AlphaInterp(InterpSpeedDecreasing=5, InterpSpeedIncreasing=10, Scale=1, bClampResult=False, Value=@Multiply.Result)
   > &RigUnit_ModifyTransforms(
    ItemToModify="((Item=(Type=Bone,Name=\"dyn_C_FrontSkirt_1\"),Transform=(Rotation=(X=0,Y=-0.2964,Z=0,W=0.9551),Translation=(X=0,Y=0,Z=0),Scale3D=(X=1,Y=1,Z=1))),(Item=(Type=Bone,Name=\"dyn_R_FrontSkirt_2\"),Transform=(Rotation=(X=0.3281,Y=0.0223,Z=-0.1081,W=0.9382),Translati ...(truncated)"
    WeightMinimum=0
    Weight=@n1.Result
   )
   > &RigUnit_ModifyTransforms(
    ItemToModify="((Item=(Type=Bone,Name=\"None\"),Transform=(Rotation=(X=0,Y=0,Z=0,W=1),Translation=(X=0,Y=0,Z=0),Scale3D=(X=1,Y=1,Z=1))),(Item=(Type=Bone,Name=\"dyn_R_BackSkirt_3\"),Transform=(Rotation=(X=0.0319,Y=-0.039,Z=-0.0012,W=-0.9987),Translation=(X=0,Y=0,Z=0),Scale3D= ...(truncated)"
    WeightMinimum=0
    Weight=@n1.Result
   )
   RigVMFunction_MathBoolToFloat = &RigVMFunction_MathBoolToFloat(Value=RigVMVariableNode() {variable=OnlyCrouch}.Value)
   n2 = &RigVMFunction_AlphaInterp(InterpSpeedDecreasing=10, InterpSpeedIncreasing=15, Scale=1, bClampResult=False, Value=@RigVMFunction_MathBoolToFloat.Result)
   &RigUnit_ModifyTransforms(
    ItemToModify="((Item=(Type=Bone,Name=\"dyn_R_BackSkirt_3\"),Transform=(Rotation=(X=-0.2308,Y=0.0837,Z=0.1047,W=-0.9637),Translation=(X=-4.6411,Y=4.9272,Z=-1.8168),Scale3D=(X=1,Y=1,Z=1))),(Item=(Type=Bone,Name=\"dyn_C_BackSkirt_3\"),Transform=(Rotation=(X=0,Y=0,Z=0.0763,W=-0 ...(truncated)"
    WeightMinimum=0
    Weight=@n2.Result
    ExecutePin=@RigVMFunction_Sequence_1.C
   )
   n3 = &RigVMFunction_AlphaInterp(
    InterpSpeedDecreasing=5
    InterpSpeedIncreasing=5
    Scale=1
    bClampResult=False
    Value=&RigVMFunction_MathDoubleMul(A=RigVMVariableNode() {variable=MoveLeft}.Value, B=RigVMVariableNode() {variable=Is Crouching}.Value).Result
   )
   > &RigUnit_ModifyTransforms(
    ItemToModify="((Item=(Type=Bone,Name=\"dyn_R_SideSkirt_2\"),Transform=(Rotation=(X=0,Y=0.1069,Z=0,W=-0.9943),Translation=(X=0,Y=0,Z=0),Scale3D=(X=1,Y=1,Z=1))),(Item=(Type=Bone,Name=\"dyn_R_BackSkirt_2\"),Transform=(Rotation=(X=0.0014,Y=-0.0858,Z=-0.0269,W=0.9959),Translatio ...(truncated)"
    WeightMinimum=0
    Weight=@n3.Result
   )
   n4 = &RigVMFunction_AlphaInterp(
    InterpSpeedDecreasing=5
    InterpSpeedIncreasing=5
    Scale=1
    bClampResult=False
    Value=&RigVMFunction_MathDoubleMul(B=RigVMVariableNode() {variable=Is Crouching}.Value, A=RigVMVariableNode() {variable=MoveRight}.Value).Result
   )
   > &RigUnit_ModifyTransforms(
    ItemToModify="((Item=(Type=Bone,Name=\"dyn_R_SideSkirt_2\"),Transform=(Rotation=(X=-0.0003,Y=-0.0016,Z=-0.0099,W=-1),Translation=(X=0,Y=0,Z=0),Scale3D=(X=1,Y=1,Z=1))),(Item=(Type=Bone,Name=\"dyn_R_BackSkirt_2\"),Transform=(Rotation=(X=-0.0105,Y=0.0655,Z=0.0764,W=0.9949),Tra ...(truncated)"
    WeightMinimum=0
    Weight=@n4.Result
   )
   n5 = &RigVMFunction_AlphaInterp(
    InterpSpeedDecreasing=5
    InterpSpeedIncreasing=5
    Scale=1
    bClampResult=False
    Value=&RigVMFunction_MathDoubleMul(A=RigVMVariableNode() {variable=MoveFw}.Value, B=RigVMVariableNode() {variable=Is Crouching}.Value).Result
   )
   > &RigUnit_ModifyTransforms(
    ItemToModify="((Item=(Type=Bone,Name=\"dyn_R_SideSkirt_2\"),Transform=(Rotation=(X=-0.0301,Y=-0.0306,Z=0.2084,W=0.9771),Translation=(X=0,Y=0,Z=0),Scale3D=(X=1,Y=1,Z=1))),(Item=(Type=Bone,Name=\"dyn_R_BackSkirt_2\"),Transform=(Rotation=(X=-0.0263,Y=-0.0119,Z=0.1572,W=0.9871) ...(truncated)"
    WeightMinimum=0
    Weight=@n5.Result
   )
   n6 = &RigVMFunction_AlphaInterp(
    InterpSpeedDecreasing=5
    InterpSpeedIncreasing=5
    Scale=1
    bClampResult=False
    Value=&RigVMFunction_MathDoubleMul(B=RigVMVariableNode() {variable=Is Crouching}.Value, A=RigVMVariableNode() {variable=MoveBw}.Value).Result
   )
   > &RigUnit_ModifyTransforms(
    ItemToModify="((Item=(Type=Bone,Name=\"dyn_R_SideSkirt_2\"),Transform=(Rotation=(X=-0.0037,Y=0.1255,Z=-0.0113,W=-0.992),Translation=(X=0,Y=0,Z=0),Scale3D=(X=1,Y=1,Z=1))),(Item=(Type=Bone,Name=\"dyn_R_BackSkirt_2\"),Transform=(Rotation=(X=0,Y=0,Z=0,W=1),Translation=(X=0,Y=0, ...(truncated)"
    WeightMinimum=0
    Weight=@n6.Result
   )
   RigVMFunctionReturnNode(ExecuteContext=@RigVMFunction_Sequence_1.D)
   RigVMVariableNode() {variable=Is Crouching}
  ]}
 New Function_ContainedGraph {nodes [
   RigVMFunctionEntryNode()
   > RigVMFunction_Sequence = RigVMAggregateNode()
   n1 = RigVMVariableNode() {variable=Is SkyDiving Float}
   &RigUnit_ModifyTransforms(
    ItemToModify="((Item=(Type=Bone,Name=\"dyn_C_BackSkirt_4\"),Transform=(Rotation=(X=0,Y=0,Z=0,W=1),Translation=(X=0,Y=0,Z=0),Scale3D=(X=1,Y=1,Z=1))),(Item=(Type=Bone,Name=\"dyn_L_BackSkirt_2\"),Transform=(Rotation=(X=0,Y=0,Z=0,W=1),Translation=(X=0,Y=0,Z=0),Scale3D=(X=1,Y=1, ...(truncated)"
    WeightMinimum=0
    ExecutePin=@RigVMFunction_Sequence.A
    Weight=@n1.Value
   )
   > &RigUnit_ModifyTransforms(
    ItemToModify="((Item=(Type=Bone,Name=\"dyn_L_BackSkirt_2\"),Transform=(Rotation=(X=0,Y=0,Z=0,W=1),Translation=(X=0,Y=0,Z=0),Scale3D=(X=1,Y=1,Z=1))),(Item=(Type=Bone,Name=\"dyn_C_BackSkirt_2\"),Transform=(Rotation=(X=0,Y=-0.0436,Z=0,W=0.999),Translation=(X=0,Y=0,Z=0),Scale3D ...(truncated)"
    WeightMinimum=0
    Weight=@n1.Value
   )
   > &RigUnit_ModifyTransforms(
    ItemToModify="((Item=(Type=Bone,Name=\"dyn_R_BackSkirt_1\"),Transform=(Rotation=(X=0,Y=-0.0436,Z=0,W=0.999),Translation=(X=0,Y=0,Z=0),Scale3D=(X=1,Y=1,Z=1))),(Item=(Type=Bone,Name=\"dyn_L_BackSkirt_1\"),Transform=(Rotation=(X=0,Y=-0.0436,Z=0,W=0.999),Translation=(X=0,Y=0,Z= ...(truncated)"
    WeightMinimum=0
    Weight=RigVMVariableNode() {variable=Is ParachuteOpen}.Value
   )
   RigVMFunctionReturnNode(ExecuteContext=@RigVMFunction_Sequence.C)
  ]}
 New Function_ContainedGraph {nodes [
   n1 = RigVMFunctionEntryNode()
   RigVMFunction_Sequence = &RigVMFunction_Sequence(ExecuteContext=@n1.ExecuteContext)
   n2 = RigVMDispatchNode(False=0, True=1, Condition=RigVMVariableNode() {variable=Emote}.Value) {template="DISPATCH_RigVMDispatch_If(in Condition,in True,in False,out Result)"}
   n3 = &RigVMFunction_AlphaInterp(InterpSpeedDecreasing=10, InterpSpeedIncreasing=10, Scale=1, bClampResult=False, Value=@n2.Result)
   &RigUnit_ModifyTransforms(
    ItemToModify="((Item=(Type=Bone,Name=\"dyn_L_BackSkirt_2\"),Transform=(Rotation=(X=0,Y=-0.0872,Z=0,W=0.9962),Translation=(X=0,Y=0,Z=0),Scale3D=(X=1,Y=1,Z=1))),(Item=(Type=Bone,Name=\"dyn_C_BackSkirt_2\"),Transform=(Rotation=(X=0,Y=-0.0872,Z=0,W=0.9962),Translation=(X=0,Y=0, ...(truncated)"
    WeightMinimum=0
    Weight=@n3.Result
    ExecutePin=@RigVMFunction_Sequence.A
   )
   n4 = &RigVMFunction_AlphaInterp(
    InterpSpeedDecreasing=10
    InterpSpeedIncreasing=10
    Scale=1
    bClampResult=False
    Value=&RigVMFunction_MathDoubleMul(A=@n2.Result, B=RigVMVariableNode() {variable=SPR L calf Bkd}.Value).Result
   )
   > &RigUnit_ModifyTransforms(
    ItemToModify="((Item=(Type=Bone,Name=\"None\"),Transform=(Rotation=(X=0,Y=0,Z=0,W=1),Translation=(X=0,Y=0,Z=0),Scale3D=(X=1,Y=1,Z=1))),(Item=(Type=Bone,Name=\"dyn_L_BackSkirt_2\"),Transform=(Rotation=(X=0.355,Y=-0.5002,Z=-0.0723,W=0.7865),Translation=(X=0,Y=0,Z=0),Scale3D=( ...(truncated)"
    WeightMinimum=0
    Weight=@n4.Result
   )
   n5 = RigVMVariableNode() {variable=Emote}
   n6 = &RigVMFunction_AlphaInterp(
    InterpSpeedDecreasing=10
    InterpSpeedIncreasing=10
    Scale=1
    bClampResult=False
    Value=&RigVMFunction_MathDoubleMul(
     A=RigVMDispatchNode(False=0, True=1, Condition=@n5.Value) {template="DISPATCH_RigVMDispatch_If(in Condition,in True,in False,out Result)"}.Result
     B=RigVMVariableNode() {variable=SPR R calf Bkd}.Value
    ).Result
   )
   > &RigUnit_ModifyTransforms(
    ItemToModify="((Item=(Type=Bone,Name=\"None\"),Transform=(Rotation=(X=0,Y=0,Z=0,W=1),Translation=(X=0,Y=0,Z=0),Scale3D=(X=1,Y=1,Z=1))),(Item=(Type=Bone,Name=\"dyn_R_BackSkirt_2\"),Transform=(Rotation=(X=-0.0208,Y=-0.6224,Z=0.0893,W=0.7773),Translation=(X=0,Y=0,Z=0),Scale3D= ...(truncated)"
    WeightMinimum=0
    Weight=@n6.Result
   )
   n7 = &RigVMFunction_MathDoubleMul(B=@n4.Result, A=@n6.Result)
   > &RigUnit_ModifyTransforms(
    ItemToModify="((Item=(Type=Bone,Name=\"dyn_R_BackSkirt_3\"),Transform=(Rotation=(X=0,Y=-0.3365,Z=0,W=-0.9417),Translation=(X=0,Y=0,Z=0),Scale3D=(X=1,Y=1,Z=1))),(Item=(Type=Bone,Name=\"dyn_C_BackSkirt_2\"),Transform=(Rotation=(X=0,Y=0.7622,Z=0,W=0.6473),Translation=(X=0,Y=0, ...(truncated)"
    WeightMinimum=0
    Weight=@n7.Result
   )
   RigVMFunction_MathBoolToFloat = &RigVMFunction_MathBoolToFloat(Value=RigVMVariableNode() {variable=Emote_Celebration}.Value)
   AlphaInterp = &RigVMFunction_AlphaInterp(InterpSpeedDecreasing=10, InterpSpeedIncreasing=2, Scale=1, bClampResult=False, Value=@RigVMFunction_MathBoolToFloat.Result)
   > &RigUnit_ModifyTransforms(
    ItemToModify="((Item=(Type=Bone,Name=\"dyn_R_SideSkirt_4\"),Transform=(Rotation=(X=0,Y=0,Z=0.2844,W=-0.9587),Translation=(X=0,Y=0,Z=0),Scale3D=(X=1,Y=1,Z=1))),(Item=(Type=Bone,Name=\"dyn_C_BackSkirt_2\"),Transform=(Rotation=(X=0,Y=0.5503,Z=0,W=0.835),Translation=(X=0,Y=0,Z= ...(truncated)"
    WeightMinimum=0
    Weight=@AlphaInterp.Result
   )
   n8 = &RigVMFunction_AlphaInterp(
    InterpSpeedDecreasing=10
    InterpSpeedIncreasing=2
    Scale=1
    bClampResult=False
    Value=&RigVMFunction_MathBoolToFloat(Value=RigVMVariableNode() {variable=EasterBloc_Emote}.Value).Result
   )
   > &RigUnit_ModifyTransforms(
    ItemToModify="((Item=(Type=Bone,Name=\"dyn_R_SideSkirt_2\"),Transform=(Rotation=(X=0,Y=0,Z=0,W=-1),Translation=(X=0,Y=0,Z=0),Scale3D=(X=1,Y=1,Z=1))),(Item=(Type=Bone,Name=\"dyn_C_BackSkirt_2\"),Transform=(Rotation=(X=0,Y=0,Z=0,W=1),Translation=(X=0,Y=0,Z=0),Scale3D=(X=1,Y=1 ...(truncated)"
    WeightMinimum=0
    Weight=@n8.Result
   )
   RigVMFunctionReturnNode(ExecuteContext=@RigVMFunction_Sequence.B)
   &RigUnit_ModifyTransforms(
    ExecutePin=()
    ItemToModify="((Item=(Type=Bone,Name=\"dyn_C_BackSkirt_4\"),Transform=(Rotation=(X=0,Y=0.5042,Z=0,W=-0.8636),Translation=(X=0,Y=0,Z=0),Scale3D=(X=1,Y=1,Z=1))),(Item=(Type=Bone,Name=\"dyn_R_BackSkirt_3\"),Transform=(Rotation=(X=0,Y=-0.316,Z=0,W=0.9488),Translation=(X=0,Y=0,Z ...(truncated)"
    WeightMinimum=0
    Weight=@n4.Result
   )
   > &RigUnit_ModifyTransforms(
    ItemToModify="((Item=(Type=Bone,Name=\"dyn_C_BackSkirt_4\"),Transform=(Rotation=(X=0,Y=-0.6392,Z=0,W=0.7691),Translation=(X=0,Y=0,Z=0),Scale3D=(X=1,Y=1,Z=1))),(Item=(Type=Bone,Name=\"dyn_R_BackSkirt_3\"),Transform=(Rotation=(X=0,Y=0.7371,Z=0,W=-0.6758),Translation=(X=0,Y=0, ...(truncated)"
    WeightMinimum=0
    Weight=@n6.Result
   )
   > &RigUnit_ModifyTransforms(
    ItemToModify="((Item=(Type=Bone,Name=\"dyn_C_BackSkirt_4\"),Transform=(Rotation=(X=0,Y=-0.9947,Z=0,W=-0.1025),Translation=(X=0,Y=0,Z=0),Scale3D=(X=1,Y=1,Z=1))),(Item=(Type=Bone,Name=\"dyn_R_BackSkirt_3\"),Transform=(Rotation=(X=0,Y=-0.9796,Z=0,W=-0.2011),Translation=(X=0,Y= ...(truncated)"
    WeightMinimum=0
    Weight=@n7.Result
   )
   RigVMVariableNode() {variable=SPR R calf Max Bkd}
   RigVMVariableNode() {variable=SPR L calf Max Bkd}
  ]}
 New Function_ContainedGraph {nodes [
   RigVMFunctionEntryNode()
   > RigVMFunction_Sequence = RigVMAggregateNode()
   n1 = RigVMVariableNode() {variable=SPR L Leg Fwd}
   &RigUnit_ModifyTransforms(
    ItemToModify="((Item=(Type=Bone,Name=\"dyn_L_FrontSkirt_1\"),Transform=(Rotation=(X=0.051,Y=0.2422,Z=0.0541,W=-0.9674),Translation=(X=-5.0016,Y=1.0812,Z=5.8115),Scale3D=(X=1,Y=1,Z=1))),(Item=(Type=Bone,Name=\"dyn_C_FrontSkirt_1\"),Transform=(Rotation=(X=0,Y=0.497,Z=0,W=-0.8 ...(truncated)"
    WeightMinimum=0
    ExecutePin=@RigVMFunction_Sequence.A
    Weight=@n1.Value
   )
   n2 = RigVMVariableNode() {variable=SPR R Leg Fwd}
   > &RigUnit_ModifyTransforms(
    ItemToModify="((Item=(Type=Bone,Name=\"dyn_C_FrontSkirt_1\"),Transform=(Rotation=(X=0,Y=-0.4032,Z=0,W=0.9151),Translation=(X=0,Y=0,Z=0),Scale3D=(X=1,Y=1,Z=1))),(Item=(Type=Bone,Name=\"dyn_R_FrontSkirt_1\"),Transform=(Rotation=(X=0,Y=-0.1134,Z=0,W=0.9935),Translation=(X=-1.9 ...(truncated)"
    WeightMinimum=0
    Weight=@n2.Value
   )
   > &RigUnit_ModifyTransforms(
    ItemToModify="((Item=(Type=Bone,Name=\"dyn_R_FrontSkirt_1\"),Transform=(Rotation=(X=-0.019,Y=0.1862,Z=0.0558,W=0.9808),Translation=(X=0,Y=0,Z=0),Scale3D=(X=1,Y=1,Z=1))),(Item=(Type=Bone,Name=\"dyn_L_FrontSkirt_1\"),Transform=(Rotation=(X=0,Y=-0.1028,Z=0,W=-0.9947),Translati ...(truncated)"
    WeightMinimum=0
    Weight=&RigVMFunction_MathDoubleMul(A=@n2.Value, B=@n1.Value).Result
   )
   n3 = RigVMVariableNode() {variable=SPR L Leg Bkd}
   > &RigUnit_ModifyTransforms(
    ItemToModify="((Item=(Type=Bone,Name=\"dyn_C_FrontSkirt_1\"),Transform=(Rotation=(X=0,Y=-0.3961,Z=0,W=0.9182),Translation=(X=0,Y=0,Z=0),Scale3D=(X=1,Y=1,Z=1))),(Item=(Type=Bone,Name=\"dyn_L_FrontSkirt_1\"),Transform=(Rotation=(X=0.0972,Y=0.6912,Z=-0.0997,W=-0.7091),Translat ...(truncated)"
    WeightMinimum=0
    Weight=@n3.Value
   )
   n4 = RigVMVariableNode() {variable=SPR R Leg Bkd}
   > &RigUnit_ModifyTransforms(
    ItemToModify="((Item=(Type=Bone,Name=\"dyn_R_SideSkirt_1\"),Transform=(Rotation=(X=0.0567,Y=0.175,Z=-0.3028,W=-0.9351),Translation=(X=0,Y=0,Z=0),Scale3D=(X=1,Y=1,Z=1))),(Item=(Type=Bone,Name=\"dyn_C_FrontSkirt_1\"),Transform=(Rotation=(X=0,Y=-0.4926,Z=0,W=0.8703),Translatio ...(truncated)"
    WeightMinimum=0
    Weight=@n4.Value
   )
   Multiply = &RigVMFunction_MathDoubleMul(B=@n3.Value, A=@n4.Value)
   > &RigUnit_ModifyTransforms(
    ItemToModify="((Item=(Type=Bone,Name=\"None\"),Transform=(Rotation=(X=0,Y=0,Z=0,W=1),Translation=(X=0,Y=0,Z=0),Scale3D=(X=1,Y=1,Z=1))),(Item=(Type=Bone,Name=\"dyn_C_BackSkirt_1\"),Transform=(Rotation=(X=0,Y=-0.6981,Z=0,W=-0.716),Translation=(X=0,Y=0,Z=0),Scale3D=(X=1,Y=1,Z= ...(truncated)"
    WeightMinimum=0
    Weight=@Multiply.Result
   )
   > &RigUnit_ModifyTransforms(
    ItemToModify="((Item=(Type=Bone,Name=\"dyn_C_FrontSkirt_1\"),Transform=(Rotation=(X=0,Y=0,Z=0,W=1),Translation=(X=0,Y=0,Z=0),Scale3D=(X=1,Y=1,Z=1))),(Item=(Type=Bone,Name=\"dyn_L_FrontSkirt_1\"),Transform=(Rotation=(X=0,Y=0,Z=-0.0182,W=-0.9998),Translation=(X=0,Y=0,Z=0),Sca ...(truncated)"
    WeightMinimum=0
    Weight=RigVMVariableNode() {variable=SPR L Leg Otw}.Value
   )
   > &RigUnit_ModifyTransforms(
    ItemToModify="((Item=(Type=Bone,Name=\"dyn_R_SideSkirt_1\"),Transform=(Rotation=(X=0,Y=-0.3947,Z=0,W=0.9188),Translation=(X=0,Y=0,Z=0),Scale3D=(X=1,Y=1,Z=1))),(Item=(Type=Bone,Name=\"dyn_C_FrontSkirt_1\"),Transform=(Rotation=(X=0,Y=0,Z=0,W=1),Translation=(X=0,Y=0,Z=0),Scale ...(truncated)"
    WeightMinimum=0
    Weight=RigVMVariableNode() {variable=SPR R Leg Otw}.Value
   )
   > &RigUnit_ModifyTransforms(
    ItemToModify="((Item=(Type=Bone,Name=\"dyn_R_SideSkirt_1\"),Transform=(Rotation=(X=0.0001,Y=0.206,Z=0.0001,W=-0.9785),Translation=(X=0,Y=0,Z=0),Scale3D=(X=1,Y=1,Z=1))),(Item=(Type=Bone,Name=\"dyn_R_FrontSkirt_1\"),Transform=(Rotation=(X=0.0553,Y=-0.1662,Z=0.3108,W=0.9342),T ...(truncated)"
    WeightMinimum=0
    Weight=RigVMVariableNode() {variable=SPR R Leg Itw}.Value
   )
   > n5 = &RigUnit_ModifyTransforms(
    ItemToModify="((Item=(Type=Bone,Name=\"dyn_C_FrontSkirt_1\"),Transform=(Rotation=(X=0,Y=0,Z=-0.196,W=0.9806),Translation=(X=0,Y=0,Z=0),Scale3D=(X=1,Y=1,Z=1))),(Item=(Type=Bone,Name=\"dyn_L_FrontSkirt_1\"),Transform=(Rotation=(X=0.0477,Y=0.1091,Z=0.3979,W=-0.9097),Translatio ...(truncated)"
    WeightMinimum=0
    Weight=RigVMVariableNode() {variable=SPR L Leg Itw}.Value
   )
   RigVMFunction_ControlFlowBranch = &RigVMFunction_ControlFlowBranch(ExecuteContext=@n5.ExecutePin, Condition=RigVMVariableNode() {variable=CrouchBool}.Value)
   RigVMFunction_MathBoolOr = &RigVMFunction_MathBoolOr(A=RigVMVariableNode() {variable=Emote}.Value, B=RigVMVariableNode() {variable=SkydiveBool}.Value)
   AlphaInterp = &RigVMFunction_AlphaInterp(
    InterpSpeedDecreasing=5
    InterpSpeedIncreasing=1
    Scale=1
    bClampResult=false
    Value=&RigVMFunction_MathDoubleMul(
     A=RigVMDispatchNode(False=1, True=0, Condition=@RigVMFunction_MathBoolOr.Result) {template="DISPATCH_RigVMDispatch_If(in Condition,in True,in False,out Result)"}.Result
     B=RigVMVariableNode() {variable=SPR L calf Bkd}.Value
    ).Result
   )
   &RigUnit_ModifyTransforms(
    ItemToModify="((Item=(Type=Bone,Name=\"None\"),Transform=(Rotation=(X=0,Y=0,Z=0,W=1),Translation=(X=0,Y=0,Z=0),Scale3D=(X=1,Y=1,Z=1))),(Item=(Type=Bone,Name=\"dyn_L_BackSkirt_2\"),Transform=(Rotation=(X=0.355,Y=-0.5002,Z=-0.0723,W=0.7865),Translation=(X=0,Y=0,Z=0),Scale3D=( ...(truncated)"
    WeightMinimum=0
    Weight=@AlphaInterp.Result
    ExecutePin=@RigVMFunction_ControlFlowBranch.False
   )
   n6 = &RigVMFunction_MathBoolOr(A=RigVMVariableNode() {variable=Emote}.Value, B=RigVMVariableNode() {variable=SkydiveBool}.Value)
   n7 = &RigVMFunction_AlphaInterp(
    ClampMax=0.8
    InterpSpeedDecreasing=5
    InterpSpeedIncreasing=1
    Scale=1
    bClampResult=false
    Value=&RigVMFunction_MathDoubleMul(
     B=RigVMDispatchNode(False=1, True=0, Condition=@n6.Result) {template="DISPATCH_RigVMDispatch_If(in Condition,in True,in False,out Result)"}.Result
     A=RigVMVariableNode() {variable=SPR R calf Bkd}.Value
    ).Result
   )
   > &RigUnit_ModifyTransforms(
    ItemToModify="((Item=(Type=Bone,Name=\"None\"),Transform=(Rotation=(X=0,Y=0,Z=0,W=1),Translation=(X=0,Y=0,Z=0),Scale3D=(X=1,Y=1,Z=1))),(Item=(Type=Bone,Name=\"dyn_R_BackSkirt_2\"),Transform=(Rotation=(X=-0.0208,Y=-0.6224,Z=0.0893,W=0.7773),Translation=(X=0,Y=0,Z=0),Scale3D= ...(truncated)"
    WeightMinimum=0
    Weight=@n7.Result
   )
   n8 = &RigVMFunction_AlphaInterp(
    ClampMax=0.8
    InterpSpeedDecreasing=5
    InterpSpeedIncreasing=1
    Scale=1
    bClampResult=false
    Value=&RigVMFunction_MathDoubleMul(B=@AlphaInterp.Result, A=@n7.Result).Result
   )
   > &RigUnit_ModifyTransforms(
    ItemToModify="((Item=(Type=Bone,Name=\"dyn_R_BackSkirt_3\"),Transform=(Rotation=(X=0,Y=-0.3365,Z=0,W=-0.9417),Translation=(X=0,Y=0,Z=0),Scale3D=(X=1,Y=1,Z=1))),(Item=(Type=Bone,Name=\"dyn_C_BackSkirt_2\"),Transform=(Rotation=(X=0,Y=0.7622,Z=0,W=0.6473),Translation=(X=0,Y=0, ...(truncated)"
    WeightMinimum=0
    Weight=@n8.Result
   )
   n9 = RigVMVariableNode() {variable=SkydiveBool}
   n10 = &RigVMFunction_AlphaInterp(
    ClampMax=0.4
    InterpSpeedDecreasing=5
    InterpSpeedIncreasing=10
    Scale=1
    bClampResult=true
    Value=&RigVMFunction_MathDoubleMul(
     A=RigVMDispatchNode(False=0, True=1, Condition=@n9.Value) {template="DISPATCH_RigVMDispatch_If(in Condition,in True,in False,out Result)"}.Result
     B=RigVMVariableNode() {variable=SPR L calf Bkd}.Value
    ).Result
   )
   > &RigUnit_ModifyTransforms(
    ItemToModify="((Item=(Type=Bone,Name=\"dyn_C_BackSkirt_3\"),Transform=(Rotation=(X=0,Y=-0.342,Z=0,W=0.9397),Translation=(X=0,Y=0,Z=0),Scale3D=(X=1,Y=1,Z=1))),(Item=(Type=Bone,Name=\"dyn_C_BackSkirt_2\"),Transform=(Rotation=(X=0,Y=0,Z=0,W=1),Translation=(X=0,Y=0,Z=0),Scale3D ...(truncated)"
    WeightMinimum=0
    Weight=@n10.Result
   )
   n11 = RigVMVariableNode() {variable=SkydiveBool}
   n12 = &RigVMFunction_AlphaInterp(
    ClampMax=0.4
    InterpSpeedDecreasing=5
    InterpSpeedIncreasing=10
    Scale=1
    bClampResult=true
    Value=&RigVMFunction_MathDoubleMul(
     B=RigVMDispatchNode(False=0, True=1, Condition=@n11.Value) {template="DISPATCH_RigVMDispatch_If(in Condition,in True,in False,out Result)"}.Result
     A=RigVMVariableNode() {variable=SPR R calf Bkd}.Value
    ).Result
   )
   > &RigUnit_ModifyTransforms(
    ItemToModify="((Item=(Type=Bone,Name=\"dyn_C_BackSkirt_3\"),Transform=(Rotation=(X=0,Y=0.6943,Z=0,W=-0.7197),Translation=(X=0,Y=0,Z=0),Scale3D=(X=1,Y=1,Z=1))),(Item=(Type=Bone,Name=\"dyn_R_BackSkirt_3\"),Transform=(Rotation=(X=-0.0349,Y=-0.7639,Z=0.0294,W=0.6437),Translatio ...(truncated)"
    WeightMinimum=0
    Weight=@n12.Result
   )
   > &RigUnit_ModifyTransforms(
    ItemToModify="((Item=(Type=Bone,Name=\"dyn_R_BackSkirt_3\"),Transform=(Rotation=(X=0,Y=-0.3839,Z=0,W=-0.9234),Translation=(X=0,Y=0,Z=0),Scale3D=(X=1,Y=1,Z=1))),(Item=(Type=Bone,Name=\"dyn_C_BackSkirt_2\"),Transform=(Rotation=(X=0,Y=-0.3114,Z=0,W=-0.9503),Translation=(X=0,Y= ...(truncated)"
    WeightMinimum=0
    Weight=&RigVMFunction_MathDoubleMul(B=@n10.Result, A=@n12.Result).Result
   )
   n13 = &RigVMFunction_MathBoolOr(A=RigVMVariableNode() {variable=Emote}.Value, B=RigVMVariableNode() {variable=SkydiveBool}.Value)
   n14 = &RigVMFunction_AlphaInterp(
    InterpSpeedDecreasing=10
    InterpSpeedIncreasing=10
    Scale=1
    bClampResult=false
    Value=&RigVMFunction_MathDoubleMul(
     A=RigVMDispatchNode(False=1, True=0, Condition=@n13.Result) {template="DISPATCH_RigVMDispatch_If(in Condition,in True,in False,out Result)"}.Result
     B=RigVMVariableNode() {variable=SPR L calf Bkd}.Value
    ).Result
   )
   &RigUnit_ModifyTransforms(
    ItemToModify="((Item=(Type=Bone,Name=\"None\"),Transform=(Rotation=(X=0,Y=0,Z=0,W=1),Translation=(X=0,Y=0,Z=0),Scale3D=(X=1,Y=1,Z=1))),(Item=(Type=Bone,Name=\"dyn_L_BackSkirt_2\"),Transform=(Rotation=(X=0.355,Y=-0.5002,Z=-0.0723,W=0.7865),Translation=(X=0,Y=0,Z=0),Scale3D=( ...(truncated)"
    WeightMinimum=0
    Weight=@n14.Result
    ExecutePin=@RigVMFunction_ControlFlowBranch.True
   )
   n15 = &RigVMFunction_MathBoolOr(A=RigVMVariableNode() {variable=Emote}.Value, B=RigVMVariableNode() {variable=SkydiveBool}.Value)
   n16 = &RigVMFunction_AlphaInterp(
    ClampMax=0.8
    InterpSpeedDecreasing=10
    InterpSpeedIncreasing=10
    Scale=1
    bClampResult=false
    Value=&RigVMFunction_MathDoubleMul(
     B=RigVMDispatchNode(False=1, True=0, Condition=@n15.Result) {template="DISPATCH_RigVMDispatch_If(in Condition,in True,in False,out Result)"}.Result
     A=RigVMVariableNode() {variable=SPR R calf Bkd}.Value
    ).Result
   )
   > &RigUnit_ModifyTransforms(
    ItemToModify="((Item=(Type=Bone,Name=\"None\"),Transform=(Rotation=(X=0,Y=0,Z=0,W=1),Translation=(X=0,Y=0,Z=0),Scale3D=(X=1,Y=1,Z=1))),(Item=(Type=Bone,Name=\"dyn_R_BackSkirt_2\"),Transform=(Rotation=(X=-0.0208,Y=-0.6224,Z=0.0893,W=0.7773),Translation=(X=0,Y=0,Z=0),Scale3D= ...(truncated)"
    WeightMinimum=0
    Weight=@n16.Result
   )
   n17 = &RigVMFunction_AlphaInterp(
    ClampMax=0.8
    InterpSpeedDecreasing=10
    InterpSpeedIncreasing=10
    Scale=1
    bClampResult=false
    Value=&RigVMFunction_MathDoubleMul(B=@n14.Result, A=@n16.Result).Result
   )
   > &RigUnit_ModifyTransforms(
    ItemToModify="((Item=(Type=Bone,Name=\"dyn_R_BackSkirt_3\"),Transform=(Rotation=(X=0,Y=-0.3365,Z=0,W=-0.9417),Translation=(X=0,Y=0,Z=0),Scale3D=(X=1,Y=1,Z=1))),(Item=(Type=Bone,Name=\"dyn_C_BackSkirt_2\"),Transform=(Rotation=(X=0,Y=0.7622,Z=0,W=0.6473),Translation=(X=0,Y=0, ...(truncated)"
    WeightMinimum=0
    Weight=@n17.Result
   )
   &RigUnit_ParentConstraint(
    AdvancedSettings=(RotationOrderForFilter=XZY)
    Child=(Type=Bone,Name="base_shoulderbackCloth_l_01")
    Parents="((Item=(Type=Bone,Name=\"deform_upperarm_bk_l\"),Weight=0.0468),(Item=(Type=Bone,Name=\"spine_04\"),Weight=0.1194),(Item=(Type=Bone,Name=\"upperarm_twist_01_l\"),Weight=0.1317),(Item=(Type=Bone,Name=\"spine_05\"),Weight=0.2065),(Item=(Type=Bone,Name=\"clavicle_l\"), ...(truncated)"
    Weight=1
    ExecutePin=@RigVMFunction_Sequence.B
   )
   > &RigUnit_ParentConstraint(
    AdvancedSettings=(RotationOrderForFilter=XZY)
    Child=(Type=Bone,Name="base_shoulderbackCloth_r_01")
    Parents="((Item=(Type=Bone,Name=\"deform_upperarm_bk_r\"),Weight=0.046),(Item=(Type=Bone,Name=\"spine_04\"),Weight=0.1187),(Item=(Type=Bone,Name=\"upperarm_twist_01_r\"),Weight=0.1301),(Item=(Type=Bone,Name=\"spine_05\"),Weight=0.2078),(Item=(Type=Bone,Name=\"clavicle_r\"),W ...(truncated)"
    Weight=1
   )
   > &RigUnit_ParentConstraint(
    AdvancedSettings=(RotationOrderForFilter=XZY)
    Child=(Type=Bone,Name="base_shoulderCloth_r")
    Parents="((Item=(Type=Bone,Name=\"deform_upperarm_up_r\"),Weight=0.0029),(Item=(Type=Bone,Name=\"deform_pec_r\"),Weight=0.0974),(Item=(Type=Bone,Name=\"deform_upperarm_fr_r\"),Weight=0.1426),(Item=(Type=Bone,Name=\"upperarm_twist_01_r\"),Weight=0.1433),(Item=(Type=Bone,Nam ...(truncated)"
    Weight=1
   )
   > &RigUnit_ParentConstraint(
    AdvancedSettings=(RotationOrderForFilter=XZY)
    Child=(Type=Bone,Name="base_shoulderCloth_l")
    Parents="((Item=(Type=Bone,Name=\"deform_upperarm_up_l\"),Weight=0.0046),(Item=(Type=Bone,Name=\"deform_upperarm_fr_l\"),Weight=0.1174),(Item=(Type=Bone,Name=\"spine_05\"),Weight=0.1253),(Item=(Type=Bone,Name=\"deform_pec_l\"),Weight=0.1492),(Item=(Type=Bone,Name=\"upperarm ...(truncated)"
    Weight=1
   )
   n18 = &RigVMFunction_AlphaInterp(InterpSpeedDecreasing=20, InterpSpeedIncreasing=10, Scale=1, bClampResult=False, Value=RigVMVariableNode() {variable=Upperarm_L_FW}.Value)
   > &RigUnit_ModifyTransforms(
    ItemToModify=((Item=(Type=Bone,Name="dyn_shoulderCloth_l_01"),Transform=(Rotation=(X=-0.1807,Y=0,Z=0,W=0.9835),Translation=(X=0,Y=3.0664,Z=0.7021),Scale3D=(X=1,Y=1,Z=1))))
    WeightMinimum=0
    Weight=@n18.Result
   )
   n19 = &RigVMFunction_AlphaInterp(InterpSpeedDecreasing=20, InterpSpeedIncreasing=10, Scale=1, bClampResult=False, Value=RigVMVariableNode() {variable=Upperarm_R_FW}.Value)
   > &RigUnit_ModifyTransforms(
    ItemToModify=((Item=(Type=Bone,Name="dyn_shoulderCloth_r_01"),Transform=(Rotation=(X=-0.2588,Y=0,Z=0,W=0.9659),Translation=(X=0,Y=0,Z=0),Scale3D=(X=1,Y=1,Z=1))))
    WeightMinimum=0
    Weight=@n19.Result
   )
   > &RigUnit_ModifyTransforms(
    ItemToModify=((Item=(Type=Bone,Name="dyn_ShoulderclPad_l_01"),Transform=(Rotation=(X=0,Y=-0.4749,Z=0,W=0.8801),Translation=(X=-4.8682,Y=0,Z=3.1973),Scale3D=(X=1,Y=1,Z=1))))
    WeightMinimum=0
    Weight=RigVMVariableNode() {variable=Upperarm_L_Dwn}.Value
   )
   > &RigUnit_ModifyTransforms(
    ItemToModify=((Item=(Type=Bone,Name="dyn_ShoulderclPad_r_01"),Transform=(Rotation=(X=0,Y=-0.4749,Z=0,W=0.88),Translation=(X=-4.8682,Y=0,Z=3.1973),Scale3D=(X=1,Y=1,Z=1))))
    WeightMinimum=0
    Weight=RigVMVariableNode() {variable=Upperarm_R_Dwn}.Value
   )
   > DeltaRotation = RigVMFunctionReferenceNode(
    Rotation Order=ZYX
    child=(Type=Bone,Name="upperarm_l")
    outputMinMax=(X=0,Y=1)
    parent=(Type=Null,Name="upperarm_l_null")
    premult=1
    rotationMinMax=(X=5,Y=50)
    triggerAxis=Y
    twistAxis=(X=1,Y=0,Z=0)
   )
   n20 = &RigVMFunction_AlphaInterp(InterpSpeedDecreasing=5, InterpSpeedIncreasing=5, Scale=1, bClampResult=False, Value=@DeltaRotation.outputWeight)
   > &RigUnit_ModifyTransforms(
    ItemToModify=((Item=(Type=Bone,Name="base_sleevescloth_l"),Transform=(Rotation=(X=0,Y=-0.5,Z=0,W=0.866),Translation=(X=0,Y=0,Z=0),Scale3D=(X=1,Y=1,Z=1))))
    WeightMinimum=0
    Weight=@n20.Result
   )
   > DeltaRotation_1 = RigVMFunctionReferenceNode(
    Rotation Order=ZYX
    child=(Type=Bone,Name="upperarm_r")
    outputMinMax=(X=0,Y=1)
    parent=(Type=Null,Name="upperarm_r_null")
    premult=1
    rotationMinMax=(X=5,Y=50)
    triggerAxis=Y
    twistAxis=(X=1,Y=0,Z=0)
   )
   n21 = &RigVMFunction_AlphaInterp(InterpSpeedDecreasing=5, InterpSpeedIncreasing=5, Scale=1, bClampResult=False, Value=@DeltaRotation_1.outputWeight)
   > &RigUnit_ModifyTransforms(
    ItemToModify=((Item=(Type=Bone,Name="dyn_sleevescloth_r_02"),Transform=(Rotation=(X=0,Y=0.5,Z=0,W=0.866),Translation=(X=0,Y=0,Z=0),Scale3D=(X=1,Y=1,Z=1))))
    WeightMinimum=0
    Weight=@n21.Result
   )
   n22 = RigVMVariableNode() {variable=idle}
   n23 = &RigVMFunction_AlphaInterp(
    InterpSpeedDecreasing=3.5
    InterpSpeedIncreasing=10
    Scale=1
    bClampResult=false
    Value=&RigVMFunction_MathDoubleMul(
     A=RigVMDispatchNode(False=0, True=1, Condition=@n22.Value) {template="DISPATCH_RigVMDispatch_If(in Condition,in True,in False,out Result)"}.Result
     B=RigVMVariableNode() {variable=ADS}.Value
    ).Result
   )
   > &RigUnit_ModifyTransforms(
    ItemToModify="((Item=(Type=Bone,Name=\"dyn_R_SideSkirt_2\"),Transform=(Rotation=(X=0,Y=0,Z=-0.1305,W=0.9914),Translation=(X=0,Y=0,Z=0),Scale3D=(X=1,Y=1,Z=1))),(Item=(Type=Bone,Name=\"dyn_R_BackSkirt_2\"),Transform=(Rotation=(X=0,Y=0.0872,Z=0,W=0.9962),Translation=(X=0,Y=0,Z ...(truncated)"
    WeightMinimum=0
    Weight=@n23.Result
   )
   RigVMFunctionReturnNode(ExecuteContext=@RigVMFunction_Sequence.D)
   n24 = &RigVMFunction_AlphaInterp(InterpSpeedDecreasing=10, InterpSpeedIncreasing=3, Scale=1, bClampResult=False, Value=RigVMVariableNode() {variable=Sprinting}.Value)
   &RigUnit_ModifyTransforms(
    ItemToModify="((Item=(Type=Bone,Name=\"dyn_C_BackSkirt_2\"),Transform=(Rotation=(X=0,Y=0,Z=0,W=1),Translation=(X=0,Y=0,Z=1),Scale3D=(X=1,Y=1,Z=1))),(Item=(Type=Bone,Name=\"dyn_R_BackSkirt_2\"),Transform=(Rotation=(X=0,Y=0,Z=0,W=1),Translation=(X=0,Y=0,Z=1),Scale3D=(X=1,Y=1, ...(truncated)"
    WeightMinimum=0
    Weight=@n24.Result
    ExecutePin=@RigVMFunction_Sequence.C
   )
   > &RigUnit_ModifyTransforms(
    ItemToModify="((Item=(Type=Bone,Name=\"dyn_C_BackSkirt_3\"),Transform=(Rotation=(X=0,Y=-0.0872,Z=0,W=0.9962),Translation=(X=0,Y=0,Z=2),Scale3D=(X=1,Y=1,Z=1))),(Item=(Type=Bone,Name=\"dyn_R_BackSkirt_3\"),Transform=(Rotation=(X=0,Y=-0.0872,Z=0,W=0.9962),Translation=(X=0,Y=0, ...(truncated)"
    WeightMinimum=0
    Weight=@n24.Result
   )
   > &RigUnit_ModifyTransforms(
    ItemToModify="((Item=(Type=Bone,Name=\"dyn_C_BackSkirt_4\"),Transform=(Rotation=(X=0,Y=0,Z=0,W=1),Translation=(X=0,Y=0,Z=0),Scale3D=(X=1,Y=1,Z=1))),(Item=(Type=Bone,Name=\"dyn_R_BackSkirt_4\"),Transform=(Rotation=(X=0,Y=0,Z=0,W=1),Translation=(X=0,Y=0,Z=0),Scale3D=(X=1,Y=1, ...(truncated)"
    WeightMinimum=0
    Weight=@n24.Result
   )
   n25 = &RigVMFunction_AlphaInterp(InterpSpeedDecreasing=10, InterpSpeedIncreasing=5, Scale=1, bClampResult=False, Value=RigVMVariableNode() {variable=tacticalSprinting}.Value)
   > &RigUnit_ModifyTransforms(
    ItemToModify="((Item=(Type=Bone,Name=\"dyn_C_BackSkirt_2\"),Transform=(Rotation=(X=0,Y=-0.1305,Z=0,W=0.9914),Translation=(X=2,Y=0,Z=4),Scale3D=(X=1,Y=1,Z=1))),(Item=(Type=Bone,Name=\"dyn_R_BackSkirt_2\"),Transform=(Rotation=(X=0,Y=-0.0436,Z=0,W=0.999),Translation=(X=2,Y=0,Z ...(truncated)"
    WeightMinimum=0
    Weight=@n25.Result
   )
   > &RigUnit_ModifyTransforms(
    ItemToModify="((Item=(Type=Bone,Name=\"dyn_C_BackSkirt_3\"),Transform=(Rotation=(X=0,Y=0.0436,Z=0,W=0.999),Translation=(X=0,Y=0,Z=3),Scale3D=(X=1,Y=1,Z=1))),(Item=(Type=Bone,Name=\"dyn_R_BackSkirt_3\"),Transform=(Rotation=(X=0,Y=0.0436,Z=0,W=0.999),Translation=(X=0,Y=0,Z=3) ...(truncated)"
    WeightMinimum=0
    Weight=@n25.Result
   )
   n26 = &RigVMFunction_AlphaInterp(
    InterpSpeedDecreasing=5
    InterpSpeedIncreasing=5
    Scale=1
    bClampResult=False
    Value=&RigVMFunction_MathDoubleMul(A=RigVMVariableNode() {variable=ADS}.Value, B=RigVMVariableNode() {variable=MoveFw}.Value).Result
   )
   > &RigUnit_ModifyTransforms(
    ItemToModify="((Item=(Type=Bone,Name=\"dyn_C_BackSkirt_2\"),Transform=(Rotation=(X=0,Y=0.1142,Z=0,W=-0.9935),Translation=(X=0,Y=0,Z=0),Scale3D=(X=1,Y=1,Z=1))),(Item=(Type=Bone,Name=\"dyn_L_BackSkirt_2\"),Transform=(Rotation=(X=0,Y=-0.1658,Z=0,W=0.9862),Translation=(X=0,Y=0, ...(truncated)"
    WeightMinimum=0
    Weight=@n26.Result
   )
   n27 = &RigVMFunction_AlphaInterp(
    InterpSpeedDecreasing=5
    InterpSpeedIncreasing=5
    Scale=1
    bClampResult=False
    Value=&RigVMFunction_MathDoubleMul(A=RigVMVariableNode() {variable=ADS}.Value, B=RigVMVariableNode() {variable=MoveBw}.Value).Result
   )
   > &RigUnit_ModifyTransforms(
    ItemToModify="((Item=(Type=Bone,Name=\"dyn_C_BackSkirt_2\"),Transform=(Rotation=(X=0,Y=0.1149,Z=0,W=-0.9934),Translation=(X=0,Y=0,Z=0),Scale3D=(X=1,Y=1,Z=1))),(Item=(Type=Bone,Name=\"dyn_L_BackSkirt_2\"),Transform=(Rotation=(X=0,Y=0.0058,Z=0,W=-1),Translation=(X=0,Y=0,Z=0), ...(truncated)"
    WeightMinimum=0
    Weight=@n27.Result
   )
   n28 = &RigVMFunction_AlphaInterp(
    InterpSpeedDecreasing=5
    InterpSpeedIncreasing=5
    Scale=1
    bClampResult=False
    Value=&RigVMFunction_MathDoubleMul(A=RigVMVariableNode() {variable=ADS}.Value, B=RigVMVariableNode() {variable=MoveLeft}.Value).Result
   )
   > &RigUnit_ModifyTransforms(
    ItemToModify="((Item=(Type=Bone,Name=\"dyn_C_BackSkirt_2\"),Transform=(Rotation=(X=-0.007,Y=0.1094,Z=-0.0632,W=-0.992),Translation=(X=0,Y=0,Z=0),Scale3D=(X=1,Y=1,Z=1))),(Item=(Type=Bone,Name=\"dyn_L_BackSkirt_2\"),Transform=(Rotation=(X=0,Y=-0.1696,Z=0,W=0.9855),Translation ...(truncated)"
    WeightMinimum=0
    Weight=@n28.Result
   )
   n29 = &RigVMFunction_AlphaInterp(
    InterpSpeedDecreasing=5
    InterpSpeedIncreasing=5
    Scale=1
    bClampResult=False
    Value=&RigVMFunction_MathDoubleMul(A=RigVMVariableNode() {variable=ADS}.Value, B=RigVMVariableNode() {variable=MoveRight}.Value).Result
   )
   > &RigUnit_ModifyTransforms(
    ItemToModify="((Item=(Type=Bone,Name=\"dyn_C_BackSkirt_2\"),Transform=(Rotation=(X=0,Y=0,Z=0,W=1),Translation=(X=0,Y=0,Z=0),Scale3D=(X=1,Y=1,Z=1))),(Item=(Type=Bone,Name=\"dyn_R_FrontSkirt_2\"),Transform=(Rotation=(X=0.0038,Y=-0.0435,Z=0.0871,W=0.9952),Translation=(X=0,Y=0, ...(truncated)"
    WeightMinimum=0
    Weight=@n29.Result
   )
   n30 = &RigVMFunction_AlphaInterp(InterpSpeedDecreasing=5, InterpSpeedIncreasing=5, Scale=1, bClampResult=False, Value=RigVMVariableNode() {variable=MoveBw}.Value)
   > &RigUnit_ModifyTransforms(
    ItemToModify="((Item=(Type=Bone,Name=\"dyn_C_BackSkirt_2\"),Transform=(Rotation=(X=0,Y=-0.1598,Z=0,W=-0.9872),Translation=(X=0,Y=0,Z=0),Scale3D=(X=1,Y=1,Z=1))),(Item=(Type=Bone,Name=\"dyn_L_BackSkirt_2\"),Transform=(Rotation=(X=0,Y=-0.1458,Z=0,W=-0.9893),Translation=(X=0,Y= ...(truncated)"
    WeightMinimum=0
    Weight=@n30.Result
   )
   > &RigUnit_ModifyTransforms(
    ItemToModify="((Item=(Type=Bone,Name=\"dyn_C_BackSkirt_4\"),Transform=(Rotation=(X=0,Y=-0.1816,Z=0,W=0.9834),Translation=(X=0,Y=0,Z=0),Scale3D=(X=1,Y=1,Z=1))),(Item=(Type=Bone,Name=\"dyn_L_BackSkirt_3\"),Transform=(Rotation=(X=0,Y=-0.2641,Z=0,W=0.9645),Translation=(X=0,Y=0, ...(truncated)"
    WeightMinimum=0
    Weight=@n30.Result
   )
   RigVMVariableNode() {variable=Sprinting}
  ]}
 New Function_ContainedGraph {nodes [
   RigVMFunctionEntryNode()
   > RigVMFunction_Sequence = RigVMAggregateNode()
   AlphaInterp = &RigVMFunction_AlphaInterp(InterpSpeedDecreasing=10, InterpSpeedIncreasing=10, Scale=1, bClampResult=False, Value=RigVMVariableNode() {variable=Is Crouching}.Value)
   &RigUnit_ModifyTransforms(
    ItemToModify="((Item=(Type=Bone,Name=\"dyn_C_FrontSkirt_1\"),Transform=(Rotation=(X=-0.1265,Y=-0.2335,Z=-0.0517,W=-0.9627),Translation=(X=0,Y=0,Z=0),Scale3D=(X=1,Y=1,Z=1))),(Item=(Type=Bone,Name=\"dyn_R_FrontSkirt_1\"),Transform=(Rotation=(X=-0.1332,Y=-0.0709,Z=-0.0302,W=-0 ...(truncated)"
    WeightMinimum=0
    ExecutePin=@RigVMFunction_Sequence.A
    Weight=@AlphaInterp.Result
   )
   > &RigUnit_ModifyTransforms(
    ItemToModify="((Item=(Type=Bone,Name=\"dyn_R_SideSkirt_4\"),Transform=(Rotation=(X=0.1567,Y=0.0742,Z=0.3744,W=-0.9109),Translation=(X=0,Y=0,Z=0),Scale3D=(X=1,Y=1,Z=1))),(Item=(Type=Bone,Name=\"dyn_R_BackSkirt_2\"),Transform=(Rotation=(X=0.0975,Y=-0.3139,Z=0.1388,W=-0.9342), ...(truncated)"
    WeightMinimum=0
    Weight=@AlphaInterp.Result
   )
   Multiply = &RigVMFunction_MathDoubleMul(A=RigVMVariableNode() {variable=Is Crouching}.Value, B=RigVMVariableNode() {variable=PelvisProx}.Value)
   n1 = &RigVMFunction_AlphaInterp(InterpSpeedDecreasing=5, InterpSpeedIncreasing=10, Scale=1, bClampResult=False, Value=@Multiply.Result)
   > &RigUnit_ModifyTransforms(
    ItemToModify="((Item=(Type=Bone,Name=\"dyn_C_FrontSkirt_1\"),Transform=(Rotation=(X=0,Y=-0.2964,Z=0,W=0.9551),Translation=(X=0,Y=0,Z=0),Scale3D=(X=1,Y=1,Z=1))),(Item=(Type=Bone,Name=\"dyn_R_FrontSkirt_2\"),Transform=(Rotation=(X=0.1646,Y=0.0471,Z=-0.1111,W=0.9789),Translati ...(truncated)"
    WeightMinimum=0
    Weight=@n1.Result
   )
   > &RigUnit_ModifyTransforms(
    ItemToModify="((Item=(Type=Bone,Name=\"None\"),Transform=(Rotation=(X=0,Y=0,Z=0,W=1),Translation=(X=0,Y=0,Z=0),Scale3D=(X=1,Y=1,Z=1))),(Item=(Type=Bone,Name=\"dyn_R_BackSkirt_3\"),Transform=(Rotation=(X=0.0003,Y=-0.1821,Z=-0.2264,W=-0.9568),Translation=(X=0,Y=0,Z=0),Scale3D ...(truncated)"
    WeightMinimum=0
    Weight=@n1.Result
   )
   RigVMFunctionReturnNode(ExecuteContext=@RigVMFunction_Sequence.B)
   RigVMFunction_MathBoolToFloat = &RigVMFunction_MathBoolToFloat(Value=RigVMVariableNode() {variable=OnlyCrouch}.Value)
   n2 = &RigVMFunction_AlphaInterp(InterpSpeedDecreasing=5, InterpSpeedIncreasing=5, Scale=1, bClampResult=False, Value=@RigVMFunction_MathBoolToFloat.Result)
   &RigUnit_ModifyTransforms(
    ItemToModify="((Item=(Type=Bone,Name=\"dyn_R_BackSkirt_3\"),Transform=(Rotation=(X=-0.1836,Y=0.1457,Z=0.0265,W=-0.9718),Translation=(X=-4.6421,Y=4.9278,Z=-1.8166),Scale3D=(X=1,Y=1,Z=1))),(Item=(Type=Bone,Name=\"dyn_C_BackSkirt_3\"),Transform=(Rotation=(X=-0.0056,Y=0.0413,Z= ...(truncated)"
    WeightMinimum=0
    Weight=@n2.Result
    ExecutePin=@RigVMFunction_Sequence.C
   )
   &RigUnit_ModifyTransforms(
    ExecutePin=()
    ItemToModify="((Item=(Type=Bone,Name=\"dyn_C_FrontSkirt_1\"),Transform=(Rotation=(X=0,Y=-0.2964,Z=0,W=0.9551),Translation=(X=0,Y=0,Z=0),Scale3D=(X=1,Y=1,Z=1))),(Item=(Type=Bone,Name=\"dyn_R_FrontSkirt_2\"),Transform=(Rotation=(X=0.1646,Y=0.0471,Z=-0.1111,W=0.9789),Translati ...(truncated)"
    Weight=1
    WeightMinimum=0
   )
   > &RigUnit_ModifyTransforms(
    ItemToModify="((Item=(Type=Bone,Name=\"None\"),Transform=(Rotation=(X=0,Y=0,Z=0,W=1),Translation=(X=0,Y=0,Z=0),Scale3D=(X=1,Y=1,Z=1))),(Item=(Type=Bone,Name=\"dyn_R_BackSkirt_3\"),Transform=(Rotation=(X=0.0003,Y=-0.1821,Z=-0.2264,W=-0.9568),Translation=(X=0,Y=0,Z=0),Scale3D ...(truncated)"
    Weight=1
    WeightMinimum=0
   )
   &RigUnit_ModifyTransforms(
    ExecutePin=()
    ItemToModify="((Item=(Type=Bone,Name=\"dyn_C_FrontSkirt_1\"),Transform=(Rotation=(X=-0.1035,Y=-0.2486,Z=0.0413,W=-0.9622),Translation=(X=0,Y=0,Z=0),Scale3D=(X=1,Y=1,Z=1))),(Item=(Type=Bone,Name=\"dyn_R_FrontSkirt_1\"),Transform=(Rotation=(X=-0.1332,Y=-0.0709,Z=-0.0302,W=-0. ...(truncated)"
    Weight=1
    WeightMinimum=0
   )
  ]}
 RigVMFunctionLibrary {nodes [
   Constraints = RigVMCollapseNode(ExecuteContext=())
   Crouch_Fix = RigVMCollapseNode(ExecuteContext=())
   Crouch_Fix_For_Backup = RigVMCollapseNode(ExecuteContext=())
   DeltaRotation = RigVMCollapseNode(
    ExecuteContext=()
    Rotation Order=ZYX
    child=(Type=Bone,Name="None")
    outputMinMax=(X=0,Y=1)
    parent=(Type=Bone,Name="None")
    premult=1
    rotationMinMax=(X=0,Y=0)
    twistAxis=(X=1,Y=0,Z=0)
   )
   Emote_Fix = RigVMCollapseNode(ExecuteContext=())
   JumpFix = RigVMCollapseNode(ExecuteContext=())
   Skirt_Fix_Backup = RigVMCollapseNode(ExecuteContext=())
   Skirt_Fix_Test = RigVMCollapseNode(ExecuteContext=())
   Skydive_Fix = RigVMCollapseNode(ExecuteContext=())
   Variables = RigVMCollapseNode(ExecuteContext=())
  ]}
 RigVMFunction_Sequence_1_ContainedGraph {nodes [
   n1 = RigVMFunctionEntryNode()
   RigVMFunction_Sequence = &RigVMFunction_Sequence(ExecuteContext=@n1.ExecuteContext)
   n2 = &RigVMFunction_Sequence(ExecuteContext=@RigVMFunction_Sequence.B)
   RigVMFunctionReturnNode(A=@RigVMFunction_Sequence.A, B=@n2.A, C=@n2.B)
  ]}
 RigVMFunction_Sequence_1_ContainedGraph {nodes [
   n1 = RigVMFunctionEntryNode()
   RigVMFunction_Sequence = &RigVMFunction_Sequence(ExecuteContext=@n1.ExecuteContext)
   n2 = &RigVMFunction_Sequence(ExecuteContext=@RigVMFunction_Sequence.B)
   RigVMFunctionReturnNode(A=@RigVMFunction_Sequence.A, B=@n2.A, C=@n2.B)
  ]}
 RigVMFunction_Sequence_1_ContainedGraph {nodes [
   n1 = RigVMFunctionEntryNode()
   RigVMFunction_Sequence = &RigVMFunction_Sequence(ExecuteContext=@n1.ExecuteContext)
   n2 = &RigVMFunction_Sequence(ExecuteContext=@RigVMFunction_Sequence.B)
   n3 = &RigVMFunction_Sequence(ExecuteContext=@n2.B)
   RigVMFunctionReturnNode(A=@RigVMFunction_Sequence.A, B=@n2.A, C=@n3.A, D=@n3.B)
  ]}
 RigVMFunction_Sequence_1_ContainedGraph {nodes [
   n1 = RigVMFunctionEntryNode()
   RigVMFunction_Sequence = &RigVMFunction_Sequence(ExecuteContext=@n1.ExecuteContext)
   n2 = &RigVMFunction_Sequence(ExecuteContext=@RigVMFunction_Sequence.B)
   n3 = &RigVMFunction_Sequence(ExecuteContext=@n2.B)
   RigVMFunctionReturnNode(A=@RigVMFunction_Sequence.A, B=@n2.A, C=@n3.A, D=@n3.B)
  ]}
 RigVMFunction_Sequence_1_ContainedGraph {nodes [
   n1 = RigVMFunctionEntryNode()
   RigVMFunction_Sequence = &RigVMFunction_Sequence(ExecuteContext=@n1.ExecuteContext)
   n2 = &RigVMFunction_Sequence(ExecuteContext=@RigVMFunction_Sequence.B)
   RigVMFunctionReturnNode(A=@RigVMFunction_Sequence.A, B=@n2.A, C=@n2.B)
  ]}
 RigVMFunction_Sequence_1_ContainedGraph {nodes [
   n1 = RigVMFunctionEntryNode()
   RigVMFunction_Sequence = &RigVMFunction_Sequence(ExecuteContext=@n1.ExecuteContext)
   n2 = &RigVMFunction_Sequence(ExecuteContext=@RigVMFunction_Sequence.B)
   n3 = &RigVMFunction_Sequence(ExecuteContext=@n2.B)
   RigVMFunctionReturnNode(A=@RigVMFunction_Sequence.A, B=@n2.A, C=@n3.A, D=@n3.B)
  ]}
 RigVMFunction_Sequence_1_ContainedGraph {nodes [
   n1 = RigVMFunctionEntryNode()
   RigVMFunction_Sequence = &RigVMFunction_Sequence(ExecuteContext=@n1.ExecuteContext)
   n2 = &RigVMFunction_Sequence(ExecuteContext=@RigVMFunction_Sequence.B)
   RigVMFunctionReturnNode(A=@RigVMFunction_Sequence.A, B=@n2.A, C=@n2.B)
  ]}
 RigVMModel {nodes [
   RigUnit_BeginExecution = &RigUnit_BeginExecution()
   FN_DeformRig_FNC = RigVMFunctionReferenceNode(
    DeformRig_ Curve="(deformRig_l_upperarm_up_alpha_72_#1=False,deformRig_r_upperarm_up_alpha_73_#2=False,deformRig_l_upperarm_fr_alpha_74_#3=False,deformRig_r_upperarm_fr_alpha_75_CB51B8 ...(truncated)"
    DeformRig_Bool="(upperarm_1_#4=True,elbow_3_#5=True,wrist_7_#6=True,groin_8_#7=True,glute_17_#8=True,knee_9_40C408324CA261 ...(truncated)"
    DeformRig_Debug=(upperarm_debug_18_#4=False,wrist_debug_20_#6=False,lat_debug_26_#9=False,groinGlute_debug_27_#7=False)
    DeformRig_Value="(upperarm_up_18_#10=(Rotation=(X=0.0,Y=-0.0,Z=0.0,W=1.0),Translation=(X=10,Y=0.0,Z=8.0),Scale3D=(X=0.70,Y=1.0,Z=1.0)),upperarm_fwd_fr_19_#11=(Rotation=(X=0.0,Y=-0.0,Z=0.0,W=1.0),Translation=(X=0.0,Y ...(truncated)"
    ExecuteContext=@RigUnit_BeginExecution.ExecutePin
   )
   > Constraints = RigVMFunctionReferenceNode()
   > Variables = RigVMFunctionReferenceNode()
   > Skirt_Fix_Test = RigVMFunctionReferenceNode()
   > Crouch_Fix = RigVMFunctionReferenceNode()
   > Skydive_Fix = RigVMFunctionReferenceNode()
   > Emote_Fix = RigVMFunctionReferenceNode()
   > JumpFix = RigVMFunctionReferenceNode()
  ]}
]
hierarchy [
 root {loc=<0 0 0>, rot=<0 0 0 1>}
 pelvis {parent=root, loc=<0 -0.5301 98.0571>, rot=<0.0127 0.707 -0.0127 -0.707>}
 spine_01 {parent=pelvis, loc=<6.1208 -0.9328 0>, rot=<0 0 0.034 -0.9994>}
 spine_02 {parent=spine_01, loc=<6.5997 -0.2095 0>, rot=<0 0 0.0238 -0.9997>}
 spine_03 {parent=spine_02, loc=<6.4914 -0.1237 0>, rot=<0 0 0.0149 -0.9999>}
 spine_04 {parent=spine_03, loc=<6.9598 0.0794 0>, rot=<0 0 -0.0132 -0.9999>}
 spine_05 {parent=spine_04, loc=<6.9987 0 0>, rot=<0 0 -0.0015 -1>}
 neck_01 {parent=spine_05, loc=<11.3138 0.1529 0>, rot=<0 0 -0.2003 -0.9797>}
 neck_02 {parent=neck_01, loc=<4.5627 0 0>, rot=<0 0 0.0504 -0.9987>}
 head {parent=neck_02, loc=<5.3914 0 0>, rot=<0 0 0.115 -0.9934>}
 FACIAL_C_FacialRoot {parent=head, loc=<2.1524 -0.746 -0.0002>, rot=<0.4879 -0.5118 -0.4879 -0.5118>}
 FACIAL_C_Forehead {parent=FACIAL_C_FacialRoot, loc=<-0.0004 -2.9029 8.8478>, rot=<-0.0473 0 0 -0.9989>}
 FACIAL_L_ForeheadIn {parent=FACIAL_C_FacialRoot, loc=<1.7175 -2.8842 8.6905>, rot=<-0.0645 -0.0696 0.004 -0.9955>}
 FACIAL_R_ForeheadIn {parent=FACIAL_C_FacialRoot, loc=<-1.7187 -2.885 8.6891>, rot=<-0.0645 0.0696 -0.004 -0.9955>}
 FACIAL_L_ForeheadMid {parent=FACIAL_C_FacialRoot, loc=<3.1727 -3.2084 8.1443>, rot=<-0.0748 -0.1755 -0.0212 -0.9814>}
 FACIAL_R_ForeheadMid {parent=FACIAL_C_FacialRoot, loc=<-3.1738 -3.2085 8.1421>, rot=<-0.0748 0.1756 0.0213 -0.9814>}
 FACIAL_L_ForeheadOut {parent=FACIAL_C_FacialRoot, loc=<4.4785 -3.2396 7.1821>, rot=<-0.125 -0.2553 -0.0801 -0.9554>}
 FACIAL_R_ForeheadOut {parent=FACIAL_C_FacialRoot, loc=<-4.4792 -3.2395 7.1802>, rot=<-0.125 0.2553 0.0802 -0.9554>}
 FACIAL_L_EyesackUpper {parent=FACIAL_C_FacialRoot, loc=<3.2288 -2.162 8.661>, rot=<0.045 -0.084 -0.0648 -0.9933>}
 FACIAL_R_EyesackUpper {parent=FACIAL_C_FacialRoot, loc=<-3.2287 -2.1626 8.6609>, rot=<0.045 0.084 0.0649 -0.9933>}
 FACIAL_L_EyelidUpperFurrow {parent=FACIAL_C_FacialRoot, loc=<3.1096 -1.9252 8.6029>, rot=<-0.0338 -0.1505 -0.0802 -0.9848>}
 FACIAL_R_EyelidUpperFurrow {parent=FACIAL_C_FacialRoot, loc=<-3.108 -1.9258 8.6028>, rot=<-0.0338 0.1505 0.0803 -0.9848>}
 FACIAL_L_EyelidUpperA {parent=FACIAL_C_FacialRoot, loc=<2.9144 -1.0308 6.7783>, rot=<-0.1516 0 0 -0.9884>}
 FACIAL_R_EyelidUpperA {parent=FACIAL_C_FacialRoot, loc=<-2.9127 -1.0307 6.7799>, rot=<-0.1516 0 0 -0.9884>}
 FACIAL_L_Eye {parent=FACIAL_C_FacialRoot, loc=<2.9187 -1.0471 6.7798>, rot=<-0.0038 0 0 -1>}
 FACIAL_R_Eye {parent=FACIAL_C_FacialRoot, loc=<-2.9195 -1.0457 6.7805>, rot=<-0.0039 0 0 -1>}
 FACIAL_L_EyelidLowerA {parent=FACIAL_C_FacialRoot, loc=<2.9144 -1.0308 6.7783>, rot=<0.2803 0 0 -0.9599>}
 FACIAL_R_EyelidLowerA {parent=FACIAL_C_FacialRoot, loc=<-2.9127 -1.0307 6.7799>, rot=<0.2803 0 0 -0.9599>}
 FACIAL_L_EyeCornerInner {parent=FACIAL_C_FacialRoot, loc=<1.6805 -0.1813 7.911>, rot=<-0.0038 0 0 -1>}
 FACIAL_R_EyeCornerInner {parent=FACIAL_C_FacialRoot, loc=<-1.6759 -0.1842 7.9114>, rot=<-0.0039 0 0 -1>}
 FACIAL_L_EyeCornerOuter {parent=FACIAL_C_FacialRoot, loc=<4.7173 -1.0107 7.2748>, rot=<-0.0036 -0.3826 -0.0014 -0.9239>}
 FACIAL_R_EyeCornerOuter {parent=FACIAL_C_FacialRoot, loc=<-4.717 -1.0086 7.2751>, rot=<-0.0036 0.3826 0.0015 -0.9239>}
 FACIAL_L_EyesackLower {parent=FACIAL_C_FacialRoot, loc=<3.4901 0.6163 7.3583>, rot=<-0.0261 -0.1396 -0.0228 -0.9896>}
 FACIAL_R_EyesackLower {parent=FACIAL_C_FacialRoot, loc=<-3.488 0.6164 7.3565>, rot=<-0.0261 0.1396 0.0228 -0.9896>}
 FACIAL_L_CheekInner {parent=FACIAL_C_FacialRoot, loc=<2.1329 1.3933 7.7015>, rot=<-0.1269 -0.0872 -0.1974 -0.9681>}
 FACIAL_R_CheekInner {parent=FACIAL_C_FacialRoot, loc=<-2.1312 1.3939 7.7011>, rot=<-0.127 0.0873 0.1975 -0.9681>}
 FACIAL_L_CheekOuter {parent=FACIAL_C_FacialRoot, loc=<3.9291 1.4151 6.0745>, rot=<0.0516 -0.3773 0.0989 -0.9193>}
 FACIAL_R_CheekOuter {parent=FACIAL_C_FacialRoot, loc=<-3.9286 1.4144 6.0714>, rot=<0.0517 0.3773 -0.0989 -0.9193>}
 FACIAL_L_NasolabialBulge {parent=FACIAL_C_FacialRoot, loc=<2.9975 4.7911 5.9525>, rot=<0.0543 -0.3774 0.0224 -0.9242>}
 FACIAL_R_NasolabialBulge {parent=FACIAL_C_FacialRoot, loc=<-3.0003 4.7906 5.9523>, rot=<0.0544 0.3774 -0.0223 -0.9242>}
 FACIAL_L_NasolabialFurrow {parent=FACIAL_C_FacialRoot, loc=<2.7123 3.4996 8.3734>, rot=<-0.0368 -0.1244 -0.3641 -0.9223>}
 FACIAL_R_NasolabialFurrow {parent=FACIAL_C_FacialRoot, loc=<-2.7147 3.4978 8.3721>, rot=<-0.0368 0.1244 0.3641 -0.9223>}
 FACIAL_L_CheekLower {parent=FACIAL_C_FacialRoot, loc=<3.3812 4.5601 5.2681>, rot=<0.0311 -0.4271 0.0561 -0.9019>}
 FACIAL_R_CheekLower {parent=FACIAL_C_FacialRoot, loc=<-3.3835 4.5592 5.2678>, rot=<0.0311 0.4271 -0.0561 -0.9019>}
 FACIAL_C_Nose {parent=FACIAL_C_FacialRoot, loc=<-0.0003 0.7256 9.2294>, rot=<-0.1342 0 0 -0.991>}
 FACIAL_C_NoseLower {parent=FACIAL_C_Nose, loc=<-0.0004 2.7878 -0.3296>, rot=<0.1305 0 0 -0.9914>}
 FACIAL_L_Nostril {parent=FACIAL_C_Nose, loc=<0.8913 2.117 -0.2739>, rot=<0.0173 0.1305 0.0023 -0.9913>}
 FACIAL_R_Nostril {parent=FACIAL_C_Nose, loc=<-0.8872 2.1162 -0.2781>, rot=<0.0173 -0.1305 -0.0023 -0.9913>}
 FACIAL_C_NoseTip {parent=FACIAL_C_Nose, loc=<-0.0005 2.0705 0.4832>, rot=<0.0872 0 0 -0.9962>}
 FACIAL_C_MouthUpper {parent=FACIAL_C_FacialRoot, loc=<-0.0011 4.9673 6.9775>, rot=<-0.0038 0 0 -1>}
 FACIAL_C_LipUpper {parent=FACIAL_C_MouthUpper, loc=<-0.0001 -0.0793 2.1071>, rot=<-0.0872 0 0 -0.9962>}
 FACIAL_L_LipUpper {parent=FACIAL_C_MouthUpper, loc=<0.8947 -0.0427 1.9461>, rot=<-0.0778 -0.1601 -0.0379 -0.9833>}
 FACIAL_R_LipUpper {parent=FACIAL_C_MouthUpper, loc=<-0.8962 -0.0434 1.9448>, rot=<-0.0778 0.1601 0.0379 -0.9833>}
 FACIAL_L_LipUpperOuter {parent=FACIAL_C_MouthUpper, loc=<1.6576 0.1034 1.3822>, rot=<-0.0984 -0.2293 -0.0983 -0.9634>}
 FACIAL_R_LipUpperOuter {parent=FACIAL_C_MouthUpper, loc=<-1.6594 0.104 1.3801>, rot=<-0.0984 0.2293 0.0983 -0.9634>}
 FACIAL_L_LipCorner {parent=FACIAL_C_MouthUpper, loc=<2.2568 0.4892 0.6414>, rot=<0 -0.2078 0 -0.9782>}
 FACIAL_R_LipCorner {parent=FACIAL_C_MouthUpper, loc=<-2.2586 0.4905 0.6457>, rot=<0 0.2078 0 -0.9782>}
 FACIAL_L_Masseter {parent=FACIAL_C_FacialRoot, loc=<5.9027 1.8783 3.8397>, rot=<0.0656 0.1129 0.0084 -0.9914>}
 FACIAL_R_Masseter {parent=FACIAL_C_FacialRoot, loc=<-5.903 1.8795 3.8386>, rot=<0.0655 -0.1128 -0.0083 -0.9914>}
 FACIAL_C_UnderChin {parent=FACIAL_C_FacialRoot, loc=<0.0008 8.8301 3.9004>, rot=<0.4965 0 0 -0.868>}
 FACIAL_L_UnderChin {parent=FACIAL_C_FacialRoot, loc=<2.6766 7.5869 2.7865>, rot=<0.3993 -0.3552 0.081 -0.8413>}
 FACIAL_R_UnderChin {parent=FACIAL_C_FacialRoot, loc=<-2.6767 7.587 2.786>, rot=<0.3993 0.3552 -0.0809 -0.8413>}
 FACIAL_C_TeethUpper {parent=FACIAL_C_FacialRoot, loc=<-0.0016 5.5783 8.2017>, rot=<-0.0038 0 0 -1>}
 FACIAL_C_LowerLipRotation {parent=FACIAL_C_FacialRoot, loc=<-0.001 2.7993 2.1237>, rot=<0.3384 0 0 -0.941>}
 FACIAL_C_MouthLower {parent=FACIAL_C_LowerLipRotation, loc=<-0.0001 -1.4201 5.1228>, rot=<-0.342 0 0 -0.9397>}
 FACIAL_C_LipLower {parent=FACIAL_C_MouthLower, loc=<0.0003 1.1669 1.6903>, rot=<0.0872 0 0 -0.9962>}
 FACIAL_L_LipLower {parent=FACIAL_C_MouthLower, loc=<0.8027 1.1237 1.5033>, rot=<0.0927 -0.1294 0.0272 -0.9869>}
 FACIAL_R_LipLower {parent=FACIAL_C_MouthLower, loc=<-0.8025 1.124 1.503>, rot=<0.0927 0.1294 -0.0272 -0.9869>}
 FACIAL_L_LipLowerOuter {parent=FACIAL_C_MouthLower, loc=<1.5655 0.8943 1.0393>, rot=<0.0913 -0.2321 0.0318 -0.9679>}
 FACIAL_R_LipLowerOuter {parent=FACIAL_C_MouthLower, loc=<-1.5675 0.8951 1.0385>, rot=<0.0913 0.2321 -0.0318 -0.9679>}
 FACIAL_C_TeethLower {parent=FACIAL_C_LowerLipRotation, loc=<-0.0005 -1.6435 6.4118>, rot=<-0.342 0 0 -0.9397>}
 FACIAL_C_Tongue1 {parent=FACIAL_C_TeethLower, loc=<0.0008 0.0498 -4.6653>, rot=<-0.2588 0 0 -0.9659>}
 FACIAL_C_Tongue2 {parent=FACIAL_C_Tongue1, loc=<-0.0002 0.3819 1.2933>, rot=<0.2588 0 0 -0.9659>}
 FACIAL_C_Tongue3 {parent=FACIAL_C_Tongue2, loc=<-0.0002 0.2002 1.23>, rot=<0.1305 0 0 -0.9914>}
 FACIAL_C_Tongue4 {parent=FACIAL_C_Tongue3, loc=<0.0003 -0.1599 1.1053>, rot=<0.0436 0 0 -0.999>}
 FACIAL_C_TongueUpper3 {parent=FACIAL_C_Tongue3, loc=<0.0007 -0.3119 -0.2015>, rot=<0 0 0 1>}
 FACIAL_C_TongueLower3 {parent=FACIAL_C_Tongue3, loc=<-0.0001 0.8149 0.0151>, rot=<0 0 0 1>}
 FACIAL_L_TongueSide3 {parent=FACIAL_C_Tongue3, loc=<1.4709 -0.0123 0.0833>, rot=<0 0 0 1>}
 FACIAL_R_TongueSide3 {parent=FACIAL_C_Tongue3, loc=<-1.4715 -0.0117 0.0825>, rot=<0 0 0 1>}
 FACIAL_C_TongueUpper2 {parent=FACIAL_C_Tongue2, loc=<0.0006 -0.5675 -0.2814>, rot=<0 0 0 1>}
 FACIAL_L_TongueSide2 {parent=FACIAL_C_Tongue2, loc=<1.7564 -0.0384 0.0732>, rot=<0 0 0 1>}
 FACIAL_R_TongueSide2 {parent=FACIAL_C_Tongue2, loc=<-1.7572 -0.0387 0.0722>, rot=<0 0 0 1>}
 FACIAL_C_Jaw {parent=FACIAL_C_FacialRoot, loc=<-0.001 2.7992 2.1237>, rot=<0.3384 0 0 -0.941>}
 FACIAL_C_Jawline {parent=FACIAL_C_Jaw, loc=<0.0005 2.2867 6.974>, rot=<0.2586 0 0 -0.966>}
 FACIAL_L_ChinSide {parent=FACIAL_C_Jaw, loc=<2.5182 0.3113 6.6084>, rot=<-0.2208 -0.3738 0.0499 -0.8995>}
 FACIAL_R_ChinSide {parent=FACIAL_C_Jaw, loc=<-2.516 0.3104 6.6088>, rot=<-0.2208 0.3738 -0.0499 -0.8995>}
 FACIAL_C_Chin {parent=FACIAL_C_Jaw, loc=<0.0005 0.9372 7.0751>, rot=<-0.2591 0 0 -0.9659>}
 FACIAL_L_Ear {parent=FACIAL_C_FacialRoot, loc=<6.1591 0.0106 0.5268>, rot=<-0.0301 -0.0006 -0.0261 -0.9992>}
 FACIAL_R_Ear {parent=FACIAL_C_FacialRoot, loc=<-6.1561 0.014 0.5274>, rot=<-0.0301 0.0006 0.0262 -0.9992>}
 FACIAL_L_Jawline {parent=FACIAL_C_FacialRoot, loc=<4.0159 5.7535 4.2761>, rot=<0.2056 0.158 -0.054 -0.9643>}
 FACIAL_R_Jawline {parent=FACIAL_C_FacialRoot, loc=<-4.0179 5.7543 4.2771>, rot=<0.2056 -0.1579 0.0541 -0.9643>}
 FACIAL_C_NoseBridge {parent=FACIAL_C_FacialRoot, loc=<-0.0001 -1.3511 9.5426>, rot=<0.04 0 0 -0.9992>}
 FACIAL_L_EyelidUpperB {parent=FACIAL_C_FacialRoot, loc=<2.9144 -1.0308 6.7783>, rot=<-0.2287 0 0 -0.9735>}
 FACIAL_L_EyelidLowerB {parent=FACIAL_C_FacialRoot, loc=<2.9144 -1.0308 6.7783>, rot=<0.3137 0 0 -0.9495>}
 FACIAL_R_EyelidLowerB {parent=FACIAL_C_FacialRoot, loc=<-2.9127 -1.0307 6.7799>, rot=<0.3219 0 0 -0.9468>}
 FACIAL_R_EyelidUpperB {parent=FACIAL_C_FacialRoot, loc=<-2.9127 -1.0307 6.7799>, rot=<-0.2287 0 0 -0.9735>}
 clavicle_l {parent=spine_05, loc=<5.4831 -0.1053 -3.2946>, rot=<0.0465 -0.7289 -0.009 -0.6829>}
 upperarm_l {parent=clavicle_l, loc=<9.8658 -1.2136 -0.4105>, rot=<0.0164 -0.3245 -0.0269 -0.9454>}
 lowerarm_l {parent=upperarm_l, loc=<26.465 -0.8345 0.259>, rot=<0.0036 0.0161 -0.2298 -0.9731>}
 hand_l {parent=lowerarm_l, loc=<22.7355 -0.3203 -0.6105>, rot=<-0.6784 -0.0223 0.0735 -0.7307>}
 thumb_01_l {parent=hand_l, loc=<1.3118 -0.3191 -2.459>, rot=<0.6546 -0.4042 0.0664 -0.6354>}
 thumb_02_l {parent=thumb_01_l, loc=<5.9336 -0.0448 0.0401>, rot=<-0.0037 0.0213 0.298 -0.9543>}
 thumb_03_l {parent=thumb_02_l, loc=<2.2223 0.3682 -0.0907>, rot=<0.0118 -0.013 0.0066 -0.9998>}
 index_metacarpal_l {parent=hand_l, loc=<1.6535 0.0183 -2.96>, rot=<0.0158 -0.0124 -0.0173 -0.9996>}
 index_01_l {parent=index_metacarpal_l, loc=<8.3008 0.0357 -0.0524>, rot=<-0.0044 0.0485 0.0924 -0.9945>}
 index_02_l {parent=index_01_l, loc=<3.3587 -0.3004 0.0452>, rot=<-0.002 0.0027 0.3225 -0.9466>}
 index_03_l {parent=index_02_l, loc=<2.5578 -0.123 -0.0252>, rot=<-0.002 0.0067 0.2558 -0.9667>}
 middle_metacarpal_l {parent=hand_l, loc=<2.1856 -0.1725 -0.5984>, rot=<0.006 0.0168 -0.0329 -0.9993>}
 middle_01_l {parent=middle_metacarpal_l, loc=<7.3275 -0.2447 0.0129>, rot=<-0.0208 0.0042 0.0858 -0.9961>}
 middle_02_l {parent=middle_01_l, loc=<4.0123 0.1401 0.1175>, rot=<-0.0218 0.0258 0.3512 -0.9357>}
 middle_03_l {parent=middle_02_l, loc=<2.6077 0.0534 0.1373>, rot=<0.0082 0.0188 0.1708 -0.9851>}
 ring_metacarpal_l {parent=hand_l, loc=<2.2074 -0.3625 1.3638>, rot=<-0.0284 0.0722 -0.0458 -0.9959>}
 ring_01_l {parent=ring_metacarpal_l, loc=<7.0295 -0.6197 -0.1446>, rot=<-0.0489 -0.0252 0.1892 -0.9804>}
 ring_02_l {parent=ring_01_l, loc=<3.6174 0.4647 -0.1027>, rot=<0 0 0.2806 -0.9598>}
 ring_03_l {parent=ring_02_l, loc=<2.4152 0.114 0.0199>, rot=<0 0 0.1189 -0.9929>}
 pinky_metacarpal_l {parent=hand_l, loc=<1.7616 -0.9753 3.3814>, rot=<-0.0397 0.0814 -0.069 -0.9935>}
 pinky_01_l {parent=pinky_metacarpal_l, loc=<7.1709 -0.4831 -0.1094>, rot=<-0.0654 -0.0035 0.2242 -0.9723>}
 pinky_02_l {parent=pinky_01_l, loc=<2.8274 0.0893 -0.1043>, rot=<0.0037 0.0011 0.2854 -0.9584>}
 pinky_03_l {parent=pinky_02_l, loc=<2.2046 0.0307 -0.0078>, rot=<-0.0002 0.0098 0.1445 -0.9895>}
 deform_wrist_dn_base_l {parent=hand_l, loc=<0 -1.5326 0>, rot=<0 0 0 -1>}
 deform_wrist_dn_l {parent=deform_wrist_dn_base_l, loc=<0 0 0>, rot=<0 0 0 1>}
 deform_wrist_up_base_l {parent=hand_l, loc=<0 2.3009 0>, rot=<0 0 0 -1>}
 deform_wrist_up_l {parent=deform_wrist_up_base_l, loc=<0 0 0>, rot=<0 0 0 1>}
 lowerarm_twist_02_l {parent=lowerarm_l, loc=<10.6707 0 0>, rot=<0 0 0 1>}
 lowerarm_twist_01_l {parent=lowerarm_l, loc=<18.316 0 0>, rot=<0 0 0 1>}
 deform_elbow_in_base_l {parent=lowerarm_l, loc=<0.1834 3.6062 -0.2572>, rot=<0 0 0 -1>}
 deform_elbow_in_l {parent=deform_elbow_in_base_l, loc=<0 0 0>, rot=<0 0 0 1>}
 deform_elbow_out_base_l {parent=lowerarm_l, loc=<0 -4 0>, rot=<0 0 0 1>}
 deform_elbow_out_l {parent=deform_elbow_out_base_l, loc=<0 0 0>, rot=<0 0 0 1>}
 upperarm_twist_01_l {parent=upperarm_l, loc=<7.2135 -0.1134 0>, rot=<0 0 0 1>}
 deform_upperarm_bk_base_l {parent=upperarm_twist_01_l, loc=<-6.4053 -4.3746 0.314>, rot=<0 0 0 -1>}
 deform_upperarm_bk_l {parent=deform_upperarm_bk_base_l, loc=<0 0 0>, rot=<0 0 0 1>}
 deform_upperarm_fr_base_l {parent=upperarm_twist_01_l, loc=<-5.6312 5.092 0.3354>, rot=<0 0 0 -1>}
 deform_upperarm_fr_l {parent=deform_upperarm_fr_base_l, loc=<0 0 0>, rot=<0 0 0 1>}
 deform_upperarm_up_base_l {parent=upperarm_twist_01_l, loc=<-8.891 0.4825 3.9517>, rot=<0 0 0 -1>}
 deform_upperarm_up_l {parent=deform_upperarm_up_base_l, loc=<0 0 0>, rot=<0 0 0 1>}
 upperarm_twist_02_l {parent=upperarm_l, loc=<19.7998 -0.1134 0>, rot=<0 0 0 1>}
 base_sleevescloth_l {parent=upperarm_l, loc=<0 0 0>, rot=<-0.0557 0.1428 0.028 -0.9878>}
 dyn_sleevescloth_l_02 {parent=base_sleevescloth_l, loc=<0 0 0>, rot=<0 0 0 1>}
 dyn_sleevescloth_l_03 {parent=dyn_sleevescloth_l_02, loc=<6.916 0.1724 -2.046>, rot=<0 0 0 1>}
 dyn_sleevescloth_l_04 {parent=dyn_sleevescloth_l_03, loc=<7.5703 0 0>, rot=<0 0 0 1>}
 upperarm_l_null {parent=clavicle_l, type=Null, loc=<9.8658 -1.2136 -0.4105>, rot=<0.0164 -0.3245 -0.0269 -0.9454>}
 base_ShoulderclPad_l_01 {parent=clavicle_l, loc=<3.6138 -0.6024 5.3538>, rot=<0.0047 -0.1188 -0.0248 -0.9926>}
 base_ShoulderclPad_l_02 {parent=base_ShoulderclPad_l_01, loc=<6.7217 -1.199 0.524>, rot=<0 0.0817 0 -0.9967>}
 dyn_ShoulderclPad_l_01 {parent=base_ShoulderclPad_l_02, loc=<0 0 0>, rot=<0 0 0 1>}
 clavicle_r {parent=spine_05, loc=<5.4831 -0.1053 3.2946>, rot=<0.7289 0.0465 -0.6829 0.009>}
 upperarm_r {parent=clavicle_r, loc=<-9.8658 1.2136 0.4105>, rot=<0.0164 -0.3245 -0.0269 -0.9454>}
 lowerarm_r {parent=upperarm_r, loc=<-26.465 0.8345 -0.259>, rot=<0.0036 0.0161 -0.2298 -0.9731>}
 hand_r {parent=lowerarm_r, loc=<-22.7355 0.3203 0.6106>, rot=<-0.6784 -0.0223 0.0735 -0.7307>}
 thumb_01_r {parent=hand_r, loc=<-1.3118 0.3191 2.459>, rot=<0.6546 -0.4042 0.0664 -0.6354>}
 thumb_02_r {parent=thumb_01_r, loc=<-5.9336 0.0448 -0.0401>, rot=<-0.0037 0.0213 0.298 -0.9543>}
 thumb_03_r {parent=thumb_02_r, loc=<-2.2223 -0.3682 0.0907>, rot=<0.0118 -0.013 0.0066 -0.9998>}
 index_metacarpal_r {parent=hand_r, loc=<-1.6535 -0.0183 2.96>, rot=<0.0158 -0.0124 -0.0173 -0.9996>}
 index_01_r {parent=index_metacarpal_r, loc=<-8.3008 -0.0357 0.0524>, rot=<-0.0044 0.0485 0.0924 -0.9945>}
 index_02_r {parent=index_01_r, loc=<-3.3587 0.3004 -0.0452>, rot=<-0.002 0.0027 0.3225 -0.9466>}
 index_03_r {parent=index_02_r, loc=<-2.5578 0.123 0.0252>, rot=<-0.002 0.0067 0.2558 -0.9667>}
 middle_metacarpal_r {parent=hand_r, loc=<-2.1856 0.1725 0.5984>, rot=<0.006 0.0168 -0.0329 -0.9993>}
 middle_01_r {parent=middle_metacarpal_r, loc=<-7.3275 0.2447 -0.0129>, rot=<-0.0208 0.0042 0.0858 -0.9961>}
 middle_02_r {parent=middle_01_r, loc=<-4.0123 -0.1401 -0.1175>, rot=<-0.0218 0.0258 0.3512 -0.9357>}
 middle_03_r {parent=middle_02_r, loc=<-2.6077 -0.0534 -0.1373>, rot=<0.0082 0.0188 0.1708 -0.9851>}
 ring_metacarpal_r {parent=hand_r, loc=<-2.2074 0.3625 -1.3638>, rot=<-0.0284 0.0722 -0.0458 -0.9959>}
 ring_01_r {parent=ring_metacarpal_r, loc=<-7.0295 0.6197 0.1446>, rot=<-0.0489 -0.0252 0.1892 -0.9804>}
 ring_02_r {parent=ring_01_r, loc=<-3.6174 -0.4647 0.1027>, rot=<0 0 0.2806 -0.9598>}
 ring_03_r {parent=ring_02_r, loc=<-2.4152 -0.114 -0.0199>, rot=<0 0 0.1189 -0.9929>}
 pinky_metacarpal_r {parent=hand_r, loc=<-1.7616 0.9753 -3.3814>, rot=<-0.0397 0.0814 -0.069 -0.9935>}
 pinky_01_r {parent=pinky_metacarpal_r, loc=<-7.1709 0.4831 0.1094>, rot=<-0.0654 -0.0035 0.2242 -0.9723>}
 pinky_02_r {parent=pinky_01_r, loc=<-2.8274 -0.0893 0.1043>, rot=<0.0037 0.0011 0.2854 -0.9584>}
 pinky_03_r {parent=pinky_02_r, loc=<-2.2046 -0.0307 0.0078>, rot=<-0.0002 0.0098 0.1445 -0.9895>}
 deform_wrist_dn_base_r {parent=hand_r, loc=<-0.0001 1.5328 0>, rot=<0 0 0 -1>}
 deform_wrist_dn_r {parent=deform_wrist_dn_base_r, loc=<0 0 0>, rot=<0 0 0 1>}
 deform_wrist_up_base_r {parent=hand_r, loc=<-0.0002 -2.3007 -0.0001>, rot=<0 0 0 -1>}
 deform_wrist_up_r {parent=deform_wrist_up_base_r, loc=<0 0 0>, rot=<0 0 0 1>}
 lowerarm_twist_02_r {parent=lowerarm_r, loc=<-10.6707 0 0>, rot=<0 0 0 -1>}
 lowerarm_twist_01_r {parent=lowerarm_r, loc=<-18.316 0 0>, rot=<0 0 0 -1>}
 deform_elbow_in_base_r {parent=lowerarm_r, loc=<-0.1834 -3.6062 0.2572>, rot=<0 0 0 -1>}
 deform_elbow_in_r {parent=deform_elbow_in_base_r, loc=<0 0 0>, rot=<0 0 0 1>}
 deform_elbow_out_base_r {parent=lowerarm_r, loc=<0 4 -0.0001>, rot=<0 0 0 -1>}
 deform_elbow_out_r {parent=deform_elbow_out_base_r, loc=<0 0 0>, rot=<0 0 0 1>}
 upperarm_twist_01_r {parent=upperarm_r, loc=<-7.2135 0.1134 0>, rot=<0 0 0 -1>}
 deform_upperarm_bk_base_r {parent=upperarm_twist_01_r, loc=<6.4053 4.3746 -0.314>, rot=<0 0 0 -1>}
 deform_upperarm_bk_r {parent=deform_upperarm_bk_base_r, loc=<0 0 0>, rot=<0 0 0 1>}
 deform_upperarm_fr_base_r {parent=upperarm_twist_01_r, loc=<5.6312 -5.092 -0.3354>, rot=<0 0 0 -1>}
 deform_upperarm_fr_r {parent=deform_upperarm_fr_base_r, loc=<0 0 0>, rot=<0 0 0 1>}
 deform_upperarm_up_base_r {parent=upperarm_twist_01_r, loc=<8.891 -0.4825 -3.9517>, rot=<0 0 0 -1>}
 deform_upperarm_up_r {parent=deform_upperarm_up_base_r, loc=<0 0 0>, rot=<0 0 0 1>}
 upperarm_twist_02_r {parent=upperarm_r, loc=<-19.7998 0.1134 0>, rot=<0 0 0 -1>}
 base_sleevescloth_r {parent=upperarm_r, loc=<0.0002 0 -0.0003>, rot=<-0.1428 -0.0557 0.9878 0.028>}
 dyn_sleevescloth_r_02 {parent=base_sleevescloth_r, loc=<0 0 0>, rot=<0 0 0 1>}
 dyn_sleevescloth_r_03 {parent=dyn_sleevescloth_r_02, loc=<6.9162 0.1723 2.0465>, rot=<0 0 0 1>}
 dyn_sleevescloth_r_04 {parent=dyn_sleevescloth_r_03, loc=<7.5703 0 0>, rot=<0 0 0 1>}
 upperarm_r_null {parent=clavicle_r, type=Null, loc=<-9.8658 1.2136 0.4105>, rot=<0.0164 -0.3245 -0.0269 -0.9454>}
 base_ShoulderclPad_r_01 {parent=clavicle_r, loc=<-3.4289 1.8119 -5.4457>, rot=<0.0251 -0.999 0.0026 0.0372>}
 base_ShoulderclPad_r_02 {parent=base_ShoulderclPad_r_01, loc=<6.9686 0 -0.6469>, rot=<0 0 0 1>}
 dyn_ShoulderclPad_r_01 {parent=base_ShoulderclPad_r_02, loc=<0 0 0>, rot=<0 0 0 1>}
 attach_backpack {parent=spine_05, loc=<-0.123 -13.473 0>, rot=<-0.0093 -0.707 0.0093 -0.707>}
 Backpack_BR {parent=attach_backpack, type=Null, loc=<0 -1 0>, rot=<-0.0436 0 0 0.999>}
 attach_cape {parent=spine_05, loc=<9.017 -11.714 0>, rot=<0.0616 -0.7044 -0.0616 -0.7044>}
 cape {parent=attach_cape, type=Null, loc=<0 0 0>, rot=<0 0 0 1>}
 base_shoulderCloth_l {parent=spine_05, loc=<1.8564 7.4509 -12.242>, rot=<-0.1279 -0.7769 -0.6062 -0.1117>}
 dyn_shoulderCloth_l_01 {parent=base_shoulderCloth_l, loc=<0 0 0>, rot=<0 0 0 1>}
 base_shoulderCloth_r {parent=spine_05, loc=<1.8569 7.451 12.242>, rot=<0.1117 0.6062 0.7769 0.1279>}
 dyn_shoulderCloth_r_01 {parent=base_shoulderCloth_r, loc=<0 0 0>, rot=<0 0 0 -1>}
 base_zipper {parent=spine_05, loc=<5.129 6.8209 0.0001>, rot=<-0.2073 -0.676 -0.676 -0.2073>}
 dyn_zipper_01 {parent=base_zipper, loc=<0 0 0>, rot=<0 0 0 1>}
 base_shoulderbackCloth_l_01 {parent=spine_05, loc=<1.6602 -9.4443 -9.3319>, rot=<0.0284 -0.7065 0.7065 -0.0284>}
 dyn_shoulderbackCloth_l_01 {parent=base_shoulderbackCloth_l_01, loc=<0 0 0>, rot=<0 0 0 1>}
 base_shoulderbackCloth_r_01 {parent=spine_05, loc=<1.6599 -9.4443 9.3319>, rot=<0.0284 -0.7065 0.7065 -0.0284>}
 dyn_shoulderbackCloth_r_01 {parent=base_shoulderbackCloth_r_01, loc=<0 0 0>, rot=<0 0 0 1>}
 deform_pec_aim_base_l {parent=spine_04, loc=<3.2456 9.3309 -1.2237>, rot=<0.0603 -0.4033 0.3318 -0.8507>}
 deform_pec_aim_l {parent=deform_pec_aim_base_l, loc=<0 0 0>, rot=<0 0 0 1>}
 deform_pec_base_l {parent=deform_pec_aim_l, loc=<9.3665 0 0>, rot=<0 0 0 1>}
 deform_pec_l {parent=deform_pec_base_l, loc=<0 0 0>, rot=<0 0 0 1>}
 deform_pec_aim_base_r {parent=spine_04, loc=<3.2452 9.3309 1.2237>, rot=<-0.4033 -0.0603 0.8507 0.3318>}
 deform_pec_aim_r {parent=deform_pec_aim_base_r, loc=<0 0 0>, rot=<0 0 0 1>}
 deform_pec_base_r {parent=deform_pec_aim_r, loc=<-9.3667 -0.0003 -0.0004>, rot=<0 0 0 1>}
 deform_pec_r {parent=deform_pec_base_r, loc=<0 0 0>, rot=<0 0 0 1>}
 thigh_l {parent=pelvis, loc=<-1.0855 -0.2109 -8.7053>, rot=<0.0009 0.0355 0.0155 -0.9993>}
 calf_l {parent=thigh_l, loc=<-42.5284 -0.0019 -0.7297>, rot=<0.0006 -0.0161 -0.0391 -0.9991>}
 foot_l {parent=calf_l, loc=<-46.7401 -2.5014 -0.592>, rot=<0.0427 -0.0238 0.0353 -0.9982>}
 ball_l {parent=foot_l, loc=<-3.5658 10.8066 -0.1544>, rot=<0.0001 -0.0001 -0.7063 -0.7079>}
 calf_twist_02_l {parent=calf_l, loc=<-19.1546 -1.7943 -0.2638>, rot=<0.0237 0.0095 -0.0035 -0.9997>}
 calf_twist_01_l {parent=calf_l, loc=<-35.8483 -1.9197 -0.5794>, rot=<0.0237 0.0095 -0.0035 -0.9997>}
 deform_knee_in_base_l {parent=calf_l, loc=<0.8031 -5.0231 0.2456>, rot=<0 0 0 -1>}
 deform_knee_in_l {parent=deform_knee_in_base_l, loc=<0 0 0>, rot=<0 0 0 1>}
 deform_knee_out_base_l {parent=calf_l, loc=<0 4 0>, rot=<0 0 0 -1>}
 deform_knee_out_l {parent=deform_knee_out_base_l, loc=<0 0 0>, rot=<0 0 0 1>}
 calf_l_null {parent=thigh_l, type=Null, loc=<-42.5284 -0.0019 -0.7297>, rot=<0.0006 -0.0161 -0.0391 -0.9991>}
 thigh_twist_01_l {parent=thigh_l, loc=<-27.952 0 0>, rot=<0.0013 0 0 -1>}
 deform_groin_base_l {parent=thigh_l, loc=<0.4801 5.6776 -0.2486>, rot=<0 0 0 -1>}
 deform_groin_l {parent=deform_groin_base_l, loc=<0 0 0>, rot=<0 0 0 1>}
 thigh_r {parent=pelvis, loc=<-1.0855 -0.2109 8.7053>, rot=<0.0355 -0.0009 0.9993 0.0155>}
 calf_r {parent=thigh_r, loc=<42.5284 0.0019 0.7297>, rot=<0.0006 -0.0161 -0.0391 -0.9991>}
 foot_r {parent=calf_r, loc=<46.7401 2.5014 0.592>, rot=<0.0427 -0.0238 0.0353 -0.9982>}
 ball_r {parent=foot_r, loc=<3.5658 -10.8066 0.1544>, rot=<0.0001 -0.0001 -0.7063 -0.7079>}
 calf_twist_02_r {parent=calf_r, loc=<19.1546 1.7943 0.2638>, rot=<0.0237 0.0095 -0.0035 -0.9997>}
 calf_twist_01_r {parent=calf_r, loc=<35.8483 1.9197 0.5794>, rot=<0.0237 0.0095 -0.0035 -0.9997>}
 deform_knee_in_base_r {parent=calf_r, loc=<-0.8031 5.0231 -0.2457>, rot=<0 0 0 -1>}
 deform_knee_in_r {parent=deform_knee_in_base_r, loc=<0 0 0>, rot=<0 0 0 1>}
 deform_knee_out_base_r {parent=calf_r, loc=<0 -4 0>, rot=<0 0 0 -1>}
 deform_knee_out_r {parent=deform_knee_out_base_r, loc=<0 0 0>, rot=<0 0 0 1>}
 calf_r_null {parent=thigh_r, type=Null, loc=<42.5284 0.0019 0.7297>, rot=<0.0006 -0.0161 -0.0391 -0.9991>}
 thigh_twist_01_r {parent=thigh_r, loc=<27.952 0 0>, rot=<0.0013 0 0 -1>}
 deform_groin_base_r {parent=thigh_r, loc=<-0.4801 -5.6776 0.2486>, rot=<0 0 0 -1>}
 deform_groin_r {parent=deform_groin_base_r, loc=<0 0 0>, rot=<0 0 0 1>}
 thigh_l_null {parent=pelvis, type=Null, loc=<-1.0855 -0.2109 -8.7053>, rot=<0.0009 0.0355 0.0155 -0.9993>}
 thigh_r_null {parent=pelvis, type=Null, loc=<-1.0855 -0.2109 8.7053>, rot=<0.0355 -0.0009 0.9993 0.0155>}
 deform_glute_base_l {parent=pelvis, loc=<-7.3466 -9.0213 -9.1245>, rot=<0.0172 0.031 -0.4477 -0.8935>}
 deform_glute_l {parent=deform_glute_base_l, loc=<0 0 0>, rot=<0 0 0 1>}
 deform_glute_base_r {parent=pelvis, loc=<-7.3464 -9.0213 9.1245>, rot=<0.031 -0.0172 0.8935 -0.4477>}
 deform_glute_r {parent=deform_glute_base_r, loc=<0 0 0>, rot=<0 0 0 1>}
 base_skull_l_old {parent=pelvis, loc=<6.6523 6.1927 -12.1709>, rot=<-0.0402 -0.921 -0.3823 -0.0637>}
 dyn_skull_l_old {parent=base_skull_l_old, loc=<0 0 0>, rot=<0 0 0 1>}
 base_L_FrontSkirt {parent=pelvis, loc=<6.5729 5.2603 -8.9922>, rot=<-0.0662 -0.8148 -0.5698 -0.0839>}
 dyn_L_FrontSkirt_1 {parent=base_L_FrontSkirt, loc=<0 0 0>, rot=<0 0 0 1>}
 base_L_SideSkirt {parent=pelvis, loc=<7.0661 -2.4659 -11.9863>, rot=<-0.0175 -0.978 0.0037 -0.2077>}
 dyn_L_SideSkirt_1 {parent=base_L_SideSkirt, loc=<0 0 0>, rot=<0 0 0 1>}
 base_L_BackSkirt {parent=pelvis, loc=<6.0893 -7.7765 -8.3178>, rot=<0.0586 -0.8146 0.5706 -0.0864>}
 dyn_L_BackSkirt_1 {parent=base_L_BackSkirt, loc=<0 0 0>, rot=<0 0 0 1>}
 dyn_L_BackSkirt_2 {parent=dyn_L_BackSkirt_1, loc=<16.3581 0 0>, rot=<0 -0.1021 0 -0.9948>}
 dyn_L_BackSkirt_3 {parent=dyn_L_BackSkirt_2, loc=<16.5741 0 0>, rot=<0 0 0 1>}
 base_C_BackSkirt {parent=pelvis, loc=<6.0278 -9.2667 0.0001>, rot=<0.0598 -0.7046 0.7046 -0.0598>}
 dyn_C_BackSkirt_1 {parent=base_C_BackSkirt, loc=<0 0 0>, rot=<0 0 0 1>}
 dyn_C_BackSkirt_2 {parent=dyn_C_BackSkirt_1, loc=<16.3581 0 0>, rot=<0 -0.0917 0 -0.9958>}
 dyn_C_BackSkirt_3 {parent=dyn_C_BackSkirt_2, loc=<16.5916 0 0>, rot=<0 0 0 1>}
 dyn_C_BackSkirt_4 {parent=dyn_C_BackSkirt_3, loc=<18.6728 0 0>, rot=<0 0 0 1>}
 dyn_C_BackSkirt_5 {parent=dyn_C_BackSkirt_4, loc=<17.7199 0 0>, rot=<0 0 0 1>}
 dyn_C_BackSkirt_6 {parent=dyn_C_BackSkirt_5, loc=<16.5914 0 0>, rot=<0 0 0 1>}
 base_R_BackSkirt {parent=pelvis, loc=<6.083 -7.4585 7.8062>, rot=<0.1033 -0.5725 0.8123 -0.0409>}
 dyn_R_BackSkirt_1 {parent=base_R_BackSkirt, loc=<0 0 0>, rot=<0 0 0 1>}
 dyn_R_BackSkirt_2 {parent=dyn_R_BackSkirt_1, loc=<16.3581 0 0>, rot=<0 -0.1039 0 -0.9946>}
 dyn_R_BackSkirt_3 {parent=dyn_R_BackSkirt_2, loc=<16.5916 0 0>, rot=<0 0 0 1>}
 dyn_R_BackSkirt_4 {parent=dyn_R_BackSkirt_3, loc=<18.6728 0 0>, rot=<0 0 0 1>}
 dyn_R_BackSkirt_5 {parent=dyn_R_BackSkirt_4, loc=<17.8444 0 0>, rot=<0 0 0 1>}
 base_R_BackSkirt_Charm {parent=dyn_R_BackSkirt_4, loc=<2.6435 -1.6564 -0.8879>, rot=<-0.041 -0.0219 -0.019 -0.9987>}
 dyn_R_BackSkirt_Charm_1 {parent=base_R_BackSkirt_Charm, loc=<0 0 0>, rot=<0 0 0 1>}
 dyn_R_BackSkirt_Charm_2 {parent=dyn_R_BackSkirt_Charm_1, loc=<9.3113 -0.2374 0.8388>, rot=<0 0 0 1>}
 base_R_SideSkirt {parent=pelvis, loc=<6.0548 -2.4296 12.0019>, rot=<0.1395 -0.0025 0.9901 0.0177>}
 dyn_R_SideSkirt_1 {parent=base_R_SideSkirt, loc=<0 0 0>, rot=<0 0 0 1>}
 dyn_R_SideSkirt_2 {parent=dyn_R_SideSkirt_1, loc=<16.3581 0 0>, rot=<0 -0.1073 0 -0.9942>}
 dyn_R_SideSkirt_3 {parent=dyn_R_SideSkirt_2, loc=<16.5916 0 0>, rot=<0 0 0 1>}
 dyn_R_SideSkirt_4 {parent=dyn_R_SideSkirt_3, loc=<18.6727 0 0>, rot=<0 0 0 1>}
 dyn_R_SideSkirt_5 {parent=dyn_R_SideSkirt_4, loc=<18.8445 0 0>, rot=<0 0 0 1>}
 dyn_R_SideSkirt_6 {parent=dyn_R_SideSkirt_5, loc=<16.5914 0 0>, rot=<0 0 0 1>}
 base_R_FrontSkirt {parent=pelvis, loc=<6.5814 3.6944 9.6019>, rot=<0.0765 0.5704 0.8155 0.061>}
 dyn_R_FrontSkirt_1 {parent=base_R_FrontSkirt, loc=<0 0 0>, rot=<0 0 0 1>}
 dyn_R_FrontSkirt_2 {parent=dyn_R_FrontSkirt_1, loc=<16.3581 0 0>, rot=<0 -0.0543 0 -0.9985>}
 dyn_R_FrontSkirt_3 {parent=dyn_R_FrontSkirt_2, loc=<15.9986 0 0>, rot=<0 -0.026 0 -0.9997>}
 base_C_FrontSkirt {parent=pelvis, loc=<6.739 8.0903 0.0001>, rot=<0.0127 0.707 0.707 0.0127>}
 dyn_C_FrontSkirt_1 {parent=base_C_FrontSkirt, loc=<0 0 0>, rot=<0 0 0 1>}
 dyn_C_FrontSkirt_2 {parent=dyn_C_FrontSkirt_1, loc=<16.3881 0 0>, rot=<0 0 0 1>}
 base_skull_l {parent=pelvis, loc=<7.9531 5.3014 -12.9217>, rot=<-0.0055 -0.9687 -0.237 -0.0733>}
 dyn_skull_l {parent=base_skull_l, loc=<0 0 0>, rot=<0 0 0 1>}
 pelvis_null {parent=root, type=Null, loc=<0 -0.5301 98.0571>, rot=<0.0127 0.707 -0.0127 -0.707>}
]
variables [
 ADS {type=double, default=0}
 CrouchBool {type=bool, default=False}
 EasterBloc_Emote {type=bool, default=False}
 Emote {type=bool, default=False}
 Emote_Celebration {type=bool, default=False}
 Is BaseballSliding {type=double, default=0}
 Is Crouching {type=double, default=0}
 Is KneeSliding {type=double, default=0}
 Is ParachuteOpen {type=double, default=0}
 Is SkyDiveDown {type=double, default=0}
 Is SkyDiving Float {type=double, default=0}
 Is Tacticaljump {type=double, default=0}
 IsJump {type=double, default=0}
 IsSkydiving {type=double, default=0}
 MoveBw {type=double, default=0}
 MoveFw {type=double, default=0}
 MoveLeft {type=double, default=0}
 MoveRight {type=double, default=0}
 OnlyCrouch {type=bool, default=False}
 PelvisProx {type=double, default=0}
 SPR L Leg Bkd {type=double, default=0}
 SPR L Leg Fwd {type=double, default=0}
 SPR L Leg Itw {type=double, default=0}
 SPR L Leg Otw {type=double, default=0}
 SPR L calf Bkd {type=double, default=0}
 SPR L calf Max Bkd {type=double, default=0}
 SPR R Leg Bkd {type=double, default=0}
 SPR R Leg Fwd {type=double, default=0}
 SPR R Leg Itw {type=double, default=0}
 SPR R Leg Otw {type=double, default=0}
 SPR R calf Bkd {type=double, default=0}
 SPR R calf Max Bkd {type=double, default=0}
 SkydiveBool {type=bool, default=False}
 Sprinting {type=double, default=0}
 Upperarm_L_Dwn {type=double, default=0}
 Upperarm_L_FW {type=double, default=0}
 Upperarm_R_Dwn {type=double, default=0}
 Upperarm_R_FW {type=double, default=0}
 idle {type=bool, default=False}
 idle standing {type=double, default=0}
 playing_emote {type=double, default=0}
 tacticalSprinting {type=double, default=0}
]
