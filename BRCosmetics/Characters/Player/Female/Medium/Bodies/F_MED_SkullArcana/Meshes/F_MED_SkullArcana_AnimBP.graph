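# Animation Blueprint for the F_MED_SkullArcana cosmetic body, adding secondary motion to skirt, sleeve cloth, charm and zipper bones.
# The pose copied from the attached parent mesh runs through a state-driven Control Rig into a cache.
# Differently tuned Anim Dynamics skirt chains branch from that cache, with a bool blend on skydiving.
# Sleeve, charm and zipper dynamics follow, with jump-driven sleeve offsets, then a post-sim Control Rig outputs the pose.
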AnimBlueprint /BRCosmetics/Characters/Player/Female/Medium/Bodies/F_MED_SkullArcana/Meshes/F_MED_SkullArcana_AnimBP
target_skeleton=&/F_MED_SkullArcana_Skeleton
graphs [
 AnimGraph {nodes [
   n1 = Anim Dynamics - Bone: dyn_C_BackSkirt_4!(ComponentPose=Local To Component(LocalPose=Use cached pose 'CR Rig'()), Alpha=1, GravityOverride:Vector="0,0,-980", ExternalForce="0,0,0") -> (Pose) {
    angularDampingOverride=0.9999
    angularSpringConstant=1000
    bAngularSpring=true
    bOverrideAngularDamping=true
    bUseGravityOverride=true
    gravityScale=0.35
    simulationSpace=RootRelative
    alphaBoolBlend {blendOption=Linear}
    boundBone {boneName=dyn_C_BackSkirt_4}
    componentAppliedLinearAccClamp=<3000 3000 3000>
    componentLinearAccScale=<0.5 0.5 0.5>
    componentLinearVelScale=<0.5 0.5 0.5>
    gravityOverride {z=-980}
    physicsBodyDefinitions [{
      boundBone {boneName=dyn_C_BackSkirt_4}
      boxExtents=<20 40 40>
      constraintSetup {angularLimitsMax=<0 -5 5>, angularLimitsMin=<0 -20 -5>, angularTarget=<1 0 0>}
      localJointOffset=<4 0 0>
     }]
   }
   n2 = Anim Dynamics - Bone: dyn_R_BackSkirt_4!(ComponentPose=@n1, Alpha=1, GravityOverride:Vector="0,0,-980", ExternalForce="0,0,0") -> (Pose) {
    angularDampingOverride=0.9999
    angularSpringConstant=1000
    bAngularSpring=true
    bOverrideAngularDamping=true
    bUseGravityOverride=true
    gravityScale=0.35
    simulationSpace=RootRelative
    alphaBoolBlend {blendOption=Linear}
    boundBone {boneName=dyn_R_BackSkirt_4}
    componentAppliedLinearAccClamp=<3000 3000 3000>
    componentLinearAccScale=<0.5 0.5 0.5>
    componentLinearVelScale=<0.5 0.5 0.5>
    gravityOverride {z=-980}
    physicsBodyDefinitions [{
      boundBone {boneName=dyn_R_BackSkirt_4}
      boxExtents=<20 40 40>
      constraintSetup {angularLimitsMax=<0 -5 5>, angularLimitsMin=<0 -20 -5>, angularTarget=<1 0 0>}
      localJointOffset=<4 0 0>
     }]
   }
   n3 = Anim Dynamics - Bone: dyn_C_BackSkirt_5!(ComponentPose=@n2, Alpha=1, GravityOverride:Vector="0,0,-980", ExternalForce="0,0,0") -> (Pose) {
    angularDampingOverride=0.9999
    angularSpringConstant=1000
    bAngularSpring=true
    bOverrideAngularDamping=true
    bUseGravityOverride=true
    gravityScale=0.35
    simulationSpace=RootRelative
    alphaBoolBlend {blendOption=Linear}
    boundBone {boneName=dyn_C_BackSkirt_5}
    componentAppliedLinearAccClamp=<3000 3000 3000>
    componentLinearAccScale=<0.5 0.5 0.5>
    componentLinearVelScale=<0.5 0.5 0.5>
    gravityOverride {z=-980}
    physicsBodyDefinitions [{
      boundBone {boneName=dyn_C_BackSkirt_5}
      boxExtents=<20 40 40>
      constraintSetup {angularLimitsMax=<0 -5 5>, angularLimitsMin=<0 -20 -5>, angularTarget=<1 0 0>}
      localJointOffset=<4 0 0>
     }]
   }
   n4 = Anim Dynamics - Bone: dyn_R_BackSkirt_5!(ComponentPose=@n3, Alpha=1, GravityOverride:Vector="0,0,-980", ExternalForce="0,0,0") -> (Pose) {
    angularDampingOverride=0.9999
    angularSpringConstant=1000
    bAngularSpring=true
    bOverrideAngularDamping=true
    bUseGravityOverride=true
    gravityScale=0.35
    simulationSpace=RootRelative
    alphaBoolBlend {blendOption=Linear}
    boundBone {boneName=dyn_R_BackSkirt_5}
    componentAppliedLinearAccClamp=<3000 3000 3000>
    componentLinearAccScale=<0.5 0.5 0.5>
    componentLinearVelScale=<0.5 0.5 0.5>
    gravityOverride {z=-980}
    physicsBodyDefinitions [{
      boundBone {boneName=dyn_R_BackSkirt_5}
      boxExtents=<20 40 40>
      constraintSetup {angularLimitsMax=<0 -5 5>, angularLimitsMin=<0 -20 -5>, angularTarget=<1 0 0>}
      localJointOffset=<4 0 0>
     }]
   }
   n5 = Anim Dynamics - Bone: dyn_R_SideSkirt_4!(ComponentPose=@n4, Alpha=1, GravityOverride:Vector="0,0,-980", ExternalForce="0,0,0") -> (Pose) {
    angularDampingOverride=0.9999
    angularSpringConstant=1000
    bAngularSpring=true
    bOverrideAngularDamping=true
    bUseGravityOverride=true
    gravityScale=0.35
    simulationSpace=RootRelative
    alphaBoolBlend {blendOption=Linear}
    boundBone {boneName=dyn_R_SideSkirt_4}
    componentAppliedLinearAccClamp=<3000 3000 3000>
    componentLinearAccScale=<0.5 0.5 0.5>
    componentLinearVelScale=<0.5 0.5 0.5>
    gravityOverride {z=-980}
    physicsBodyDefinitions [{
      boundBone {boneName=dyn_R_SideSkirt_4}
      boxExtents=<20 40 40>
      constraintSetup {angularLimitsMax=<0 -5 5>, angularLimitsMin=<0 -20 -5>, angularTarget=<1 0 0>}
      localJointOffset=<4 0 0>
     }]
   }
   n6 = Anim Dynamics - Bone: dyn_R_SideSkirt_5!(ComponentPose=@n5, Alpha=1, GravityOverride:Vector="0,0,-980", ExternalForce="0,0,0") -> (Pose) {
    angularDampingOverride=0.9999
    angularSpringConstant=1000
    bAngularSpring=true
    bOverrideAngularDamping=true
    bUseGravityOverride=true
    gravityScale=0.35
    simulationSpace=RootRelative
    alphaBoolBlend {blendOption=Linear}
    boundBone {boneName=dyn_R_SideSkirt_5}
    componentAppliedLinearAccClamp=<3000 3000 3000>
    componentLinearAccScale=<0.5 0.5 0.5>
    componentLinearVelScale=<0.5 0.5 0.5>
    gravityOverride {z=-980}
    physicsBodyDefinitions [{
      boundBone {boneName=dyn_R_SideSkirt_5}
      boxExtents=<20 40 40>
      constraintSetup {angularLimitsMax=<0 -5 5>, angularLimitsMin=<0 -20 -5>, angularTarget=<1 0 0>}
      localJointOffset=<4 0 0>
     }]
   }
   n7 = Anim Dynamics - Bone: dyn_L_BackSkirt_3!(ComponentPose=@n6, Alpha=1, GravityOverride:Vector="0,0,-980", ExternalForce="0,0,0") -> (Pose) {
    angularDampingOverride=0.9999
    angularSpringConstant=1000
    bAngularSpring=true
    bOverrideAngularDamping=true
    bUseGravityOverride=true
    gravityScale=0.35
    simulationSpace=RootRelative
    alphaBoolBlend {blendOption=Linear}
    boundBone {boneName=dyn_L_BackSkirt_3}
    componentAppliedLinearAccClamp=<3000 3000 3000>
    componentLinearAccScale=<0.5 0.5 0.5>
    componentLinearVelScale=<0.5 0.5 0.5>
    gravityOverride {z=-980}
    physicsBodyDefinitions [{
      boundBone {boneName=dyn_L_BackSkirt_3}
      boxExtents=<20 40 40>
      constraintSetup {angularLimitsMax=<0 -5 5>, angularLimitsMin=<0 -20 -5>, angularTarget=<1 0 0>}
      localJointOffset=<4 0 0>
     }]
   }
   n8 = Anim Dynamics - Bone: dyn_C_BackSkirt_3!(ComponentPose=@n7, Alpha=1, GravityOverride:Vector="0,0,-980", ExternalForce="0,0,0") -> (Pose) {
    angularDampingOverride=0.9999
    angularSpringConstant=1000
    bAngularSpring=true
    bOverrideAngularDamping=true
    bUseGravityOverride=true
    gravityScale=0.35
    simulationSpace=RootRelative
    alphaBoolBlend {blendOption=Linear}
    boundBone {boneName=dyn_C_BackSkirt_3}
    componentAppliedLinearAccClamp=<3000 3000 3000>
    componentLinearAccScale=<0.5 0.5 0.5>
    componentLinearVelScale=<0.5 0.5 0.5>
    gravityOverride {z=-980}
    physicsBodyDefinitions [{
      boundBone {boneName=dyn_C_BackSkirt_3}
      boxExtents=<20 40 40>
      constraintSetup {angularLimitsMax=<0 -5 5>, angularLimitsMin=<0 -20 -5>, angularTarget=<1 0 0>}
      localJointOffset=<4 0 0>
     }]
   }
   n9 = Anim Dynamics - Bone: dyn_R_BackSkirt_3!(ComponentPose=@n8, Alpha=1, GravityOverride:Vector="0,0,-980", ExternalForce="0,0,0") -> (Pose) {
    angularDampingOverride=0.9999
    angularSpringConstant=1000
    bAngularSpring=true
    bOverrideAngularDamping=true
    bUseGravityOverride=true
    gravityScale=0.35
    simulationSpace=RootRelative
    alphaBoolBlend {blendOption=Linear}
    boundBone {boneName=dyn_R_BackSkirt_3}
    componentAppliedLinearAccClamp=<3000 3000 3000>
    componentLinearAccScale=<0.5 0.5 0.5>
    componentLinearVelScale=<0.5 0.5 0.5>
    gravityOverride {z=-980}
    physicsBodyDefinitions [{
      boundBone {boneName=dyn_R_BackSkirt_3}
      boxExtents=<20 40 40>
      constraintSetup {angularLimitsMax=<0 -5 5>, angularLimitsMin=<0 -20 -5>, angularTarget=<1 0 0>}
      localJointOffset=<4 0 0>
     }]
   }
   n10 = Anim Dynamics - Bone: dyn_R_SideSkirt_3!(ComponentPose=@n9, Alpha=1, GravityOverride:Vector="0,0,-980", ExternalForce="0,0,0") -> (Pose) {
    angularDampingOverride=0.9999
    angularSpringConstant=1000
    bAngularSpring=true
    bOverrideAngularDamping=true
    bUseGravityOverride=true
    gravityScale=0.35
    simulationSpace=RootRelative
    alphaBoolBlend {blendOption=Linear}
    boundBone {boneName=dyn_R_SideSkirt_3}
    componentAppliedLinearAccClamp=<3000 3000 3000>
    componentLinearAccScale=<0.5 0.5 0.5>
    componentLinearVelScale=<0.5 0.5 0.5>
    gravityOverride {z=-980}
    physicsBodyDefinitions [{
      boundBone {boneName=dyn_R_SideSkirt_3}
      boxExtents=<20 40 40>
      constraintSetup {angularLimitsMax=<0 -5 5>, angularLimitsMin=<0 -20 -5>, angularTarget=<1 0 0>}
      localJointOffset=<4 0 0>
     }]
   }
   n11 = "Transform (Modify) Bone - Bone: dyn_R_SideSkirt_5"!(ComponentPose=@n10, Alpha=1, Rotation="0,0,0", Scale="1,1,1") -> (Pose) {rotationMode=BMM_Additive, alphaBoolBlend {blendOption=Linear}, boneToModify {boneName=dyn_R_SideSkirt_5}}
   n12 = "Transform (Modify) Bone - Bone: dyn_R_BackSkirt_5"!(
    ComponentPose="Transform (Modify) Bone - Bone: dyn_R_BackSkirt_4"!(
     ComponentPose="Transform (Modify) Bone - Bone: dyn_R_SideSkirt_4"!(ComponentPose=@n11, Alpha=1, Rotation="0,0,0", Scale="1,1,1") -> (Pose) {rotationMode=BMM_Additive, alphaBoolBlend {blendOption=Linear}, boneToModify {boneName=dyn_R_SideSkirt_4}}
     Alpha=1
     Rotation="0,0,0"
     Scale="1,1,1"
    ) -> (Pose) {rotationMode=BMM_Additive, alphaBoolBlend {blendOption=Linear}, boneToModify {boneName=dyn_R_BackSkirt_4}}
    Alpha=1
    Rotation="0,0,0"
    Scale="1,1,1"
   ) -> (Pose) {rotationMode=BMM_Additive, alphaBoolBlend {blendOption=Linear}, boneToModify {boneName=dyn_R_BackSkirt_5}}
   n13 = Component To Local(ComponentPose="Transform (Modify) Bone - Bone: dyn_C_BackSkirt_5"!(
     ComponentPose="Transform (Modify) Bone - Bone: dyn_C_BackSkirt_4"!(ComponentPose=@n12, Alpha=1, Rotation="0,0,0", Scale="1,1,1") -> (Pose) {rotationMode=BMM_Additive, alphaBoolBlend {blendOption=Linear}, boneToModify {boneName=dyn_C_BackSkirt_4}}
     Alpha=1
     Rotation="0,0,0"
     Scale="1,1,1"
    ) -> (Pose) {rotationMode=BMM_Additive, alphaBoolBlend {blendOption=Linear}, boneToModify {boneName=dyn_C_BackSkirt_5}})
   n14 = Use cached pose 'SimCache'()
   n15 = Get bIsSkyDiving()
   n16 = Anim Dynamics - Bone: dyn_sleevescloth_l_02!(
    ComponentPose=Local To Component(LocalPose=Blend Poses by bool(BlendPose_0=@n13, BlendPose_1=@n14, bActiveValue=@n15) {blendType=Linear})
    Alpha=1
    GravityOverride:Vector="0,0,-980"
    ExternalForce="0,0,0"
   ) -> (Pose) {
    angularBiasOverride=0.5
    angularDampingOverride=0.9999
    angularSpringConstant=500
    bAngularSpring=true
    bOverrideAngularBias=true
    bOverrideAngularDamping=true
    bUseGravityOverride=true
    gravityScale=0.35
    simulationSpace=RootRelative
    alphaBoolBlend {blendOption=Linear}
    boundBone {boneName=dyn_sleevescloth_l_02}
    componentAppliedLinearAccClamp=<5000 5000 5000>
    componentLinearAccScale=<1 1 1>
    componentLinearVelScale=<1 1 1>
    gravityOverride {z=-980}
    physicsBodyDefinitions [{
      boundBone {boneName=dyn_sleevescloth_l_02}
      boxExtents=<20 15 15>
      constraintSetup {angularLimitsMax=<0 0 0>, angularLimitsMin=<0 -20 0>, angularTarget=<1 0 0>}
      localJointOffset=<2 0 0>
     }]
   }
   n17 = Anim Dynamics - Bone: dyn_sleevescloth_l_03!(ComponentPose=@n16, Alpha=1, GravityOverride:Vector="0,0,-980", ExternalForce="0,0,0") -> (Pose) {
    angularDampingOverride=0.9999
    angularSpringConstant=500
    bAngularSpring=true
    bOverrideAngularDamping=true
    bUseGravityOverride=true
    gravityScale=0.35
    simulationSpace=RootRelative
    alphaBoolBlend {blendOption=Linear}
    boundBone {boneName=dyn_sleevescloth_l_03}
    componentAppliedLinearAccClamp=<5000 5000 5000>
    componentLinearAccScale=<1 1 1>
    componentLinearVelScale=<1 1 1>
    gravityOverride {z=-980}
    physicsBodyDefinitions [{
      boundBone {boneName=dyn_sleevescloth_l_03}
      boxExtents=<20 15 15>
      constraintSetup {angularLimitsMax=<0 0 0>, angularLimitsMin=<0 -20 0>, angularTarget=<1 0 0>}
      localJointOffset=<2 0 0>
     }]
   }
   n18 = Anim Dynamics - Bone: dyn_sleevescloth_r_02!(ComponentPose=@n17, Alpha=1, GravityOverride:Vector="0,0,-980", ExternalForce="0,0,0") -> (Pose) {
    angularBiasOverride=0.5
    angularDampingOverride=0.9999
    angularSpringConstant=500
    bAngularSpring=true
    bOverrideAngularBias=true
    bOverrideAngularDamping=true
    bUseGravityOverride=true
    gravityScale=0.35
    simulationSpace=RootRelative
    alphaBoolBlend {blendOption=Linear}
    boundBone {boneName=dyn_sleevescloth_r_02}
    componentAppliedLinearAccClamp=<5000 5000 5000>
    componentLinearAccScale=<1 1 1>
    componentLinearVelScale=<1 1 1>
    gravityOverride {z=-980}
    physicsBodyDefinitions [{
      boundBone {boneName=dyn_sleevescloth_r_02}
      boxExtents=<20 15 15>
      constraintSetup {angularLimitsMax=<0 20 0>, angularLimitsMin=<0 0 0>, angularTarget=<1 0 0>}
      localJointOffset=<2 0 0>
     }]
   }
   n19 = Anim Dynamics - Bone: dyn_sleevescloth_r_03!(ComponentPose=@n18, Alpha=1, GravityOverride:Vector="0,0,-980", ExternalForce="0,0,0") -> (Pose) {
    angularDampingOverride=0.9999
    angularSpringConstant=500
    bAngularSpring=true
    bOverrideAngularDamping=true
    bUseGravityOverride=true
    gravityScale=0.35
    simulationSpace=RootRelative
    alphaBoolBlend {blendOption=Linear}
    boundBone {boneName=dyn_sleevescloth_r_03}
    componentAppliedLinearAccClamp=<5000 5000 5000>
    componentLinearAccScale=<1 1 1>
    componentLinearVelScale=<1 1 1>
    gravityOverride {z=-980}
    physicsBodyDefinitions [{
      boundBone {boneName=dyn_sleevescloth_r_03}
      boxExtents=<20 15 15>
      constraintSetup {angularLimitsMax=<0 20 0>, angularLimitsMin=<0 0 0>, angularTarget=<1 0 0>}
      localJointOffset=<2 0 0>
     }]
   }
   n20 = Get IsJumping() -> (IsJumping:double)
   n21 = "Transform (Modify) Bone - Bone: base_sleevescloth_l"(ComponentPose=@n19, Alpha=@n20, Translation="0,0,0", Rotation="20,0,0", Scale="1,1,1") {
    rotationMode=BMM_Additive
    translationMode=BMM_Additive
    alphaBoolBlend {blendOption=Linear}
    alphaScaleBiasClamp {bInterpResult=true, interpSpeedIncreasing=5}
    boneToModify {boneName=base_sleevescloth_l}
    rotation {pitch=20}
   }
   n22 = "Transform (Modify) Bone - Bone: base_sleevescloth_r"(ComponentPose=@n21, Alpha=@n20, Translation="0,0,0", Rotation="-20,0,0", Scale="1,1,1") {
    rotationMode=BMM_Additive
    translationMode=BMM_Additive
    alphaBoolBlend {blendOption=Linear}
    alphaScaleBiasClamp {bInterpResult=true, interpSpeedIncreasing=5}
    boneToModify {boneName=base_sleevescloth_r}
    rotation {pitch=-20}
   }
   n23 = Anim Dynamics - Bone: dyn_R_BackSkirt_Charm_1(ComponentPose=@n22, Alpha=1, ExternalForce="0,0,0") {
    angularDampingOverride=0.9999
    angularSpringConstant=5000
    bAngularSpring=true
    bOverrideAngularDamping=true
    bUseGravityOverride=true
    gravityScale=0.35
    simulationSpace=RootRelative
    alphaBoolBlend {blendOption=Linear}
    boundBone {boneName=dyn_R_BackSkirt_Charm_1}
    componentAppliedLinearAccClamp=<3000 3000 3000>
    componentLinearAccScale=<1 1 1>
    componentLinearVelScale=<1 1 1>
    gravityOverride {z=-980}
    physicsBodyDefinitions [{
      boundBone {boneName=dyn_R_BackSkirt_Charm_1}
      boxExtents=<25 30 30>
      constraintSetup {angularLimitsMax=<0 -5 5>, angularLimitsMin=<0 -20 -5>, angularTarget=<1 0 0>}
      localJointOffset=<8 0 0>
     }]
   }
   n24 = Anim Dynamics - Bone: dyn_R_BackSkirt_Charm_2(ComponentPose=@n23, Alpha=1, ExternalForce="0,0,0") {
    angularDampingOverride=0.9999
    angularSpringConstant=5000
    bAngularSpring=true
    bOverrideAngularDamping=true
    bUseGravityOverride=true
    gravityScale=0.35
    simulationSpace=RootRelative
    alphaBoolBlend {blendOption=Linear}
    boundBone {boneName=dyn_R_BackSkirt_Charm_2}
    componentAppliedLinearAccClamp=<3000 3000 3000>
    componentLinearAccScale=<1 1 1>
    componentLinearVelScale=<1 1 1>
    gravityOverride {z=-980}
    physicsBodyDefinitions [{
      boundBone {boneName=dyn_R_BackSkirt_Charm_2}
      boxExtents=<25 30 30>
      constraintSetup {angularLimitsMax=<0 -5 5>, angularLimitsMin=<0 -20 -5>, angularTarget=<1 0 0>}
      localJointOffset=<8 0 0>
     }]
   }
   n25 = Anim Dynamics - Bone: dyn_zipper_01!(SimulationSpace:AnimPhysSimSpaceType Enum, ComponentPose=@n24, Alpha=1, ExternalForce="0,0,0") -> (Pose) {
    angularDampingOverride=1
    angularSpringConstant=20000
    bAngularSpring=true
    bOverrideAngularDamping=true
    bUseGravityOverride=true
    gravityScale=0.35
    simulationSpace=RootRelative
    alphaBoolBlend {blendOption=Linear}
    boundBone {boneName=dyn_zipper_01}
    componentAppliedLinearAccClamp=<2000 2000 2000>
    componentLinearAccScale=<5 5 5>
    componentLinearVelScale=<0.5 0.5 5>
    gravityOverride {z=-980}
    physicsBodyDefinitions [{
      boundBone {boneName=dyn_zipper_01}
      boxExtents=<20 30 30>
      constraintSetup {angularLimitsMax=<0 0 15>, angularLimitsMin=<0 -20 -15>, angularTarget=<1 0 0>}
      localJointOffset=<2 0 0>
     }]
   }
   n26 = Control Rig!(
    Source=Component To Local(ComponentPose=@n25)
    Alpha
    Is KneeSliding:double=Get isKneeSlidling float() -> (isKneeSlidling float:double)
    Is BaseballSliding:double=Get isBaseballSlidling float() -> (isBaseballSlidling float:double)
    Swiming:bool=Get bIsSwimSprinting() -> (bIsSwimSprinting:bool)
    SurfaceSwiming:bool=Get bIsSurfaceSwimming()
    Is Crouching:double=Get IsCrouching() -> (IsCrouching:double)
    MoveFw:double=Get moveForward() -> (moveForward:double)
    MoveBw:double=Get moveBackward() -> (moveBackward:double)
    MoveLeft:double="Get Move Left "() -> ("Move Left ":double)
    MoveRight:double=Get Move Right() -> (Move Right:double)
    OnlyCrouch:bool=Get OnlyCrouch() -> (OnlyCrouch:bool)
    PickaxeSwing:bool=Get bIsPlayingMeleeAnim()
    Targeting:bool=Get bIsTargeting()
    IN ZipLine:bool=Get OnZipline() -> (OnZipline:bool)
    FrontEnd:bool=Get bIsInFrontEnd()
   ) -> (Pose) {controlRigAssetReference {blueprintRigClass=&/F_MED_SkullArcana_PostSim_CtrlRig.F_MED_SkullArcana_PostSim_CtrlRig_C}}
   Output Pose(Result=@n26) {layerGroup=None}
   n27 = Control Rig!(
    Source=Copy Pose From Mesh() {bUseAttachedParent=true}
    Alpha
    Is Crouching=Get IsCrouching()
    Is SkyDiving Float:double=Get IsSkydiveFloat() -> (IsSkydiveFloat:double)
    idle:bool=Get isIdle_Bool() -> (isIdle_Bool:bool)
    Is ParachuteOpen:double=Get IsParachuteOpened() -> (IsParachuteOpened:double)
    MoveFw=Get moveForward()
    MoveBw=Get moveBackward()
    MoveLeft="Get Move Left "()
    MoveRight=Get Move Right()
    Is SkyDiveDown:double=Get IsSkydivingDown() -> (IsSkydivingDown:double)
    IsSkydiving:double=Get IsSkydive() -> (IsSkydive:double)
    playing_emote:double=0
    SkydiveBool:bool=Get bIsSkyDiving()
    Emote:bool=Get bIsPlayingEmote()
    OnlyCrouch=Get OnlyCrouch()
    Emote_Celebration:bool=Get Celebration Custom() -> (Celebration Custom:bool)
    EasterBloc_Emote:bool=Get IsEasterBlock() -> (IsEasterBlock:bool)
    tacticalSprinting:double=Get TacticalSprinting() -> (TacticalSprinting:double)
    Sprinting:double=Get Sprinting() -> (Sprinting:double)
    ADS:double=Get IsADS() -> (IsADS:double)
    CrouchBool:bool=Get bIsCrouching()
    IsJump:double=Get IsJumping()
   ) -> (Pose) {controlRigAssetReference {blueprintRigClass=&/F_MED_SkullArcana_CtrlRig.F_MED_SkullArcana_CtrlRig_C}}
   Save cached pose 'CR Rig'(Pose=@n27) {cacheName=CR Rig}
   n28 = Anim Dynamics - Bone: dyn_C_BackSkirt_2!(
    SimulationSpace:AnimPhysSimSpaceType Enum
    ComponentPose=Local To Component(LocalPose=Use cached pose 'CR Rig'())
    Alpha=1
    GravityOverride:Vector="0,0,-980"
    ExternalForce="0,0,0"
   ) -> (Pose) {
    angularDampingOverride=0.9999
    angularSpringConstant=2000
    bAngularSpring=true
    bOverrideAngularDamping=true
    bUseGravityOverride=true
    gravityScale=0.35
    simulationSpace=RootRelative
    alphaBoolBlend {blendOption=Linear}
    boundBone {boneName=dyn_C_BackSkirt_2}
    componentAppliedLinearAccClamp=<3000 3000 3000>
    componentLinearAccScale=<0.7 0.7 0.7>
    componentLinearVelScale=<0.7 0.7 0.7>
    gravityOverride {z=-980}
    physicsBodyDefinitions [{
      boundBone {boneName=dyn_C_BackSkirt_2}
      boxExtents=<30 40 40>
      constraintSetup {angularLimitsMax=<0 0 10>, angularLimitsMin=<0 -5 -10>, angularTarget=<1 0 0>}
      localJointOffset=<6 0 0>
     }]
   }
   n29 = Anim Dynamics - Bone: dyn_C_BackSkirt_3!(SimulationSpace:AnimPhysSimSpaceType Enum, ComponentPose=@n28, Alpha=1, GravityOverride="0,0,-980", ExternalForce="0,0,0") -> (Pose) {
    angularDampingOverride=0.9999
    angularSpringConstant=2000
    bAngularSpring=true
    bOverrideAngularDamping=true
    bUseGravityOverride=true
    gravityScale=0.35
    simulationSpace=RootRelative
    alphaBoolBlend {blendOption=Linear}
    boundBone {boneName=dyn_C_BackSkirt_3}
    componentAppliedLinearAccClamp=<3000 3000 3000>
    componentLinearAccScale=<0.7 0.7 0.7>
    componentLinearVelScale=<0.7 0.7 0.7>
    gravityOverride {z=-980}
    physicsBodyDefinitions [{
      boundBone {boneName=dyn_C_BackSkirt_3}
      boxExtents=<30 40 40>
      constraintSetup {angularLimitsMax=<0 0 10>, angularLimitsMin=<0 -10 -10>, angularTarget=<1 0 0>}
      localJointOffset=<6 0 0>
     }]
   }
   n30 = Anim Dynamics - Bone: dyn_C_BackSkirt_4!(SimulationSpace:AnimPhysSimSpaceType Enum, ComponentPose=@n29, Alpha=1, GravityOverride="0,0,-980", ExternalForce="0,0,0") -> (Pose) {
    angularDampingOverride=0.9999
    angularSpringConstant=2000
    bAngularSpring=true
    bOverrideAngularDamping=true
    bUseGravityOverride=true
    gravityScale=0.35
    simulationSpace=RootRelative
    alphaBoolBlend {blendOption=Linear}
    boundBone {boneName=dyn_C_BackSkirt_4}
    componentAppliedLinearAccClamp=<3000 3000 3000>
    componentLinearAccScale=<0.7 0.7 0.7>
    componentLinearVelScale=<0.7 0.7 0.7>
    gravityOverride {z=-980}
    physicsBodyDefinitions [{
      boundBone {boneName=dyn_C_BackSkirt_4}
      boxExtents=<30 40 40>
      constraintSetup {angularLimitsMax=<10 0 10>, angularLimitsMin=<-10 -30 -10>, angularTarget=<1 0 0>}
      localJointOffset=<6 0 0>
     }]
   }
   n31 = Anim Dynamics - Bone: dyn_C_BackSkirt_5!(SimulationSpace:AnimPhysSimSpaceType Enum, ComponentPose=@n30, Alpha=1, GravityOverride="0,0,-980", ExternalForce="0,0,0") -> (Pose) {
    angularDampingOverride=0.9999
    angularSpringConstant=2000
    bAngularSpring=true
    bOverrideAngularDamping=true
    bUseGravityOverride=true
    gravityScale=0.35
    simulationSpace=RootRelative
    alphaBoolBlend {blendOption=Linear}
    boundBone {boneName=dyn_C_BackSkirt_5}
    componentAppliedLinearAccClamp=<3000 3000 3000>
    componentLinearAccScale=<0.7 0.7 0.7>
    componentLinearVelScale=<0.7 0.7 0.7>
    gravityOverride {z=-980}
    physicsBodyDefinitions [{
      boundBone {boneName=dyn_C_BackSkirt_5}
      boxExtents=<30 40 40>
      constraintSetup {angularLimitsMax=<10 0 10>, angularLimitsMin=<-10 -30 -10>, angularTarget=<1 0 0>}
      localJointOffset=<6 0 0>
     }]
   }
   n32 = Anim Dynamics - Bone: dyn_R_BackSkirt_2!(SimulationSpace:AnimPhysSimSpaceType Enum, ComponentPose=@n31, Alpha=1, GravityOverride:Vector="0,0,-980", ExternalForce="0,0,0") -> (Pose) {
    angularDampingOverride=0.9999
    angularSpringConstant=2000
    bAngularSpring=true
    bOverrideAngularDamping=true
    bUseGravityOverride=true
    gravityScale=0.35
    simulationSpace=RootRelative
    alphaBoolBlend {blendOption=Linear}
    boundBone {boneName=dyn_R_BackSkirt_2}
    componentAppliedLinearAccClamp=<3000 3000 3000>
    componentLinearAccScale=<0.7 0.7 0.7>
    componentLinearVelScale=<0.7 0.7 0.7>
    gravityOverride {z=-980}
    physicsBodyDefinitions [{
      boundBone {boneName=dyn_R_BackSkirt_2}
      boxExtents=<30 40 40>
      constraintSetup {angularLimitsMax=<0 0 10>, angularLimitsMin=<0 -5 -10>, angularTarget=<1 0 0>}
      localJointOffset=<6 0 0>
     }]
   }
   n33 = Anim Dynamics - Bone: dyn_R_BackSkirt_3!(SimulationSpace:AnimPhysSimSpaceType Enum, ComponentPose=@n32, Alpha=1, GravityOverride="0,0,-980", ExternalForce="0,0,0") -> (Pose) {
    angularDampingOverride=0.9999
    angularSpringConstant=2000
    bAngularSpring=true
    bOverrideAngularDamping=true
    bUseGravityOverride=true
    gravityScale=0.35
    simulationSpace=RootRelative
    alphaBoolBlend {blendOption=Linear}
    boundBone {boneName=dyn_R_BackSkirt_3}
    componentAppliedLinearAccClamp=<3000 3000 3000>
    componentLinearAccScale=<0.7 0.7 0.7>
    componentLinearVelScale=<0.7 0.7 0.7>
    gravityOverride {z=-980}
    physicsBodyDefinitions [{
      boundBone {boneName=dyn_R_BackSkirt_3}
      boxExtents=<30 40 40>
      constraintSetup {angularLimitsMax=<0 0 10>, angularLimitsMin=<0 -10 -10>, angularTarget=<1 0 0>}
      localJointOffset=<6 0 0>
     }]
   }
   n34 = Anim Dynamics - Bone: dyn_R_BackSkirt_4!(SimulationSpace:AnimPhysSimSpaceType Enum, ComponentPose=@n33, Alpha=1, GravityOverride="0,0,-980", ExternalForce="0,0,0") -> (Pose) {
    angularDampingOverride=0.9999
    angularSpringConstant=2000
    bAngularSpring=true
    bOverrideAngularDamping=true
    bUseGravityOverride=true
    gravityScale=0.35
    simulationSpace=RootRelative
    alphaBoolBlend {blendOption=Linear}
    boundBone {boneName=dyn_R_BackSkirt_4}
    componentAppliedLinearAccClamp=<3000 3000 3000>
    componentLinearAccScale=<0.7 0.7 0.7>
    componentLinearVelScale=<0.7 0.7 0.7>
    gravityOverride {z=-980}
    physicsBodyDefinitions [{
      boundBone {boneName=dyn_R_BackSkirt_4}
      boxExtents=<30 40 40>
      constraintSetup {angularLimitsMax=<10 0 10>, angularLimitsMin=<-10 -30 -10>, angularTarget=<1 0 0>}
      localJointOffset=<6 0 0>
     }]
   }
   n35 = Anim Dynamics - Bone: dyn_R_BackSkirt_5!(SimulationSpace:AnimPhysSimSpaceType Enum, ComponentPose=@n34, Alpha=1, GravityOverride="0,0,-980", ExternalForce="0,0,0") -> (Pose) {
    angularDampingOverride=0.9999
    angularSpringConstant=2000
    bAngularSpring=true
    bOverrideAngularDamping=true
    bUseGravityOverride=true
    gravityScale=0.35
    simulationSpace=RootRelative
    alphaBoolBlend {blendOption=Linear}
    boundBone {boneName=dyn_R_BackSkirt_5}
    componentAppliedLinearAccClamp=<3000 3000 3000>
    componentLinearAccScale=<0.7 0.7 0.7>
    componentLinearVelScale=<0.7 0.7 0.7>
    gravityOverride {z=-980}
    physicsBodyDefinitions [{
      boundBone {boneName=dyn_R_BackSkirt_5}
      boxExtents=<30 40 40>
      constraintSetup {angularLimitsMax=<10 0 10>, angularLimitsMin=<-10 -30 -10>, angularTarget=<1 0 0>}
      localJointOffset=<6 0 0>
     }]
   }
   n36 = Anim Dynamics - Bone: dyn_L_BackSkirt_2!(SimulationSpace:AnimPhysSimSpaceType Enum, ComponentPose=@n35, Alpha=1, GravityOverride:Vector="0,0,-980", ExternalForce="0,0,0") -> (Pose) {
    angularDampingOverride=0.9999
    angularSpringConstant=2000
    bAngularSpring=true
    bOverrideAngularDamping=true
    bUseGravityOverride=true
    gravityScale=0.35
    simulationSpace=RootRelative
    alphaBoolBlend {blendOption=Linear}
    boundBone {boneName=dyn_L_BackSkirt_2}
    componentAppliedLinearAccClamp=<3000 3000 3000>
    componentLinearAccScale=<0.7 0.7 0.7>
    componentLinearVelScale=<0.7 0.7 0.7>
    gravityOverride {z=-980}
    physicsBodyDefinitions [{
      boundBone {boneName=dyn_L_BackSkirt_2}
      boxExtents=<30 40 40>
      constraintSetup {angularLimitsMax=<0 0 10>, angularLimitsMin=<0 -5 -10>, angularTarget=<1 0 0>}
      localJointOffset=<6 0 0>
     }]
   }
   n37 = Anim Dynamics - Bone: dyn_L_BackSkirt_3!(SimulationSpace:AnimPhysSimSpaceType Enum, ComponentPose=@n36, Alpha=1, GravityOverride="0,0,-980", ExternalForce="0,0,0") -> (Pose) {
    angularDampingOverride=0.9999
    angularSpringConstant=2000
    bAngularSpring=true
    bOverrideAngularDamping=true
    bUseGravityOverride=true
    gravityScale=0.35
    simulationSpace=RootRelative
    alphaBoolBlend {blendOption=Linear}
    boundBone {boneName=dyn_L_BackSkirt_3}
    componentAppliedLinearAccClamp=<3000 3000 3000>
    componentLinearAccScale=<0.7 0.7 0.7>
    componentLinearVelScale=<0.7 0.7 0.7>
    gravityOverride {z=-980}
    physicsBodyDefinitions [{
      boundBone {boneName=dyn_L_BackSkirt_3}
      boxExtents=<30 40 40>
      constraintSetup {angularLimitsMax=<0 0 10>, angularLimitsMin=<0 -10 -10>, angularTarget=<1 0 0>}
      localJointOffset=<6 0 0>
     }]
   }
   n38 = Anim Dynamics - Bone: dyn_L_SideSkirt_1!(SimulationSpace:AnimPhysSimSpaceType Enum, ComponentPose=@n37, Alpha=1, GravityOverride:Vector="0,0,-980", ExternalForce="0,0,0") -> (Pose) {
    angularDampingOverride=0.9999
    angularSpringConstant=2000
    bAngularSpring=true
    bOverrideAngularDamping=true
    bUseGravityOverride=true
    gravityScale=0.35
    simulationSpace=RootRelative
    boundBone {boneName=dyn_L_SideSkirt_1}
    componentAppliedLinearAccClamp=<3000 3000 3000>
    componentLinearAccScale=<0.7 0.7 0.7>
    componentLinearVelScale=<0.7 0.7 0.7>
    gravityOverride {z=-980}
    physicsBodyDefinitions [{
      boundBone {boneName=dyn_L_SideSkirt_1}
      boxExtents=<30 40 40>
      constraintSetup {angularLimitsMax=<0 0 10>, angularLimitsMin=<0 -20 -10>, angularTarget=<1 0 0>}
      localJointOffset=<6 0 0>
     }]
   }
   n39 = Anim Dynamics - Bone: dyn_L_FrontSkirt_1!(SimulationSpace:AnimPhysSimSpaceType Enum, ComponentPose=@n38, Alpha=1, GravityOverride:Vector="0,0,-980", ExternalForce="0,0,0") -> (Pose) {
    angularDampingOverride=0.9999
    angularSpringConstant=2000
    bAngularSpring=true
    bOverrideAngularDamping=true
    bUseGravityOverride=true
    gravityScale=0.35
    simulationSpace=RootRelative
    boundBone {boneName=dyn_L_FrontSkirt_1}
    componentAppliedLinearAccClamp=<3000 3000 3000>
    componentLinearAccScale=<0.7 0.7 0.7>
    componentLinearVelScale=<0.7 0.7 0.7>
    gravityOverride {z=-980}
    physicsBodyDefinitions [{
      boundBone {boneName=dyn_L_FrontSkirt_1}
      boxExtents=<30 40 40>
      constraintSetup {angularLimitsMax=<0 0 10>, angularLimitsMin=<0 -20 -10>, angularTarget=<1 0 0>}
      localJointOffset=<6 0 0>
     }]
   }
   n40 = Anim Dynamics - Bone: dyn_R_SideSkirt_2!(SimulationSpace:AnimPhysSimSpaceType Enum, ComponentPose=@n39, Alpha=1, GravityOverride:Vector="0,0,-980", ExternalForce="0,0,0") -> (Pose) {
    angularDampingOverride=0.9999
    angularSpringConstant=2000
    bAngularSpring=true
    bOverrideAngularDamping=true
    bUseGravityOverride=true
    gravityScale=0.35
    simulationSpace=RootRelative
    alphaBoolBlend {blendOption=Linear}
    boundBone {boneName=dyn_R_SideSkirt_2}
    componentAppliedLinearAccClamp=<3000 3000 3000>
    componentLinearAccScale=<0.7 0.7 0.7>
    componentLinearVelScale=<0.7 0.7 0.7>
    gravityOverride {z=-980}
    physicsBodyDefinitions [{
      boundBone {boneName=dyn_R_SideSkirt_2}
      boxExtents=<30 40 40>
      constraintSetup {angularLimitsMax=<0 0 10>, angularLimitsMin=<0 -5 -10>, angularTarget=<1 0 0>}
      localJointOffset=<6 0 0>
     }]
   }
   n41 = Anim Dynamics - Bone: dyn_R_SideSkirt_3!(SimulationSpace:AnimPhysSimSpaceType Enum, ComponentPose=@n40, Alpha=1, GravityOverride="0,0,-980", ExternalForce="0,0,0") -> (Pose) {
    angularDampingOverride=0.9999
    angularSpringConstant=2000
    bAngularSpring=true
    bOverrideAngularDamping=true
    bUseGravityOverride=true
    gravityScale=0.35
    simulationSpace=RootRelative
    alphaBoolBlend {blendOption=Linear}
    boundBone {boneName=dyn_R_SideSkirt_3}
    componentAppliedLinearAccClamp=<3000 3000 3000>
    componentLinearAccScale=<0.7 0.7 0.7>
    componentLinearVelScale=<0.7 0.7 0.7>
    gravityOverride {z=-980}
    physicsBodyDefinitions [{
      boundBone {boneName=dyn_R_SideSkirt_3}
      boxExtents=<30 40 40>
      constraintSetup {angularLimitsMax=<0 0 10>, angularLimitsMin=<0 -10 -10>, angularTarget=<1 0 0>}
      localJointOffset=<6 0 0>
     }]
   }
   n42 = Anim Dynamics - Bone: dyn_R_SideSkirt_4!(SimulationSpace:AnimPhysSimSpaceType Enum, ComponentPose=@n41, Alpha=1, GravityOverride="0,0,-980", ExternalForce="0,0,0") -> (Pose) {
    angularDampingOverride=0.9999
    angularSpringConstant=2000
    bAngularSpring=true
    bOverrideAngularDamping=true
    bUseGravityOverride=true
    gravityScale=0.35
    simulationSpace=RootRelative
    alphaBoolBlend {blendOption=Linear}
    boundBone {boneName=dyn_R_SideSkirt_4}
    componentAppliedLinearAccClamp=<3000 3000 3000>
    componentLinearAccScale=<0.7 0.7 0.7>
    componentLinearVelScale=<0.7 0.7 0.7>
    gravityOverride {z=-980}
    physicsBodyDefinitions [{
      boundBone {boneName=dyn_R_SideSkirt_4}
      boxExtents=<30 40 40>
      constraintSetup {angularLimitsMax=<10 0 10>, angularLimitsMin=<-10 -30 -10>, angularTarget=<1 0 0>}
      localJointOffset=<6 0 0>
     }]
   }
   n43 = Anim Dynamics - Bone: dyn_R_SideSkirt_5!(SimulationSpace:AnimPhysSimSpaceType Enum, ComponentPose=@n42, Alpha=1, GravityOverride="0,0,-980", ExternalForce="0,0,0") -> (Pose) {
    angularDampingOverride=0.9999
    angularSpringConstant=2000
    bAngularSpring=true
    bOverrideAngularDamping=true
    bUseGravityOverride=true
    gravityScale=0.35
    simulationSpace=RootRelative
    alphaBoolBlend {blendOption=Linear}
    boundBone {boneName=dyn_R_SideSkirt_5}
    componentAppliedLinearAccClamp=<3000 3000 3000>
    componentLinearAccScale=<0.7 0.7 0.7>
    componentLinearVelScale=<0.7 0.7 0.7>
    gravityOverride {z=-980}
    physicsBodyDefinitions [{
      boundBone {boneName=dyn_R_SideSkirt_5}
      boxExtents=<30 40 40>
      constraintSetup {angularLimitsMax=<10 0 10>, angularLimitsMin=<-10 -30 -10>, angularTarget=<1 0 0>}
      localJointOffset=<6 0 0>
     }]
   }
   n44 = Anim Dynamics - Bone: dyn_R_FrontSkirt_2!(SimulationSpace:AnimPhysSimSpaceType Enum, ComponentPose=@n43, Alpha=0, GravityOverride:Vector="0,0,-980", ExternalForce="0,0,0") -> (Pose) {
    angularDampingOverride=0.9999
    angularSpringConstant=2000
    bAngularSpring=true
    bOverrideAngularDamping=true
    bUseGravityOverride=true
    gravityScale=0.35
    simulationSpace=RootRelative
    alphaBoolBlend {blendOption=Linear}
    boundBone {boneName=dyn_R_FrontSkirt_2}
    componentAppliedLinearAccClamp=<3000 3000 3000>
    componentLinearAccScale=<1 1 1>
    componentLinearVelScale=<1 1 1>
    gravityOverride {z=-980}
    physicsBodyDefinitions [{
      boundBone {boneName=dyn_R_FrontSkirt_2}
      boxExtents=<30 40 40>
      constraintSetup {angularLimitsMax=<0 0 5>, angularLimitsMin=<0 -5 -5>, angularTarget=<1 0 0>}
      localJointOffset=<6 0 0>
     }]
   }
   n45 = Anim Dynamics - Bone: dyn_R_FrontSkirt_3!(SimulationSpace:AnimPhysSimSpaceType Enum, ComponentPose=@n44, Alpha=0, GravityOverride:Vector="0,0,-980", ExternalForce="0,0,0") -> (Pose) {
    angularDampingOverride=0.9999
    angularSpringConstant=200
    bAngularSpring=true
    bOverrideAngularDamping=true
    bUseGravityOverride=true
    gravityScale=0.35
    simulationSpace=RootRelative
    alphaBoolBlend {blendOption=Linear}
    boundBone {boneName=dyn_R_FrontSkirt_3}
    componentAppliedLinearAccClamp=<3000 3000 3000>
    componentLinearAccScale=<1 1 1>
    componentLinearVelScale=<1 1 1>
    gravityOverride {z=-980}
    physicsBodyDefinitions [{
      boundBone {boneName=dyn_R_FrontSkirt_3}
      boxExtents=<30 40 40>
      constraintSetup {angularLimitsMax=<0 0 5>, angularLimitsMin=<0 -10 -5>, angularTarget=<1 0 0>}
      localJointOffset=<6 0 0>
     }]
   }
   n46 = Anim Dynamics - Bone: dyn_C_FrontSkirt_2!(SimulationSpace:AnimPhysSimSpaceType Enum, ComponentPose=@n45, Alpha=0, GravityOverride:Vector="0,0,-980", ExternalForce="0,0,0") -> (Pose) {
    angularDampingOverride=0.9999
    angularSpringConstant=2000
    bAngularSpring=true
    bOverrideAngularDamping=true
    bUseGravityOverride=true
    gravityScale=0.35
    simulationSpace=RootRelative
    alphaBoolBlend {blendOption=Linear}
    boundBone {boneName=dyn_C_FrontSkirt_2}
    componentAppliedLinearAccClamp=<3000 3000 3000>
    componentLinearAccScale=<1 1 1>
    componentLinearVelScale=<1 1 1>
    gravityOverride {z=-980}
    physicsBodyDefinitions [{
      boundBone {boneName=dyn_C_FrontSkirt_2}
      boxExtents=<30 40 40>
      constraintSetup {angularLimitsMax=<0 0 10>, angularLimitsMin=<0 -5 -10>, angularTarget=<1 0 0>}
      localJointOffset=<6 0 0>
     }]
   }
   n47 = Anim Dynamics - Bone: dyn_C_BackSkirt_3!(ComponentPose=Local To Component(LocalPose=Use cached pose 'CR Rig'()), Alpha=1, GravityOverride="0,0,-980", ExternalForce="0,0,0") -> (Pose) {
    angularDampingOverride=1
    angularSpringConstant=8000
    bAngularSpring=true
    bOverrideAngularDamping=true
    bUseGravityOverride=true
    gravityScale=0.35
    simulationSpace=RootRelative
    alphaBoolBlend {blendOption=Linear}
    boundBone {boneName=dyn_C_BackSkirt_3}
    componentAppliedLinearAccClamp=<3000 3000 3000>
    componentLinearAccScale=<0.2 0.2 0.2>
    componentLinearVelScale=<0.2 0.2 0.2>
    gravityOverride {z=-980}
    physicsBodyDefinitions [{
      boundBone {boneName=dyn_C_BackSkirt_3}
      boxExtents=<30 40 40>
      constraintSetup {angularLimitsMax=<0 -2 5>, angularLimitsMin=<0 -10 -5>, angularTarget=<1 0 0>}
      localJointOffset=<6 0 0>
     }]
   }
   n48 = Anim Dynamics - Bone: dyn_C_BackSkirt_4!(ComponentPose=@n47, Alpha=1, GravityOverride="0,0,-980", ExternalForce="0,0,0") -> (Pose) {
    angularDampingOverride=1
    angularSpringConstant=8000
    bAngularSpring=true
    bOverrideAngularDamping=true
    bUseGravityOverride=true
    gravityScale=0.35
    simulationSpace=RootRelative
    alphaBoolBlend {blendOption=Linear}
    boundBone {boneName=dyn_C_BackSkirt_4}
    componentAppliedLinearAccClamp=<3000 3000 3000>
    componentLinearAccScale=<0.2 0.2 0.2>
    componentLinearVelScale=<0.2 0.2 0.2>
    gravityOverride {z=-980}
    physicsBodyDefinitions [{
      boundBone {boneName=dyn_C_BackSkirt_4}
      boxExtents=<30 40 40>
      constraintSetup {angularLimitsMax=<0 -2 5>, angularLimitsMin=<0 -30 -5>, angularTarget=<1 0 0>}
      localJointOffset=<6 0 0>
     }]
   }
   n49 = Anim Dynamics - Bone: dyn_C_BackSkirt_5!(ComponentPose=@n48, Alpha=1, GravityOverride="0,0,-980", ExternalForce="0,0,0") -> (Pose) {
    angularDampingOverride=1
    angularSpringConstant=8000
    bAngularSpring=true
    bOverrideAngularDamping=true
    bUseGravityOverride=true
    gravityScale=0.35
    simulationSpace=RootRelative
    alphaBoolBlend {blendOption=Linear}
    boundBone {boneName=dyn_C_BackSkirt_5}
    componentAppliedLinearAccClamp=<3000 3000 3000>
    componentLinearAccScale=<0.2 0.2 0.2>
    componentLinearVelScale=<0.2 0.2 0.2>
    gravityOverride {z=-980}
    physicsBodyDefinitions [{
      boundBone {boneName=dyn_C_BackSkirt_5}
      boxExtents=<30 40 40>
      constraintSetup {angularLimitsMax=<0 -2 5>, angularLimitsMin=<0 -30 -5>, angularTarget=<1 0 0>}
      localJointOffset=<6 0 0>
     }]
   }
   n50 = Anim Dynamics - Bone: dyn_R_BackSkirt_3!(ComponentPose=@n49, Alpha=1, GravityOverride="0,0,-980", ExternalForce="0,0,0") -> (Pose) {
    angularDampingOverride=1
    angularSpringConstant=8000
    bAngularSpring=true
    bOverrideAngularDamping=true
    bUseGravityOverride=true
    gravityScale=0.35
    simulationSpace=RootRelative
    alphaBoolBlend {blendOption=Linear}
    boundBone {boneName=dyn_R_BackSkirt_3}
    componentAppliedLinearAccClamp=<3000 3000 3000>
    componentLinearAccScale=<0.2 0.2 0.2>
    componentLinearVelScale=<0.2 0.2 0.2>
    gravityOverride {z=-980}
    physicsBodyDefinitions [{
      boundBone {boneName=dyn_R_BackSkirt_3}
      boxExtents=<30 40 40>
      constraintSetup {angularLimitsMax=<0 -2 5>, angularLimitsMin=<0 -10 -5>, angularTarget=<1 0 0>}
      localJointOffset=<6 0 0>
     }]
   }
   n51 = Anim Dynamics - Bone: dyn_R_BackSkirt_4!(ComponentPose=@n50, Alpha=1, GravityOverride="0,0,-980", ExternalForce="0,0,0") -> (Pose) {
    angularDampingOverride=1
    angularSpringConstant=8000
    bAngularSpring=true
    bOverrideAngularDamping=true
    bUseGravityOverride=true
    gravityScale=0.35
    simulationSpace=RootRelative
    alphaBoolBlend {blendOption=Linear}
    boundBone {boneName=dyn_R_BackSkirt_4}
    componentAppliedLinearAccClamp=<3000 3000 3000>
    componentLinearAccScale=<0.2 0.2 0.2>
    componentLinearVelScale=<0.2 0.2 0.2>
    gravityOverride {z=-980}
    physicsBodyDefinitions [{
      boundBone {boneName=dyn_R_BackSkirt_4}
      boxExtents=<30 40 40>
      constraintSetup {angularLimitsMax=<0 -2 5>, angularLimitsMin=<0 -30 -5>, angularTarget=<1 0 0>}
      localJointOffset=<6 0 0>
     }]
   }
   n52 = Anim Dynamics - Bone: dyn_R_BackSkirt_5!(ComponentPose=@n51, Alpha=1, GravityOverride="0,0,-980", ExternalForce="0,0,0") -> (Pose) {
    angularDampingOverride=1
    angularSpringConstant=8000
    bAngularSpring=true
    bOverrideAngularDamping=true
    bUseGravityOverride=true
    gravityScale=0.35
    simulationSpace=RootRelative
    alphaBoolBlend {blendOption=Linear}
    boundBone {boneName=dyn_R_BackSkirt_5}
    componentAppliedLinearAccClamp=<3000 3000 3000>
    componentLinearAccScale=<0.2 0.2 0.2>
    componentLinearVelScale=<0.2 0.2 0.2>
    gravityOverride {z=-980}
    physicsBodyDefinitions [{
      boundBone {boneName=dyn_R_BackSkirt_5}
      boxExtents=<30 40 40>
      constraintSetup {angularLimitsMax=<0 -2 5>, angularLimitsMin=<0 -30 -5>, angularTarget=<1 0 0>}
      localJointOffset=<6 0 0>
     }]
   }
   n53 = Anim Dynamics - Bone: dyn_R_SideSkirt_3!(ComponentPose=@n52, Alpha=1, GravityOverride="0,0,-980", ExternalForce="0,0,0") -> (Pose) {
    angularDampingOverride=1
    angularSpringConstant=8000
    bAngularSpring=true
    bOverrideAngularDamping=true
    bUseGravityOverride=true
    gravityScale=0.35
    simulationSpace=RootRelative
    alphaBoolBlend {blendOption=Linear}
    boundBone {boneName=dyn_R_SideSkirt_3}
    componentAppliedLinearAccClamp=<3000 3000 3000>
    componentLinearAccScale=<0.2 0.2 0.2>
    componentLinearVelScale=<0.2 0.2 0.2>
    gravityOverride {z=-980}
    physicsBodyDefinitions [{
      boundBone {boneName=dyn_R_SideSkirt_3}
      boxExtents=<30 40 40>
      constraintSetup {angularLimitsMax=<0 -2 5>, angularLimitsMin=<0 -10 -5>, angularTarget=<1 0 0>}
      localJointOffset=<6 0 0>
     }]
   }
   n54 = Anim Dynamics - Bone: dyn_R_SideSkirt_4!(ComponentPose=@n53, Alpha=1, GravityOverride="0,0,-980", ExternalForce="0,0,0") -> (Pose) {
    angularDampingOverride=1
    angularSpringConstant=8000
    bAngularSpring=true
    bOverrideAngularDamping=true
    bUseGravityOverride=true
    gravityScale=0.35
    simulationSpace=RootRelative
    alphaBoolBlend {blendOption=Linear}
    boundBone {boneName=dyn_R_SideSkirt_4}
    componentAppliedLinearAccClamp=<3000 3000 3000>
    componentLinearAccScale=<0.2 0.2 0.2>
    componentLinearVelScale=<0.2 0.2 0.2>
    gravityOverride {z=-980}
    physicsBodyDefinitions [{
      boundBone {boneName=dyn_R_SideSkirt_4}
      boxExtents=<30 40 40>
      constraintSetup {angularLimitsMax=<0 -2 5>, angularLimitsMin=<0 -30 -5>, angularTarget=<1 0 0>}
      localJointOffset=<6 0 0>
     }]
   }
   n55 = Anim Dynamics - Bone: dyn_R_SideSkirt_5!(ComponentPose=@n54, Alpha=1, GravityOverride="0,0,-980", ExternalForce="0,0,0") -> (Pose) {
    angularDampingOverride=1
    angularSpringConstant=8000
    bAngularSpring=true
    bOverrideAngularDamping=true
    bUseGravityOverride=true
    gravityScale=0.35
    simulationSpace=RootRelative
    alphaBoolBlend {blendOption=Linear}
    boundBone {boneName=dyn_R_SideSkirt_5}
    componentAppliedLinearAccClamp=<3000 3000 3000>
    componentLinearAccScale=<0.2 0.2 0.2>
    componentLinearVelScale=<0.2 0.2 0.2>
    gravityOverride {z=-980}
    physicsBodyDefinitions [{
      boundBone {boneName=dyn_R_SideSkirt_5}
      boxExtents=<30 40 40>
      constraintSetup {angularLimitsMax=<0 -2 5>, angularLimitsMin=<0 -30 -5>, angularTarget=<1 0 0>}
      localJointOffset=<6 0 0>
     }]
   }
   Save cached pose 'SimCache'(Pose=Blend Poses by bool(BlendPose_0=Component To Local(ComponentPose=@n55), BlendPose_1=Component To Local(ComponentPose=@n46), bActiveValue=Get bIsPlayingEmote()) {blendType=Linear}) {cacheName=SimCache}
  ]}
 EventGraph {nodes [
   Event BlueprintUpdateAnimation()
   > n1 = Sequence!(execute) -> (then_0, then_1, then_2:Exec, then_3:Exec, then_4:Exec, then_5:Exec, then_6:Exec, then_7:Exec, then_8:Exec, then_9:Exec, then_10:Exec)
   > n2 = "Gravity Override State Function (Parameter Asset) (Nativized)"(
    CharacterPartAnimInstance=Self-Reference()
    GameTimeInSeconds=GetGameTimeInSeconds()
    Parameters=&/F_MED_SkullArcana_Parameters
   )
   > Set Gravity Override(Gravity Override=@n2)
   > Set RBAN Settings Out(RBAN Settings Out=@n2.OutRigidBodyAnimNodeParameters)
   n3 = Branch(execute=@n1.then_10, Condition=Get bIsCrouching())
   > Set Crouch_DynAlpha(Crouch_DynAlpha:double=1) -> (Output_Get:double)
   Set Crouch_DynAlpha(execute=@n3.else, Crouch_DynAlpha=0)
   n4 = Branch(execute=@n1.then_4, Condition=Get bIsInFrontEnd())
   > Set Sim Space Skirt(Sim Space Skirt:AnimPhysSimSpaceType Enum=World) -> (Output_Get:AnimPhysSimSpaceType Enum)
   > Set SkirtDynAlpha(SkirtDynAlpha:double=0.3) -> (Output_Get:double)
   Set Sim Space Skirt(execute=@n4.else)
   > Set SkirtDynAlpha(SkirtDynAlpha=1)
   n5 = Sequence!(execute=@n1.then_8) -> (then_0, then_1, then_2, then_3, then_4, then_5, then_6)
   > n6 = Branch(Condition=OR Boolean(A=Get bIsSkydiveDiveMode(), B=Get bIsSkydiveFloating()))
   > Set enableSin(enableSin:double=1) -> (Output_Get:double)
   Set enableSin(execute=@n6.else, enableSin=0)
   n7 = Branch(execute=@n5.then_1, Condition=Get bIsSkydiveDiveMode())
   > Set freqVar1(freqVar1:double=35) -> (Output_Get:double)
   > Set freqVar2(freqVar2:double=32) -> (Output_Get:double)
   > Set freqVar3(freqVar3:double=28) -> (Output_Get:double)
   Set freqVar1(execute=@n7.else, freqVar1=22)
   > Set freqVar2(freqVar2=25)
   > Set freqVar3(freqVar3=20)
   Set skirtRotVar1(
    execute=@n5.then_2
    skirtRotVar1:Rotator=Rotator Rumble!(frequency:double=Get freqVar1() -> (freqVar1:double), amplitude:double=0.8, time offset:Vector="0,1.5,0", coefficient:Vector="0,1,0") -> (rotatorOutput:Rotator=~) {is=title K2Node_MacroInstance}
   ) -> (Output_Get:Rotator)
   > Set skirtRotVar2(skirtRotVar2:Rotator=Rotator Rumble(frequency=Get freqVar2() -> (freqVar2:double), amplitude=1.4, time offset="0,1,0", coefficient="0,1,0") -> (rotatorOutput=~)) -> (Output_Get:Rotator)
   > Set skirtRotVar3(skirtRotVar3:Rotator=Rotator Rumble(frequency=Get freqVar3() -> (freqVar3:double), amplitude=1, time offset="0,1,0", coefficient="0,1,2") -> (rotatorOutput=~)) -> (Output_Get:Rotator)
   n8 = Get bIsJumping()
   n9 = Get bIsFalling()
   n10 = Get bIsSprinting()
   n11 = Branch(execute=@n5.then_4, Condition=AND Boolean(A=Get bIsTacticalSprinting(), B=OR Boolean!(A=@n9, B=@n8, C:bool=@n10) -> (ReturnValue)))
   > Set IsTacticalandJump(IsTacticalandJump:double=1) -> (Output_Get:double)
   Set IsTacticalandJump(execute=@n11.else, IsTacticalandJump=0)
   n12 = GetAnimInstance(self=GetOwnerSkeletalMeshComponent())
   Is Valid(exec=@n5.then_5, InputObject=@n12)
   > Set Celebration Custom(Celebration Custom:bool=Montage_IsPlaying(self=@n12, Montage=/BRCosmetics/Animation/Game/MainPlayer/Emotes/Emote_Celebration_Female_M)) -> (Output_Get:bool)
   n13 = Branch(execute=@n5.then_3, Condition=Get bIsTargeting())
   > Set IsADS(IsADS:double=1) -> (Output_Get:double)
   Set IsADS(execute=@n13.else, IsADS=0)
   Is Valid(exec=@n5.then_6, InputObject=@n12)
   > Set IsEasterBlock(IsEasterBlock:bool=Montage_IsPlaying(self=@n12, Montage=/BRCosmetics/Animation/Game/MainPlayer/Emotes/Eastern_Bloc/Emote_EasternBloc_CMF_M)) -> (Output_Get:bool)
   n14 = Sequence!(execute=@n1.then_1) -> (then_0, then_1, then_2, then_3, then_4)
   > n15 = Branch(Condition=OR Boolean(A=false, B=Get bIsSkydiveFloating()))
   > Set IsSkydiveFloat(IsSkydiveFloat:double=1) -> (Output_Get:double)
   Set IsSkydiveFloat(execute=@n15.else, IsSkydiveFloat=0)
   n16 = Branch(execute=@n14.then_1, Condition=Get bIsParachuteOpened())
   > Set IsParachuteOpened(IsParachuteOpened:double=1) -> (Output_Get:double)
   Set IsParachuteOpened(execute=@n16.else, IsParachuteOpened=0)
   n17 = Branch(execute=@n14.then_3, Condition=Get bIsSkydiveDiveMode())
   > Set IsSkydivingDown(IsSkydivingDown:double=1) -> (Output_Get:double)
   Set IsSkydivingDown(execute=@n17.else, IsSkydivingDown=0)
   n18 = Branch(execute=@n14.then_2, Condition=OR Boolean!(A=false, B=Get bIsSkyDiving(), C=Get bIsParachuteOpened(), D:bool) -> (ReturnValue))
   > Set SkydiveDynamics(SkydiveDynamics:double=1) -> (Output_Get:double)
   Set SkydiveDynamics(execute=@n18.else, SkydiveDynamics=0)
   n19 = PlayerMovement(execute=@n1.then_9, AnimInstance=Self-Reference(), XDotProductCoefficient=1, YDotProductCoefficient=1, ZDotProductCoefficient=1)
   n20 = BreakVector(InVec=MakeVector(X=@n19, Y=@n19.YMovement))
   > "Set Move Left "("Move Left ":double="Absolute (Float)"(A="Clamp (Float)"(Value=@n20, Min=-1, Max=0))) -> (Output_Get:double)
   > Set Move Right(Move Right:double="Clamp (Float)"(Value=@n20, Min=0, Max=1)) -> (Output_Get:double)
   > Set moveForward(moveForward:double="Clamp (Float)"(Value=@n20.Y, Min=0, Max=1)) -> (Output_Get:double)
   > Set moveBackward(moveBackward:double="Absolute (Float)"(A="Clamp (Float)"(Value=@n20.Y, Min=-1, Max=0))) -> (Output_Get:double)
   Set Sprinting(execute=@n1.then_6, Sprinting:double="To Float (Boolean)"(InBool=Get bIsSprinting())) -> (Output_Get:double)
   > Set TacticalSprinting(TacticalSprinting:double="To Float (Boolean)"(InBool=Get bIsTacticalSprinting())) -> (Output_Get:double)
   n21 = Get bIsMoving2D()
   n22 = Get bIsCrouching()
   > Set Crouch Move(Crouch Move:double="To Float (Boolean)"(InBool=AND Boolean(A=@n21, B=@n22))) -> (Output_Get:double)
   n23 = OR Boolean!(A=Get bIsSliding(), B=Get bIsBaseballSliding()) -> (ReturnValue)
   > Set Crouch(Crouch:bool=AND Boolean(A=Get bIsCrouching(), B=NOT Boolean(A=@n23))) -> (Output_Get:bool)
   n24 = OR Boolean!(A=Get bIsSliding(), B=Get bIsBaseballSliding(), C:bool=Get bIsMoving2D()) -> (ReturnValue)
   > Set OnlyCrouch(OnlyCrouch:bool=AND Boolean(A=Get bIsCrouching(), B=NOT Boolean(A=@n24))) -> (Output_Get:bool)
   n25 = OR Boolean(A=Get bIsSkydiveFloating(), B=Get bIsSkydiveDiveMode(), C=Get bIsParachuteOpened())
   n26 = Get bIsJumping()
   n27 = Get bIsFalling()
   n28 = Branch(execute=@n1.then_3, Condition=AND Boolean(A=OR Boolean!(A=@n26, B=@n27) -> (ReturnValue), B=NOT Boolean(A=@n25)))
   > Set IsJumping(IsJumping:double=1) -> (Output_Get:double)
   Set IsJumping(execute=@n28.else, IsJumping=0)
   n29 = Branch(
    execute=@n1.then_7
    Condition=OR Boolean!(
     A=Get bIsSkyDiving()
     B=Get bIsCrouching()
     C:bool=Get bIsPlayingEmote()
     D:bool=Get bIsJumping()
     E:bool=Get bIsFalling()
     F:bool=Get bIsMoving2D()
    ) -> (ReturnValue)
   )
   > n30 = Set isIdle(isIdle:double=0) -> (Output_Get:double)
   > Set isIdleInFE(isIdleInFE:double=0) -> (Output_Get:double)
   > Set isIdle_Bool(isIdle_Bool:bool=false) -> (Output_Get:bool)
   > Set isNotIdle(isNotIdle:double=@n30) -> (Output_Get:double)
   Set isIdle(execute=@n29.else, isIdle=1)
   > Set isIdle_Bool(isIdle_Bool=true)
   n31 = Branch(
    execute=@n1.then_2
    Condition=OR Boolean!(A=Get bIsSkydiveFloating(), B=Get bIsSkyDiving(), C=Get bIsParachuteOpened(), D=Get bIsInFrontEnd()) -> (ReturnValue)
   )
   > Set AlphaSkydive(AlphaSkydive:double=0) -> (Output_Get:double)
   Set AlphaSkydive(execute=@n31.else, AlphaSkydive=1)
   Event OnBeginCrouching()
   > Set IsCrouching(IsCrouching:double=1) -> (Output_Get:double)
   Event OnEndPlayingEmote() {is=title K2Node_Event}
   > Set playing Emote(playing Emote:double=0) -> (Output_Get:double)
   Event OnEndCrouching()
   > Set IsCrouching(IsCrouching=0)
   n32 = Event BlueprintProcessAnimInputEvent()
   > n33 = Branch(Condition=IsNamed(self=@n32.AnimInputEvent, InEventName=aie.Player.Zipline.Entered))
   > Set OnZipline(OnZipline:bool=1) -> (Output_Get:bool)
   Branch(execute=@n33.else, Condition=IsNamed(self=@n32.AnimInputEvent, InEventName=aie.Player.Zipline.Exited))
   > Set OnZipline(OnZipline=false)
   Event OnBeginKneeSliding() {is=title K2Node_Event}
   > Set isKneeSlidling float(isKneeSlidling float:double=1) -> (Output_Get:double)
   Event OnEndKneeSliding() {is=title K2Node_Event}
   > Set isKneeSlidling float(isKneeSlidling float=0)
   Event OnBeginBaseballSliding() {is=title K2Node_Event}
   > Set isBaseballSlidling float(isBaseballSlidling float:double=1) -> (Output_Get:double)
   Event OnEndBaseballSliding() {is=title K2Node_Event}
   > Set isBaseballSlidling float(isBaseballSlidling float=0)
   Event OnBeginSkydiving()
   > Set IsSkydive(IsSkydive:double=1) -> (Output_Get:double)
   Event OnEndSkydiving()
   > Set IsSkydive(IsSkydive=0)
   Event OnBeginPlayingEmote() {is=title K2Node_Event}
   > Set playing Emote(playing Emote=1)
   n34 = Branch(Condition=Get bIsSkydiveFloating())
   > Set SkydiveDynamics(SkydiveDynamics=0.2)
   Set SkydiveDynamics(execute=@n34.else, SkydiveDynamics=0)
   n35 = Get moveBackward()
   n36 = "Get Move Left "()
   n37 = Get Move Right()
   n38 = Get moveForward()
   Select(Option 0:double=1, Option 1:double=0, Index="float > float"(A=float + float!(A=@n36, B=@n37, C:double=@n38, D:double=@n35) -> (ReturnValue))) -> (ReturnValue:double)
   Get bIsTacticalSprinting()
   Set IsSkydiveFloat(IsSkydiveFloat=1)
  ]}
 floatSinFunc {nodes [
   n1 = Inputs!() -> (time:double=~, time offset:double=~, frequency:double=~, amplitude:double=~, coefficient:double=~)
   n2 = float + float!(A=@n1, B=@n1.time offset) -> (ReturnValue)
   n3 = MultiplyByPi!(Value:double="Sin (Radians)"(A=float * float(A=@n2, B=@n1.frequency))) -> (ReturnValue:double)
   Outputs!(floatOutput:double=float * float(A=float * float(A=@n3, B=@n1.amplitude), B=@n1.coefficient))
  ]}
 rotatorRumble {nodes [
   n1 = Inputs!() -> (frequency, amplitude, time offset:Vector=~, coefficient:Vector=~)
   n2 = BreakVector(InVec=@n1.coefficient)
   n3 = BreakVector(InVec=@n1.time offset)
   n4 = GetGameTimeInSeconds()
   Outputs!(rotatorOutput:Rotator=MakeRotator(
     Roll=Float Sin Func!(time:double=@n4, time offset:double=@n3, frequency:double=@n1, amplitude:double=@n1.amplitude, coefficient:double=@n2) -> (floatOutput:double=~) {is=title K2Node_MacroInstance}
     Pitch=Float Sin Func(time=@n4, time offset=@n3.Y, frequency=@n1, amplitude=@n1.amplitude, coefficient=@n2.Y) -> (floatOutput=~)
     Yaw=Float Sin Func(time=@n4, time offset=@n3.Z, frequency=@n1, amplitude=@n1.amplitude, coefficient=@n2.Z) -> (floatOutput=~)
    ))
  ]}
]
variables [
 AlphaSkydive {type=number}
 Celebration Custom {type=boolean}
 Crouch {type=boolean}
 Crouch Move {type=number}
 Crouch_DynAlpha {type=number}
 Emote_Crouch {type=boolean}
 enableSin {type=number}
 FEAlpha {type=number}
 freqVar1 {type=number}
 freqVar2 {type=number}
 freqVar3 {type=number}
 Gravity Override {type=Vector}
 IsADS {type=number}
 isBaseballSlidling float {type=number}
 IsCrouching {type=number}
 IsEasterBlock {type=boolean}
 isIdle {type=number}
 isIdle_Bool {type=boolean}
 isIdleInFE {type=number}
 IsJumping {type=number}
 isKneeSlidling float {type=number}
 isNotIdle {type=number}
 IsParachuteOpened {type=number}
 IsSkydive {type=number}
 IsSkydiveFloat {type=number}
 IsSkydivingDown {type=number}
 IsTacticalandJump {type=number}
 "Move Left " {type=number}
 Move Right {type=number}
 moveBackward {type=number}
 moveForward {type=number}
 OnlyCrouch {type=boolean}
 OnZipline {type=boolean}
 playing Emote {type=number}
 RBAN Settings Out {type=FortRigidBodyAnimNodeParameters}
 Sim Space Skirt {type=AnimPhysSimSpaceType}
 SkirtDynAlpha {type=number}
 skirtRotVar1 {type=Rotator}
 skirtRotVar2 {type=Rotator}
 skirtRotVar3 {type=Rotator}
 SkydiveDynamics {type=number}
 Sprinting {type=number}
 TacticalSprinting {type=number}
]
