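# Test physics asset for the F_MED_SkullArcana female body, built around jointed capsule chains for its skirt.
# Eight tiny skirt base capsules are locked to the pelvis, each anchoring a dynamic chain of one to five segments, uneven between left and right.
# Chain joints allow narrow swing, wider toward the longest tips, and very little twist.
# Pelvis and rotation-locked leg capsules form the body, with free pelvis rotation against a wide, flat root box.
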PhysicsAsset /BRCosmetics/Characters/Player/Female/Medium/Bodies/F_MED_SkullArcana/Meshes/F_MED_SkullArcana_Physics_Test
bodies {
 base_C_BackSkirt [{shapeType=Capsule, length=0.505, radius=0.505, shapeName=base_C_BackSkirt_capsule, rotation=<rot 0 -90 0>}]
 base_C_FrontSkirt [{shapeType=Capsule, length=0.505, radius=0.505, shapeName=base_C_FrontSkirt_capsule, rotation=<rot 0 -90 0>}]
 base_L_BackSkirt [{shapeType=Capsule, length=0.505, radius=0.505, shapeName=base_L_BackSkirt_capsule, rotation=<rot 0 -90 0>}]
 base_L_FrontSkirt [{shapeType=Capsule, length=0.505, radius=0.505, shapeName=base_L_FrontSkirt_capsule, rotation=<rot 0 -90 0>}]
 base_L_SideSkirt [{shapeType=Capsule, length=0.505, radius=0.505, shapeName=base_L_SideSkirt_capsule, rotation=<rot 0 -90 0>}]
 base_R_BackSkirt [{shapeType=Capsule, length=0.505, radius=0.505, shapeName=base_R_BackSkirt_capsule, rotation=<rot 0 -90 0>}]
 base_R_FrontSkirt [{shapeType=Capsule, length=0.505, radius=0.505, shapeName=base_R_FrontSkirt_capsule, rotation=<rot 0 -90 0>}]
 base_R_SideSkirt [{shapeType=Capsule, length=0.505, radius=0.505, shapeName=base_R_SideSkirt_capsule, rotation=<rot 0 -90 0>}]
 calf_l [{shapeType=Capsule, length=32.5614, radius=9, center=<-25 0 -0.7708>, rotation=<rot 90 0 0>}]
 calf_r [{shapeType=Capsule, length=32.5614, radius=9, center=<25 0 0.7708>, rotation=<rot 90 180 0>}]
 dyn_C_BackSkirt_1 [{shapeType=Capsule, length=15, radius=8, shapeName=dyn_C_BackSkirt_1_capsule, center=<8 0 3>, rotation=<rot 0 0 90>}]
 dyn_C_BackSkirt_2 [{shapeType=Capsule, length=18, radius=8, shapeName=dyn_C_BackSkirt_2_capsule, center=<8 0 4>, rotation=<rot 0 0 90>}]
 dyn_C_BackSkirt_3 [{shapeType=Capsule, length=18, radius=9, shapeName=dyn_C_BackSkirt_3_capsule, center=<10 0 10>, rotation=<rot 0 0 90>}]
 dyn_C_BackSkirt_4 [{shapeType=Capsule, length=20, radius=10, shapeName=dyn_C_BackSkirt_4_capsule, center=<10 0 8>, rotation=<rot 0 0 90>}]
 dyn_C_BackSkirt_5 [{shapeType=Capsule, length=25, radius=15, shapeName=dyn_C_BackSkirt_5_capsule, center=<12 0 12>, rotation=<rot 0 0 90>}]
 dyn_C_FrontSkirt_1 [{shapeType=Capsule, length=15, radius=8, shapeName=dyn_C_FrontSkirt_1_capsule, center=<8 0 3>, rotation=<rot 0 0 90>}]
 dyn_C_FrontSkirt_2 [{shapeType=Capsule, length=18, radius=8, shapeName=dyn_C_FrontSkirt_2_capsule, center=<8 0 4>, rotation=<rot 0 0 90>}]
 dyn_L_BackSkirt_1 [{shapeType=Capsule, length=15, radius=8, shapeName=dyn_L_BackSkirt_1_capsule, center=<8 0 3>, rotation=<rot 0 0 90>}]
 dyn_L_BackSkirt_2 [{shapeType=Capsule, length=18, radius=8, shapeName=dyn_L_BackSkirt_2_capsule, center=<10 0 3>, rotation=<rot 0 0 90>}]
 dyn_L_BackSkirt_3 [{shapeType=Capsule, length=18, radius=9, shapeName=dyn_L_BackSkirt_3_capsule, center=<10 0 10>, rotation=<rot 0 0 90>}]
 dyn_L_FrontSkirt_1 [{shapeType=Capsule, length=15, radius=8, shapeName=dyn_L_FrontSkirt_1_capsule, center=<8 0 3>, rotation=<rot 0 0 90>}]
 dyn_L_SideSkirt_1 [{shapeType=Capsule, length=15, radius=8, shapeName=dyn_L_SideSkirt_1_capsule, center=<8 0 3>, rotation=<rot 0 0 90>}]
 dyn_R_BackSkirt_1 [{shapeType=Capsule, length=15, radius=8, shapeName=dyn_R_BackSkirt_1_capsule, center=<8 0 3>, rotation=<rot 0 0 90>}]
 dyn_R_BackSkirt_2 [{shapeType=Capsule, length=18, radius=8, shapeName=dyn_R_BackSkirt_2_capsule, center=<8 0 4>, rotation=<rot 0 0 90>}]
 dyn_R_BackSkirt_3 [{shapeType=Capsule, length=18, radius=9, shapeName=dyn_R_BackSkirt_3_capsule, center=<10 0 10>, rotation=<rot 0 0 90>}]
 dyn_R_BackSkirt_4 [{shapeType=Capsule, length=16, radius=10, shapeName=dyn_R_BackSkirt_4_capsule, center=<10 0 7>, rotation=<rot 0 0 90>}]
 dyn_R_BackSkirt_5 [{shapeType=Capsule, length=25, radius=15, shapeName=dyn_R_BackSkirt_5_capsule, center=<12 0 12>, rotation=<rot 0 0 90>}]
 dyn_R_FrontSkirt_1 [{shapeType=Capsule, length=15, radius=8, shapeName=dyn_R_FrontSkirt_1_capsule, center=<8 0 3>, rotation=<rot 0 0 90>}]
 dyn_R_FrontSkirt_2 [{shapeType=Capsule, length=18, radius=8, shapeName=dyn_R_FrontSkirt_2_capsule, center=<8 0 10.4072>, rotation=<rot 0 0 90>}]
 dyn_R_FrontSkirt_3 [{shapeType=Capsule, length=18, radius=9, shapeName=dyn_R_FrontSkirt_3_capsule, center=<10 0 10>, rotation=<rot 0 0 90>}]
 dyn_R_SideSkirt_1 [{shapeType=Capsule, length=15, radius=8, shapeName=dyn_R_SideSkirt_1_capsule, center=<8 0 3>, rotation=<rot 0 0 90>}]
 dyn_R_SideSkirt_2 [{shapeType=Capsule, length=18, radius=8, shapeName=dyn_R_SideSkirt_2_capsule, center=<8 0 7>, rotation=<rot 0 0 90>}]
 dyn_R_SideSkirt_3 [{shapeType=Capsule, length=18, radius=9, shapeName=dyn_R_SideSkirt_3_capsule, center=<10 0 10>, rotation=<rot 0 0 90>}]
 dyn_R_SideSkirt_4 [{shapeType=Capsule, length=16, radius=10, shapeName=dyn_R_SideSkirt_4_capsule, center=<10 0 10>, rotation=<rot 0 0 90>}]
 dyn_R_SideSkirt_5 [{shapeType=Capsule, length=25, radius=15, shapeName=dyn_R_SideSkirt_5_capsule, center=<10 0 12>, rotation=<rot 0 0 90>}]
 foot_l [{shapeType=Capsule, length=15, radius=10, shapeName=foot_l_capsule, center=<-2.9073 3.1965 0>, rotation=<rot -90 89.9187 0>}]
 foot_r [{shapeType=Capsule, length=15, radius=10, shapeName=foot_r_capsule, center=<2.9073 -3.1965 0>, rotation=<rot -90 -90.0813 0>}]
 pelvis [{shapeType=Capsule, length=8, radius=10, center=<0 -1.561 0>}]
 root [{shapeType=Box, extentX=300, extentY=300, extentZ=5, shapeName=root_box}]
 thigh_l [{shapeType=Capsule, length=41.1029, radius=10, center=<-25 0 0>, rotation=<rot 90 0 0>}]
 thigh_r [{shapeType=Capsule, length=41.1029, radius=10, center=<25 0 0>, rotation=<rot 90 180 0>}]
}
constraints [
 {
  bone1Name=thigh_l
  bone2Name=pelvis
  swing1LimitDegrees=45
  swing1Motion=Locked
  swing2LimitDegrees=45
  swing2Motion=Locked
  twistLimitDegrees=45
  twistMotion=Locked
 }
 {
  bone1Name=calf_l
  bone2Name=thigh_l
  swing1LimitDegrees=45
  swing1Motion=Locked
  swing2LimitDegrees=45
  swing2Motion=Locked
  twistLimitDegrees=45
  twistMotion=Locked
 }
 {
  bone1Name=thigh_r
  bone2Name=pelvis
  swing1LimitDegrees=45
  swing1Motion=Locked
  swing2LimitDegrees=45
  swing2Motion=Locked
  twistLimitDegrees=45
  twistMotion=Locked
 }
 {
  bone1Name=calf_r
  bone2Name=thigh_r
  swing1LimitDegrees=45
  swing1Motion=Locked
  swing2LimitDegrees=45
  swing2Motion=Locked
  twistLimitDegrees=45
  twistMotion=Locked
 }
 {
  bone1Name=foot_l
  bone2Name=calf_l
  swing1LimitDegrees=45
  swing1Motion=Locked
  swing2LimitDegrees=45
  swing2Motion=Locked
  twistLimitDegrees=45
  twistMotion=Locked
 }
 {
  bone1Name=foot_r
  bone2Name=calf_r
  swing1LimitDegrees=45
  swing1Motion=Locked
  swing2LimitDegrees=45
  swing2Motion=Locked
  twistLimitDegrees=45
  twistMotion=Locked
 }
 {
  bone1Name=base_C_BackSkirt
  bone2Name=pelvis
  swing1LimitDegrees=45
  swing1Motion=Locked
  swing2LimitDegrees=45
  swing2Motion=Locked
  twistLimitDegrees=45
  twistMotion=Locked
 }
 {bone1Name=dyn_C_BackSkirt_1, bone2Name=base_C_BackSkirt, swing1LimitDegrees=10, swing2LimitDegrees=10, twistLimitDegrees=2}
 {bone1Name=dyn_C_BackSkirt_2, bone2Name=dyn_C_BackSkirt_1, swing1LimitDegrees=10, swing2LimitDegrees=10, twistLimitDegrees=2}
 {bone1Name=dyn_C_BackSkirt_3, bone2Name=dyn_C_BackSkirt_2, swing1LimitDegrees=10, swing2LimitDegrees=10, twistLimitDegrees=2}
 {bone1Name=dyn_C_BackSkirt_4, bone2Name=dyn_C_BackSkirt_3, swing1LimitDegrees=20, swing2LimitDegrees=20, twistLimitDegrees=2}
 {bone1Name=dyn_C_BackSkirt_5, bone2Name=dyn_C_BackSkirt_4, swing1LimitDegrees=20, swing2LimitDegrees=20, twistLimitDegrees=2}
 {
  bone1Name=base_C_FrontSkirt
  bone2Name=pelvis
  swing1LimitDegrees=45
  swing1Motion=Locked
  swing2LimitDegrees=45
  swing2Motion=Locked
  twistLimitDegrees=45
  twistMotion=Locked
 }
 {bone1Name=dyn_C_FrontSkirt_1, bone2Name=base_C_FrontSkirt, swing1LimitDegrees=10, swing2LimitDegrees=10, twistLimitDegrees=2}
 {bone1Name=dyn_C_FrontSkirt_2, bone2Name=dyn_C_FrontSkirt_1, swing1LimitDegrees=10, swing2LimitDegrees=10, twistLimitDegrees=2}
 {
  bone1Name=base_L_BackSkirt
  bone2Name=pelvis
  swing1LimitDegrees=45
  swing1Motion=Locked
  swing2LimitDegrees=45
  swing2Motion=Locked
  twistLimitDegrees=45
  twistMotion=Locked
 }
 {bone1Name=dyn_L_BackSkirt_1, bone2Name=base_L_BackSkirt, swing1LimitDegrees=10, swing2LimitDegrees=10, twistLimitDegrees=2}
 {bone1Name=dyn_L_BackSkirt_2, bone2Name=dyn_L_BackSkirt_1, swing1LimitDegrees=10, swing2LimitDegrees=10, twistLimitDegrees=2}
 {bone1Name=dyn_L_BackSkirt_3, bone2Name=dyn_L_BackSkirt_2, swing1LimitDegrees=10, swing2LimitDegrees=10, twistLimitDegrees=2}
 {
  bone1Name=base_L_FrontSkirt
  bone2Name=pelvis
  swing1LimitDegrees=45
  swing1Motion=Locked
  swing2LimitDegrees=45
  swing2Motion=Locked
  twistLimitDegrees=45
  twistMotion=Locked
 }
 {bone1Name=dyn_L_FrontSkirt_1, bone2Name=base_L_FrontSkirt, swing1LimitDegrees=10, swing2LimitDegrees=10, twistLimitDegrees=2}
 {
  bone1Name=base_L_SideSkirt
  bone2Name=pelvis
  swing1LimitDegrees=45
  swing1Motion=Locked
  swing2LimitDegrees=45
  swing2Motion=Locked
  twistLimitDegrees=45
  twistMotion=Locked
 }
 {bone1Name=dyn_L_SideSkirt_1, bone2Name=base_L_SideSkirt, swing1LimitDegrees=10, swing2LimitDegrees=10, twistLimitDegrees=2}
 {
  bone1Name=base_R_BackSkirt
  bone2Name=pelvis
  swing1LimitDegrees=45
  swing1Motion=Locked
  swing2LimitDegrees=45
  swing2Motion=Locked
  twistLimitDegrees=45
  twistMotion=Locked
 }
 {bone1Name=dyn_R_BackSkirt_1, bone2Name=base_R_BackSkirt, swing1LimitDegrees=10, swing2LimitDegrees=10, twistLimitDegrees=2}
 {bone1Name=dyn_R_BackSkirt_2, bone2Name=dyn_R_BackSkirt_1, swing1LimitDegrees=10, swing2LimitDegrees=10, twistLimitDegrees=2}
 {bone1Name=dyn_R_BackSkirt_3, bone2Name=dyn_R_BackSkirt_2, swing1LimitDegrees=10, swing2LimitDegrees=10, twistLimitDegrees=2}
 {bone1Name=dyn_R_BackSkirt_4, bone2Name=dyn_R_BackSkirt_3, swing1LimitDegrees=20, swing2LimitDegrees=20, twistLimitDegrees=2}
 {bone1Name=dyn_R_BackSkirt_5, bone2Name=dyn_R_BackSkirt_4, swing1LimitDegrees=20, swing2LimitDegrees=20, twistLimitDegrees=2}
 {
  bone1Name=base_R_FrontSkirt
  bone2Name=pelvis
  swing1LimitDegrees=45
  swing1Motion=Locked
  swing2LimitDegrees=45
  swing2Motion=Locked
  twistLimitDegrees=45
  twistMotion=Locked
 }
 {bone1Name=dyn_R_FrontSkirt_1, bone2Name=base_R_FrontSkirt, swing1LimitDegrees=10, swing2LimitDegrees=10, twistLimitDegrees=2}
 {bone1Name=dyn_R_FrontSkirt_2, bone2Name=dyn_R_FrontSkirt_1, swing1LimitDegrees=10, swing2LimitDegrees=10, twistLimitDegrees=2}
 {bone1Name=dyn_R_FrontSkirt_3, bone2Name=dyn_R_FrontSkirt_2, swing1LimitDegrees=10, swing2LimitDegrees=10, twistLimitDegrees=2}
 {
  bone1Name=base_R_SideSkirt
  bone2Name=pelvis
  swing1LimitDegrees=45
  swing1Motion=Locked
  swing2LimitDegrees=45
  swing2Motion=Locked
  twistLimitDegrees=45
  twistMotion=Locked
 }
 {bone1Name=dyn_R_SideSkirt_1, bone2Name=base_R_SideSkirt, swing1LimitDegrees=10, swing2LimitDegrees=10, twistLimitDegrees=2}
 {bone1Name=dyn_R_SideSkirt_2, bone2Name=dyn_R_SideSkirt_1, swing1LimitDegrees=10, swing2LimitDegrees=10, twistLimitDegrees=2}
 {bone1Name=dyn_R_SideSkirt_3, bone2Name=dyn_R_SideSkirt_2, swing1LimitDegrees=10, swing2LimitDegrees=10, twistLimitDegrees=2}
 {bone1Name=dyn_R_SideSkirt_4, bone2Name=dyn_R_SideSkirt_3, swing1LimitDegrees=20, swing2LimitDegrees=20, twistLimitDegrees=2}
 {bone1Name=dyn_R_SideSkirt_5, bone2Name=dyn_R_SideSkirt_4, swing1LimitDegrees=20, swing2LimitDegrees=20, twistLimitDegrees=2}
 {
  bone1Name=pelvis
  bone2Name=root
  swing1LimitDegrees=45
  swing1Motion=Free
  swing2LimitDegrees=45
  swing2Motion=Free
  twistLimitDegrees=45
  twistMotion=Free
 }
]
