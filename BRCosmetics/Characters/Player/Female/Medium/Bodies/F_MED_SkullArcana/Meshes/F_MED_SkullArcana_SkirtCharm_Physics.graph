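# Physics asset for the right back-skirt charm on the SkullArcana medium female player cosmetic body.
# Capsule bodies cover two back-skirt bones, a tiny charm base and two charm segments.
# Fully locked angular constraints tie the lower skirt bone and charm base to skirt bone 4, forming a rigid mount.
# The two charm segments chain from that base with twist locked and swing left unlocked, so only the charm dangles.
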PhysicsAsset /BRCosmetics/Characters/Player/Female/Medium/Bodies/F_MED_SkullArcana/Meshes/F_MED_SkullArcana_SkirtCharm_Physics
bodies {
 base_R_BackSkirt_Charm [{shapeType=Capsule, length=0.505, radius=0.505, shapeName=base_R_BackSkirt_Charm_capsule, rotation=<rot 0 -90 0>}]
 dyn_R_BackSkirt_4 [{shapeType=Capsule, length=15, radius=10, shapeName=dyn_R_BackSkirt_4_capsule, center=<5.5065 0 -9.7682>, rotation=<rot 0 0 90>}]
 dyn_R_BackSkirt_5 [{shapeType=Capsule, length=15, radius=12, shapeName=dyn_R_BackSkirt_5_capsule, center=<5.784 0 -11.704>, rotation=<rot 0 0 90>}]
 dyn_R_BackSkirt_Charm_1 [{shapeType=Capsule, length=8, radius=2, shapeName=dyn_R_BackSkirt_Charm_1_capsule, center=<8 0 2.5>, rotation=<rot 89.4293 174.4015 174.3466>}]
 dyn_R_BackSkirt_Charm_2 [{shapeType=Capsule, length=8, radius=2, shapeName=dyn_R_BackSkirt_Charm_2_capsule, center=<8 0 2>, rotation=<rot 90 0 0>}]
}
constraints [
 {
  bone1Name=dyn_R_BackSkirt_5
  bone2Name=dyn_R_BackSkirt_4
  swing1LimitDegrees=5
  swing1Motion=Locked
  swing2LimitDegrees=20
  swing2Motion=Locked
  twistLimitDegrees=5
  twistMotion=Locked
 }
 {
  bone1Name=base_R_BackSkirt_Charm
  bone2Name=dyn_R_BackSkirt_4
  swing1LimitDegrees=45
  swing1Motion=Locked
  swing2LimitDegrees=45
  swing2Motion=Locked
  twistLimitDegrees=45
  twistMotion=Locked
 }
 {
  bone1Name=dyn_R_BackSkirt_Charm_1
  bone2Name=base_R_BackSkirt_Charm
  swing1LimitDegrees=10
  swing2LimitDegrees=10
  twistLimitDegrees=2
  twistMotion=Locked
 }
 {
  bone1Name=dyn_R_BackSkirt_Charm_2
  bone2Name=dyn_R_BackSkirt_Charm_1
  swing1LimitDegrees=30
  swing2LimitDegrees=30
  twistLimitDegrees=2
  twistMotion=Locked
 }
]
