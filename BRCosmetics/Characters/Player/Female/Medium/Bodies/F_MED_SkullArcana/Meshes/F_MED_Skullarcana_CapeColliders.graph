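# Physics asset of cape colliders for the medium female SkullArcana player body, going by its name.
# Capsules cover head, clavicles, spine, feet, a dense pelvis cluster and the dynamic back-skirt bones, while arm, leg and neck bodies have no shapes.
# A huge, offset capsule on root likely stands in for the ground.
# Constraints link the bodies along the skeleton with free swing and twist.
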PhysicsAsset /BRCosmetics/Characters/Player/Female/Medium/Bodies/F_MED_SkullArcana/Meshes/F_MED_Skullarcana_CapeColliders
bodies {
 calf_l []
 calf_r []
 clavicle_l [{shapeType=Capsule, length=6.5623, radius=7.5194, center=<3.9566 -1.0862 -2.1127>, rotation=<rot -77.5566 177.3915 7.7314>}]
 clavicle_r [{shapeType=Capsule, length=6.5623, radius=7.5194, center=<-3.9566 1.0862 2.1127>, rotation=<rot -77.5566 177.3915 -172.2686>}]
 dyn_C_BackSkirt_1 [{shapeType=Capsule, length=20, radius=10, shapeName=dyn_C_BackSkirt_1_capsule, center=<10 0 -2>, rotation=<rot 90 0 0>}]
 dyn_L_BackSkirt_1 [
  {shapeType=Capsule, length=20, radius=10, shapeName=dyn_L_BackSkirt_1_capsule, center=<20.5344 0 -5.742>, rotation=<rot 90 0 0>}
  {
   shapeType=Capsule
   length=31.0927
   radius=15
   shapeName=dyn_L_BackSkirt_1_capsule
   center=<18.2988 23.4363 -1.6457>
   rotation=<rot -0.0001 0 68.0934>
  }
 ]
 dyn_L_BackSkirt_2 [{
   shapeType=Capsule
   length=23.6961
   radius=8.9325
   shapeName=dyn_L_BackSkirt_2_capsule
   center=<13.3462 8.3245 -5.2198>
   rotation=<rot 81.9602 -179.9999 -180>
  }]
 dyn_R_BackSkirt_1 [{shapeType=Capsule, length=20, radius=10, shapeName=dyn_R_BackSkirt_1_capsule, center=<10 0 -2>, rotation=<rot 90 0 0>}]
 dyn_R_BackSkirt_2 [
  {
   shapeType=Capsule
   length=31.6212
   radius=31.8725
   shapeName=dyn_R_BackSkirt_2_capsule
   center=<48.0513 6.6918 -25.5099>
   rotation=<rot 0 0 100>
  }
  {
   shapeType=Capsule
   length=44.9977
   radius=28.9628
   shapeName=dyn_R_BackSkirt_2_capsule
   center=<57.889 14.8005 -20.2939>
   rotation=<rot 0 0 100>
  }
 ]
 foot_l [{shapeType=Capsule, length=11.0765, radius=8.0344, center=<-4.1811 6.0027 -0.338>, rotation=<rot 0.0388 1.4701 -89.3982>}]
 foot_r [{shapeType=Capsule, length=11.0765, radius=8.0344, center=<4.1811 -6.0027 0.338>, rotation=<rot 0.0388 1.4701 90.6018>}]
 head [{shapeType=Capsule, length=0, radius=18, center=<9.0697 0.0746 0>, rotation=<rot 0 -90 0>}]
 lowerarm_l []
 lowerarm_r []
 neck_02 []
 pelvis [
  {shapeType=Capsule, length=25.872, radius=16.9438, center=<-1.6419 -3.4639 -0.1981>}
  {shapeType=Capsule, length=50, radius=18, shapeName=pelvis_capsule, center=<-9.3567 -3.6918 7.6581>}
  {shapeType=Capsule, length=50, radius=20, shapeName=pelvis_capsule, center=<-34.4453 -5.3453 7.94>, rotation=<rot 0 -0.0001 2.6069>}
  {shapeType=Capsule, length=66.7233, radius=20, shapeName=pelvis_capsule, center=<-69.1808 -21.2146 8.6396>, rotation=<rot 0 -0.0001 2.6069>}
  {shapeType=Capsule, length=61.0356, radius=20, shapeName=pelvis_capsule, center=<-49.4542 -14.8336 -19.0464>, rotation=<rot 90 25.4971 0>}
  {shapeType=Capsule, length=61, radius=20, shapeName=pelvis_capsule, center=<-46.8212 -14.6121 9.3178>, rotation=<rot 90 27.4663 0>}
 ]
 root [{shapeType=Capsule, length=0.505, radius=400, shapeName=root_capsule, center=<0 -22.576 -386.63>, rotation=<rot 0 -90 0>}]
 spine_01 [{shapeType=Capsule, length=4.3888, radius=14.3638, center=<-8.9021 -0.8973 0>}]
 spine_02 [
  {shapeType=Capsule, length=2.5219, radius=9.7317, center=<-2.6885 -1.2641 0>}
  {shapeType=Capsule, length=20.6875, radius=9.3125, center=<-1.8317 1.9362 0>, rotation=<rot 90 0 0>}
 ]
 spine_03 [
  {shapeType=Capsule, length=32.5565, radius=11.7913, center=<-3.2674 -2.2349 0>}
  {shapeType=Capsule, length=20.7188, radius=9.2812, center=<-1.2163 0.549 0>, rotation=<rot 90 180 -175.0732>}
 ]
 spine_04 [
  {shapeType=Capsule, length=9.0076, radius=13.6976, center=<4.1842 1.4535 0>}
  {shapeType=Capsule, length=10.0312, radius=9.2812, center=<-1.2326 1.308 0>, rotation=<rot 90 0 0>}
 ]
 spine_05 [{shapeType=Capsule, length=9.678, radius=12.3571, center=<0.6969 1.91 0>}]
 thigh_l []
 thigh_r []
 upperarm_l []
 upperarm_r []
}
constraints [
 {
  bone1Name=spine_01
  bone2Name=pelvis
  swing1LimitDegrees=45
  swing1Motion=Free
  swing2LimitDegrees=45
  swing2Motion=Free
  twistLimitDegrees=45
  twistMotion=Free
 }
 {
  bone1Name=spine_02
  bone2Name=spine_01
  swing1LimitDegrees=45
  swing1Motion=Free
  swing2LimitDegrees=45
  swing2Motion=Free
  twistLimitDegrees=45
  twistMotion=Free
 }
 {
  bone1Name=spine_03
  bone2Name=spine_02
  swing1LimitDegrees=45
  swing1Motion=Free
  swing2LimitDegrees=45
  swing2Motion=Free
  twistLimitDegrees=45
  twistMotion=Free
 }
 {
  bone1Name=spine_04
  bone2Name=spine_03
  swing1LimitDegrees=45
  swing1Motion=Free
  swing2LimitDegrees=45
  swing2Motion=Free
  twistLimitDegrees=45
  twistMotion=Free
 }
 {
  bone1Name=spine_05
  bone2Name=spine_04
  swing1LimitDegrees=45
  swing1Motion=Free
  swing2LimitDegrees=45
  swing2Motion=Free
  twistLimitDegrees=45
  twistMotion=Free
 }
 {
  bone1Name=clavicle_l
  bone2Name=spine_05
  swing1LimitDegrees=45
  swing1Motion=Free
  swing2LimitDegrees=45
  swing2Motion=Free
  twistLimitDegrees=45
  twistMotion=Free
 }
 {
  bone1Name=upperarm_l
  bone2Name=clavicle_l
  swing1LimitDegrees=45
  swing1Motion=Free
  swing2LimitDegrees=45
  swing2Motion=Free
  twistLimitDegrees=45
  twistMotion=Free
 }
 {
  bone1Name=lowerarm_l
  bone2Name=upperarm_l
  swing1LimitDegrees=45
  swing1Motion=Free
  swing2LimitDegrees=45
  swing2Motion=Free
  twistLimitDegrees=45
  twistMotion=Free
 }
 {
  bone1Name=neck_02
  bone2Name=spine_05
  swing1LimitDegrees=45
  swing1Motion=Free
  swing2LimitDegrees=45
  swing2Motion=Free
  twistLimitDegrees=45
  twistMotion=Free
 }
 {
  bone1Name=head
  bone2Name=neck_02
  swing1LimitDegrees=45
  swing1Motion=Free
  swing2LimitDegrees=45
  swing2Motion=Free
  twistLimitDegrees=45
  twistMotion=Free
 }
 {
  bone1Name=thigh_l
  bone2Name=pelvis
  swing1LimitDegrees=45
  swing1Motion=Free
  swing2LimitDegrees=45
  swing2Motion=Free
  twistLimitDegrees=45
  twistMotion=Free
 }
 {
  bone1Name=foot_l
  bone2Name=calf_l
  swing1LimitDegrees=45
  swing1Motion=Free
  swing2LimitDegrees=45
  swing2Motion=Free
  twistLimitDegrees=45
  twistMotion=Free
 }
 {
  bone1Name=lowerarm_r
  bone2Name=upperarm_r
  swing1LimitDegrees=45
  swing1Motion=Free
  swing2LimitDegrees=45
  swing2Motion=Free
  twistLimitDegrees=45
  twistMotion=Free
 }
 {
  bone1Name=upperarm_r
  bone2Name=clavicle_r
  swing1LimitDegrees=45
  swing1Motion=Free
  swing2LimitDegrees=45
  swing2Motion=Free
  twistLimitDegrees=45
  twistMotion=Free
 }
 {
  bone1Name=clavicle_r
  bone2Name=spine_05
  swing1LimitDegrees=45
  swing1Motion=Free
  swing2LimitDegrees=45
  swing2Motion=Free
  twistLimitDegrees=45
  twistMotion=Free
 }
 {
  bone1Name=thigh_r
  bone2Name=pelvis
  swing1LimitDegrees=45
  swing1Motion=Free
  swing2LimitDegrees=45
  swing2Motion=Free
  twistLimitDegrees=45
  twistMotion=Free
 }
 {
  bone1Name=foot_r
  bone2Name=calf_r
  swing1LimitDegrees=45
  swing1Motion=Free
  swing2LimitDegrees=45
  swing2Motion=Free
  twistLimitDegrees=45
  twistMotion=Free
 }
 {
  bone1Name=calf_r
  bone2Name=thigh_r
  swing1LimitDegrees=45
  swing1Motion=Free
  swing2LimitDegrees=45
  swing2Motion=Free
  twistLimitDegrees=45
  twistMotion=Free
 }
 {
  bone1Name=calf_l
  bone2Name=thigh_l
  swing1LimitDegrees=45
  swing1Motion=Free
  swing2LimitDegrees=45
  swing2Motion=Free
  twistLimitDegrees=45
  twistMotion=Free
 }
 {
  bone1Name=dyn_C_BackSkirt_1
  bone2Name=pelvis
  swing1LimitDegrees=45
  swing1Motion=Free
  swing2LimitDegrees=45
  swing2Motion=Free
  twistLimitDegrees=45
  twistMotion=Free
 }
 {
  bone1Name=dyn_L_BackSkirt_1
  bone2Name=pelvis
  swing1LimitDegrees=45
  swing1Motion=Free
  swing2LimitDegrees=45
  swing2Motion=Free
  twistLimitDegrees=45
  twistMotion=Free
 }
 {
  bone1Name=dyn_L_BackSkirt_2
  bone2Name=dyn_L_BackSkirt_1
  swing1LimitDegrees=45
  swing1Motion=Free
  swing2LimitDegrees=45
  swing2Motion=Free
  twistLimitDegrees=45
  twistMotion=Free
 }
 {
  bone1Name=dyn_R_BackSkirt_1
  bone2Name=pelvis
  swing1LimitDegrees=45
  swing1Motion=Free
  swing2LimitDegrees=45
  swing2Motion=Free
  twistLimitDegrees=45
  twistMotion=Free
 }
 {
  bone1Name=dyn_R_BackSkirt_2
  bone2Name=dyn_R_BackSkirt_1
  swing1LimitDegrees=45
  swing1Motion=Free
  swing2LimitDegrees=45
  swing2Motion=Free
  twistLimitDegrees=45
  twistMotion=Free
 }
 {
  bone1Name=pelvis
  bone2Name=root
  swing1LimitDegrees=45
  swing1Motion=Free
  swing2LimitDegrees=45
  swing2Motion=Free
  twistLimitDegrees=45
  twistMotion=Free
 }
]
