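# Physics asset for the hair on the female medium SkullArcana face accessory, simulating hair chains against capsule body collision.
# Locked constraints join torso, arm, thigh and backpack capsules and anchor center, left and right chain roots to spine_05.
# Lower segments use narrow swing limits without twist, and short branches hang off left segment 3 and right segment 4.
# Angularly free constraints link both side tips to the center tip, cross-tying the chains.
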PhysicsAsset /BRCosmetics/Characters/Player/Female/Medium/Bodies/F_MED_SkullArcana/Meshes/Parts/F_MED_SkullArcana_FaceAcc_BackBiling_Physics_Hair
bodies {
 attach_backpack [{shapeType=Capsule, length=30, radius=20, shapeName=attach_backpack_capsule, center=<0 -25.0145 -0.1238>, rotation=<rot 3.949 -90 0>}]
 dyn_C_hair_2 [{shapeType=Capsule, length=8, radius=4, shapeName=dyn_C_hair_2_capsule, center=<10 0 -2.5>, rotation=<rot 90 0 0>}]
 dyn_C_hair_3 [{shapeType=Capsule, length=12.5741, radius=4, shapeName=dyn_C_hair_3_capsule, center=<10 0 -2.5>, rotation=<rot 83.9962 -180 180>}]
 dyn_C_hair_4 [{shapeType=Capsule, length=8, radius=4, shapeName=dyn_C_hair_1_capsule, center=<10 0 -2.6>, rotation=<rot 90 0 0>}]
 dyn_C_hair_5 [{shapeType=Capsule, length=6.5538, radius=4.5, shapeName=dyn_C_hair_1_capsule, center=<10 0 -3>, rotation=<rot 90 0 0>}]
 dyn_C_hair_6 [{shapeType=Capsule, length=6.6721, radius=3.7182, shapeName=dyn_C_hair_1_capsule, center=<10 0 -1.4903>, rotation=<rot 80.7239 0 0>}]
 dyn_C_hair_7 [{shapeType=Capsule, length=6.6721, radius=3.7182, shapeName=dyn_C_hair_1_capsule, center=<10 0 -2.2971>, rotation=<rot 88.2004 -180 180>}]
 dyn_L_hair_2 [{shapeType=Capsule, length=8, radius=4, shapeName=dyn_L_hair_2_capsule, center=<10 0 -2.5>, rotation=<rot 90 0 0>}]
 dyn_L_hair_3 [{shapeType=Capsule, length=12.5741, radius=4, shapeName=dyn_L_hair_3_capsule, center=<10 0 -2.5>, rotation=<rot 83.9962 -179.9999 -179.9999>}]
 dyn_L_hair_3_a_01 [{shapeType=Capsule, length=8, radius=2, shapeName=dyn_L_hair_3_a_01_capsule, center=<6 -1 0>, rotation=<rot 90 0 0>}]
 dyn_L_hair_3_a_02 [{shapeType=Capsule, length=8, radius=2.5, shapeName=dyn_L_hair_3_a_02_capsule, center=<4 -1 0>, rotation=<rot 90 0 0>}]
 dyn_L_hair_3_a_03 [{shapeType=Capsule, length=8, radius=2.5, shapeName=dyn_L_hair_3_a_03_capsule, center=<4 -1 0>, rotation=<rot 90 0 0>}]
 dyn_L_hair_4 [{shapeType=Capsule, length=8, radius=4, shapeName=dyn_C_hair_1_capsule, center=<10 0 -2.6>, rotation=<rot 90 0 0>}]
 dyn_L_hair_5 [{shapeType=Capsule, length=6.5538, radius=4.5, shapeName=dyn_C_hair_1_capsule, center=<10 0 -3>, rotation=<rot 90 0 0>}]
 dyn_L_hair_6 [{shapeType=Capsule, length=5.6846, radius=5.7248, shapeName=dyn_C_hair_1_capsule, center=<10 0 -3.9471>, rotation=<rot 90 0 0>}]
 dyn_L_hair_7 [{shapeType=Capsule, length=3.3462, radius=6.91, shapeName=dyn_C_hair_1_capsule, center=<10 0 -4.9589>, rotation=<rot 90 0 0>}]
 dyn_R_hair_2 [{shapeType=Capsule, length=8, radius=4, shapeName=dyn_R_hair_2_capsule, center=<10 0 -2.5>, rotation=<rot 90 0 0>}]
 dyn_R_hair_3 [{shapeType=Capsule, length=12.5741, radius=4, shapeName=dyn_R_hair_3_capsule, center=<10 0 -2.5>, rotation=<rot 83.9962 179.9999 179.9999>}]
 dyn_R_hair_4 [{shapeType=Capsule, length=8, radius=4, shapeName=dyn_C_hair_1_capsule, center=<15 0 -2.6>, rotation=<rot 90 0 0>}]
 dyn_R_hair_4_a_01 [{shapeType=Capsule, length=8, radius=2.5, shapeName=dyn_R_hair_4_a_01_capsule, center=<6 0 0>, rotation=<rot 90 0 0>}]
 dyn_R_hair_4_a_02 [{shapeType=Capsule, length=8, radius=2.5, shapeName=dyn_R_hair_4_a_02_capsule, center=<4 0 0>, rotation=<rot 90 0 0>}]
 dyn_R_hair_4_a_03 [{shapeType=Capsule, length=8, radius=2.5, shapeName=dyn_R_hair_4_a_03_capsule, center=<4 0 0>, rotation=<rot 90 0 0>}]
 dyn_R_hair_5 [{shapeType=Capsule, length=6.5538, radius=4.5, shapeName=dyn_C_hair_1_capsule, center=<10 0 -3>, rotation=<rot 90 0 0>}]
 dyn_R_hair_6 [{shapeType=Capsule, length=5.6846, radius=5.7248, shapeName=dyn_C_hair_1_capsule, center=<10 0 -3.9471>, rotation=<rot 90 0 0>}]
 dyn_R_hair_7 [{shapeType=Capsule, length=3.3462, radius=6.91, shapeName=dyn_C_hair_1_capsule, center=<10 0 -4.9589>, rotation=<rot 90 0 0>}]
 hand_l [{shapeType=Capsule, length=7.242, radius=5.9632, shapeName=hand_l_capsule, center=<8.7 0 -1.2>, rotation=<rot -90 0 0>}]
 hand_r [{shapeType=Capsule, length=7.242, radius=5.9632, shapeName=hand_l_capsule, center=<-8.7 0 1.2>, rotation=<rot -90 0 0>}]
 lowerarm_l [{shapeType=Capsule, length=16.4463, radius=3.8719, shapeName=lowerarm_r_capsule, center=<11.498 0 0>, rotation=<rot 90 0 0>}]
 lowerarm_r [{shapeType=Capsule, length=16.4463, radius=3.8719, shapeName=lowerarm_r_capsule, center=<-11.498 0 0>, rotation=<rot 90 0 0>}]
 pelvis [{shapeType=Capsule, length=15.648, radius=11.0693, shapeName=pelvis_capsule, center=<0.7253 -0.5324 0>, rotation=<rot 0 -90 0>}]
 spine_01 [{shapeType=Capsule, length=8.3135, radius=9.6685, shapeName=spine_01_capsule, center=<0.9493 2.1703 0>, rotation=<rot 0 -90 0>}]
 spine_03 [{shapeType=Capsule, length=6.4618, radius=8.7935, shapeName=spine_03_capsule, center=<-2.5591 2.4404 0>, rotation=<rot 0 -90 0>}]
 spine_04 [{shapeType=Capsule, length=3.967, radius=12.0063, shapeName=spine_04_capsule, center=<0 3.8958 0>, rotation=<rot 0 -90 0>}]
 spine_05 [{shapeType=Capsule, length=11.2972, radius=9.6481, shapeName=spine_05_capsule, center=<2.4027 2.0454 0.4254>}]
 thigh_l [{
   shapeType=Capsule
   length=20.3508
   radius=12.5358
   shapeName=thigh_l_capsule
   center=<-13.2994 -0.7814 0.9974>
   rotation=<rot 88.8079 -5.1314 -7.4225>
  }]
 thigh_r [{
   shapeType=Capsule
   length=20.3508
   radius=12.5358
   shapeName=thigh_r_capsule
   center=<13.2994 0.7814 -0.9974>
   rotation=<rot 88.8079 -5.1314 172.5775>
  }]
 upperarm_l [{shapeType=Capsule, length=20.7492, radius=6.8746, shapeName=upperarm_r_capsule, center=<13 0 0>, rotation=<rot -90 180 0>}]
 upperarm_r [{shapeType=Capsule, length=20.7492, radius=6.8746, shapeName=upperarm_l_capsule, center=<-13 0 0>, rotation=<rot -90 180 180>}]
}
constraints [
 {bone1Name=dyn_L_hair_4, bone2Name=dyn_L_hair_3, swing1LimitDegrees=5, swing2LimitDegrees=20, twistLimitDegrees=0}
 {bone1Name=dyn_L_hair_5, bone2Name=dyn_L_hair_4, swing1LimitDegrees=5, swing2LimitDegrees=10, twistLimitDegrees=0}
 {bone1Name=dyn_L_hair_6, bone2Name=dyn_L_hair_5, swing1LimitDegrees=5, swing2LimitDegrees=30, twistLimitDegrees=0}
 {bone1Name=dyn_L_hair_7, bone2Name=dyn_L_hair_6, swing1LimitDegrees=5, swing2LimitDegrees=30, twistLimitDegrees=0}
 {bone1Name=dyn_R_hair_4, bone2Name=dyn_R_hair_3, swing1LimitDegrees=5, swing2LimitDegrees=20, twistLimitDegrees=0}
 {bone1Name=dyn_R_hair_5, bone2Name=dyn_R_hair_4, swing1LimitDegrees=5, swing2LimitDegrees=10, twistLimitDegrees=0}
 {bone1Name=dyn_R_hair_6, bone2Name=dyn_R_hair_5, swing1LimitDegrees=5, swing2LimitDegrees=30, twistLimitDegrees=0}
 {bone1Name=dyn_R_hair_7, bone2Name=dyn_R_hair_6, swing1LimitDegrees=5, swing2LimitDegrees=30, twistLimitDegrees=0}
 {bone1Name=dyn_C_hair_4, bone2Name=dyn_C_hair_3, swing1LimitDegrees=5, swing2LimitDegrees=20, twistLimitDegrees=0}
 {bone1Name=dyn_C_hair_5, bone2Name=dyn_C_hair_4, swing1LimitDegrees=5, swing2LimitDegrees=10, twistLimitDegrees=0}
 {bone1Name=dyn_C_hair_6, bone2Name=dyn_C_hair_5, swing1LimitDegrees=5, swing2LimitDegrees=30, twistLimitDegrees=0}
 {bone1Name=dyn_C_hair_7, bone2Name=dyn_C_hair_6, swing1LimitDegrees=5, swing2LimitDegrees=30, twistLimitDegrees=0}
 {
  bone1Name=spine_04
  bone2Name=spine_03
  swing1LimitDegrees=45
  swing1Motion=Locked
  swing2LimitDegrees=45
  swing2Motion=Locked
  twistLimitDegrees=45
  twistMotion=Locked
 }
 {
  bone1Name=upperarm_r
  bone2Name=spine_05
  swing1LimitDegrees=45
  swing1Motion=Locked
  swing2LimitDegrees=45
  swing2Motion=Locked
  twistLimitDegrees=45
  twistMotion=Locked
 }
 {
  bone1Name=upperarm_l
  bone2Name=spine_05
  swing1LimitDegrees=45
  swing1Motion=Locked
  swing2LimitDegrees=45
  swing2Motion=Locked
  twistLimitDegrees=45
  twistMotion=Locked
 }
 {
  bone1Name=spine_03
  bone2Name=spine_01
  swing1LimitDegrees=45
  swing1Motion=Locked
  swing2LimitDegrees=45
  swing2Motion=Locked
  twistLimitDegrees=45
  twistMotion=Locked
 }
 {
  bone1Name=spine_01
  bone2Name=pelvis
  swing1LimitDegrees=45
  swing1Motion=Locked
  swing2LimitDegrees=45
  swing2Motion=Locked
  twistLimitDegrees=45
  twistMotion=Locked
 }
 {
  bone1Name=thigh_r
  bone2Name=pelvis
  swing1LimitDegrees=45
  swing1Motion=Locked
  swing2LimitDegrees=45
  swing2Motion=Locked
  twistLimitDegrees=45
  twistMotion=Locked
 }
 {
  bone1Name=thigh_l
  bone2Name=pelvis
  swing1LimitDegrees=45
  swing1Motion=Locked
  swing2LimitDegrees=45
  swing2Motion=Locked
  twistLimitDegrees=45
  twistMotion=Locked
 }
 {
  bone1Name=hand_l
  bone2Name=lowerarm_l
  swing1LimitDegrees=45
  swing1Motion=Locked
  swing2LimitDegrees=45
  swing2Motion=Locked
  twistLimitDegrees=45
  twistMotion=Locked
 }
 {
  bone1Name=lowerarm_l
  bone2Name=upperarm_l
  swing1LimitDegrees=45
  swing1Motion=Locked
  swing2LimitDegrees=45
  swing2Motion=Locked
  twistLimitDegrees=45
  twistMotion=Locked
 }
 {
  bone1Name=lowerarm_r
  bone2Name=upperarm_r
  swing1LimitDegrees=45
  swing1Motion=Locked
  swing2LimitDegrees=45
  swing2Motion=Locked
  twistLimitDegrees=45
  twistMotion=Locked
 }
 {
  bone1Name=hand_r
  bone2Name=lowerarm_r
  swing1LimitDegrees=45
  swing1Motion=Locked
  swing2LimitDegrees=45
  swing2Motion=Locked
  twistLimitDegrees=45
  twistMotion=Locked
 }
 {
  bone1Name=dyn_C_hair_3
  bone2Name=dyn_C_hair_2
  swing1LimitDegrees=5
  swing1Motion=Locked
  swing2LimitDegrees=20
  swing2Motion=Locked
  twistLimitDegrees=0
  twistMotion=Locked
 }
 {
  bone1Name=dyn_L_hair_3
  bone2Name=dyn_L_hair_2
  swing1LimitDegrees=5
  swing1Motion=Locked
  swing2LimitDegrees=20
  swing2Motion=Locked
  twistLimitDegrees=0
  twistMotion=Locked
 }
 {
  bone1Name=dyn_R_hair_3
  bone2Name=dyn_R_hair_2
  swing1LimitDegrees=5
  swing1Motion=Locked
  swing2LimitDegrees=20
  swing2Motion=Locked
  twistLimitDegrees=0
  twistMotion=Locked
 }
 {
  bone1Name=dyn_C_hair_2
  bone2Name=spine_05
  swing1LimitDegrees=5
  swing1Motion=Locked
  swing2LimitDegrees=20
  swing2Motion=Locked
  twistLimitDegrees=0
  twistMotion=Locked
 }
 {
  bone1Name=dyn_L_hair_2
  bone2Name=spine_05
  swing1LimitDegrees=5
  swing1Motion=Locked
  swing2LimitDegrees=20
  swing2Motion=Locked
  twistLimitDegrees=0
  twistMotion=Locked
 }
 {
  bone1Name=dyn_R_hair_2
  bone2Name=spine_05
  swing1LimitDegrees=5
  swing1Motion=Locked
  swing2LimitDegrees=20
  swing2Motion=Locked
  twistLimitDegrees=0
  twistMotion=Locked
 }
 {
  bone1Name=spine_05
  bone2Name=spine_04
  swing1LimitDegrees=45
  swing1Motion=Locked
  swing2LimitDegrees=45
  swing2Motion=Locked
  twistLimitDegrees=45
  twistMotion=Locked
 }
 {
  bone1Name=dyn_R_hair_7
  bone2Name=dyn_C_hair_7
  swing1LimitDegrees=45
  swing1Motion=Free
  swing2LimitDegrees=45
  swing2Motion=Free
  twistLimitDegrees=45
  twistMotion=Free
 }
 {
  bone1Name=dyn_L_hair_7
  bone2Name=dyn_C_hair_7
  swing1LimitDegrees=45
  swing1Motion=Free
  swing2LimitDegrees=45
  swing2Motion=Free
  twistLimitDegrees=45
  twistMotion=Free
 }
 {
  bone1Name=attach_backpack
  bone2Name=spine_05
  swing1LimitDegrees=45
  swing1Motion=Locked
  swing2LimitDegrees=45
  swing2Motion=Locked
  twistLimitDegrees=45
  twistMotion=Locked
 }
 {
  bone1Name=dyn_R_hair_4_a_01
  bone2Name=dyn_R_hair_4
  swing1LimitDegrees=5
  swing1Motion=Locked
  swing2LimitDegrees=5
  swing2Motion=Locked
  twistLimitDegrees=0
  twistMotion=Locked
 }
 {bone1Name=dyn_R_hair_4_a_02, bone2Name=dyn_R_hair_4_a_01, swing1LimitDegrees=5, swing2LimitDegrees=10, twistLimitDegrees=0}
 {bone1Name=dyn_R_hair_4_a_03, bone2Name=dyn_R_hair_4_a_02, swing1LimitDegrees=5, swing2LimitDegrees=10, twistLimitDegrees=0}
 {
  bone1Name=dyn_L_hair_3_a_01
  bone2Name=dyn_L_hair_3
  swing1LimitDegrees=5
  swing1Motion=Locked
  swing2LimitDegrees=5
  swing2Motion=Locked
  twistLimitDegrees=0
  twistMotion=Locked
 }
 {bone1Name=dyn_L_hair_3_a_02, bone2Name=dyn_L_hair_3_a_01, swing1LimitDegrees=5, swing2LimitDegrees=10, twistLimitDegrees=0}
 {bone1Name=dyn_L_hair_3_a_03, bone2Name=dyn_L_hair_3_a_02, swing1LimitDegrees=5, swing2LimitDegrees=10, twistLimitDegrees=0}
]
