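# Physics asset for the female medium SkullArcana face accessory, rigging three hair bone chains (center, left, right) with capsule bodies for secondary motion.
# Each chain anchors to spine_05 with locked angular constraints through its second dynamic segment, then allows limited swing with locked twist.
# Cross constraints tie the lower side segments to the center chain.
# Spine, clavicle, arm, hand and thigh capsules provide body collision shapes.
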PhysicsAsset /BRCosmetics/Characters/Player/Female/Medium/Bodies/F_MED_SkullArcana/Meshes/Parts/F_MED_SkullArcana_FaceAcc_Physics
bodies {
 base_C_hair_1 [{shapeType=Capsule, length=0.505, radius=0.505, shapeName=base_C_hair_1_capsule, rotation=<rot 0 -90 0>}]
 base_L_hair_1 [{shapeType=Capsule, length=0.505, radius=0.505, shapeName=base_L_hair_1_capsule, rotation=<rot 0 -90 0>}]
 base_R_hair_1 [{shapeType=Capsule, length=0.505, radius=0.505, shapeName=base_R_hair_1_capsule, rotation=<rot 0 -90 0>}]
 clavicle_l [{shapeType=Capsule, length=15, radius=7, shapeName=clavicle_l_capsule, rotation=<rot 90 180 0>}]
 clavicle_r [{shapeType=Capsule, length=15, radius=7, shapeName=clavicle_r_capsule, rotation=<rot 90 0 0>}]
 dyn_C_hair_1 [{shapeType=Capsule, length=6, radius=6, shapeName=dyn_C_hair_1_capsule, center=<5 0 0>, rotation=<rot 90 0 0>}]
 dyn_C_hair_2 [{shapeType=Capsule, length=6, radius=6, shapeName=dyn_C_hair_2_capsule, center=<5 0 -2.5>, rotation=<rot 90 0 0>}]
 dyn_C_hair_3 [{shapeType=Capsule, length=6, radius=6, shapeName=dyn_C_hair_3_capsule, center=<5 0 -4>, rotation=<rot 90 0 0>}]
 dyn_C_hair_4 [{shapeType=Capsule, length=6, radius=6, shapeName=dyn_C_hair_4_capsule, center=<5 0 -3>, rotation=<rot 90 0 0>}]
 dyn_C_hair_5 [{shapeType=Capsule, length=6, radius=6, shapeName=dyn_C_hair_5_capsule, center=<5 0 -2>, rotation=<rot 90 0 0>}]
 dyn_C_hair_6 [{shapeType=Capsule, length=6, radius=6, shapeName=dyn_C_hair_6_capsule, center=<5 0 0>, rotation=<rot 90 0 0>}]
 dyn_C_hair_7 [{shapeType=Capsule, length=6, radius=6, shapeName=dyn_C_hair_7_capsule, center=<5 0 0>, rotation=<rot 90 0 0>}]
 dyn_L_hair_1 [{shapeType=Capsule, length=6, radius=6, shapeName=dyn_L_hair_1_capsule, center=<5 0 0>, rotation=<rot 90 0 0>}]
 dyn_L_hair_2 [{shapeType=Capsule, length=6, radius=6, shapeName=dyn_L_hair_2_capsule, center=<5 0 -2.5>, rotation=<rot 90 0 0>}]
 dyn_L_hair_3 [{shapeType=Capsule, length=6, radius=6, shapeName=dyn_L_hair_3_capsule, center=<5 0 -3>, rotation=<rot 90 0 0>}]
 dyn_L_hair_4 [{shapeType=Capsule, length=6, radius=6, shapeName=dyn_L_hair_4_capsule, center=<5 0 -3>, rotation=<rot 90 0 0>}]
 dyn_L_hair_5 [{shapeType=Capsule, length=6, radius=6, shapeName=dyn_L_hair_5_capsule, center=<5 0 -2>, rotation=<rot 90 0 0>}]
 dyn_L_hair_6 [{shapeType=Capsule, length=6, radius=6, shapeName=dyn_L_hair_6_capsule, center=<5 0 0>, rotation=<rot 90 0 0>}]
 dyn_L_hair_7 [{shapeType=Capsule, length=6, radius=6, shapeName=dyn_L_hair_7_capsule, center=<5 0 0>, rotation=<rot 90 0 0>}]
 dyn_R_hair_1 [{shapeType=Capsule, length=6, radius=6, shapeName=dyn_R_hair_1_capsule, center=<5 0 0>, rotation=<rot 90 0 0>}]
 dyn_R_hair_2 [{shapeType=Capsule, length=6, radius=6, shapeName=dyn_R_hair_2_capsule, center=<5 0 -2.5>, rotation=<rot 90 0 0>}]
 dyn_R_hair_3 [{shapeType=Capsule, length=6, radius=6, shapeName=dyn_R_hair_3_capsule, center=<5 0 -4>, rotation=<rot 90 0 0>}]
 dyn_R_hair_4 [{shapeType=Capsule, length=6, radius=6, shapeName=dyn_R_hair_4_capsule, center=<5 0 -3>, rotation=<rot 90 0 0>}]
 dyn_R_hair_5 [{shapeType=Capsule, length=6, radius=6, shapeName=dyn_R_hair_5_capsule, center=<5 1.8708 -2>, rotation=<rot 90 0 0>}]
 dyn_R_hair_6 [{shapeType=Capsule, length=6, radius=6, shapeName=dyn_R_hair_6_capsule, center=<5 3.046 0>, rotation=<rot 90 0 0>}]
 dyn_R_hair_7 [{shapeType=Capsule, length=6, radius=6, shapeName=dyn_R_hair_7_capsule, center=<5 0 0>, rotation=<rot 90 0 0>}]
 hand_l [{shapeType=Capsule, length=8, radius=6, shapeName=hand_l_capsule, center=<10 -2.4462 0>, rotation=<rot 90 0 0>}]
 hand_r [{shapeType=Capsule, length=8, radius=6, shapeName=hand_r_capsule, center=<-10 2.4462 0>, rotation=<rot 90 180 0>}]
 lowerarm_l [{shapeType=Capsule, length=20, radius=6, shapeName=lowerarm_l_capsule, center=<10 0 0>, rotation=<rot 90 0 0>}]
 lowerarm_r [{shapeType=Capsule, length=20, radius=6, shapeName=lowerarm_r_capsule, center=<-10 0 0>, rotation=<rot 90 180 0>}]
 spine_02 [{shapeType=Capsule, length=20, radius=12, shapeName=spine_02_capsule, center=<-17.8506 -0.9195 0>, rotation=<rot 0 -90 0>}]
 spine_03 [{shapeType=Capsule, length=20, radius=10, shapeName=spine_03_capsule, center=<-10.2394 1.0033 0>, rotation=<rot 0 -90 0>}]
 spine_04 [{shapeType=Capsule, length=20, radius=8, shapeName=spine_04_capsule, center=<-7.9794 1.0073 0>, rotation=<rot 0 -90 0>}]
 spine_05 [{shapeType=Capsule, length=20, radius=8, shapeName=spine_05_capsule, center=<-5 -0.1136 0>, rotation=<rot 0 -90 0>}]
 thigh_l [{shapeType=Capsule, length=30, radius=8, shapeName=thigh_l_capsule, center=<-20.5534 -4.4255 0>, rotation=<rot 90 0 0>}]
 thigh_r [{shapeType=Capsule, length=30, radius=8, shapeName=thigh_r_capsule, center=<20.5534 4.4255 0>, rotation=<rot 90 180 0>}]
 upperarm_l [{shapeType=Capsule, length=15, radius=7, shapeName=upperarm_l_capsule, center=<13.3467 0.001 0.5365>, rotation=<rot -87.698 0.1311 179.8733>}]
 upperarm_r [{
   shapeType=Capsule
   length=15
   radius=7
   shapeName=upperarm_r_capsule
   center=<-13.3467 -0.001 -0.5365>
   rotation=<rot -87.698 0.1311 -0.1267>
  }]
}
constraints [
 {
  bone1Name=base_C_hair_1
  bone2Name=spine_05
  swing1LimitDegrees=45
  swing1Motion=Locked
  swing2LimitDegrees=45
  swing2Motion=Locked
  twistLimitDegrees=45
  twistMotion=Locked
 }
 {
  bone1Name=dyn_C_hair_1
  bone2Name=base_C_hair_1
  swing1LimitDegrees=40
  swing1Motion=Locked
  swing2LimitDegrees=40
  swing2Motion=Locked
  twistLimitDegrees=45
  twistMotion=Locked
 }
 {
  bone1Name=dyn_C_hair_2
  bone2Name=dyn_C_hair_1
  swing1LimitDegrees=40
  swing1Motion=Locked
  swing2LimitDegrees=40
  swing2Motion=Locked
  twistLimitDegrees=45
  twistMotion=Locked
 }
 {bone1Name=dyn_C_hair_3, bone2Name=dyn_C_hair_2, swing1LimitDegrees=20, swing2LimitDegrees=20, twistLimitDegrees=45, twistMotion=Locked}
 {bone1Name=dyn_C_hair_4, bone2Name=dyn_C_hair_3, swing1LimitDegrees=20, swing2LimitDegrees=20, twistLimitDegrees=45, twistMotion=Locked}
 {bone1Name=dyn_C_hair_5, bone2Name=dyn_C_hair_4, swing1LimitDegrees=20, swing2LimitDegrees=20, twistLimitDegrees=45, twistMotion=Locked}
 {bone1Name=dyn_C_hair_6, bone2Name=dyn_C_hair_5, swing1LimitDegrees=20, swing2LimitDegrees=20, twistLimitDegrees=45, twistMotion=Locked}
 {bone1Name=dyn_C_hair_7, bone2Name=dyn_C_hair_6, swing1LimitDegrees=20, swing2LimitDegrees=20, twistLimitDegrees=45, twistMotion=Locked}
 {
  bone1Name=base_L_hair_1
  bone2Name=spine_05
  swing1LimitDegrees=45
  swing1Motion=Locked
  swing2LimitDegrees=45
  swing2Motion=Locked
  twistLimitDegrees=45
  twistMotion=Locked
 }
 {
  bone1Name=dyn_L_hair_1
  bone2Name=base_L_hair_1
  swing1LimitDegrees=40
  swing1Motion=Locked
  swing2LimitDegrees=40
  swing2Motion=Locked
  twistLimitDegrees=45
  twistMotion=Locked
 }
 {
  bone1Name=dyn_L_hair_2
  bone2Name=dyn_L_hair_1
  swing1LimitDegrees=40
  swing1Motion=Locked
  swing2LimitDegrees=40
  swing2Motion=Locked
  twistLimitDegrees=45
  twistMotion=Locked
 }
 {bone1Name=dyn_L_hair_3, bone2Name=dyn_L_hair_2, swing1LimitDegrees=20, swing2LimitDegrees=20, twistLimitDegrees=45, twistMotion=Locked}
 {bone1Name=dyn_L_hair_4, bone2Name=dyn_L_hair_3, swing1LimitDegrees=20, swing2LimitDegrees=20, twistLimitDegrees=45, twistMotion=Locked}
 {bone1Name=dyn_L_hair_5, bone2Name=dyn_L_hair_4, swing1LimitDegrees=20, swing2LimitDegrees=20, twistLimitDegrees=45, twistMotion=Locked}
 {bone1Name=dyn_L_hair_6, bone2Name=dyn_L_hair_5, swing1LimitDegrees=20, swing2LimitDegrees=20, twistLimitDegrees=45, twistMotion=Locked}
 {bone1Name=dyn_L_hair_7, bone2Name=dyn_L_hair_6, swing1LimitDegrees=20, swing2LimitDegrees=20, twistLimitDegrees=45, twistMotion=Locked}
 {
  bone1Name=base_R_hair_1
  bone2Name=spine_05
  swing1LimitDegrees=45
  swing1Motion=Locked
  swing2LimitDegrees=45
  swing2Motion=Locked
  twistLimitDegrees=45
  twistMotion=Locked
 }
 {
  bone1Name=dyn_R_hair_1
  bone2Name=base_R_hair_1
  swing1LimitDegrees=40
  swing1Motion=Locked
  swing2LimitDegrees=40
  swing2Motion=Locked
  twistLimitDegrees=45
  twistMotion=Locked
 }
 {
  bone1Name=dyn_R_hair_2
  bone2Name=dyn_R_hair_1
  swing1LimitDegrees=40
  swing1Motion=Locked
  swing2LimitDegrees=40
  swing2Motion=Locked
  twistLimitDegrees=45
  twistMotion=Locked
 }
 {bone1Name=dyn_R_hair_3, bone2Name=dyn_R_hair_2, swing1LimitDegrees=20, swing2LimitDegrees=20, twistLimitDegrees=45, twistMotion=Locked}
 {bone1Name=dyn_R_hair_4, bone2Name=dyn_R_hair_3, swing1LimitDegrees=20, swing2LimitDegrees=20, twistLimitDegrees=45, twistMotion=Locked}
 {bone1Name=dyn_R_hair_5, bone2Name=dyn_R_hair_4, swing1LimitDegrees=20, swing2LimitDegrees=20, twistLimitDegrees=45, twistMotion=Locked}
 {bone1Name=dyn_R_hair_6, bone2Name=dyn_R_hair_5, swing1LimitDegrees=20, swing2LimitDegrees=20, twistLimitDegrees=45, twistMotion=Locked}
 {bone1Name=dyn_R_hair_7, bone2Name=dyn_R_hair_6, swing1LimitDegrees=20, swing2LimitDegrees=20, twistLimitDegrees=45, twistMotion=Locked}
 {
  bone1Name=dyn_R_hair_7
  bone2Name=dyn_C_hair_7
  swing1LimitDegrees=45
  swing1Motion=Free
  swing2LimitDegrees=45
  swing2Motion=Free
  twistLimitDegrees=45
  twistMotion=Locked
 }
 {
  bone1Name=dyn_L_hair_7
  bone2Name=dyn_C_hair_7
  swing1LimitDegrees=45
  swing1Motion=Free
  swing2LimitDegrees=45
  swing2Motion=Free
  twistLimitDegrees=45
  twistMotion=Locked
 }
 {
  bone1Name=lowerarm_r
  bone2Name=upperarm_r
  swing1LimitDegrees=45
  swing1Motion=Locked
  swing2LimitDegrees=45
  swing2Motion=Locked
  twistLimitDegrees=45
  twistMotion=Locked
 }
 {
  bone1Name=hand_r
  bone2Name=lowerarm_r
  swing1LimitDegrees=45
  swing1Motion=Locked
  swing2LimitDegrees=45
  swing2Motion=Locked
  twistLimitDegrees=45
  twistMotion=Locked
 }
 {
  bone1Name=upperarm_l
  bone2Name=clavicle_l
  swing1LimitDegrees=45
  swing1Motion=Locked
  swing2LimitDegrees=45
  swing2Motion=Locked
  twistLimitDegrees=45
  twistMotion=Locked
 }
 {
  bone1Name=lowerarm_l
  bone2Name=upperarm_l
  swing1LimitDegrees=45
  swing1Motion=Locked
  swing2LimitDegrees=45
  swing2Motion=Locked
  twistLimitDegrees=45
  twistMotion=Locked
 }
 {
  bone1Name=hand_l
  bone2Name=lowerarm_l
  swing1LimitDegrees=45
  swing1Motion=Locked
  swing2LimitDegrees=45
  swing2Motion=Locked
  twistLimitDegrees=45
  twistMotion=Locked
 }
 {
  bone1Name=upperarm_l
  bone2Name=hand_l
  swing1LimitDegrees=45
  swing1Motion=Locked
  swing2LimitDegrees=45
  swing2Motion=Locked
  twistLimitDegrees=45
  twistMotion=Locked
 }
 {
  bone1Name=lowerarm_l
  bone2Name=hand_l
  swing1LimitDegrees=45
  swing1Motion=Locked
  swing2LimitDegrees=45
  swing2Motion=Locked
  twistLimitDegrees=45
  twistMotion=Locked
 }
 {
  bone1Name=dyn_R_hair_5
  bone2Name=dyn_C_hair_5
  swing1LimitDegrees=45
  swing1Motion=Free
  swing2LimitDegrees=45
  swing2Motion=Free
  twistLimitDegrees=45
  twistMotion=Free
 }
 {
  bone1Name=dyn_L_hair_5
  bone2Name=dyn_C_hair_5
  swing1LimitDegrees=45
  swing1Motion=Free
  swing2LimitDegrees=45
  swing2Motion=Free
  twistLimitDegrees=45
  twistMotion=Free
 }
 {
  bone1Name=dyn_R_hair_4
  bone2Name=dyn_C_hair_4
  swing1LimitDegrees=45
  swing1Motion=Free
  swing2LimitDegrees=45
  swing2Motion=Free
  twistLimitDegrees=45
  twistMotion=Free
 }
 {
  bone1Name=dyn_L_hair_4
  bone2Name=dyn_C_hair_4
  swing1LimitDegrees=45
  swing1Motion=Free
  swing2LimitDegrees=45
  swing2Motion=Free
  twistLimitDegrees=45
  twistMotion=Free
 }
 {
  bone1Name=dyn_R_hair_6
  bone2Name=dyn_C_hair_6
  swing1LimitDegrees=45
  swing1Motion=Free
  swing2LimitDegrees=45
  swing2Motion=Free
  twistLimitDegrees=45
  twistMotion=Free
 }
 {
  bone1Name=dyn_L_hair_6
  bone2Name=dyn_C_hair_6
  swing1LimitDegrees=45
  swing1Motion=Free
  swing2LimitDegrees=45
  swing2Motion=Free
  twistLimitDegrees=45
  twistMotion=Free
 }
]
