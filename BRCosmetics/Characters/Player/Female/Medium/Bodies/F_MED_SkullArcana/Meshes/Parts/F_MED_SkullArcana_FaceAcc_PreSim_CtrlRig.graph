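# Pre-simulation Control Rig for the female SkullArcana face accessory, posing dynamic hair and an earring from body motion.
# Weighted parent constraints tie left, center and right hair controls and nulls to head, shoulders, spine and pelvis.
# TwistBones and B-spline IK shape the three hair chains.
# Spherical pose readers and DeltaRotation weights blend corrective offsets for arm, clavicle, neck, head and spine poses, with some gated by state flags.
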ControlRigBlueprint /BRCosmetics/Characters/Player/Female/Medium/Bodies/F_MED_SkullArcana/Meshes/Parts/F_MED_SkullArcana_FaceAcc_PreSim_CtrlRig
graphs [
 CollapseNode_ContainedGraph {nodes [
   RigVMFunctionEntryNode()
   > &RigUnit_ModifyTransforms(
    ItemToModify=((Item=(Type=Bone,Name="spine_05"),Transform=(Rotation=(X=0,Y=0,Z=0,W=1),Translation=(X=0,Y=0,Z=0),Scale3D=(X=1,Y=1,Z=1))))
    Weight=1
    WeightMinimum=0
   )
   > &RigUnit_ParentConstraint(
    AdvancedSettings=(RotationOrderForFilter=XZY)
    Child=(Type=Control,Name="dyn_L_hair_1_ctrl")
    Filter=(TranslationFilter=(bX=True,bY=True,bZ=True),RotationFilter=(bX=True,bY=True,bZ=True),ScaleFilter=(bX=False,bY=False,bZ=False))
    Parents=((Item=(Type=Bone,Name="head"),Weight=1))
    Weight=1
   )
   > &RigUnit_ParentConstraint(
    AdvancedSettings=(RotationOrderForFilter=XZY)
    Child=(Type=Control,Name="dyn_C_hair_1_ctrl")
    Filter=(TranslationFilter=(bX=True,bY=True,bZ=True),RotationFilter=(bX=True,bY=True,bZ=True),ScaleFilter=(bX=False,bY=False,bZ=False))
    Parents=((Item=(Type=Bone,Name="head"),Weight=1))
    Weight=1
   )
   > &RigUnit_ParentConstraint(
    AdvancedSettings=(RotationOrderForFilter=XZY)
    Child=(Type=Control,Name="dyn_R_hair_1_ctrl")
    Filter=(TranslationFilter=(bX=True,bY=True,bZ=True),RotationFilter=(bX=True,bY=True,bZ=True),ScaleFilter=(bX=False,bY=False,bZ=False))
    Parents=((Item=(Type=Bone,Name="head"),Weight=1))
    Weight=1
   )
   > &RigUnit_ParentConstraint(
    AdvancedSettings=(RotationOrderForFilter=XZY)
    Child=(Type=Control,Name="dyn_L_hair_2_ctrl")
    Filter=(TranslationFilter=(bX=True,bY=True,bZ=True),RotationFilter=(bX=True,bY=True,bZ=True),ScaleFilter=(bX=False,bY=False,bZ=False))
    Parents=((Item=(Type=Bone,Name="head"),Weight=0.4),(Item=(Type=Bone,Name="spine_05"),Weight=0.6))
    Weight=1
   )
   > &RigUnit_ParentConstraint(
    AdvancedSettings=(RotationOrderForFilter=XZY)
    Child=(Type=Control,Name="dyn_C_hair_2_ctrl")
    Filter=(TranslationFilter=(bX=True,bY=True,bZ=True),RotationFilter=(bX=True,bY=True,bZ=True),ScaleFilter=(bX=False,bY=False,bZ=False))
    Parents=((Item=(Type=Bone,Name="head"),Weight=0.4),(Item=(Type=Bone,Name="spine_05"),Weight=0.6))
    Weight=1
   )
   > &RigUnit_ParentConstraint(
    AdvancedSettings=(RotationOrderForFilter=XZY)
    Child=(Type=Control,Name="dyn_R_hair_2_ctrl")
    Filter=(TranslationFilter=(bX=True,bY=True,bZ=True),RotationFilter=(bX=True,bY=True,bZ=True),ScaleFilter=(bX=False,bY=False,bZ=False))
    Parents=((Item=(Type=Bone,Name="head"),Weight=0.4),(Item=(Type=Bone,Name="spine_05"),Weight=0.6))
    Weight=1
   )
   > &RigUnit_ParentConstraint(
    AdvancedSettings=(RotationOrderForFilter=XZY)
    Child=(Type=Control,Name="dyn_L_hair_3_ctrl")
    Filter=(TranslationFilter=(bX=True,bY=True,bZ=True),RotationFilter=(bX=True,bY=True,bZ=True),ScaleFilter=(bX=False,bY=False,bZ=False))
    Parents=((Item=(Type=Bone,Name="upperarm_twist_01_l"),Weight=0.05),(Item=(Type=Bone,Name="spine_05"),Weight=0.35),(Item=(Type=Bone,Name="clavicle_l"),Weight=0.6))
    Weight=1
   )
   > &RigUnit_ParentConstraint(
    AdvancedSettings=(RotationOrderForFilter=XZY)
    Child=(Type=Control,Name="dyn_C_hair_3_ctrl")
    Filter=(TranslationFilter=(bX=True,bY=True,bZ=True),RotationFilter=(bX=True,bY=True,bZ=True),ScaleFilter=(bX=False,bY=False,bZ=False))
    Parents=((Item=(Type=Bone,Name="clavicle_l"),Weight=0.05),(Item=(Type=Bone,Name="clavicle_r"),Weight=0.05),(Item=(Type=Bone,Name="spine_05"),Weight=0.9))
    Weight=1
   )
   > &RigUnit_ParentConstraint(
    AdvancedSettings=(RotationOrderForFilter=XZY)
    Child=(Type=Control,Name="dyn_R_hair_3_ctrl")
    Filter=(TranslationFilter=(bX=True,bY=True,bZ=True),RotationFilter=(bX=True,bY=True,bZ=True),ScaleFilter=(bX=False,bY=False,bZ=False))
    Parents=((Item=(Type=Bone,Name="upperarm_twist_01_r"),Weight=0.05),(Item=(Type=Bone,Name="spine_05"),Weight=0.35),(Item=(Type=Bone,Name="clavicle_r"),Weight=0.6))
    Weight=1
   )
   > &RigUnit_ParentConstraint(
    AdvancedSettings=(RotationOrderForFilter=XZY)
    Child=(Type=Null,Name="dyn_L_hair_4_ctrl_null")
    Parents="((Item=(Type=Control,Name=\"dyn_L_hair_4_ctrl\"),Weight=0.0064),(Item=(Type=Bone,Name=\"deform_upperarm_bk_l\"),Weight=0.0543),(Item=(Type=Bone,Name=\"upperarm_twist_01_l\"),Weight=0.1277),(Item=(Type=Bone,Name=\"spine_05\"),Weight=0.1823),(Item=(Type=Bone,Name=\"s ...(truncated)"
    Weight=1
   )
   > &RigUnit_ParentConstraint(
    AdvancedSettings=(RotationOrderForFilter=XZY)
    Child=(Type=Control,Name="dyn_C_hair_4_ctrl")
    Parents="((Item=(Type=Bone,Name=\"spine_01\"),Weight=0.013),(Item=(Type=Bone,Name=\"spine_02\"),Weight=0.0436),(Item=(Type=Bone,Name=\"spine_03\"),Weight=0.0465),(Item=(Type=Bone,Name=\"clavicle_l\"),Weight=0.0833),(Item=(Type=Bone,Name=\"clavicle_r\"),Weight=0.1683),(Item=( ...(truncated)"
    Weight=1
   )
   > &RigUnit_ParentConstraint(
    AdvancedSettings=(RotationOrderForFilter=XZY)
    Child=(Type=Control,Name="dyn_R_hair_4_ctrl")
    Parents="((Item=(Type=Bone,Name=\"clavicle_l\"),Weight=0.0085),(Item=(Type=Bone,Name=\"deform_upperarm_bk_r\"),Weight=0.0119),(Item=(Type=Bone,Name=\"spine_02\"),Weight=0.0241),(Item=(Type=Bone,Name=\"spine_03\"),Weight=0.027),(Item=(Type=Bone,Name=\"upperarm_twist_01_r\"),W ...(truncated)"
    Weight=1
   )
   > &RigUnit_ParentConstraint(
    AdvancedSettings=(RotationOrderForFilter=XZY)
    Child=(Type=Control,Name="dyn_L_hair_5_ctrl")
    Filter=(TranslationFilter=(bX=True,bY=True,bZ=True),RotationFilter=(bX=True,bY=True,bZ=True),ScaleFilter=(bX=False,bY=False,bZ=False))
    Parents=((Item=(Type=Bone,Name="clavicle_l"),Weight=0.1),(Item=(Type=Bone,Name="spine_04"),Weight=0.15),(Item=(Type=Bone,Name="spine_05"),Weight=0.15),(Item=(Type=Bone,Name="spine_02"),Weight=0.2),(Item=(Type=Bone,Name="spine_03"),Weight=0.4))
    Weight=1
   )
   > &RigUnit_ParentConstraint(
    AdvancedSettings=(RotationOrderForFilter=XZY)
    Child=(Type=Control,Name="dyn_C_hair_5_ctrl")
    Filter=(TranslationFilter=(bX=True,bY=True,bZ=True),RotationFilter=(bX=True,bY=True,bZ=True),ScaleFilter=(bX=False,bY=False,bZ=False))
    Parents=((Item=(Type=Bone,Name="spine_04"),Weight=0.15),(Item=(Type=Bone,Name="spine_05"),Weight=0.1),(Item=(Type=Bone,Name="spine_02"),Weight=0.25),(Item=(Type=Bone,Name="spine_03"),Weight=0.5))
    Weight=1
   )
   > &RigUnit_ParentConstraint(
    AdvancedSettings=(RotationOrderForFilter=XZY)
    Child=(Type=Control,Name="dyn_R_hair_5_ctrl")
    Filter=(TranslationFilter=(bX=True,bY=True,bZ=True),RotationFilter=(bX=True,bY=True,bZ=True),ScaleFilter=(bX=False,bY=False,bZ=False))
    Parents=((Item=(Type=Bone,Name="clavicle_r"),Weight=0.1),(Item=(Type=Bone,Name="spine_04"),Weight=0.15),(Item=(Type=Bone,Name="spine_05"),Weight=0.15),(Item=(Type=Bone,Name="spine_02"),Weight=0.2),(Item=(Type=Bone,Name="spine_03"),Weight=0.4))
    Weight=1
   )
   > &RigUnit_ParentConstraint(
    AdvancedSettings=(RotationOrderForFilter=XZY)
    Child=(Type=Control,Name="dyn_L_hair_6_ctrl")
    Filter=(TranslationFilter=(bX=True,bY=True,bZ=True),RotationFilter=(bX=True,bY=True,bZ=True),ScaleFilter=(bX=False,bY=False,bZ=False))
    Parents=((Item=(Type=Bone,Name="spine_02"),Weight=0.075),(Item=(Type=Bone,Name="spine_01"),Weight=0.425),(Item=(Type=Bone,Name="pelvis"),Weight=0.5))
    Weight=1
   )
   > &RigUnit_ParentConstraint(
    AdvancedSettings=(RotationOrderForFilter=XZY)
    Child=(Type=Control,Name="dyn_C_hair_6_ctrl")
    Filter=(TranslationFilter=(bX=True,bY=True,bZ=True),RotationFilter=(bX=True,bY=True,bZ=True),ScaleFilter=(bX=False,bY=False,bZ=False))
    Parents=((Item=(Type=Bone,Name="spine_02"),Weight=0.075),(Item=(Type=Bone,Name="spine_01"),Weight=0.425),(Item=(Type=Bone,Name="pelvis"),Weight=0.5))
    Weight=1
   )
   > &RigUnit_ParentConstraint(
    AdvancedSettings=(RotationOrderForFilter=XZY)
    Child=(Type=Control,Name="dyn_R_hair_6_ctrl")
    Filter=(TranslationFilter=(bX=True,bY=True,bZ=True),RotationFilter=(bX=True,bY=True,bZ=True),ScaleFilter=(bX=False,bY=False,bZ=False))
    Parents=((Item=(Type=Bone,Name="spine_02"),Weight=0.075),(Item=(Type=Bone,Name="spine_01"),Weight=0.425),(Item=(Type=Bone,Name="pelvis"),Weight=0.5))
    Weight=1
   )
   > &RigUnit_ParentConstraint(
    AdvancedSettings=(RotationOrderForFilter=XZY)
    Child=(Type=Control,Name="dyn_L_hair_7_ctrl")
    Filter=(TranslationFilter=(bX=True,bY=True,bZ=True),RotationFilter=(bX=True,bY=True,bZ=True),ScaleFilter=(bX=False,bY=False,bZ=False))
    Parents=((Item=(Type=Bone,Name="spine_01"),Weight=0.02),(Item=(Type=Bone,Name="pelvis"),Weight=0.98))
    Weight=1
   )
   > &RigUnit_ParentConstraint(
    AdvancedSettings=(RotationOrderForFilter=XZY)
    Child=(Type=Control,Name="dyn_C_hair_7_ctrl")
    Filter=(TranslationFilter=(bX=True,bY=True,bZ=True),RotationFilter=(bX=True,bY=True,bZ=True),ScaleFilter=(bX=False,bY=False,bZ=False))
    Parents=((Item=(Type=Bone,Name="spine_01"),Weight=0.02),(Item=(Type=Bone,Name="pelvis"),Weight=0.98))
    Weight=1
   )
   > &RigUnit_ParentConstraint(
    AdvancedSettings=(RotationOrderForFilter=XZY)
    Child=(Type=Control,Name="dyn_R_hair_7_ctrl")
    Filter=(TranslationFilter=(bX=True,bY=True,bZ=True),RotationFilter=(bX=True,bY=True,bZ=True),ScaleFilter=(bX=False,bY=False,bZ=False))
    Parents=((Item=(Type=Bone,Name="spine_01"),Weight=0.02),(Item=(Type=Bone,Name="pelvis"),Weight=0.98))
    Weight=1
   )
   > RigVMFunctionReturnNode()
  ]}
 CollapseNode_ContainedGraph {nodes [
   RigVMFunctionEntryNode()
   > &RigUnit_ParentConstraint(
    AdvancedSettings=(RotationOrderForFilter=XZY)
    Child=(Type=Null,Name="dyn_L_hair_1_ctrl_null")
    Filter=(TranslationFilter=(bX=True,bY=True,bZ=True),RotationFilter=(bX=True,bY=True,bZ=True),ScaleFilter=(bX=False,bY=False,bZ=False))
    Parents=((Item=(Type=Bone,Name="head"),Weight=1))
    Weight=1
   )
   > &RigUnit_ParentConstraint(
    AdvancedSettings=(RotationOrderForFilter=XZY)
    Child=(Type=Null,Name="dyn_C_hair_1_ctrl_null")
    Filter=(TranslationFilter=(bX=True,bY=True,bZ=True),RotationFilter=(bX=True,bY=True,bZ=True),ScaleFilter=(bX=False,bY=False,bZ=False))
    Parents=((Item=(Type=Bone,Name="head"),Weight=1))
    Weight=1
   )
   > &RigUnit_ParentConstraint(
    AdvancedSettings=(RotationOrderForFilter=XZY)
    Child=(Type=Null,Name="dyn_R_hair_1_ctrl_null")
    Filter=(TranslationFilter=(bX=True,bY=True,bZ=True),RotationFilter=(bX=True,bY=True,bZ=True),ScaleFilter=(bX=False,bY=False,bZ=False))
    Parents=((Item=(Type=Bone,Name="head"),Weight=1))
    Weight=1
   )
   > &RigUnit_ParentConstraint(
    AdvancedSettings=(RotationOrderForFilter=XZY)
    Child=(Type=Null,Name="dyn_L_hair_2_ctrl_null")
    Filter=(TranslationFilter=(bX=True,bY=True,bZ=True),RotationFilter=(bX=True,bY=True,bZ=True),ScaleFilter=(bX=False,bY=False,bZ=False))
    Parents=((Item=(Type=Bone,Name="head"),Weight=0.5),(Item=(Type=Bone,Name="spine_05"),Weight=0.5))
    Weight=1
   )
   > &RigUnit_ParentConstraint(
    AdvancedSettings=(RotationOrderForFilter=XZY)
    Child=(Type=Null,Name="dyn_C_hair_2_ctrl_null")
    Filter=(TranslationFilter=(bX=True,bY=True,bZ=True),RotationFilter=(bX=True,bY=True,bZ=True),ScaleFilter=(bX=False,bY=False,bZ=False))
    Parents=((Item=(Type=Bone,Name="head"),Weight=0.5),(Item=(Type=Bone,Name="spine_05"),Weight=0.5))
    Weight=1
   )
   > &RigUnit_ParentConstraint(
    AdvancedSettings=(RotationOrderForFilter=XZY)
    Child=(Type=Null,Name="dyn_R_hair_2_ctrl_null")
    Filter=(TranslationFilter=(bX=True,bY=True,bZ=True),RotationFilter=(bX=True,bY=True,bZ=True),ScaleFilter=(bX=False,bY=False,bZ=False))
    Parents=((Item=(Type=Bone,Name="head"),Weight=0.5),(Item=(Type=Bone,Name="spine_05"),Weight=0.5))
    Weight=1
   )
   > &RigUnit_ParentConstraint(
    AdvancedSettings=(RotationOrderForFilter=XZY)
    Child=(Type=Null,Name="dyn_L_hair_3_ctrl_null")
    Filter=(TranslationFilter=(bX=True,bY=True,bZ=True),RotationFilter=(bX=True,bY=True,bZ=True),ScaleFilter=(bX=False,bY=False,bZ=False))
    Parents=((Item=(Type=Bone,Name="upperarm_twist_01_l"),Weight=0.05),(Item=(Type=Bone,Name="spine_05"),Weight=0.35),(Item=(Type=Bone,Name="clavicle_l"),Weight=0.6))
    Weight=1
   )
   > &RigUnit_ParentConstraint(
    AdvancedSettings=(RotationOrderForFilter=XZY)
    Child=(Type=Null,Name="dyn_C_hair_3_ctrl_null")
    Filter=(TranslationFilter=(bX=True,bY=True,bZ=True),RotationFilter=(bX=True,bY=True,bZ=True),ScaleFilter=(bX=False,bY=False,bZ=False))
    Parents=((Item=(Type=Bone,Name="clavicle_l"),Weight=0.05),(Item=(Type=Bone,Name="clavicle_r"),Weight=0.05),(Item=(Type=Bone,Name="spine_05"),Weight=0.9))
    Weight=1
   )
   > &RigUnit_ParentConstraint(
    AdvancedSettings=(RotationOrderForFilter=XZY)
    Child=(Type=Null,Name="dyn_R_hair_3_ctrl_null")
    Filter=(TranslationFilter=(bX=True,bY=True,bZ=True),RotationFilter=(bX=True,bY=True,bZ=True),ScaleFilter=(bX=False,bY=False,bZ=False))
    Parents=((Item=(Type=Bone,Name="upperarm_twist_01_r"),Weight=0.05),(Item=(Type=Bone,Name="spine_05"),Weight=0.35),(Item=(Type=Bone,Name="clavicle_r"),Weight=0.6))
    Weight=1
   )
   > &RigUnit_ParentConstraint(
    AdvancedSettings=(RotationOrderForFilter=XZY)
    Child=(Type=Null,Name="dyn_L_hair_4_ctrl_null")
    Parents="((Item=(Type=Bone,Name=\"spine_03\"),Weight=0.0064),(Item=(Type=Bone,Name=\"deform_upperarm_bk_l\"),Weight=0.0543),(Item=(Type=Bone,Name=\"upperarm_twist_01_l\"),Weight=0.1277),(Item=(Type=Bone,Name=\"spine_05\"),Weight=0.1823),(Item=(Type=Bone,Name=\"spine_04\"),We ...(truncated)"
    Weight=1
   )
   > &RigUnit_ParentConstraint(
    AdvancedSettings=(RotationOrderForFilter=XZY)
    Child=(Type=Null,Name="dyn_C_hair_4_ctrl_null")
    Parents="((Item=(Type=Bone,Name=\"spine_01\"),Weight=0.013),(Item=(Type=Bone,Name=\"spine_02\"),Weight=0.0436),(Item=(Type=Bone,Name=\"spine_03\"),Weight=0.0465),(Item=(Type=Bone,Name=\"clavicle_l\"),Weight=0.0833),(Item=(Type=Bone,Name=\"clavicle_r\"),Weight=0.1683),(Item=( ...(truncated)"
    Weight=1
   )
   > &RigUnit_ParentConstraint(
    AdvancedSettings=(RotationOrderForFilter=XZY)
    Child=(Type=Null,Name="dyn_R_hair_4_ctrl_null")
    Parents="((Item=(Type=Bone,Name=\"clavicle_l\"),Weight=0.0085),(Item=(Type=Bone,Name=\"deform_upperarm_bk_r\"),Weight=0.0119),(Item=(Type=Bone,Name=\"spine_02\"),Weight=0.0241),(Item=(Type=Bone,Name=\"spine_03\"),Weight=0.027),(Item=(Type=Bone,Name=\"upperarm_twist_01_r\"),W ...(truncated)"
    Weight=1
   )
   > &RigUnit_ParentConstraint(
    AdvancedSettings=(RotationOrderForFilter=XZY)
    Child=(Type=Null,Name="dyn_L_hair_5_ctrl_null")
    Filter=(TranslationFilter=(bX=True,bY=True,bZ=True),RotationFilter=(bX=True,bY=True,bZ=True),ScaleFilter=(bX=False,bY=False,bZ=False))
    Parents=((Item=(Type=Bone,Name="clavicle_l"),Weight=0.1),(Item=(Type=Bone,Name="spine_04"),Weight=0.15),(Item=(Type=Bone,Name="spine_05"),Weight=0.15),(Item=(Type=Bone,Name="spine_02"),Weight=0.2),(Item=(Type=Bone,Name="spine_03"),Weight=0.4))
    Weight=1
   )
   > &RigUnit_ParentConstraint(
    AdvancedSettings=(RotationOrderForFilter=XZY)
    Child=(Type=Null,Name="dyn_C_hair_5_ctrl_null")
    Filter=(TranslationFilter=(bX=True,bY=True,bZ=True),RotationFilter=(bX=True,bY=True,bZ=True),ScaleFilter=(bX=False,bY=False,bZ=False))
    Parents=((Item=(Type=Bone,Name="spine_04"),Weight=0.15),(Item=(Type=Bone,Name="spine_05"),Weight=0.1),(Item=(Type=Bone,Name="spine_02"),Weight=0.25),(Item=(Type=Bone,Name="spine_03"),Weight=0.5))
    Weight=1
   )
   > &RigUnit_ParentConstraint(
    AdvancedSettings=(RotationOrderForFilter=XZY)
    Child=(Type=Null,Name="dyn_R_hair_5_ctrl_null")
    Filter=(TranslationFilter=(bX=True,bY=True,bZ=True),RotationFilter=(bX=True,bY=True,bZ=True),ScaleFilter=(bX=False,bY=False,bZ=False))
    Parents=((Item=(Type=Bone,Name="clavicle_r"),Weight=0.1),(Item=(Type=Bone,Name="spine_04"),Weight=0.15),(Item=(Type=Bone,Name="spine_05"),Weight=0.15),(Item=(Type=Bone,Name="spine_02"),Weight=0.2),(Item=(Type=Bone,Name="spine_03"),Weight=0.4))
    Weight=1
   )
   > &RigUnit_ParentConstraint(
    AdvancedSettings=(RotationOrderForFilter=XZY)
    Child=(Type=Null,Name="dyn_L_hair_6_ctrl_null")
    Filter=(TranslationFilter=(bX=True,bY=True,bZ=True),RotationFilter=(bX=True,bY=True,bZ=True),ScaleFilter=(bX=False,bY=False,bZ=False))
    Parents=((Item=(Type=Bone,Name="spine_02"),Weight=0.075),(Item=(Type=Bone,Name="spine_01"),Weight=0.425),(Item=(Type=Bone,Name="pelvis"),Weight=0.5))
    Weight=1
   )
   > &RigUnit_ParentConstraint(
    AdvancedSettings=(RotationOrderForFilter=XZY)
    Child=(Type=Null,Name="dyn_C_hair_6_ctrl_null")
    Filter=(TranslationFilter=(bX=True,bY=True,bZ=True),RotationFilter=(bX=True,bY=True,bZ=True),ScaleFilter=(bX=False,bY=False,bZ=False))
    Parents=((Item=(Type=Bone,Name="spine_02"),Weight=0.075),(Item=(Type=Bone,Name="spine_01"),Weight=0.425),(Item=(Type=Bone,Name="pelvis"),Weight=0.5))
    Weight=1
   )
   > &RigUnit_ParentConstraint(
    AdvancedSettings=(RotationOrderForFilter=XZY)
    Child=(Type=Null,Name="dyn_R_hair_6_ctrl_null")
    Filter=(TranslationFilter=(bX=True,bY=True,bZ=True),RotationFilter=(bX=True,bY=True,bZ=True),ScaleFilter=(bX=False,bY=False,bZ=False))
    Parents=((Item=(Type=Bone,Name="spine_02"),Weight=0.075),(Item=(Type=Bone,Name="spine_01"),Weight=0.425),(Item=(Type=Bone,Name="pelvis"),Weight=0.5))
    Weight=1
   )
   > &RigUnit_ParentConstraint(
    AdvancedSettings=(RotationOrderForFilter=XZY)
    Child=(Type=Null,Name="dyn_L_hair_7_ctrl_null")
    Filter=(TranslationFilter=(bX=True,bY=True,bZ=True),RotationFilter=(bX=True,bY=True,bZ=True),ScaleFilter=(bX=False,bY=False,bZ=False))
    Parents=((Item=(Type=Bone,Name="spine_01"),Weight=0.02),(Item=(Type=Bone,Name="pelvis"),Weight=0.98))
    Weight=1
   )
   > &RigUnit_ParentConstraint(
    AdvancedSettings=(RotationOrderForFilter=XZY)
    Child=(Type=Null,Name="dyn_C_hair_7_ctrl_null")
    Filter=(TranslationFilter=(bX=True,bY=True,bZ=True),RotationFilter=(bX=True,bY=True,bZ=True),ScaleFilter=(bX=False,bY=False,bZ=False))
    Parents=((Item=(Type=Bone,Name="spine_01"),Weight=0.02),(Item=(Type=Bone,Name="pelvis"),Weight=0.98))
    Weight=1
   )
   > &RigUnit_ParentConstraint(
    AdvancedSettings=(RotationOrderForFilter=XZY)
    Child=(Type=Null,Name="dyn_R_hair_7_ctrl_null")
    Filter=(TranslationFilter=(bX=True,bY=True,bZ=True),RotationFilter=(bX=True,bY=True,bZ=True),ScaleFilter=(bX=False,bY=False,bZ=False))
    Parents=((Item=(Type=Bone,Name="spine_01"),Weight=0.02),(Item=(Type=Bone,Name="pelvis"),Weight=0.98))
    Weight=1
   )
   > RigVMFunctionReturnNode()
   &RigUnit_ParentConstraint(
    AdvancedSettings=(RotationOrderForFilter=XZY)
    Child=(Type=Null,Name="dyn_L_hair_4_ctrl_null")
    ExecutePin=()
    Filter=(TranslationFilter=(bX=True,bY=True,bZ=True),RotationFilter=(bX=True,bY=True,bZ=True),ScaleFilter=(bX=False,bY=False,bZ=False))
    Parents="((Item=(Type=Bone,Name=\"spine_02\"),Weight=0.0014),(Item=(Type=Bone,Name=\"spine_03\"),Weight=0.0235),(Item=(Type=Bone,Name=\"spine_04\"),Weight=0.0379),(Item=(Type=Bone,Name=\"upperarm_twist_01_l\"),Weight=0.0733),(Item=(Type=Bone,Name=\"spine_05\"),Weight=0.3523) ...(truncated)"
    Weight=1
   )
   > &RigUnit_ParentConstraint(
    AdvancedSettings=(RotationOrderForFilter=XZY)
    Child=(Type=Null,Name="dyn_C_hair_4_ctrl_null")
    Filter=(TranslationFilter=(bX=True,bY=True,bZ=True),RotationFilter=(bX=True,bY=True,bZ=True),ScaleFilter=(bX=False,bY=False,bZ=False))
    Parents="((Item=(Type=Bone,Name=\"spine_01\"),Weight=0.0001),(Item=(Type=Bone,Name=\"spine_02\"),Weight=0.0068),(Item=(Type=Bone,Name=\"clavicle_l\"),Weight=0.063),(Item=(Type=Bone,Name=\"clavicle_r\"),Weight=0.0705),(Item=(Type=Bone,Name=\"spine_03\"),Weight=0.1021),(Item=( ...(truncated)"
    Weight=1
   )
   > &RigUnit_ParentConstraint(
    AdvancedSettings=(RotationOrderForFilter=XZY)
    Child=(Type=Null,Name="dyn_R_hair_4_ctrl_null")
    Filter=(TranslationFilter=(bX=True,bY=True,bZ=True),RotationFilter=(bX=True,bY=True,bZ=True),ScaleFilter=(bX=False,bY=False,bZ=False))
    Parents="((Item=(Type=Bone,Name=\"spine_02\"),Weight=0.0014),(Item=(Type=Bone,Name=\"spine_03\"),Weight=0.0235),(Item=(Type=Bone,Name=\"spine_04\"),Weight=0.038),(Item=(Type=Bone,Name=\"upperarm_twist_01_r\"),Weight=0.0733),(Item=(Type=Bone,Name=\"spine_05\"),Weight=0.3528), ...(truncated)"
    Weight=1
   )
  ]}
 CollapseNode_ContainedGraph {nodes [
   RigVMFunctionEntryNode()
   n1 = &RigVMFunction_AlphaInterp(
    InterpSpeedDecreasing=10
    InterpSpeedIncreasing=15
    Scale=1
    bClampResult=False
    Value=&RigVMFunction_MathBoolToFloat(Value=RigVMVariableNode() {variable=pub isTargeting}.Value).Result
   )
   > RigVMVariableNode(Value=@n1.Result) {variable=pub isTargeting_lerp}
   > RigVMFunctionReturnNode()
  ]}
 New Function_ContainedGraph {nodes [
   n1 = RigVMFunctionEntryNode()
   RigVMFunction_ControlFlowBranch = &RigVMFunction_ControlFlowBranch(ExecuteContext=@n1.ExecuteContext, Condition=@n1.debug)
   RigVMFunctionReturnNode(ExecuteContext=@RigVMFunction_ControlFlowBranch.Completed)
   RigUnit_ItemArray = &RigUnit_ItemArray(Items="((Type=Bone,Name=\"spine_01\"),(Type=Bone,Name=\"spine_02\"),(Type=Bone,Name=\"spine_03\"),(Type=Bone,Name=\"spine_04\"),(Type=Bone,Name=\"spine_05\"),(Type=Bone,Name=\"neck_01\"),(Type=Bone,Name=\"neck_02\"),(Type=Bone,Name=\"head\"),(Type=Bone,Name=\"clavicle_l\"),(Type=B ...(truncated)")
   DISPATCH_RigVMDispatch_ArrayIterator = RigVMDispatchNode(Array=@RigUnit_ItemArray.Items, ExecuteContext=@RigVMFunction_ControlFlowBranch.True) {template="DISPATCH_RigVMDispatch_ArrayIterator(in Array,out Element,out Index,out Count,out Ratio)"}
   Concat = /Script/RigVM.RigVMFunction_NameConcat(B=_ctrl, A=@DISPATCH_RigVMDispatch_ArrayIterator.Element.Name) {is=struct RigVMUnitNode, template="Concat::Execute(in A,in B,out Result)"}
   > &RigUnit_SetTransform(
    Item=(Type=Bone,Name="spine_01")
    Weight=1
    bInitial=False
    Item.Name=@DISPATCH_RigVMDispatch_ArrayIterator.Element.Name
    Value=&RigUnit_GetTransform(Item=(Type=Control,Name="spine_01_ctrl"), bInitial=False, Item.Name=@Concat.Result).Transform
   )
   &RigUnit_SetTransform(
    Item=(Type=Bone,Name="pelvis")
    Weight=1
    bInitial=False
    bPropagateToChildren=false
    Value=&RigUnit_GetTransform(Item=(Type=Control,Name="pelvis_ctrl"), bInitial=False).Transform
    ExecutePin=@DISPATCH_RigVMDispatch_ArrayIterator.Completed
   )
  ]}
 New Function_ContainedGraph {nodes [
   RigVMFunctionEntryNode()
   > &RigUnit_ModifyTransforms(
    ItemToModify=((Item=(Type=Bone,Name="head"),Transform=(Rotation=(X=0,Y=0,Z=0,W=1),Translation=(X=0,Y=0,Z=0),Scale3D=(X=1,Y=1,Z=1))))
    Weight=1
    WeightMinimum=0
   )
   > RigVMCollapseNode()
   > RigVMFunction_Sequence = RigVMAggregateNode()
   TwistBones_2 = RigVMFunctionReferenceNode(
    Axis to Twist=(bX=true,bY=true,bZ=true)
    Bones to Twist=((Type=Bone,Name="dyn_C_hair_3"),(Type=Bone,Name="dyn_C_hair_4"),(Type=Bone,Name="dyn_C_hair_5"),(Type=Bone,Name="dyn_C_hair_6"),(Type=Bone,Name="dyn_C_hair_7"))
    End=(Type=Bone,Name="neck_01")
    Start=(Type=Bone,Name="pelvis")
    ExecuteContext=@RigVMFunction_Sequence.A
   )
   > TwistBones_2_1 = RigVMFunctionReferenceNode(
    Axis to Twist=(bX=True,bY=true,bZ=true)
    Bones to Twist=((Type=Bone,Name="dyn_L_hair_3"),(Type=Bone,Name="dyn_L_hair_4"),(Type=Bone,Name="dyn_L_hair_5"),(Type=Bone,Name="dyn_L_hair_6"),(Type=Bone,Name="dyn_L_hair_7"))
    End=(Type=Bone,Name="neck_01")
    Start=(Type=Bone,Name="pelvis")
   )
   > TwistBones_2_1_1 = RigVMFunctionReferenceNode(
    Axis to Twist=(bX=True,bY=true,bZ=true)
    Bones to Twist=((Type=Bone,Name="dyn_R_hair_3"),(Type=Bone,Name="dyn_R_hair_4"),(Type=Bone,Name="dyn_R_hair_5"),(Type=Bone,Name="dyn_R_hair_6"),(Type=Bone,Name="dyn_R_hair_7"))
    End=(Type=Bone,Name="neck_01")
    Start=(Type=Bone,Name="pelvis")
   )
   n1 = &RigUnit_ItemArray(Items=((Type=Bone,Name="dyn_C_hair_1"),(Type=Bone,Name="dyn_C_hair_2"),(Type=Bone,Name="dyn_C_hair_3"),(Type=Bone,Name="dyn_C_hair_4"),(Type=Bone,Name="dyn_C_hair_5"),(Type=Bone,Name="dyn_C_hair_6"),(Type=Bone,Name="dyn_C_hair_7")))
   SplineIK_2 = RigVMFunctionReferenceNode(
    Primary Axis=(X=1,Y=0,Z=0)
    Samples Per Segment=1
    Secondary Spline Direction=(X=0,Y=0,Z=0)
    Spline Mode=BSpline
    Stretch=false
    Up Axis=(X=0,Y=0,Z=0)
    debug=false
    Bones=@n1.Items
    ExecuteContext=@RigVMFunction_Sequence.B
    Controls=RigVMVariableNode() {variable=c_dyn_hair_controls}.Value
   )
   RigUnit_ItemArray = &RigUnit_ItemArray(Items=((Type=Bone,Name="dyn_L_hair_1"),(Type=Bone,Name="dyn_L_hair_2"),(Type=Bone,Name="dyn_L_hair_3"),(Type=Bone,Name="dyn_L_hair_4"),(Type=Bone,Name="dyn_L_hair_5"),(Type=Bone,Name="dyn_L_hair_6"),(Type=Bone,Name="dyn_L_hair_7")))
   > SplineIK_2_1 = RigVMFunctionReferenceNode(
    Primary Axis=(X=1,Y=0,Z=0)
    Samples Per Segment=1
    Secondary Spline Direction=(X=0,Y=0,Z=0)
    Spline Mode=BSpline
    Stretch=false
    Up Axis=(X=0,Y=0,Z=0)
    debug=false
    Bones=@RigUnit_ItemArray.Items
    Controls=RigVMVariableNode() {variable=l_dyn_hair_controls}.Value
   )
   n2 = &RigUnit_ItemArray(Items=((Type=Bone,Name="dyn_R_hair_1"),(Type=Bone,Name="dyn_R_hair_2"),(Type=Bone,Name="dyn_R_hair_3"),(Type=Bone,Name="dyn_R_hair_4"),(Type=Bone,Name="dyn_R_hair_5"),(Type=Bone,Name="dyn_R_hair_6"),(Type=Bone,Name="dyn_R_hair_7")))
   > SplineIK_2_1_1 = RigVMFunctionReferenceNode(
    Primary Axis=(X=1,Y=0,Z=0)
    Samples Per Segment=1
    Secondary Spline Direction=(X=0,Y=0,Z=0)
    Spline Mode=BSpline
    Stretch=false
    Up Axis=(X=0,Y=0,Z=0)
    debug=false
    Bones=@n2.Items
    Controls=RigVMVariableNode() {variable=r_dyn_hair_controls}.Value
   )
   DeltaRotation = RigVMFunctionReferenceNode(
    Rotation Order=ZYX
    child=spine_02
    max=80
    parent=spine_02_null
    parentType=Null
    primAxis=2
    twistAxis=(X=1,Y=0,Z=0)
    ExecuteContext=@RigVMFunction_Sequence.C
   )
   RigVMFunction_MathQuaternionFromEuler = &RigVMFunction_MathQuaternionFromEuler(Euler=(X=0,Y=40,Z=0))
   n3 = /Script/RigVM.RigVMFunction_MathVectorMake(X=-2, Y=0, Z=-5) {is=struct RigVMUnitNode}
   > &RigUnit_ModifyTransforms(
    ItemToModify="((Item=(Type=Bone,Name=\"dyn_C_hair_6\"),Transform=(Rotation=(X=0,Y=0,Z=0,W=1),Translation=(X=0,Y=0,Z=0),Scale3D=(X=1,Y=1,Z=1))),(Item=(Type=Bone,Name=\"dyn_L_hair_6\"),Transform=(Rotation=(X=0,Y=0,Z=0,W=1),Translation=(X=0,Y=0,Z=0),Scale3D=(X=1,Y=1,Z=1))),(It ...(truncated)"
    WeightMinimum=0
    Weight=@DeltaRotation.outputWeight
    ItemToModify.0.Transform.Rotation=@RigVMFunction_MathQuaternionFromEuler.Result
    ItemToModify.1.Transform.Rotation=@RigVMFunction_MathQuaternionFromEuler.Result
    ItemToModify.2.Transform.Rotation=@RigVMFunction_MathQuaternionFromEuler.Result
    ItemToModify.0.Transform.Translation=@n3.Result
    ItemToModify.1.Transform.Translation=@n3.Result
    ItemToModify.2.Transform.Translation=@n3.Result
   )
   RigVMFunctionReturnNode(ExecuteContext=@RigVMFunction_Sequence.D)
   RigVMCollapseNode(ExecuteContext=())
   &RigVMFunction_Sequence(ExecuteContext=())
   &RigUnit_SphericalPoseReader(
    ActiveRegionScaleFactors=(PositiveWidth=1,NegativeWidth=1,PositiveHeight=1,NegativeHeight=1)
    ActiveRegionSize=0.15
    Debug=(bDrawDebug=false,bDraw2D=False,bDrawLocalAxes=False,DebugScale=25,DebugSegments=20,DebugThickness=0.25)
    DriverAxis=(X=1,Y=0,Z=0)
    DriverItem=(Type=Bone,Name="spine_02")
    ExecutePin=()
    FalloffRegionScaleFactors=(PositiveWidth=1,NegativeWidth=1,PositiveHeight=1,NegativeHeight=1)
    FalloffSize=0.35
    FlipHeightScaling=False
    FlipWidthScaling=False
    OptionalParentItem=(Type=Bone,Name="pelvis")
    RotationOffset=(X=90,Y=0,Z=0)
   )
  ]}
 New Function_ContainedGraph {nodes [
   RigVMFunctionEntryNode()
   > RigVMFunction_Sequence_2 = RigVMAggregateNode()
   n1 = &RigVMFunction_ControlFlowBranch(ExecuteContext=@RigVMFunction_Sequence_2.A, Condition=RigVMVariableNode() {variable=pub diveFloating}.Value)
   claviclesBackFoward_3 = RigVMFunctionReferenceNode(ExecuteContext=@n1.True)
   n2 = &RigVMFunction_ControlFlowBranch(ExecuteContext=@RigVMFunction_Sequence_2.B, Condition=RigVMVariableNode() {variable=pub disableUpperLimbsCorrection}.Value)
   n3 = &RigVMFunction_Sequence(ExecuteContext=@n2.False)
   n4 = &RigUnit_SphericalPoseReader(
    ActiveRegionScaleFactors=(PositiveWidth=1,NegativeWidth=1,PositiveHeight=1,NegativeHeight=1)
    ActiveRegionSize=0.1
    Debug=(bDrawDebug=true,bDraw2D=False,bDrawLocalAxes=False,DebugScale=25,DebugSegments=20,DebugThickness=0.25)
    DriverAxis=(X=1,Y=0,Z=0)
    DriverItem=(Type=Bone,Name="upperarm_l")
    FalloffRegionScaleFactors=(PositiveWidth=1,NegativeWidth=1,PositiveHeight=1,NegativeHeight=1)
    FalloffSize=0.3
    FlipHeightScaling=False
    FlipWidthScaling=False
    OptionalParentItem=(Type=Bone,Name="None")
    RotationOffset=(X=-140,Y=50,Z=0)
    ExecutePin=@n3.A
   )
   > alterTransform_9 = RigVMFunctionReferenceNode(
    type=Bone
    transforms="((Bone=\"dyn_L_hair_2\",Transform=(Rotation=(X=0,Y=0,Z=0,W=1),Translation=(X=0,Y=2,Z=0),Scale3D=(X=1,Y=1,Z=1))),(Bone=\"dyn_L_hair_3\",Transform=(Rotation=(X=0,Y=0,Z=0,W=1),Translation=(X=0,Y=4,Z=0),Scale3D=(X=1,Y=1,Z=1))),(Bone=\"dyn_L_hair_4\",Transform=(Rotat ...(truncated)"
    weight=@n4.OutputParam
   )
   > n5 = &RigUnit_SphericalPoseReader(
    ActiveRegionScaleFactors=(PositiveWidth=1,NegativeWidth=1,PositiveHeight=1,NegativeHeight=1)
    ActiveRegionSize=0.2
    Debug=(bDrawDebug=true,DebugScale=25,DebugSegments=20,DebugThickness=0.25)
    DriverAxis=(X=-1,Y=0,Z=0)
    DriverItem=(Type=Bone,Name="upperarm_r")
    FalloffRegionScaleFactors=(PositiveWidth=1,NegativeWidth=1,PositiveHeight=1,NegativeHeight=1)
    FalloffSize=0.3
    FlipHeightScaling=false
    FlipWidthScaling=False
    OptionalParentItem=(Type=Bone,Name="None")
    RotationOffset=(X=40,Y=50,Z=0)
   )
   > alterTransform_10 = RigVMFunctionReferenceNode(
    type=Bone
    transforms="((Bone=\"dyn_R_hair_2\",Transform=(Rotation=(X=0,Y=0,Z=0,W=1),Translation=(X=0,Y=-2,Z=0),Scale3D=(X=1,Y=1,Z=1))),(Bone=\"dyn_R_hair_3\",Transform=(Rotation=(X=0,Y=0,Z=0,W=1),Translation=(X=0,Y=2,Z=0),Scale3D=(X=1,Y=1,Z=1))),(Bone=\"dyn_R_hair_4\",Transform=(Rota ...(truncated)"
    weight=@n5.OutputParam
   )
   > &RigUnit_ModifyTransforms(
    ItemToModify="((Item=(Type=Bone,Name=\"dyn_L_hair_2\"),Transform=(Rotation=(X=0,Y=0,Z=0,W=1),Translation=(X=0,Y=1,Z=0),Scale3D=(X=1,Y=1,Z=1))),(Item=(Type=Bone,Name=\"dyn_L_hair_3\"),Transform=(Rotation=(X=0,Y=0,Z=0,W=1),Translation=(X=0,Y=1,Z=0),Scale3D=(X=1,Y=1,Z=1))),(It ...(truncated)"
    WeightMinimum=0
    Weight=@n4.OutputParam
   )
   > &RigUnit_ModifyTransforms(
    ItemToModify="((Item=(Type=Bone,Name=\"dyn_R_hair_2\"),Transform=(Rotation=(X=0,Y=0,Z=0,W=1),Translation=(X=0,Y=-2,Z=0),Scale3D=(X=1,Y=1,Z=1))),(Item=(Type=Bone,Name=\"dyn_R_hair_3\"),Transform=(Rotation=(X=0,Y=0,Z=0,W=1),Translation=(X=0,Y=-2,Z=0),Scale3D=(X=1,Y=1,Z=1))),( ...(truncated)"
    WeightMinimum=0
    Weight=@n5.OutputParam
   )
   claviclesBackFoward_2 = RigVMFunctionReferenceNode(ExecuteContext=@n3.B)
   > n6 = &RigUnit_SphericalPoseReader(
    ActiveRegionScaleFactors=(PositiveWidth=1,NegativeWidth=1,PositiveHeight=1,NegativeHeight=1)
    ActiveRegionSize=0.1
    Debug=(bDrawDebug=false,bDraw2D=False,bDrawLocalAxes=False,DebugScale=25,DebugSegments=20,DebugThickness=0.25)
    DriverAxis=(X=1,Y=0,Z=0)
    DriverItem=(Type=Bone,Name="clavicle_l")
    FalloffRegionScaleFactors=(PositiveWidth=0.5,NegativeWidth=0.5,PositiveHeight=0.5,NegativeHeight=0.5)
    FalloffSize=0.3
    FlipHeightScaling=False
    FlipWidthScaling=False
    OptionalParentItem=(Type=Bone,Name="None")
    RotationOffset=(X=-45,Y=6,Z=60)
   )
   > alterTransform_13 = RigVMFunctionReferenceNode(
    type=Bone
    transforms=((Bone="dyn_L_hair_2",Transform=(Rotation=(X=0,Y=0,Z=0,W=1),Translation=(X=0,Y=0,Z=-1),Scale3D=(X=1,Y=1,Z=1))),(Bone="dyn_L_hair_3",Transform=(Rotation=(X=0,Y=0,Z=0,W=1),Translation=(X=0,Y=2,Z=-1),Scale3D=(X=1,Y=1,Z=1))))
    weight=@n6.OutputParam
   )
   > n7 = &RigUnit_SphericalPoseReader(
    ActiveRegionScaleFactors=(PositiveWidth=1,NegativeWidth=1,PositiveHeight=1,NegativeHeight=1)
    ActiveRegionSize=0.1
    Debug=(bDrawDebug=false,bDraw2D=False,bDrawLocalAxes=False,DebugScale=25,DebugSegments=20,DebugThickness=0.25)
    DriverAxis=(X=-1,Y=0,Z=0)
    DriverItem=(Type=Bone,Name="clavicle_r")
    FalloffRegionScaleFactors=(PositiveWidth=0.5,NegativeWidth=0.5,PositiveHeight=0.5,NegativeHeight=0.5)
    FalloffSize=0.3
    FlipHeightScaling=False
    FlipWidthScaling=False
    OptionalParentItem=(Type=Bone,Name="None")
    RotationOffset=(X=45,Y=174,Z=30)
   )
   > alterTransform_5_5 = RigVMFunctionReferenceNode(
    type=Bone
    transforms=((Bone="dyn_R_hair_2",Transform=(Rotation=(X=0,Y=0,Z=0,W=1),Translation=(X=0,Y=-1,Z=-1),Scale3D=(X=1,Y=1,Z=1))),(Bone="dyn_R_hair_3",Transform=(Rotation=(X=0,Y=0,Z=0,W=1),Translation=(X=0,Y=-2,Z=-1),Scale3D=(X=1,Y=1,Z=1))))
    weight=@n7.OutputParam
   )
   > DeltaRotation_7 = RigVMFunctionReferenceNode(
    Rotation Order=ZXY
    child=clavicle_r
    debugRotation=false
    max=-15
    parent=clavicle_r_null
    parentType=Null
    primAxis=1
    twistAxis=(X=1,Y=0,Z=0)
   )
   > alterTransform_4_2_1_2_4 = RigVMFunctionReferenceNode(
    type=Bone
    transforms=((Bone="dyn_R_hair_2",Transform=(Rotation=(X=0,Y=-0.0436,Z=0,W=0.999),Translation=(X=-1,Y=0,Z=-1),Scale3D=(X=1,Y=1,Z=1))),(Bone="dyn_R_hair_3",Transform=(Rotation=(X=0,Y=-0.0436,Z=0,W=0.999),Translation=(X=-2,Y=-1,Z=-2),Scale3D=(X=1,Y=1,Z=1))))
    weight=@DeltaRotation_7.outputWeight
   )
   > DeltaRotation_4_2 = RigVMFunctionReferenceNode(
    Rotation Order=ZXY
    child=clavicle_l
    debugRotation=false
    max=-10
    parent=clavicle_l_null
    parentType=Null
    primAxis=1
    twistAxis=(X=1,Y=0,Z=0)
   )
   > alterTransform_4_2_1_2_1_4 = RigVMFunctionReferenceNode(
    type=Bone
    transforms="((Bone=\"dyn_L_hair_2\",Transform=(Rotation=(X=0,Y=0,Z=0,W=1),Translation=(X=-1,Y=0,Z=0),Scale3D=(X=1,Y=1,Z=1))),(Bone=\"dyn_L_hair_3\",Transform=(Rotation=(X=0,Y=0,Z=0,W=1),Translation=(X=-4,Y=1,Z=-1),Scale3D=(X=1,Y=1,Z=1))),(Bone=\"dyn_L_hair_1\",Transform=(Ro ...(truncated)"
    weight=@DeltaRotation_4_2.outputWeight
   )
   RigVMFunctionReturnNode(ExecuteContext=@RigVMFunction_Sequence_2.E)
   RigVMFunction_Sequence_5 = RigVMAggregateNode(ExecuteContext=@RigVMFunction_Sequence_2.C)
   DeltaRotation_9 = RigVMFunctionReferenceNode(
    Rotation Order=ZYX
    child=neck_01
    debugRotation=false
    max=-23
    min=0
    parent=neck_01_null
    parentType=Null
    primAxis=2
    twistAxis=(X=1,Y=0,Z=0)
    ExecuteContext=@RigVMFunction_Sequence_5.A
   )
   > alterTransform_4_1_1 = RigVMFunctionReferenceNode(
    type=Bone
    transforms="((Bone=\"dyn_C_hair_2\",Transform=(Rotation=(X=0,Y=0,Z=0,W=1),Translation=(X=0,Y=0,Z=-1),Scale3D=(X=1,Y=1,Z=1))),(Bone=\"dyn_L_hair_2\",Transform=(Rotation=(X=0,Y=0,Z=0,W=1),Translation=(X=0,Y=0,Z=-1),Scale3D=(X=1,Y=1,Z=1))),(Bone=\"dyn_R_hair_2\",Transform=(Rot ...(truncated)"
    weight=@DeltaRotation_9.outputWeight
   )
   > alterTransform_4_1_1_1 = RigVMFunctionReferenceNode(
    type=Bone
    transforms="((Bone=\"dyn_C_hair_3\",Transform=(Rotation=(X=0,Y=0,Z=0,W=1),Translation=(X=0,Y=0,Z=0),Scale3D=(X=1,Y=1,Z=1))),(Bone=\"dyn_L_hair_3\",Transform=(Rotation=(X=0,Y=0,Z=0,W=1),Translation=(X=0,Y=0,Z=0),Scale3D=(X=1,Y=1,Z=1))),(Bone=\"dyn_R_hair_3\",Transform=(Rotat ...(truncated)"
    weight=@DeltaRotation_9.outputWeight
   )
   > DeltaRotation_1_4 = RigVMFunctionReferenceNode(Rotation Order=ZYX, child=head, debugRotation=false, max=-20, min=0, parent=head_null, parentType=Null, primAxis=2, twistAxis=(X=1,Y=0,Z=0))
   n8 = &RigVMFunction_AnimEvalRichCurve(
    Curve=(EditorCurveData=(Keys=((InterpMode=RCIM_Cubic),(InterpMode=RCIM_Cubic,Time=0.3879,Value=0.174,ArriveTangent=1,LeaveTangent=1),(InterpMode=RCIM_Cubic,TangentMode=RCTM_User,Time=1,Value=1,ArriveTangent=1.9233,LeaveTangent=1.9233))))
    SourceMinimum=0
    TargetMinimum=0
    Value=@DeltaRotation_1_4.outputWeight
   )
   > alterTransform_4_2_3_2 = RigVMFunctionReferenceNode(
    type=Bone
    transforms="((Bone=\"dyn_C_hair_1\",Transform=(Rotation=(X=0,Y=0.0436,Z=0,W=0.999),Translation=(X=0,Y=0,Z=-1),Scale3D=(X=1,Y=1,Z=1))),(Bone=\"dyn_L_hair_1\",Transform=(Rotation=(X=0,Y=0,Z=0,W=1),Translation=(X=0,Y=0,Z=0),Scale3D=(X=1,Y=1,Z=1))),(Bone=\"dyn_R_hair_1\",Transf ...(truncated)"
    weight=@n8.Result
   )
   > alterTransform_4_1_2_2 = RigVMFunctionReferenceNode(
    type=Bone
    transforms="((Bone=\"dyn_C_hair_2\",Transform=(Rotation=(X=0,Y=0,Z=0,W=1),Translation=(X=0,Y=0,Z=-1),Scale3D=(X=1,Y=1,Z=1))),(Bone=\"dyn_L_hair_2\",Transform=(Rotation=(X=0,Y=0,Z=0,W=1),Translation=(X=0,Y=0,Z=-1),Scale3D=(X=1,Y=1,Z=1))),(Bone=\"dyn_R_hair_2\",Transform=(Rot ...(truncated)"
    weight=@n8.Result
   )
   > alterTransform_4_1_3_2 = RigVMFunctionReferenceNode(
    type=Bone
    transforms="((Bone=\"dyn_C_hair_2\",Transform=(Rotation=(X=0,Y=-0.0872,Z=0,W=0.9962),Translation=(X=0,Y=0,Z=-1),Scale3D=(X=1,Y=1,Z=1))),(Bone=\"dyn_L_hair_2\",Transform=(Rotation=(X=0,Y=0,Z=0,W=1),Translation=(X=0,Y=0,Z=0),Scale3D=(X=1,Y=1,Z=1))),(Bone=\"dyn_R_hair_2\",Tran ...(truncated)"
    weight=&RigVMFunction_MathDoubleMul(B=@DeltaRotation_9.outputWeight, A=@n8.Result).Result
   )
   > &RigUnit_ModifyTransforms(
    ItemToModify=((Item=(Type=Bone,Name="dyn_earring_l"),Transform=(Rotation=(X=0,Y=0.1736,Z=0,W=0.9848),Translation=(X=0,Y=0,Z=0),Scale3D=(X=1,Y=1,Z=1))))
    WeightMinimum=0
    Weight=@n8.Result
   )
   > DeltaRotation_8 = RigVMFunctionReferenceNode(Rotation Order=ZYX, child=head, debugRotation=false, max=60, min=0, parent=head_null, parentType=Null, primAxis=2, twistAxis=(X=1,Y=0,Z=0))
   > alterTransform_4_2_4 = RigVMFunctionReferenceNode(
    type=Bone
    transforms="((Bone=\"dyn_C_hair_2\",Transform=(Rotation=(X=0,Y=0,Z=0,W=1),Translation=(X=0,Y=0,Z=-3),Scale3D=(X=1,Y=1,Z=1))),(Bone=\"dyn_L_hair_2\",Transform=(Rotation=(X=0,Y=0,Z=0,W=1),Translation=(X=0,Y=0,Z=-3),Scale3D=(X=1,Y=1,Z=1))),(Bone=\"dyn_R_hair_2\",Transform=(Rot ...(truncated)"
    weight=@DeltaRotation_8.outputWeight
   )
   > &RigUnit_ModifyTransforms(
    ItemToModify=((Item=(Type=Bone,Name="dyn_earring_l"),Transform=(Rotation=(X=0,Y=-0.766,Z=0,W=0.6428),Translation=(X=0,Y=0,Z=0),Scale3D=(X=1,Y=1,Z=1))))
    WeightMinimum=0
    Weight=@DeltaRotation_8.outputWeight
   )
   > DeltaRotation_1_1_1_1_2 = RigVMFunctionReferenceNode(
    Rotation Order=ZYX
    child=head
    debugRotation=false
    max=50
    min=15
    onlyTwist=true
    parent=head_null_sp
    parentType=Null
    primAxis=0
    twistAxis=(X=1,Y=0,Z=0)
   )
   > alterTransform_4_2_2_1_1_1_2 = RigVMFunctionReferenceNode(
    type=Bone
    transforms="((Bone=\"dyn_C_hair_1\",Transform=(Rotation=(X=0,Y=0,Z=0,W=1),Translation=(X=0,Y=0,Z=0),Scale3D=(X=1,Y=1,Z=1))),(Bone=\"dyn_R_hair_2\",Transform=(Rotation=(X=0,Y=0,Z=0,W=1),Translation=(X=0,Y=0,Z=-2),Scale3D=(X=1,Y=1,Z=1))),(Bone=\"dyn_R_hair_3\",Transform=(Rota ...(truncated)"
    weight=&RigVMFunction_MathDoubleMul(B=0.6, A=@DeltaRotation_1_1_1_1_2.outputWeight).Result
   )
   > DeltaRotation_1_1_1_3 = RigVMFunctionReferenceNode(
    Rotation Order=ZYX
    child=head
    debugRotation=false
    max=-50
    min=-15
    onlyTwist=true
    parent=head_null_sp
    parentType=Null
    primAxis=0
    twistAxis=(X=1,Y=0,Z=0)
   )
   > alterTransform_4_2_2_1_1_2 = RigVMFunctionReferenceNode(
    type=Bone
    transforms="((Bone=\"dyn_C_hair_3\",Transform=(Rotation=(X=0,Y=0,Z=0,W=1),Translation=(X=0,Y=0,Z=0),Scale3D=(X=1,Y=1,Z=1))),(Bone=\"dyn_R_hair_2\",Transform=(Rotation=(X=0,Y=0,Z=0,W=1),Translation=(X=0,Y=0,Z=-3),Scale3D=(X=1,Y=1,Z=1))),(Bone=\"dyn_R_hair_3\",Transform=(Rota ...(truncated)"
    weight=@DeltaRotation_1_1_1_3.outputWeight
   )
   > DeltaRotation_1_2 = RigVMFunctionReferenceNode(
    Rotation Order=ZYX
    child=head
    debugRotation=false
    max=-100
    min=-60
    parent=head_null_sp
    parentType=Null
    primAxis=2
    twistAxis=(X=1,Y=0,Z=0)
   )
   > alterTransform_4_2_2_2 = RigVMFunctionReferenceNode(
    type=Bone
    transforms="((Bone=\"dyn_C_hair_1\",Transform=(Rotation=(X=0,Y=0,Z=0,W=1),Translation=(X=2,Y=0,Z=-2),Scale3D=(X=1,Y=1,Z=1))),(Bone=\"dyn_L_hair_1\",Transform=(Rotation=(X=0,Y=0,Z=0,W=1),Translation=(X=0,Y=0,Z=-3),Scale3D=(X=1,Y=1,Z=1))),(Bone=\"dyn_R_hair_1\",Transform=(Rot ...(truncated)"
    weight=@DeltaRotation_1_2.outputWeight
   )
   > alterTransform_4_2_2_1_2 = RigVMFunctionReferenceNode(
    type=Bone
    transforms="((Bone=\"dyn_C_hair_2\",Transform=(Rotation=(X=0,Y=0,Z=0,W=1),Translation=(X=0,Y=0,Z=2),Scale3D=(X=1,Y=1,Z=1))),(Bone=\"dyn_L_hair_2\",Transform=(Rotation=(X=0,Y=0,Z=0,W=1),Translation=(X=0,Y=0,Z=2),Scale3D=(X=1,Y=1,Z=1))),(Bone=\"dyn_R_hair_2\",Transform=(Rotat ...(truncated)"
    weight=@DeltaRotation_1_2.outputWeight
   )
   > n9 = &RigUnit_SphericalPoseReader(
    ActiveRegionScaleFactors=(PositiveWidth=0.5,NegativeWidth=0.5,PositiveHeight=1,NegativeHeight=1)
    ActiveRegionSize=0.1
    Debug=(bDrawDebug=false,bDrawLocalAxes=false,DebugScale=25,DebugSegments=20,DebugThickness=0.25)
    DriverAxis=(X=0,Y=-1,Z=0)
    DriverItem=(Type=Bone,Name="head")
    FalloffRegionScaleFactors=(PositiveWidth=1,NegativeWidth=0.5,PositiveHeight=1,NegativeHeight=1)
    FalloffSize=0.2
    OptionalParentItem=(Type=Bone,Name="neck_02")
    RotationOffset=(X=-55,Y=0,Z=-10)
   )
   n10 = &RigVMFunction_AlphaInterp(ClampMax=0.8, InterpSpeedDecreasing=10, InterpSpeedIncreasing=10, Scale=1, bClampResult=true, Value=@n9.OutputParam)
   > alterTransform_4_2_2_1_1_3 = RigVMFunctionReferenceNode(
    type=Bone
    transforms="((Bone=\"dyn_C_hair_3\",Transform=(Rotation=(X=0,Y=0,Z=0,W=1),Translation=(X=0,Y=0,Z=-5),Scale3D=(X=1,Y=1,Z=1))),(Bone=\"dyn_R_hair_2\",Transform=(Rotation=(X=0,Y=0,Z=0,W=1),Translation=(X=0,Y=0,Z=-7),Scale3D=(X=1,Y=1,Z=1))),(Bone=\"dyn_R_hair_3\",Transform=(Rot ...(truncated)"
    weight=@n10.Result
   )
   RigVMFunction_ControlFlowBranch = &RigVMFunction_ControlFlowBranch(Condition=RigVMVariableNode() {variable=CrouchingTargeting}.Value, ExecuteContext=@alterTransform_4_2_2_1_1_3.ExecuteContext)
   n11 = &RigUnit_SphericalPoseReader(
    ActiveRegionScaleFactors=(PositiveWidth=1,NegativeWidth=0.5,PositiveHeight=1,NegativeHeight=1)
    ActiveRegionSize=0.1
    Debug=(bDrawDebug=true,bDrawLocalAxes=false,DebugScale=25,DebugSegments=20,DebugThickness=0.25)
    DriverAxis=(X=0,Y=-1,Z=0)
    DriverItem=(Type=Bone,Name="head")
    FalloffRegionScaleFactors=(PositiveWidth=1,NegativeWidth=0.5,PositiveHeight=1,NegativeHeight=1)
    FalloffSize=0.2
    OptionalParentItem=(Type=Bone,Name="None")
    RotationOffset=(X=-70,Y=0,Z=-50)
    ExecutePin=@RigVMFunction_ControlFlowBranch.True
   )
   > alterTransform_4_2_2_1_1_1 = RigVMFunctionReferenceNode(
    type=Bone
    transforms="((Bone=\"dyn_R_hair_1\",Transform=(Rotation=(X=0,Y=0,Z=0,W=1),Translation=(X=0,Y=-5,Z=0),Scale3D=(X=1,Y=1,Z=1))),(Bone=\"dyn_R_hair_2\",Transform=(Rotation=(X=0,Y=0,Z=0,W=1),Translation=(X=0,Y=0,Z=0),Scale3D=(X=1,Y=1,Z=1))),(Bone=\"dyn_R_hair_3\",Transform=(Rota ...(truncated)"
    weight=@n11.OutputParam
   )
   > &RigUnit_ModifyTransforms(
    ItemToModify="((Item=(Type=Bone,Name=\"dyn_R_hair_2\"),Transform=(Rotation=(X=0,Y=0,Z=0,W=-1),Translation=(X=0,Y=-7.9101,Z=8.4578),Scale3D=(X=1,Y=1,Z=1))),(Item=(Type=Bone,Name=\"dyn_R_hair_1\"),Transform=(Rotation=(X=0,Y=0,Z=0,W=1),Translation=(X=-7,Y=0,Z=-10),Scale3D=(X=1 ...(truncated)"
    WeightMinimum=0
    Weight=@n11.OutputParam
   )
   DeltaRotation_6 = RigVMFunctionReferenceNode(
    Rotation Order=ZYX
    child=spine_05
    debugRotation=false
    max=-50
    min=-10
    onlyTwist=true
    parent=spine_01
    parentType=Bone
    twistAxis=(X=1,Y=0,Z=0)
    ExecuteContext=@RigVMFunction_Sequence_5.B
   )
   n12 = &RigVMFunction_AnimEvalRichCurve(
    Curve=(EditorCurveData=(Keys=((InterpMode=RCIM_Cubic),(InterpMode=RCIM_Cubic,TangentMode=RCTM_User,Time=0.5279,Value=0.4056,ArriveTangent=1.3758,LeaveTangent=1.3758),(InterpMode=RCIM_Cubic,Time=1,Value=1))))
    SourceMinimum=0
    TargetMinimum=0
    Value=@DeltaRotation_6.outputWeight
   )
   > alterTransform_4_2_3 = RigVMFunctionReferenceNode(
    type=Bone
    transforms="((Bone=\"dyn_L_hair_2\",Transform=(Rotation=(X=0,Y=0,Z=0,W=1),Translation=(X=0,Y=-1,Z=0),Scale3D=(X=1,Y=1,Z=1))),(Bone=\"dyn_L_hair_3\",Transform=(Rotation=(X=0,Y=0,Z=0,W=1),Translation=(X=0,Y=-1,Z=0),Scale3D=(X=1,Y=1,Z=1))),(Bone=\"dyn_L_hair_4\",Transform=(Rot ...(truncated)"
    weight=@n12.Result
   )
   > DeltaRotation_5 = RigVMFunctionReferenceNode(
    Rotation Order=ZYX
    child=spine_05
    debugRotation=false
    max=50
    min=10
    onlyTwist=true
    parent=spine_01
    parentType=Bone
    twistAxis=(X=1,Y=0,Z=0)
   )
   n13 = &RigVMFunction_AnimEvalRichCurve(
    Curve=(EditorCurveData=(Keys=((InterpMode=RCIM_Cubic),(InterpMode=RCIM_Cubic,TangentMode=RCTM_User,Time=0.5279,Value=0.4056,ArriveTangent=1.3758,LeaveTangent=1.3758),(InterpMode=RCIM_Cubic,Time=1,Value=1))))
    SourceMinimum=0
    TargetMinimum=0
    Value=@DeltaRotation_5.outputWeight
   )
   > alterTransform_4_2_1_2 = RigVMFunctionReferenceNode(
    type=Bone
    transforms="((Bone=\"dyn_R_hair_2\",Transform=(Rotation=(X=0,Y=0,Z=0,W=1),Translation=(X=0,Y=-1,Z=0),Scale3D=(X=1,Y=1,Z=1))),(Bone=\"dyn_R_hair_3\",Transform=(Rotation=(X=0,Y=0,Z=0,W=1),Translation=(X=0,Y=-1,Z=0),Scale3D=(X=1,Y=1,Z=1))),(Bone=\"dyn_R_hair_4\",Transform=(Rot ...(truncated)"
    weight=@n13.Result
   )
   n14 = &RigVMFunction_ControlFlowBranch(ExecuteContext=@RigVMFunction_Sequence_5.C, Condition=RigVMVariableNode() {variable=pub isTargeting}.Value)
   n15 = RigVMVariableNode() {variable=pub aimPitch}
   n16 = &RigVMFunction_AnimEvalRichCurve(
    Curve=(EditorCurveData=(Keys=((InterpMode=RCIM_Cubic),(InterpMode=RCIM_Cubic,TangentMode=RCTM_User,Time=0.4302,Value=0.2017,ArriveTangent=0.8571,LeaveTangent=0.8571),(Time=1,Value=1))))
    SourceMinimum=0
    TargetMinimum=0
    Value=&RigVMFunction_MathDoubleMul(
     B=&RigVMFunction_MathDoubleRemap(SourceMaximum=0, SourceMinimum=88, TargetMaximum=0, TargetMinimum=1, Value=@n15.Value).Result
     A=RigVMVariableNode() {variable=pub isAssaultRifleEquipped}.Value
    ).Result
   )
   &RigUnit_ModifyTransforms(
    ItemToModify=((Item=(Type=Bone,Name="base_R_hair_1"),Transform=(Rotation=(X=0,Y=0,Z=0,W=1),Translation=(X=0,Y=3,Z=-1.5),Scale3D=(X=1,Y=1,Z=1))))
    WeightMinimum=0
    Weight=@n16.Result
    ExecutePin=@n14.True
   )
   > alterTransform_12 = RigVMFunctionReferenceNode(
    type=Bone
    transforms=((Bone="dyn_R_hair_1",Transform=(Rotation=(X=-0.0872,Y=0,Z=0,W=0.9962),Translation=(X=0,Y=0,Z=0),Scale3D=(X=1,Y=1,Z=1))))
    weight=@n16.Result
   )
   AlphaInterp = &RigVMFunction_AlphaInterp(InterpSpeedDecreasing=10, InterpSpeedIncreasing=10, Scale=1, bClampResult=False, Value=RigVMVariableNode() {variable=AimUp}.Value)
   alterTransform_0 = RigVMFunctionReferenceNode(
    type=Bone
    transforms="((Bone=\"dyn_R_hair_1\",Transform=(Rotation=(X=0,Y=0,Z=0,W=1),Translation=(X=0,Y=0,Z=-10),Scale3D=(X=1,Y=1,Z=1))),(Bone=\"dyn_R_hair_2\",Transform=(Rotation=(X=0,Y=0,Z=0,W=1),Translation=(X=0,Y=0,Z=-1),Scale3D=(X=1,Y=1,Z=1))),(Bone=\"dyn_R_hair_3\",Transform=(Ro ...(truncated)"
    ExecuteContext=@RigVMFunction_Sequence_5.D
    weight=@AlphaInterp.Result
   )
   > &RigUnit_ModifyTransforms(
    ItemToModify=((Item=(Type=Bone,Name="dyn_earring_l"),Transform=(Rotation=(X=0.0298,Y=0.3407,Z=-0.0819,W=0.9361),Translation=(X=0,Y=0,Z=0),Scale3D=(X=1,Y=1,Z=1))))
    WeightMinimum=0
    Weight=@AlphaInterp.Result
   )
   n17 = &RigVMFunction_AnimEvalRichCurve(
    Curve="(EditorCurveData=(Keys=((InterpMode=RCIM_Cubic,TangentMode=RCTM_Break,TangentWeightMode=RCTWM_WeightedNone,Time=0,Value=0,ArriveTangent=0,ArriveTangentWeight=0,LeaveTangent=0,LeaveTangentWeight=0),(InterpMode=RCIM_Cubic,TangentMode=RCTM_Break,TangentWeight ...(truncated)"
    SourceMinimum=0.5
    TargetMinimum=0
    Value=&RigVMFunction_MathDoubleRemap(
     SourceMaximum=30
     SourceMinimum=60
     TargetMaximum=0
     TargetMinimum=1
     bClamp=False
     Value=&RigUnit_GetRelativeTransformForItem(Child=(Type=Bone,Name="pelvis"), Parent=(Type=Bone,Name="root"), bChildInitial=False, bParentInitial=False).RelativeTransform.Translation.Z
    ).Result
   )
   RigVMVariableNode(Value=@n17.Result, ExecuteContext=@RigVMFunction_Sequence_5.F) {variable=GrounMovement}
   n18 = &RigVMFunction_AlphaInterp(InterpSpeedDecreasing=5, InterpSpeedIncreasing=10, Scale=1, bClampResult=False, Value=RigVMVariableNode() {variable=Pickaxe}.Value)
   > alterTransform_4_2_5 = RigVMFunctionReferenceNode(
    type=Bone
    transforms="((Bone=\"dyn_R_hair_2\",Transform=(Rotation=(X=0,Y=0,Z=0,W=1),Translation=(X=0,Y=0,Z=-10),Scale3D=(X=1,Y=1,Z=1))),(Bone=\"dyn_C_hair_2\",Transform=(Rotation=(X=0,Y=0,Z=0,W=1),Translation=(X=0,Y=0,Z=-10),Scale3D=(X=1,Y=1,Z=1))),(Bone=\"dyn_L_hair_2\",Transform=(R ...(truncated)"
    weight=@n18.Result
   )
   RigVMFunction_MathBoolAnd = &RigVMFunction_MathBoolAnd(A=RigVMVariableNode() {variable=wearingBackpack}.Value, B=RigVMVariableNode() {variable=Parachuteopened}.Value)
   n19 = &RigVMFunction_AlphaInterp(
    InterpSpeedDecreasing=10
    InterpSpeedIncreasing=10
    Scale=1
    bClampResult=False
    Value=RigVMDispatchNode(False=0, True=1, Condition=@RigVMFunction_MathBoolAnd.Result) {template="DISPATCH_RigVMDispatch_If(in Condition,in True,in False,out Result)"}.Result
   )
   &RigUnit_ModifyTransforms(
    ItemToModify="((Item=(Type=Bone,Name=\"dyn_R_hair_2\"),Transform=(Rotation=(X=0,Y=-0.0872,Z=0,W=0.9962),Translation=(X=0,Y=0,Z=0),Scale3D=(X=1,Y=1,Z=1))),(Item=(Type=Bone,Name=\"dyn_C_hair_2\"),Transform=(Rotation=(X=0,Y=-0.0872,Z=0,W=0.9962),Translation=(X=0,Y=0,Z=0),Scale ...(truncated)"
    WeightMinimum=0
    Weight=@n19.Result
    ExecutePin=@RigVMFunction_Sequence_5.E
   )
   RigVMFunction_MathBoolToFloat = &RigVMFunction_MathBoolToFloat(Value=RigVMVariableNode() {variable=pub diveFloating}.Value)
   Multiply = &RigVMFunction_MathDoubleMul(
    B=@RigVMFunction_MathBoolToFloat.Result
    A=&RigVMFunction_MathBoolToFloat(Value=RigVMVariableNode() {variable=wearingBackpack}.Value).Result
   )
   n20 = &RigVMFunction_AlphaInterp(InterpSpeedDecreasing=10, InterpSpeedIncreasing=15, Scale=1, bClampResult=False, Value=@Multiply.Result)
   > &RigUnit_ModifyTransforms(
    ItemToModify="((Item=(Type=Bone,Name=\"dyn_R_hair_3\"),Transform=(Rotation=(X=0,Y=-0.0436,Z=0,W=0.999),Translation=(X=0,Y=0,Z=-2),Scale3D=(X=1,Y=1,Z=1))),(Item=(Type=Bone,Name=\"dyn_C_hair_3\"),Transform=(Rotation=(X=0,Y=-0.0436,Z=0,W=0.999),Translation=(X=0,Y=0,Z=-2),Scale ...(truncated)"
    WeightMinimum=0
    Weight=@n20.Result
   )
   > &RigUnit_ModifyTransforms(
    ItemToModify="((Item=(Type=Bone,Name=\"dyn_R_hair_6\"),Transform=(Rotation=(X=0,Y=0.0436,Z=0,W=0.999),Translation=(X=0,Y=0,Z=0),Scale3D=(X=1,Y=1,Z=1))),(Item=(Type=Bone,Name=\"dyn_C_hair_6\"),Transform=(Rotation=(X=0,Y=0.0436,Z=0,W=0.999),Translation=(X=0,Y=0,Z=0),Scale3D=( ...(truncated)"
    WeightMinimum=0
    Weight=@n20.Result
   )
   > &RigUnit_ModifyTransforms(
    ItemToModify="((Item=(Type=Bone,Name=\"dyn_R_hair_4\"),Transform=(Rotation=(X=0,Y=0.0872,Z=0,W=0.9962),Translation=(X=0,Y=0,Z=-1),Scale3D=(X=1,Y=1,Z=1))),(Item=(Type=Bone,Name=\"dyn_C_hair_4\"),Transform=(Rotation=(X=0,Y=0.0872,Z=0,W=0.9962),Translation=(X=0,Y=0,Z=-1),Scale ...(truncated)"
    WeightMinimum=0
    Weight=@n20.Result
   )
   n21 = RigVMVariableNode() {variable=wearingBackpack}
   n22 = RigVMVariableNode() {variable=Parachuteopened}
   > &RigUnit_ModifyTransforms(
    ItemToModify="((Item=(Type=Bone,Name=\"dyn_R_hair_3\"),Transform=(Rotation=(X=0,Y=0.0436,Z=0,W=0.999),Translation=(X=0,Y=0,Z=0),Scale3D=(X=1,Y=1,Z=1))),(Item=(Type=Bone,Name=\"dyn_C_hair_3\"),Transform=(Rotation=(X=0,Y=0.0436,Z=0,W=0.999),Translation=(X=0,Y=0,Z=0),Scale3D=( ...(truncated)"
    WeightMinimum=0
    Weight=&RigVMFunction_MathDoubleMul(
     A=&RigVMFunction_MathBoolToFloat(Value=@n21.Value).Result
     B=&RigVMFunction_MathBoolToFloat(Value=@n22.Value).Result
    ).Result
   )
   &RigVMFunction_AlphaInterp(InterpSpeedDecreasing=10, InterpSpeedIncreasing=10, Scale=1, Value=0, bClampResult=False)
   RigVMUnitNode(
    Bias=0
    ClampMax=1
    ClampMin=0
    InRange=(Min=0,Max=1)
    InterpSpeedDecreasing=10
    InterpSpeedIncreasing=10
    OutRange=(Min=0,Max=1)
    Value=()
    bClampResult=False
    bInterpResult=False
    bMapRange=False
    Scale=@RigVMFunction_MathBoolToFloat.Result
   ) {template="AlphaInterp::Execute(in Value,in Scale,in Bias,in bMapRange,in InRange,in OutRange,in bClampResult,in ClampMin,in ClampMax,in bInterpResult,in InterpSpeedIncreasing,in InterpSpeedDecreasing,out Result)"}
   DeltaRotation_1_1_1_2 = RigVMFunctionReferenceNode(
    ExecuteContext=()
    Rotation Order=ZYX
    child=neck_01
    debugRotation=false
    max=-40
    min=-10
    onlyTwist=true
    parent=neck_01_null
    parentType=Null
    primAxis=0
    twistAxis=(X=1,Y=0,Z=0)
   )
   RigVMDispatchNode(Condition=False, False=0, True=1) {template="DISPATCH_RigVMDispatch_If(in Condition,in True,in False,out Result)"}
   &RigVMFunction_MathDoubleMul(A=1, B=1)
   &RigVMFunction_ControlFlowBranch(Condition=False, ExecuteContext=())
   &RigUnit_SphericalPoseReader(
    ActiveRegionScaleFactors=(PositiveWidth=1,NegativeWidth=1,PositiveHeight=1,NegativeHeight=1)
    ActiveRegionSize=0.1
    Debug=(bDrawDebug=true,DebugScale=25,DebugSegments=20,DebugThickness=0.25)
    DriverAxis=(X=-1,Y=0,Z=0)
    DriverItem=(Type=Bone,Name="upperarm_r")
    ExecutePin=()
    FalloffRegionScaleFactors=(PositiveWidth=1,NegativeWidth=1,PositiveHeight=1,NegativeHeight=1)
    FalloffSize=0.3
    FlipHeightScaling=false
    FlipWidthScaling=False
    OptionalParentItem=(Type=Bone,Name="None")
    RotationOffset=(X=40,Y=50,Z=0)
   )
   &RigUnit_SphericalPoseReader(
    ActiveRegionScaleFactors=(PositiveWidth=1,NegativeWidth=0.5,PositiveHeight=1,NegativeHeight=1)
    ActiveRegionSize=0.1
    Debug=(bDrawDebug=true,bDrawLocalAxes=false,DebugScale=25,DebugSegments=20,DebugThickness=0.25)
    DriverAxis=(X=0,Y=-1,Z=0)
    DriverItem=(Type=Bone,Name="head")
    ExecutePin=()
    FalloffRegionScaleFactors=(PositiveWidth=1,NegativeWidth=0.5,PositiveHeight=1,NegativeHeight=1)
    FalloffSize=0.2
    OptionalParentItem=(Type=Bone,Name="neck_02")
    RotationOffset=(X=-55,Y=0,Z=-10)
   )
   RigVMVariableNode() {variable=pub isTargeting}
   n23 = &RigVMFunction_AlphaInterp(InterpSpeedDecreasing=10, InterpSpeedIncreasing=10, Scale=1, bClampResult=False, Value=RigVMVariableNode() {variable=pistolHead Down}.Value)
   alterTransform_11 = RigVMFunctionReferenceNode(
    type=Bone
    ExecuteContext=()
    transforms="((Bone=\"dyn_R_hair_1\",Transform=(Rotation=(X=0,Y=0,Z=0,W=1),Translation=(X=0,Y=0,Z=0),Scale3D=(X=1,Y=1,Z=1))),(Bone=\"dyn_R_hair_2\",Transform=(Rotation=(X=0,Y=0,Z=0,W=1),Translation=(X=0,Y=0,Z=3),Scale3D=(X=1,Y=1,Z=1))),(Bone=\"dyn_R_hair_3\",Transform=(Rotat ...(truncated)"
    weight=@n23.Result
   )
   n24 = &RigVMFunction_AlphaInterp(InterpSpeedDecreasing=10, InterpSpeedIncreasing=10, Scale=1, bClampResult=False, Value=RigVMVariableNode() {variable=HeadUpCrouch}.Value)
   alterTransform_15 = RigVMFunctionReferenceNode(
    type=Bone
    ExecuteContext=()
    transforms="((Bone=\"dyn_R_hair_1\",Transform=(Rotation=(X=0,Y=0,Z=0,W=1),Translation=(X=0,Y=0,Z=-10),Scale3D=(X=1,Y=1,Z=1))),(Bone=\"dyn_R_hair_2\",Transform=(Rotation=(X=0,Y=0,Z=0,W=1),Translation=(X=0,Y=0,Z=-10),Scale3D=(X=1,Y=1,Z=1))),(Bone=\"dyn_R_hair_3\",Transform=(R ...(truncated)"
    weight=@n24.Result
   )
   n25 = &RigVMFunction_AlphaInterp(
    InterpSpeedDecreasing=10
    InterpSpeedIncreasing=10
    Scale=1
    bClampResult=False
    Value=&RigVMFunction_MathBoolToFloat(Value=RigVMVariableNode() {variable=Crouch}.Value).Result
   )
   alterTransform_4_2_1 = RigVMFunctionReferenceNode(
    type=Bone
    ExecuteContext=()
    transforms="((Bone=\"dyn_R_hair_4\",Transform=(Rotation=(X=0,Y=0,Z=0,W=1),Translation=(X=0,Y=0,Z=3),Scale3D=(X=1,Y=1,Z=1))),(Bone=\"dyn_C_hair_4\",Transform=(Rotation=(X=0,Y=0,Z=0,W=1),Translation=(X=0,Y=0,Z=3),Scale3D=(X=1,Y=1,Z=1))),(Bone=\"dyn_L_hair_4\",Transform=(Rotat ...(truncated)"
    weight=@n25.Result
   )
   alterTransform_4_2_1_2_1 = RigVMFunctionReferenceNode(
    type=Bone
    ExecuteContext=()
    transforms=((Bone="dyn_R_hair_2",Transform=(Rotation=(X=0,Y=-0.0872,Z=0,W=0.9962),Translation=(X=-1,Y=0,Z=-2),Scale3D=(X=1,Y=1,Z=1))),(Bone="dyn_R_hair_3",Transform=(Rotation=(X=0,Y=-0.0872,Z=0,W=0.9962),Translation=(X=-2,Y=-1,Z=-5),Scale3D=(X=1,Y=1,Z=1))))
   )
   alterTransform_4_2_2 = RigVMFunctionReferenceNode(
    type=Bone
    ExecuteContext=()
    transforms="((Bone=\"dyn_C_hair_2\",Transform=(Rotation=(X=0,Y=0,Z=0,W=1),Translation=(X=0,Y=0,Z=-3),Scale3D=(X=1,Y=1,Z=1))),(Bone=\"dyn_L_hair_2\",Transform=(Rotation=(X=0,Y=0,Z=0,W=1),Translation=(X=0,Y=0,Z=-3),Scale3D=(X=1,Y=1,Z=1))),(Bone=\"dyn_R_hair_2\",Transform=(Rot ...(truncated)"
   )
  ]}
 New Function_ContainedGraph {nodes [
   n1 = RigVMFunctionEntryNode()
   > DISPATCH_RigVMDispatch_ArrayIterator = RigVMDispatchNode(Array=@n1.transforms) {template="DISPATCH_RigVMDispatch_ArrayIterator(in Array,out Element,out Index,out Count,out Ratio)"}
   Multiply = &RigVMFunction_MathTransformMul(
    A=@DISPATCH_RigVMDispatch_ArrayIterator.Element.Transform
    B=&RigUnit_GetTransform(
     Item=(Type=Bone,Name="dyn_L_hair_3")
     Space=LocalSpace
     bInitial=False
     Item.Name=@DISPATCH_RigVMDispatch_ArrayIterator.Element.Bone
     Item.Type=@n1.type
    ).Transform
   ) {template="Multiply::Execute(in A,in B,out Result)"}
   > &RigUnit_SetTransform(
    Item=(Type=Bone,Name="dyn_L_hair_3")
    Space=LocalSpace
    bInitial=false
    Value=@Multiply.Result
    bPropagateToChildren=@n1.inheritsTransform
    Weight=@n1.weight
    Item.Name=@DISPATCH_RigVMDispatch_ArrayIterator.Element.Bone
    Item.Type=@n1.type
   )
   RigVMFunctionReturnNode(ExecuteContext=@DISPATCH_RigVMDispatch_ArrayIterator.Completed)
   &RigUnit_OffsetTransformForItem(
    ExecutePin=()
    Item=(Type=Bone,Name="None")
    OffsetTransform=(Rotation=(X=0,Y=0,Z=0,W=1),Translation=(X=0,Y=0,Z=0),Scale3D=(X=1,Y=1,Z=1))
    Weight=1
    bPropagateToChildren=True
   )
  ]}
 New Function_ContainedGraph {nodes [
   n1 = RigVMFunctionEntryNode()
   RigVMFunction_MathBoolToInteger = &RigVMFunction_MathBoolToInteger(Value=@n1.onlyTwist)
   RigVMFunction_MathQuaternionSwingTwist = &RigVMFunction_MathQuaternionSwingTwist(
    TwistAxis=@n1.twistAxis
    Input=&RigUnit_GetRelativeTransformForItem(
     Child=(Type=Bone,Name="None")
     Parent=(Type=Bone,Name="None")
     bChildInitial=False
     bParentInitial=False
     Child.Name=@n1.child
     Parent.Name=@n1.parent
     Parent.Type=@n1.parentType
    ).RelativeTransform.Rotation
   )
   RigVMFunction_MathQuaternionToEuler = &RigVMFunction_MathQuaternionToEuler(
    Value=RigVMDispatchNode(
     Values=((),())
     Index=@RigVMFunction_MathBoolToInteger.Result
     Values.0=@RigVMFunction_MathQuaternionSwingTwist.Swing
     Values.1=@RigVMFunction_MathQuaternionSwingTwist.Twist
    ) {template="DISPATCH_RigVMDispatch_SelectInt32(in Index,in Values,out Result)"}.Result
    RotationOrder=@n1.Rotation Order
   )
   > DISPATCH_RigVMDispatch_Print = RigVMDispatchNode(ScreenColor=(R=1,G=1,B=1,A=1), ScreenDuration=1.5, Enabled=@n1.debugRotation, Value=@RigVMFunction_MathQuaternionToEuler.Result) {template="DISPATCH_RigVMDispatch_Print(in Prefix,in Value,in Enabled,in ScreenDuration,in ScreenColor)"}
   Clamp = &RigVMFunction_MathIntClamp(Value=@n1.primAxis)
   DISPATCH_RigVMDispatch_SelectInt32 = RigVMDispatchNode(
    Values=(0,0,0)
    Index=@Clamp.Result
    Values.0=@RigVMFunction_MathQuaternionToEuler.Result.X
    Values.1=@RigVMFunction_MathQuaternionToEuler.Result.Y
    Values.2=@RigVMFunction_MathQuaternionToEuler.Result.Z
   ) {template="DISPATCH_RigVMDispatch_SelectInt32(in Index,in Values,out Result)"}
   Remap = &RigVMFunction_MathDoubleRemap(
    TargetMaximum=1
    TargetMinimum=0
    bClamp=False
    Value=@DISPATCH_RigVMDispatch_SelectInt32.Result
    SourceMaximum=@n1.max
    SourceMinimum=@n1.min
   )
   > RigVMFunctionReturnNode(outputWeight=&RigVMFunction_MathDoubleClamp(Value=@Remap.Result).Result)
  ]}
 New Function_ContainedGraph {nodes [
   RigVMFunctionEntryNode()
   > &RigUnit_ModifyTransforms(
    ItemToModify=((Item=(Type=Bone,Name="clavicle_l"),Transform=(Rotation=(X=0,Y=0,Z=0,W=1),Translation=(X=0,Y=0,Z=0),Scale3D=(X=1,Y=1,Z=1))))
    Weight=1
    WeightMinimum=0
   )
   > &RigUnit_ModifyTransforms(
    ItemToModify=((Item=(Type=Bone,Name="clavicle_r"),Transform=(Rotation=(X=0,Y=0,Z=0,W=1),Translation=(X=0,Y=0,Z=0),Scale3D=(X=1,Y=1,Z=1))))
    Weight=1
    WeightMinimum=0
   )
   > n1 = &RigUnit_SphericalPoseReader(
    ActiveRegionScaleFactors=(PositiveWidth=0.5,NegativeWidth=0.5,PositiveHeight=0.5,NegativeHeight=0.5)
    ActiveRegionSize=0.12
    Debug=(bDrawDebug=false,bDraw2D=False,bDrawLocalAxes=False,DebugScale=25,DebugSegments=20,DebugThickness=0.25)
    DriverAxis=(X=1,Y=0,Z=0)
    DriverItem=(Type=Bone,Name="clavicle_l")
    FalloffRegionScaleFactors=(PositiveWidth=0.5,NegativeWidth=0.5,PositiveHeight=0.5,NegativeHeight=0.5)
    FalloffSize=0.3
    FlipHeightScaling=False
    FlipWidthScaling=False
    OptionalParentItem=(Type=Bone,Name="None")
    RotationOffset=(X=0,Y=90,Z=140)
   )
   > alterTransform = RigVMFunctionReferenceNode(
    type=Bone
    transforms="((Bone=\"dyn_L_hair_3\",Transform=(Rotation=(X=0,Y=0,Z=0,W=1),Translation=(X=0,Y=0,Z=-5),Scale3D=(X=1,Y=1,Z=1))),(Bone=\"dyn_L_hair_4\",Transform=(Rotation=(X=0,Y=0,Z=0,W=1),Translation=(X=0,Y=4,Z=-1),Scale3D=(X=1,Y=1,Z=1))),(Bone=\"dyn_L_hair_2\",Transform=(Rot ...(truncated)"
    weight=@n1.OutputParam
   )
   > n2 = &RigUnit_SphericalPoseReader(
    ActiveRegionScaleFactors=(PositiveWidth=0.5,NegativeWidth=0.5,PositiveHeight=0.5,NegativeHeight=0.5)
    ActiveRegionSize=0.12
    Debug=(bDrawDebug=false,bDraw2D=False,bDrawLocalAxes=False,DebugScale=25,DebugSegments=20,DebugThickness=0.25)
    DriverAxis=(X=-1,Y=0,Z=0)
    DriverItem=(Type=Bone,Name="clavicle_r")
    FalloffRegionScaleFactors=(PositiveWidth=0.5,NegativeWidth=0.5,PositiveHeight=0.5,NegativeHeight=0.5)
    FalloffSize=0.3
    FlipHeightScaling=False
    FlipWidthScaling=False
    OptionalParentItem=(Type=Bone,Name="None")
    RotationOffset=(X=0,Y=-90,Z=140)
   )
   > alterTransform_1 = RigVMFunctionReferenceNode(
    type=Bone
    transforms="((Bone=\"dyn_R_hair_3\",Transform=(Rotation=(X=0,Y=0,Z=0,W=1),Translation=(X=0,Y=-6,Z=-6),Scale3D=(X=1,Y=1,Z=1))),(Bone=\"dyn_R_hair_4\",Transform=(Rotation=(X=0,Y=0,Z=0,W=1),Translation=(X=0,Y=-6,Z=-6),Scale3D=(X=1,Y=1,Z=1))),(Bone=\"dyn_R_hair_2\",Transform=(R ...(truncated)"
    weight=@n2.OutputParam
   )
   > RigVMFunctionReturnNode()
   alterTransform_2 = RigVMFunctionReferenceNode(
    type=Bone
    ExecuteContext=()
    transforms="((Bone=\"dyn_L_hair_3\",Transform=(Rotation=(X=0,Y=0,Z=0,W=1),Translation=(X=0,Y=0,Z=-5),Scale3D=(X=1,Y=1,Z=1))),(Bone=\"dyn_L_hair_4\",Transform=(Rotation=(X=0,Y=0,Z=0,W=1),Translation=(X=0,Y=4,Z=-1),Scale3D=(X=1,Y=1,Z=1))),(Bone=\"dyn_L_hair_2\",Transform=(Rot ...(truncated)"
   )
   alterTransform_3 = RigVMFunctionReferenceNode(
    type=Bone
    ExecuteContext=()
    transforms="((Bone=\"dyn_R_hair_3\",Transform=(Rotation=(X=0,Y=0,Z=0,W=1),Translation=(X=0,Y=0,Z=-5),Scale3D=(X=1,Y=1,Z=1))),(Bone=\"dyn_R_hair_4\",Transform=(Rotation=(X=0,Y=0,Z=0,W=1),Translation=(X=0,Y=-4,Z=-1),Scale3D=(X=1,Y=1,Z=1))),(Bone=\"dyn_R_hair_2\",Transform=(Ro ...(truncated)"
   )
  ]}
 New Function_ContainedGraph {nodes [
   RigVMFunctionEntryNode()
   > lerp bools = RigVMCollapseNode()
   > RigVMFunction_Sequence = RigVMAggregateNode()
   n1 = &RigUnit_SphericalPoseReader(
    ActiveRegionScaleFactors=(PositiveWidth=1,NegativeWidth=1,PositiveHeight=1,NegativeHeight=1)
    ActiveRegionSize=0.1
    Debug=(bDrawDebug=false,bDraw2D=False,bDrawLocalAxes=False,DebugScale=25,DebugSegments=20,DebugThickness=0.25)
    DriverAxis=(X=1,Y=0,Z=0)
    DriverItem=(Type=Bone,Name="head")
    FalloffRegionScaleFactors=(PositiveWidth=1,NegativeWidth=1,PositiveHeight=1,NegativeHeight=1)
    FalloffSize=0.3
    FlipHeightScaling=False
    FlipWidthScaling=False
    OptionalParentItem=(Type=Bone,Name="spine_05")
    RotationOffset=(X=-90,Y=0,Z=20)
    ExecutePin=@RigVMFunction_Sequence.B
   )
   > n2 = &RigUnit_SphericalPoseReader(
    ActiveRegionScaleFactors=(PositiveWidth=1,NegativeWidth=1,PositiveHeight=1,NegativeHeight=1)
    ActiveRegionSize=0.1
    Debug=(bDrawDebug=false,bDraw2D=False,bDrawLocalAxes=False,DebugScale=25,DebugSegments=20,DebugThickness=0.25)
    DriverAxis=(X=1,Y=0,Z=0)
    DriverItem=(Type=Bone,Name="head")
    FalloffRegionScaleFactors=(PositiveWidth=1,NegativeWidth=1,PositiveHeight=1,NegativeHeight=1)
    FalloffSize=0.35
    FlipHeightScaling=False
    FlipWidthScaling=False
    OptionalParentItem=(Type=Bone,Name="spine_05")
    RotationOffset=(X=-20,Y=-15,Z=0)
   )
   > &RigUnit_ModifyTransforms(
    ItemToModify="((Item=(Type=Bone,Name=\"base_hat\"),Transform=(Rotation=(X=0,Y=0.1736,Z=0,W=0.9848),Translation=(X=-2.5821,Y=0,Z=0.9398),Scale3D=(X=1,Y=1,Z=1))),(Item=(Type=Bone,Name=\"dyn_hat_rear_1\"),Transform=(Rotation=(X=0,Y=0.2329,Z=0,W=-0.9725),Translation=(X=0,Y=0,Z= ...(truncated)"
    WeightMinimum=0
    Weight=&RigVMFunction_MathDoubleMul(A=@n2.OutputParam, B=RigVMVariableNode() {variable=pub isTargeting_lerp}.Value).Result
   )
   Multiply = &RigVMFunction_MathDoubleMul(B=0.5, A=@n1.OutputParam)
   > n3 = &RigUnit_ModifyTransforms(
    ItemToModify="((Item=(Type=Bone,Name=\"base_hat\"),Transform=(Rotation=(X=0,Y=0.2299,Z=0,W=0.9732),Translation=(X=0,Y=0,Z=0),Scale3D=(X=1,Y=1,Z=1))),(Item=(Type=Bone,Name=\"dyn_hat_rear_1\"),Transform=(Rotation=(X=0,Y=-0.2692,Z=0,W=0.9631),Translation=(X=0,Y=0,Z=0),Scale3D= ...(truncated)"
    WeightMinimum=0
    Weight=@Multiply.Result
   )
   n4 = &RigVMFunction_ControlFlowBranch(ExecuteContext=@n3.ExecutePin, Condition=RigVMVariableNode() {variable=pub isTargeting}.Value)
   DeltaRotation_v2 = RigVMFunctionReferenceNode(
    Rotation Order=ZYX
    child=(Type=Bone,Name="head")
    outputMinMax=(X=0,Y=1)
    parent=(Type=Null,Name="head_null")
    premult=1
    rotationMinMax=(X=-50,Y=-100)
    triggerAxis=Z
    twistAxis=(X=1,Y=0,Z=0)
    ExecuteContext=@n4.True
   )
   > &RigUnit_ModifyTransforms(
    ItemToModify="((Item=(Type=Bone,Name=\"base_hat\"),Transform=(Rotation=(X=0,Y=0.1736,Z=0,W=0.9848),Translation=(X=0,Y=0,Z=0),Scale3D=(X=1,Y=1,Z=1))),(Item=(Type=Bone,Name=\"dyn_hat_rear_1\"),Transform=(Rotation=(X=0,Y=-0.1947,Z=0,W=0.9809),Translation=(X=0,Y=0,Z=0),Scale3D= ...(truncated)"
    WeightMinimum=0
    Weight=@DeltaRotation_v2.outputWeight
   )
   n5 = &RigVMFunction_ControlFlowBranch(ExecuteContext=@RigVMFunction_Sequence.D, Condition=RigVMVariableNode() {variable=Celebration Emote}.Value)
   n6 = &RigUnit_SphericalPoseReader(
    ActiveRegionScaleFactors=(PositiveWidth=1,NegativeWidth=1.8,PositiveHeight=1.5,NegativeHeight=1.5)
    ActiveRegionSize=0.1
    Debug=(bDrawDebug=false,bDraw2D=False,bDrawLocalAxes=False,DebugScale=25,DebugSegments=20,DebugThickness=0.25)
    DriverAxis=(X=-1,Y=0,Z=0)
    DriverItem=(Type=Bone,Name="upperarm_r")
    FalloffRegionScaleFactors=(PositiveWidth=1,NegativeWidth=2.5,PositiveHeight=2,NegativeHeight=2)
    FalloffSize=0.25
    FlipHeightScaling=False
    FlipWidthScaling=False
    OptionalParentItem=(Type=Bone,Name="None")
    RotationOffset=(X=165,Y=-20,Z=0)
    ExecutePin=@n5.True
   )
   > &RigUnit_ModifyTransforms(
    ItemToModify="((Item=(Type=Bone,Name=\"dyn_hat_SideFront_r_1\"),Transform=(Rotation=(X=0,Y=-0.4226,Z=0,W=0.9063),Translation=(X=0,Y=0,Z=0),Scale3D=(X=1,Y=1,Z=1))),(Item=(Type=Bone,Name=\"dyn_hat_side_r\"),Transform=(Rotation=(X=0,Y=-0.2588,Z=0,W=0.9659),Translation=(X=0,Y=0 ...(truncated)"
    WeightMinimum=0
    Weight=@n6.OutputParam
   )
   RigVMFunction_ControlFlowBranch = &RigVMFunction_ControlFlowBranch(ExecuteContext=@RigVMFunction_Sequence.E, Condition=RigVMVariableNode() {variable=Parachuteopened}.Value)
   DeltaRotation_2 = RigVMFunctionReferenceNode(
    Rotation Order=ZXY
    child=clavicle_r
    debugRotation=false
    max=-80
    min=-25
    parent=clavicle_r_null
    parentType=Null
    primAxis=1
    twistAxis=(X=1,Y=0,Z=0)
    ExecuteContext=@RigVMFunction_ControlFlowBranch.True
   )
   > &RigUnit_ModifyTransforms(
    ItemToModify="((Item=(Type=Bone,Name=\"dyn_hat_SideFront_r_1\"),Transform=(Rotation=(X=0,Y=-0.6428,Z=0,W=0.766),Translation=(X=0,Y=0,Z=0),Scale3D=(X=1,Y=1,Z=1))),(Item=(Type=Bone,Name=\"dyn_hat_side_r\"),Transform=(Rotation=(X=0,Y=-0.4226,Z=0,W=0.9063),Translation=(X=0,Y=0, ...(truncated)"
    WeightMinimum=0
    Weight=@DeltaRotation_2.outputWeight
   )
   RigVMFunction_MathBoolOr = RigVMAggregateNode(
    B=RigVMVariableNode() {variable=Celebration Emote}.Value
    A=RigVMVariableNode() {variable=Parachuteopened}.Value
    C=RigVMVariableNode() {variable=pub isTargeting}.Value
    D=RigVMVariableNode() {variable=pub diveFloating}.Value
   )
   n7 = &RigVMFunction_ControlFlowBranch(Condition=@RigVMFunction_MathBoolOr.Result, ExecuteContext=@RigVMFunction_Sequence.F)
   n8 = &RigUnit_SphericalPoseReader(
    ActiveRegionScaleFactors=(PositiveWidth=1.5,NegativeWidth=1.5,PositiveHeight=1,NegativeHeight=1)
    ActiveRegionSize=0.1
    Debug=(bDrawDebug=false,bDraw2D=False,bDrawLocalAxes=False,DebugScale=25,DebugSegments=20,DebugThickness=0.25)
    DriverAxis=(X=-1,Y=0,Z=0)
    DriverItem=(Type=Bone,Name="clavicle_r")
    FalloffRegionScaleFactors=(PositiveWidth=1,NegativeWidth=2.5,PositiveHeight=1.5,NegativeHeight=1.5)
    FalloffSize=0.2
    FlipHeightScaling=False
    FlipWidthScaling=False
    OptionalParentItem=(Type=Bone,Name="None")
    RotationOffset=(X=190,Y=-5,Z=0)
    ExecutePin=@n7.False
   )
   > &RigUnit_ModifyTransforms(
    ItemToModify="((Item=(Type=Bone,Name=\"dyn_hat_SideFront_r_1\"),Transform=(Rotation=(X=0.2339,Y=-0.9398,Z=-0.2411,W=0.0629),Translation=(X=0,Y=0,Z=0),Scale3D=(X=1,Y=1,Z=1))),(Item=(Type=Bone,Name=\"dyn_hat_front_1\"),Transform=(Rotation=(X=0,Y=0,Z=0,W=1),Translation=(X=0,Y= ...(truncated)"
    WeightMinimum=0
    Weight=@n8.OutputParam
   )
   > n9 = &RigUnit_SphericalPoseReader(
    ActiveRegionScaleFactors=(PositiveWidth=2.5,NegativeWidth=1,PositiveHeight=1,NegativeHeight=1)
    ActiveRegionSize=0.1
    Debug=(bDrawDebug=false,bDraw2D=False,bDrawLocalAxes=False,DebugScale=25,DebugSegments=20,DebugThickness=0.25)
    DriverAxis=(X=1,Y=0,Z=0)
    DriverItem=(Type=Bone,Name="clavicle_l")
    FalloffRegionScaleFactors=(PositiveWidth=1.5,NegativeWidth=1,PositiveHeight=0.5,NegativeHeight=0.5)
    FalloffSize=0.3
    FlipHeightScaling=true
    FlipWidthScaling=true
    OptionalParentItem=(Type=Bone,Name="None")
    RotationOffset=(X=-200,Y=180,Z=0)
   )
   > &RigUnit_ModifyTransforms(
    ItemToModify="((Item=(Type=Bone,Name=\"dyn_hat_sidefront_l_1\"),Transform=(Rotation=(X=-0.3021,Y=-0.305,Z=-0.1065,W=0.8969),Translation=(X=0,Y=0,Z=0),Scale3D=(X=1,Y=1,Z=1))),(Item=(Type=Bone,Name=\"dyn_hat_Side_l_1\"),Transform=(Rotation=(X=-0.1361,Y=0.26,Z=-0.014,W=-0.9559 ...(truncated)"
    WeightMinimum=0
    Weight=@n9.OutputParam
   )
   RigVMFunctionReturnNode(ExecuteContext=@RigVMFunction_Sequence.G)
   &RigUnit_ParentConstraint(
    AdvancedSettings=(RotationOrderForFilter=XZY)
    Child=(Type=Bone,Name="dyn_hatCharms_b_01")
    ExecutePin=()
    Parents=((Item=(Type=Bone,Name="base_hat"),Weight=0.0527),(Item=(Type=Bone,Name="dyn_hat_front_2"),Weight=0.1699),(Item=(Type=Bone,Name="dyn_hat_front_1"),Weight=0.1996),(Item=(Type=Bone,Name="dyn_hat_sidefront_l_2"),Weight=0.5777))
    Weight=1
   )
   > &RigUnit_ParentConstraint(
    AdvancedSettings=(RotationOrderForFilter=XZY)
    Child=(Type=Bone,Name="dyn_hatCharms_a_01")
    Parents="((Item=(Type=Bone,Name=\"base_hat\"),Weight=0.0237),(Item=(Type=Bone,Name=\"dyn_hat_front_1\"),Weight=0.0789),(Item=(Type=Bone,Name=\"dyn_hat_SideFront_r_1\"),Weight=0.079),(Item=(Type=Bone,Name=\"dyn_hat_front_2\"),Weight=0.2201),(Item=(Type=Bone,Name=\"dyn_hat_Si ...(truncated)"
    Weight=1
   )
   > &RigUnit_ParentConstraint(
    AdvancedSettings=(RotationOrderForFilter=XZY)
    Child=(Type=Bone,Name="dyn_hatCharms_c_01")
    Parents=((Item=(Type=Bone,Name="dyn_hat_Side_l_1"),Weight=0.0277),(Item=(Type=Bone,Name="dyn_hat_Side_l_2"),Weight=0.1029),(Item=(Type=Bone,Name="base_hat"),Weight=0.2556),(Item=(Type=Bone,Name="dyn_hat_sidefront_l_2"),Weight=0.6138))
    Weight=1
   )
   > &RigUnit_ParentConstraint(
    AdvancedSettings=(RotationOrderForFilter=XZY)
    Child=(Type=Bone,Name="dyn_hatCharms_e_01")
    Parents=((Item=(Type=Bone,Name="dyn_hat_rear_1"),Weight=0.053),(Item=(Type=Bone,Name="dyn_hat_rear_2"),Weight=0.2229),(Item=(Type=Bone,Name="dyn_hat_Side_l_2"),Weight=0.2676),(Item=(Type=Bone,Name="base_hat"),Weight=0.4565))
    Weight=1
   )
   > &RigUnit_ParentConstraint(
    AdvancedSettings=(RotationOrderForFilter=XZY)
    Child=(Type=Bone,Name="dyn_hatCharms_d_01")
    Parents=((Item=(Type=Bone,Name="base_hat"),Weight=0.2654),(Item=(Type=Bone,Name="dyn_hat_Side_l_2"),Weight=0.7346))
    Weight=1
   )
   &RigUnit_ModifyTransforms(
    ExecutePin=()
    ItemToModify=((Item=(Type=Bone,Name="head"),Transform=(Rotation=(X=0,Y=0,Z=0,W=1),Translation=(X=0,Y=0,Z=0),Scale3D=(X=1,Y=1,Z=1))))
    Weight=1
    WeightMinimum=0
   )
   RigVMVariableNode() {variable=pub isTargeting_lerp}
  ]}
 New Function_ContainedGraph {nodes [
   n1 = RigVMFunctionEntryNode()
   RigVMFunction_MathBoolToInteger = &RigVMFunction_MathBoolToInteger(Value=@n1.onlyTwist)
   RigVMFunction_MathQuaternionSwingTwist = &RigVMFunction_MathQuaternionSwingTwist(
    TwistAxis=@n1.twistAxis
    Input=&RigUnit_GetRelativeTransformForItem(bChildInitial=False, bParentInitial=False, Child=@n1.child, Parent=@n1.parent).RelativeTransform.Rotation
   )
   n2 = RigVMDispatchNode(
    Values=((),())
    Index=@RigVMFunction_MathBoolToInteger.Result
    Values.0=@RigVMFunction_MathQuaternionSwingTwist.Swing
    Values.1=@RigVMFunction_MathQuaternionSwingTwist.Twist
   ) {template="DISPATCH_RigVMDispatch_SelectInt32(in Index,in Values,out Result)"}
   RigVMFunction_MathQuaternionToEuler = &RigVMFunction_MathQuaternionToEuler(Value=@n2.Result, RotationOrder=@n1.Rotation Order)
   > DISPATCH_RigVMDispatch_Print = RigVMDispatchNode(ScreenColor=(R=1,G=1,B=1,A=1), ScreenDuration=1.5, Enabled=@n1.debugRotation, Value=@RigVMFunction_MathQuaternionToEuler.Result) {template="DISPATCH_RigVMDispatch_Print(in Prefix,in Value,in Enabled,in ScreenDuration,in ScreenColor)"}
   Cast = RigVMDispatchNode(Value=@n1.triggerAxis) {template="DISPATCH_RigVMDispatch_CastEnumToInt(in Value,out Result)"}
   DISPATCH_RigVMDispatch_SelectInt32 = RigVMDispatchNode(
    Values=(0,0,0)
    Index=@Cast.Result
    Values.0=@RigVMFunction_MathQuaternionToEuler.Result.X
    Values.1=@RigVMFunction_MathQuaternionToEuler.Result.Y
    Values.2=@RigVMFunction_MathQuaternionToEuler.Result.Z
   ) {template="DISPATCH_RigVMDispatch_SelectInt32(in Index,in Values,out Result)"}
   Remap = &RigVMFunction_MathDoubleRemap(
    bClamp=False
    Value=@DISPATCH_RigVMDispatch_SelectInt32.Result
    TargetMinimum=@n1.outputMinMax.X
    TargetMaximum=@n1.outputMinMax.Y
    SourceMinimum=@n1.rotationMinMax.X
    SourceMaximum=@n1.rotationMinMax.Y
   )
   Absolute = &RigVMFunction_MathDoubleAbs(Value=@Remap.Result)
   Multiply = &RigVMFunction_MathDoubleMul(
    A=&RigVMFunction_MathDoubleClamp(Value=RigVMDispatchNode(True=@Absolute.Result, Condition=@n1.absoluteValue, False=@Remap.Result) {template="DISPATCH_RigVMDispatch_If(in Condition,in True,in False,out Result)"}.Result).Result
    B=@n1.premult
   )
   > RigVMFunctionReturnNode(outputQuat=@n2.Result, outputWeight=@Multiply.Result)
  ]}
 New Function_ContainedGraph {nodes [
   n1 = RigVMFunctionEntryNode()
   > DISPATCH_RigVMDispatch_ArrayIterator = RigVMDispatchNode(Array=@n1.transforms) {template="DISPATCH_RigVMDispatch_ArrayIterator(in Array,out Element,out Index,out Count,out Ratio)"}
   Multiply = &RigVMFunction_MathTransformMul(
    A=@DISPATCH_RigVMDispatch_ArrayIterator.Element.Transform
    B=&RigUnit_GetTransform(
     Item=(type=Bone,Name="dyn_L_hair_3")
     Space=LocalSpace
     bInitial=False
     Item.Name=@DISPATCH_RigVMDispatch_ArrayIterator.Element.Bone
     Item.type=@n1.type
    ).Transform
   ) {template="Multiply::Execute(in A,in B,out Result)"}
   > &RigUnit_SetTransform(
    Item=(type=Bone,Name="dyn_L_hair_3")
    Space=LocalSpace
    bInitial=false
    Value=@Multiply.Result
    bPropagateToChildren=@n1.inheritsTransform
    weight=@n1.weight
    Item.Name=@DISPATCH_RigVMDispatch_ArrayIterator.Element.Bone
    Item.type=@n1.type
   )
   RigVMFunctionReturnNode(ExecuteContext=@DISPATCH_RigVMDispatch_ArrayIterator.Completed)
   &RigUnit_OffsetTransformForItem(
    ExecutePin=()
    Item=(type=Bone,Name="None")
    OffsetTransform=(Rotation=(X=0,Y=0,Z=0,W=1),Translation=(X=0,Y=0,Z=0),Scale3D=(X=1,Y=1,Z=1))
    bPropagateToChildren=True
    weight=1
   )
  ]}
 New Function_ContainedGraph {nodes [
   n1 = RigVMFunctionEntryNode()
   RigVMFunction_MathBoolToInteger = &RigVMFunction_MathBoolToInteger(Value=@n1.onlyTwist)
   RigVMFunction_MathQuaternionSwingTwist = &RigVMFunction_MathQuaternionSwingTwist(
    twistAxis=@n1.twistAxis
    Input=&RigUnit_GetRelativeTransformForItem(
     bChildInitial=False
     bParentInitial=False
     child=(Type=Bone,Name="None")
     parent=(Type=Bone,Name="None")
     child.Name=@n1.child
     parent.Name=@n1.parent
     parent.Type=@n1.parentType
    ).RelativeTransform.Rotation
   )
   RigVMFunction_MathQuaternionToEuler = &RigVMFunction_MathQuaternionToEuler(
    Value=RigVMDispatchNode(
     Values=((),())
     Index=@RigVMFunction_MathBoolToInteger.Result
     Values.0=@RigVMFunction_MathQuaternionSwingTwist.Swing
     Values.1=@RigVMFunction_MathQuaternionSwingTwist.Twist
    ) {template="DISPATCH_RigVMDispatch_SelectInt32(in Index,in Values,out Result)"}.Result
    RotationOrder=@n1.Rotation Order
   )
   > DISPATCH_RigVMDispatch_Print = RigVMDispatchNode(ScreenColor=(R=1,G=1,B=1,A=1), ScreenDuration=1.5, Enabled=@n1.debugRotation, Value=@RigVMFunction_MathQuaternionToEuler.Result) {template="DISPATCH_RigVMDispatch_Print(in Prefix,in Value,in Enabled,in ScreenDuration,in ScreenColor)"}
   Clamp = &RigVMFunction_MathIntClamp(Value=@n1.primAxis)
   DISPATCH_RigVMDispatch_SelectInt32 = RigVMDispatchNode(
    Values=(0,0,0)
    Index=@Clamp.Result
    Values.0=@RigVMFunction_MathQuaternionToEuler.Result.X
    Values.1=@RigVMFunction_MathQuaternionToEuler.Result.Y
    Values.2=@RigVMFunction_MathQuaternionToEuler.Result.Z
   ) {template="DISPATCH_RigVMDispatch_SelectInt32(in Index,in Values,out Result)"}
   Remap = &RigVMFunction_MathDoubleRemap(
    TargetMaximum=1
    TargetMinimum=0
    bClamp=False
    Value=@DISPATCH_RigVMDispatch_SelectInt32.Result
    SourceMaximum=@n1.max
    SourceMinimum=@n1.min
   )
   > RigVMFunctionReturnNode(outputWeight=&RigVMFunction_MathDoubleClamp(Value=@Remap.Result).Result)
  ]}
 New Function_ContainedGraph {nodes [
   n1 = RigVMFunctionEntryNode()
   > DISPATCH_RigVMDispatch_ArrayIterator = RigVMDispatchNode(Array=@n1.transforms) {template="DISPATCH_RigVMDispatch_ArrayIterator(in Array,out Element,out Index,out Count,out Ratio)"}
   Multiply = &RigVMFunction_MathTransformMul(
    A=@DISPATCH_RigVMDispatch_ArrayIterator.Element.Transform
    B=&RigUnit_GetTransform(
     Item=(type=Bone,Name="dyn_L_hair_3")
     Space=LocalSpace
     bInitial=False
     Item.Name=@DISPATCH_RigVMDispatch_ArrayIterator.Element.Bone
     Item.type=@n1.type
    ).Transform
   ) {template="Multiply::Execute(in A,in B,out Result)"}
   > &RigUnit_SetTransform(
    Item=(type=Bone,Name="dyn_L_hair_3")
    Space=LocalSpace
    bInitial=false
    Value=@Multiply.Result
    bPropagateToChildren=@n1.inheritsTransform
    weight=@n1.weight
    Item.Name=@DISPATCH_RigVMDispatch_ArrayIterator.Element.Bone
    Item.type=@n1.type
   )
   RigVMFunctionReturnNode(ExecuteContext=@DISPATCH_RigVMDispatch_ArrayIterator.Completed)
   &RigUnit_OffsetTransformForItem(
    ExecutePin=()
    Item=(type=Bone,Name="None")
    OffsetTransform=(Rotation=(X=0,Y=0,Z=0,W=1),Translation=(X=0,Y=0,Z=0),Scale3D=(X=1,Y=1,Z=1))
    bPropagateToChildren=True
    weight=1
   )
  ]}
 RigVMFunctionLibrary {nodes [
   DeltaRotation = RigVMCollapseNode(ExecuteContext=(), Rotation Order=ZYX, twistAxis=(X=1,Y=0,Z=0))
   DeltaRotation_1 = RigVMCollapseNode(ExecuteContext=(), Rotation Order=ZYX, twistAxis=(X=1,Y=0,Z=0))
   DeltaRotation_v2 = RigVMCollapseNode(
    ExecuteContext=()
    Rotation Order=ZYX
    child=(Type=Bone,Name="None")
    outputMinMax=(X=0,Y=1)
    parent=(Type=Bone,Name="None")
    premult=1
    rotationMinMax=(X=0,Y=0)
    twistAxis=(X=1,Y=0,Z=0)
   )
   Hair Adjustments = RigVMCollapseNode(ExecuteContext=())
   Hat_fix = RigVMCollapseNode(ExecuteContext=())
   alterTransform = RigVMCollapseNode(type=Bone, ExecuteContext=(), transforms=())
   alterTransform_1 = RigVMCollapseNode(type=Bone, ExecuteContext=(), transforms=())
   alterTransform_2 = RigVMCollapseNode(type=Bone, ExecuteContext=(), transforms=())
   claviclesBackFoward = RigVMCollapseNode(ExecuteContext=())
   controllers = RigVMCollapseNode(ExecuteContext=(), debug=False)
   splineIK = RigVMCollapseNode(ExecuteContext=())
  ]}
 RigVMFunction_MathBoolOr_1_ContainedGraph {nodes [
   n1 = RigVMFunctionEntryNode()
   RigVMFunction_MathBoolOr = &RigVMFunction_MathBoolOr(A=@n1.A, B=@n1.B)
   RigVMFunctionReturnNode(Result=&RigVMFunction_MathBoolOr(B=@n1.D, A=&RigVMFunction_MathBoolOr(B=@n1.C, A=@RigVMFunction_MathBoolOr.Result).Result).Result)
  ]}
 RigVMFunction_Sequence_1_ContainedGraph {nodes [
   n1 = RigVMFunctionEntryNode()
   RigVMFunction_Sequence = &RigVMFunction_Sequence(ExecuteContext=@n1.ExecuteContext)
   n2 = &RigVMFunction_Sequence(ExecuteContext=@RigVMFunction_Sequence.B)
   RigVMFunctionReturnNode(A=@RigVMFunction_Sequence.A, B=@n2.A, C=@n2.B)
  ]}
 RigVMFunction_Sequence_1_ContainedGraph {nodes [
   n1 = RigVMFunctionEntryNode()
   RigVMFunction_Sequence = &RigVMFunction_Sequence(ExecuteContext=@n1.ExecuteContext)
   n2 = &RigVMFunction_Sequence(ExecuteContext=@RigVMFunction_Sequence.B)
   n3 = &RigVMFunction_Sequence(ExecuteContext=@n2.B)
   n4 = &RigVMFunction_Sequence(ExecuteContext=@n3.B)
   n5 = &RigVMFunction_Sequence(ExecuteContext=@n4.B)
   n6 = &RigVMFunction_Sequence(ExecuteContext=@n5.B)
   RigVMFunctionReturnNode(A=@RigVMFunction_Sequence.A, B=@n2.A, C=@n3.A, D=@n4.A, E=@n5.A, F=@n6.A, G=@n6.B)
  ]}
 RigVMFunction_Sequence_1_ContainedGraph {nodes [
   n1 = RigVMFunctionEntryNode()
   RigVMFunction_Sequence = &RigVMFunction_Sequence(ExecuteContext=@n1.ExecuteContext)
   n2 = &RigVMFunction_Sequence(ExecuteContext=@RigVMFunction_Sequence.B)
   n3 = &RigVMFunction_Sequence(ExecuteContext=@n2.B)
   n4 = &RigVMFunction_Sequence(ExecuteContext=@n3.B)
   RigVMFunctionReturnNode(A=@RigVMFunction_Sequence.A, B=@n2.A, C=@n3.A, D=@n4.A, E=@n4.B)
  ]}
 RigVMFunction_Sequence_1_ContainedGraph {nodes [
   n1 = RigVMFunctionEntryNode()
   RigVMFunction_Sequence = &RigVMFunction_Sequence(ExecuteContext=@n1.ExecuteContext)
   n2 = &RigVMFunction_Sequence(ExecuteContext=@RigVMFunction_Sequence.B)
   n3 = &RigVMFunction_Sequence(ExecuteContext=@n2.B)
   n4 = &RigVMFunction_Sequence(ExecuteContext=@n3.B)
   n5 = &RigVMFunction_Sequence(ExecuteContext=@n4.B)
   n6 = &RigVMFunction_Sequence(ExecuteContext=@n5.B)
   RigVMFunctionReturnNode(A=@RigVMFunction_Sequence.A, B=@n2.A, C=@n3.A, D=@n4.A, E=@n5.A, F=@n6.A, G=@n6.B)
  ]}
 RigVMFunction_Sequence_3_1_ContainedGraph {nodes [
   n1 = RigVMFunctionEntryNode()
   n2 = &RigVMFunction_Sequence(ExecuteContext=@n1.ExecuteContext)
   n3 = &RigVMFunction_Sequence(ExecuteContext=@n2.B)
   n4 = &RigVMFunction_Sequence(ExecuteContext=@n3.B)
   n5 = &RigVMFunction_Sequence(ExecuteContext=@n4.B)
   n6 = &RigVMFunction_Sequence(ExecuteContext=@n5.B)
   RigVMFunctionReturnNode(C=@n4.A, D=@n5.A, A=@n2.A, B=@n3.A, E=@n6.A, F=@n6.B)
  ]}
 RigVMModel {nodes [
   RigUnit_BeginExecution = &RigUnit_BeginExecution()
   &RigUnit_ModifyTransforms(
    ItemToModify=((Item=(Type=Bone,Name="head"),Transform=(Rotation=(X=0,Y=0,Z=0,W=1),Translation=(X=0,Y=0,Z=0),Scale3D=(X=1,Y=1,Z=1))))
    Weight=1
    WeightMinimum=0
    ExecutePin=@RigUnit_BeginExecution.ExecutePin
   )
   > splineIK = RigVMFunctionReferenceNode()
   > Hair Adjustments = RigVMFunctionReferenceNode()
   > Hat_fix = RigVMFunctionReferenceNode()
   controllers = RigVMFunctionReferenceNode(ExecuteContext=(), debug=RigVMVariableNode() {variable=pub disablecontrolRig}.Value)
  ]}
 RigVMModel Rig_4 {nodes [
   n1 = &RigUnit_PrepareForExecution()
   RigVMFunction_Sequence = RigVMAggregateNode(ExecuteContext=@n1.ExecutePin)
   n2 = &RigUnit_ItemArray(Items="((Type=Control,Name=\"dyn_L_hair_1_ctrl\"),(Type=Control,Name=\"dyn_L_hair_2_ctrl\"),(Type=Control,Name=\"dyn_L_hair_3_ctrl\"),(Type=Control,Name=\"dyn_L_hair_4_ctrl\"),(Type=Control,Name=\"dyn_L_hair_5_ctrl\"),(Type=Control,Name=\"dyn_L_hair_6_ctrl\"),(Type=Control,N ...(truncated)")
   RigVMVariableNode(Value=@n2.Items, ExecuteContext=@RigVMFunction_Sequence.A) {variable=l_dyn_hair_controls}
   > RigVMVariableNode(Value=/Script/ControlRig.RigUnit_HierarchyGetPoseItemArray(ElementType=All, Initial=False, ItemsToGet=@n2.Items) {is=struct RigVMUnitNode}.Pose) {variable=l_dyn_hair_pose}
   RigUnit_ItemArray = &RigUnit_ItemArray(Items="((Type=Control,Name=\"dyn_R_hair_1_ctrl\"),(Type=Control,Name=\"dyn_R_hair_2_ctrl\"),(Type=Control,Name=\"dyn_R_hair_3_ctrl\"),(Type=Control,Name=\"dyn_R_hair_4_ctrl\"),(Type=Control,Name=\"dyn_R_hair_5_ctrl\"),(Type=Control,Name=\"dyn_R_hair_6_ctrl\"),(Type=Control,N ...(truncated)")
   RigVMVariableNode(Value=@RigUnit_ItemArray.Items, ExecuteContext=@RigVMFunction_Sequence.B) {variable=r_dyn_hair_controls}
   > RigVMVariableNode(Value=&RigUnit_HierarchyGetPoseItemArray(ElementType=All, Initial=False, ItemsToGet=@RigUnit_ItemArray.Items).Pose) {variable=r_dyn_hair_pose}
   n3 = &RigUnit_ItemArray(Items="((Type=Control,Name=\"dyn_C_hair_1_ctrl\"),(Type=Control,Name=\"dyn_C_hair_2_ctrl\"),(Type=Control,Name=\"dyn_C_hair_3_ctrl\"),(Type=Control,Name=\"dyn_C_hair_4_ctrl\"),(Type=Control,Name=\"dyn_C_hair_5_ctrl\"),(Type=Control,Name=\"dyn_C_hair_6_ctrl\"),(Type=Control,N ...(truncated)")
   RigVMVariableNode(Value=@n3.Items, ExecuteContext=@RigVMFunction_Sequence.C) {variable=c_dyn_hair_controls}
   > RigVMVariableNode(Value=&RigUnit_HierarchyGetPoseItemArray(ElementType=All, Initial=False, ItemsToGet=@n3.Items).Pose) {variable=c_dyn_hair_pose}
  ]}
]
hierarchy [
 root {loc=<0 0 0>, rot=<0 0 0 1>}
 pelvis {parent=root, loc=<0 -0.5301 98.0571>, rot=<0.0127 0.707 -0.0127 -0.707>}
 spine_01 {parent=pelvis, loc=<6.1208 -0.9328 0>, rot=<0 0 0.034 -0.9994>}
 spine_02 {parent=spine_01, loc=<6.5996 -0.2095 0>, rot=<0 0 0.0238 -0.9997>}
 spine_03 {parent=spine_02, loc=<6.4914 -0.1237 0>, rot=<0 0 0.0149 -0.9999>}
 spine_04 {parent=spine_03, loc=<6.9598 0.0794 0>, rot=<0 0 -0.0132 -0.9999>}
 spine_05 {parent=spine_04, loc=<6.9987 0 0>, rot=<0 0 -0.0015 -1>}
 upperarm_r_null {parent=spine_05, type=Null, loc=<4.5519 -1.8922 13.0371>, rot=<-0.9118 -0.0385 0.4081 -0.0238>}
 neck_01 {parent=spine_05, loc=<11.3138 0.1529 0>, rot=<0 0 -0.2003 -0.9797>}
 neck_02 {parent=neck_01, loc=<4.5627 0 0>, rot=<0 0 0.0504 -0.9987>}
 head {parent=neck_02, loc=<5.3914 0 0>, rot=<0 0 0.115 -0.9934>}
 base_C_hair_1 {parent=head, loc=<6.9517 -5.4626 0>, rot=<-0.0949 0.7007 0.7007 -0.0949>}
 dyn_C_hair_1 {parent=base_C_hair_1, loc=<0 0 0>, rot=<0 0 0 1>}
 dyn_C_hair_2 {parent=dyn_C_hair_1, loc=<10.6015 0 0>, rot=<0 -0.0172 0 -0.9999>}
 dyn_C_hair_3 {parent=dyn_C_hair_2, loc=<10.6008 0 0>, rot=<0 0.0789 0 -0.9969>}
 dyn_C_hair_4 {parent=dyn_C_hair_3, loc=<10.5829 0 0>, rot=<0 0.0898 0 -0.996>}
 dyn_C_hair_5 {parent=dyn_C_hair_4, loc=<10.5971 0 0>, rot=<0 0.0001 0 -1>}
 dyn_C_hair_6 {parent=dyn_C_hair_5, loc=<10.5971 0 0>, rot=<0 -0.0903 0 -0.9959>}
 dyn_C_hair_7 {parent=dyn_C_hair_6, loc=<10.5831 0 0>, rot=<0 -0.0859 0 -0.9963>}
 dyn_C_hair_7_end {parent=dyn_C_hair_7, loc=<10.5979 0 0>, rot=<0 -0.0983 0 -0.9952>}
 base_L_hair_1 {parent=head, loc=<6.9314 -3.0124 -7.9806>, rot=<0.0956 -0.7006 -0.7001 0.0996>}
 dyn_L_hair_1 {parent=base_L_hair_1, loc=<0 0 0>, rot=<0 0 0 1>}
 dyn_L_hair_2 {parent=dyn_L_hair_1, loc=<10.7733 0 0>, rot=<0.0014 -0.0775 0.0234 -0.9967>}
 dyn_L_hair_3 {parent=dyn_L_hair_2, loc=<10.7695 0 0>, rot=<-0.0067 0.1556 0.0018 -0.9878>}
 dyn_L_hair_4 {parent=dyn_L_hair_3, loc=<10.67 0 0>, rot=<-0.0019 0.0735 -0.0194 -0.9971>}
 dyn_L_hair_5 {parent=dyn_L_hair_4, loc=<10.7742 0 0>, rot=<-0.0002 0.0115 0.0111 -0.9999>}
 dyn_L_hair_6 {parent=dyn_L_hair_5, loc=<10.7637 0 0>, rot=<0.004 -0.1481 -0.0009 -0.989>}
 dyn_L_hair_7 {parent=dyn_L_hair_6, loc=<10.7571 0 0>, rot=<0.0021 -0.0478 0.0183 -0.9987>}
 dyn_L_hair_7_end {parent=dyn_L_hair_7, loc=<10.7622 0 0>, rot=<0 -0.0983 0 -0.9952>}
 dyn_L_hair_3_a_01 {parent=dyn_L_hair_3, loc=<9.2661 -3.6002 -1.4399>, rot=<-0.0022 0.0258 0.0404 -0.9988>}
 dyn_L_hair_3_a_02 {parent=dyn_L_hair_3_a_01, loc=<10.053 0 0>, rot=<0.0056 -0.036 0.0293 -0.9989>}
 dyn_L_hair_3_a_03 {parent=dyn_L_hair_3_a_02, loc=<10.0505 0 0>, rot=<0 0 0 1>}
 base_R_hair_1 {parent=head, loc=<6.9319 -3.0124 7.9806>, rot=<-0.0996 0.7001 0.7006 -0.0956>}
 dyn_R_hair_1 {parent=base_R_hair_1, loc=<0 0 0>, rot=<0 0 0 1>}
 dyn_R_hair_2 {parent=dyn_R_hair_1, loc=<10.7735 0 0.0001>, rot=<-0.0014 -0.0775 -0.0234 -0.9967>}
 dyn_R_hair_3 {parent=dyn_R_hair_2, loc=<10.7696 0 0>, rot=<0.0067 0.1556 -0.0018 -0.9878>}
 dyn_R_hair_4 {parent=dyn_R_hair_3, loc=<10.6703 0 0.0001>, rot=<0.0019 0.0735 0.0194 -0.9971>}
 dyn_R_hair_5 {parent=dyn_R_hair_4, loc=<10.7743 0 0>, rot=<0.0002 0.0115 -0.0111 -0.9999>}
 dyn_R_hair_6 {parent=dyn_R_hair_5, loc=<10.7632 0 0>, rot=<-0.004 -0.1481 0.0009 -0.989>}
 dyn_R_hair_7 {parent=dyn_R_hair_6, loc=<10.7574 0 0>, rot=<-0.0021 -0.0478 -0.0183 -0.9987>}
 dyn_R_hair_7_end {parent=dyn_R_hair_7, loc=<10.7622 0 0.0001>, rot=<0 -0.0983 0 -0.9952>}
 dyn_R_hair_4_a_01 {parent=dyn_R_hair_4, loc=<2.5174 5.7251 -1.8019>, rot=<-0.0039 -0.0366 -0.1003 -0.9943>}
 dyn_R_hair_4_a_02 {parent=dyn_R_hair_4_a_01, loc=<9.8886 0 0>, rot=<-0.0029 -0.0165 0.0337 -0.9993>}
 dyn_R_hair_4_a_03 {parent=dyn_R_hair_4_a_02, loc=<11.2472 0 0>, rot=<0 0 0 1>}
 dyn_earring_l {parent=head, loc=<0.4505 0.0271 -6.2969>, rot=<0.0034 -0.8155 -0.5787 0.0024>}
 base_hat {parent=head, loc=<7.6428 0.3962 0.3568>, rot=<-0.0029 0.7071 0.7071 -0.0029>}
 dyn_hatCharms_d_01 {parent=base_hat, loc=<-1.1512 -14.906 -17.1704>, rot=<-0.8492 0 0 -0.528>}
 dyn_hatCharms_e_01 {parent=base_hat, loc=<1.6597 -8.3236 -20.8559>, rot=<0.9998 0 0 0.0186>}
 dyn_hatCharms_e_02 {parent=dyn_hatCharms_e_01, loc=<0.9396 -0.0207 -0.0387>, rot=<0.6938 0 0 -0.7201>}
 dyn_hatCharms_c_01 {parent=base_hat, loc=<-10.7912 -19.794 6.5926>, rot=<-0.708 0 0 -0.7062>}
 dyn_hatCharms_b_01 {parent=base_hat, loc=<-4.355 -10.1476 20.8766>, rot=<-0.3544 0 0 -0.9351>}
 dyn_hatCharms_b_02 {parent=dyn_hatCharms_b_01, loc=<1.8698 -0.0663 0.074>, rot=<0.0023 0 0 -1>}
 dyn_hatCharms_a_01 {parent=base_hat, loc=<7.0208 14.5282 17.2638>, rot=<-0.3724 0 0 -0.9281>}
 dyn_hatCharms_a_02 {parent=dyn_hatCharms_a_01, loc=<0.8609 -0.0437 0.007>, rot=<0.3724 0 0 -0.9281>}
 base_hat_side_r {parent=base_hat, loc=<1.8478 7.1684 -6.5786>, rot=<0.6533 0.3262 -0.5797 -0.3615>}
 dyn_hat_side_r {parent=base_hat_side_r, loc=<0 0 0>, rot=<0 0 0 1>}
 base_hat_SideFront_r {parent=base_hat, loc=<2.3114 8.6852 3.4102>, rot=<0.4906 0.4282 -0.4324 -0.6238>}
 dyn_hat_SideFront_r_1 {parent=base_hat_SideFront_r, loc=<0 0 0>, rot=<0 0 0 1>}
 base_hat_Side_l {parent=base_hat, loc=<-2.7631 -8.5483 -3.4995>, rot=<-0.571 0.5349 0.5873 -0.2072>}
 dyn_hat_Side_l_1 {parent=base_hat_Side_l, loc=<0 0 0>, rot=<0 0 0 1>}
 base_hat_front {parent=base_hat, loc=<0.8942 1.6565 10.2977>, rot=<0.2194 0.596 0.0607 -0.77>}
 dyn_hat_front_1 {parent=base_hat_front, loc=<0 0 0>, rot=<0 0 0 1>}
 base_hat_sidefront_l {parent=base_hat, loc=<-2.4806 -6.8502 6.7978>, rot=<-0.1234 0.6844 0.5071 -0.5091>}
 dyn_hat_sidefront_l_1 {parent=base_hat_sidefront_l, loc=<0 0 0>, rot=<0 0 0 1>}
 base_hat_rear1 {parent=base_hat, loc=<0.2496 -0.0922 -10.409>, rot=<0.8026 -0.0563 -0.5889 -0.0767>}
 dyn_hat_rear_1 {parent=base_hat_rear1, loc=<0 0 0>, rot=<0 0 0 1>}
 head_null {parent=neck_02, type=Null, loc=<5.3914 0 0>, rot=<0 0 0.115 -0.9934>}
 clavicle_l {parent=spine_05, loc=<5.4831 -0.1053 -3.2946>, rot=<0.0465 -0.7289 -0.009 -0.6829>}
 upperarm_l {parent=clavicle_l, loc=<9.8658 -1.2136 -0.4105>, rot=<0.0164 -0.3245 -0.0269 -0.9454>}
 upperarm_twist_01_l {parent=upperarm_l, loc=<7.2135 -0.1134 0>, rot=<0 0 0 1>}
 deform_upperarm_bk_base_l {parent=upperarm_twist_01_l, loc=<-6.4053 -4.3746 0.314>, rot=<0 0 0 -1>}
 deform_upperarm_bk_l {parent=deform_upperarm_bk_base_l, loc=<0 0 0>, rot=<0 0 0 1>}
 deform_upperarm_fr_base_l {parent=upperarm_twist_01_l, loc=<-5.6312 5.092 0.3354>, rot=<0 0 0 -1>}
 deform_upperarm_fr_l {parent=deform_upperarm_fr_base_l, loc=<0 0 0>, rot=<0 0 0 1>}
 deform_upperarm_up_base_l {parent=upperarm_twist_01_l, loc=<-8.891 0.4825 3.9517>, rot=<0 0 0 -1>}
 deform_upperarm_up_l {parent=deform_upperarm_up_base_l, loc=<0 0 0>, rot=<0 0 0 1>}
 lowerarm_l {parent=upperarm_l, loc=<26.465 -0.8345 0.259>, rot=<0.0036 0.0161 -0.2298 -0.9731>}
 hand_l {parent=lowerarm_l, loc=<22.7355 -0.3203 -0.6105>, rot=<-0.6784 -0.0223 0.0735 -0.7307>}
 thumb_01_l {parent=hand_l, loc=<1.3118 -0.3191 -2.459>, rot=<0.6546 -0.4042 0.0664 -0.6354>}
 thumb_02_l {parent=thumb_01_l, loc=<5.9336 -0.0448 0.0401>, rot=<-0.0037 0.0213 0.298 -0.9543>}
 thumb_03_l {parent=thumb_02_l, loc=<2.2223 0.3682 -0.0907>, rot=<0.0118 -0.013 0.0066 -0.9998>}
 index_metacarpal_l {parent=hand_l, loc=<1.6535 0.0183 -2.96>, rot=<0.0158 -0.0124 -0.0173 -0.9996>}
 index_01_l {parent=index_metacarpal_l, loc=<8.3008 0.0357 -0.0524>, rot=<-0.0044 0.0485 0.0924 -0.9945>}
 index_02_l {parent=index_01_l, loc=<3.3587 -0.3004 0.0452>, rot=<-0.002 0.0027 0.3225 -0.9466>}
 index_03_l {parent=index_02_l, loc=<2.5578 -0.123 -0.0252>, rot=<-0.002 0.0067 0.2558 -0.9667>}
 middle_metacarpal_l {parent=hand_l, loc=<2.1856 -0.1725 -0.5984>, rot=<0.006 0.0168 -0.0329 -0.9993>}
 middle_01_l {parent=middle_metacarpal_l, loc=<7.3275 -0.2447 0.0129>, rot=<-0.0208 0.0042 0.0858 -0.9961>}
 middle_02_l {parent=middle_01_l, loc=<4.0123 0.1401 0.1175>, rot=<-0.0218 0.0258 0.3512 -0.9357>}
 middle_03_l {parent=middle_02_l, loc=<2.6077 0.0534 0.1373>, rot=<0.0082 0.0188 0.1708 -0.9851>}
 ring_metacarpal_l {parent=hand_l, loc=<2.2074 -0.3625 1.3638>, rot=<-0.0284 0.0722 -0.0458 -0.9959>}
 ring_01_l {parent=ring_metacarpal_l, loc=<7.0295 -0.6197 -0.1446>, rot=<-0.0489 -0.0252 0.1892 -0.9804>}
 ring_02_l {parent=ring_01_l, loc=<3.6174 0.4647 -0.1027>, rot=<0 0 0.2806 -0.9598>}
 ring_03_l {parent=ring_02_l, loc=<2.4152 0.114 0.0199>, rot=<0 0 0.1189 -0.9929>}
 pinky_metacarpal_l {parent=hand_l, loc=<1.7616 -0.9753 3.3814>, rot=<-0.0397 0.0814 -0.069 -0.9935>}
 pinky_01_l {parent=pinky_metacarpal_l, loc=<7.1709 -0.4831 -0.1094>, rot=<-0.0654 -0.0035 0.2242 -0.9723>}
 pinky_02_l {parent=pinky_01_l, loc=<2.8274 0.0893 -0.1043>, rot=<0.0037 0.0011 0.2854 -0.9584>}
 pinky_03_l {parent=pinky_02_l, loc=<2.2046 0.0307 -0.0078>, rot=<-0.0002 0.0098 0.1445 -0.9895>}
 deform_wrist_dn_base_l {parent=hand_l, loc=<0 -1.5326 0>, rot=<0 0 0 -1>}
 deform_wrist_dn_l {parent=deform_wrist_dn_base_l, loc=<0 0 0>, rot=<0 0 0 1>}
 deform_wrist_up_base_l {parent=hand_l, loc=<0 2.3009 0>, rot=<0 0 0 -1>}
 deform_wrist_up_l {parent=deform_wrist_up_base_l, loc=<0 0 0>, rot=<0 0 0 1>}
 lowerarm_twist_02_l {parent=lowerarm_l, loc=<10.6707 0 0>, rot=<0 0 0 1>}
 lowerarm_twist_01_l {parent=lowerarm_l, loc=<18.316 0 0>, rot=<0 0 0 1>}
 deform_elbow_in_base_l {parent=lowerarm_l, loc=<0.1834 3.6062 -0.2572>, rot=<0 0 0 -1>}
 deform_elbow_in_l {parent=deform_elbow_in_base_l, loc=<0 0 0>, rot=<0 0 0 1>}
 deform_elbow_out_base_l {parent=lowerarm_l, loc=<0 -4 0>, rot=<0 0 0 1>}
 deform_elbow_out_l {parent=deform_elbow_out_base_l, loc=<0 0 0>, rot=<0 0 0 1>}
 upperarm_twist_02_l {parent=upperarm_l, loc=<19.7998 -0.1134 0>, rot=<0 0 0 1>}
 base_sleevescloth_l {parent=upperarm_l, loc=<0 0 0>, rot=<-0.0557 0.1428 0.028 -0.9878>}
 dyn_sleevescloth_l_02 {parent=base_sleevescloth_l, loc=<0 0 0>, rot=<0 0 0 1>}
 dyn_sleevescloth_l_03 {parent=dyn_sleevescloth_l_02, loc=<6.916 0.1724 -2.046>, rot=<0 0 0 1>}
 dyn_sleevescloth_l_04 {parent=dyn_sleevescloth_l_03, loc=<7.5703 0 0>, rot=<0 0 0 1>}
 clavicle_r {parent=spine_05, loc=<5.4831 -0.1053 3.2946>, rot=<0.7289 0.0465 -0.6829 0.009>}
 upperarm_r {parent=clavicle_r, loc=<-9.8658 1.2136 0.4105>, rot=<0.0164 -0.3245 -0.0269 -0.9454>}
 upperarm_twist_01_r {parent=upperarm_r, loc=<-7.2135 0.1134 0>, rot=<0 0 0 -1>}
 deform_upperarm_bk_base_r {parent=upperarm_twist_01_r, loc=<6.4053 4.3746 -0.314>, rot=<0 0 0 -1>}
 deform_upperarm_bk_r {parent=deform_upperarm_bk_base_r, loc=<0 0 0>, rot=<0 0 0 1>}
 deform_upperarm_fr_base_r {parent=upperarm_twist_01_r, loc=<5.6312 -5.092 -0.3354>, rot=<0 0 0 -1>}
 deform_upperarm_fr_r {parent=deform_upperarm_fr_base_r, loc=<0 0 0>, rot=<0 0 0 1>}
 deform_upperarm_up_base_r {parent=upperarm_twist_01_r, loc=<8.891 -0.4825 -3.9517>, rot=<0 0 0 -1>}
 deform_upperarm_up_r {parent=deform_upperarm_up_base_r, loc=<0 0 0>, rot=<0 0 0 1>}
 lowerarm_r {parent=upperarm_r, loc=<-26.465 0.8345 -0.259>, rot=<0.0036 0.0161 -0.2298 -0.9731>}
 hand_r {parent=lowerarm_r, loc=<-22.7355 0.3203 0.6106>, rot=<-0.6784 -0.0223 0.0735 -0.7307>}
 thumb_01_r {parent=hand_r, loc=<-1.3118 0.3191 2.459>, rot=<0.6546 -0.4042 0.0664 -0.6354>}
 thumb_02_r {parent=thumb_01_r, loc=<-5.9336 0.0448 -0.0401>, rot=<-0.0037 0.0213 0.298 -0.9543>}
 thumb_03_r {parent=thumb_02_r, loc=<-2.2223 -0.3682 0.0907>, rot=<0.0118 -0.013 0.0066 -0.9998>}
 index_metacarpal_r {parent=hand_r, loc=<-1.6535 -0.0183 2.96>, rot=<0.0158 -0.0124 -0.0173 -0.9996>}
 index_01_r {parent=index_metacarpal_r, loc=<-8.3008 -0.0357 0.0524>, rot=<-0.0044 0.0485 0.0924 -0.9945>}
 index_02_r {parent=index_01_r, loc=<-3.3587 0.3004 -0.0452>, rot=<-0.002 0.0027 0.3225 -0.9466>}
 index_03_r {parent=index_02_r, loc=<-2.5578 0.123 0.0252>, rot=<-0.002 0.0067 0.2558 -0.9667>}
 middle_metacarpal_r {parent=hand_r, loc=<-2.1856 0.1725 0.5984>, rot=<0.006 0.0168 -0.0329 -0.9993>}
 middle_01_r {parent=middle_metacarpal_r, loc=<-7.3275 0.2447 -0.0129>, rot=<-0.0208 0.0042 0.0858 -0.9961>}
 middle_02_r {parent=middle_01_r, loc=<-4.0123 -0.1401 -0.1175>, rot=<-0.0218 0.0258 0.3512 -0.9357>}
 middle_03_r {parent=middle_02_r, loc=<-2.6077 -0.0534 -0.1373>, rot=<0.0082 0.0188 0.1708 -0.9851>}
 ring_metacarpal_r {parent=hand_r, loc=<-2.2074 0.3625 -1.3638>, rot=<-0.0284 0.0722 -0.0458 -0.9959>}
 ring_01_r {parent=ring_metacarpal_r, loc=<-7.0295 0.6197 0.1446>, rot=<-0.0489 -0.0252 0.1892 -0.9804>}
 ring_02_r {parent=ring_01_r, loc=<-3.6174 -0.4647 0.1027>, rot=<0 0 0.2806 -0.9598>}
 ring_03_r {parent=ring_02_r, loc=<-2.4152 -0.114 -0.0199>, rot=<0 0 0.1189 -0.9929>}
 pinky_metacarpal_r {parent=hand_r, loc=<-1.7616 0.9753 -3.3814>, rot=<-0.0397 0.0814 -0.069 -0.9935>}
 pinky_01_r {parent=pinky_metacarpal_r, loc=<-7.1709 0.4831 0.1094>, rot=<-0.0654 -0.0035 0.2242 -0.9723>}
 pinky_02_r {parent=pinky_01_r, loc=<-2.8274 -0.0893 0.1043>, rot=<0.0037 0.0011 0.2854 -0.9584>}
 pinky_03_r {parent=pinky_02_r, loc=<-2.2046 -0.0307 0.0078>, rot=<-0.0002 0.0098 0.1445 -0.9895>}
 deform_wrist_dn_base_r {parent=hand_r, loc=<-0.0001 1.5328 0>, rot=<0 0 0 -1>}
 deform_wrist_dn_r {parent=deform_wrist_dn_base_r, loc=<0 0 0>, rot=<0 0 0 1>}
 deform_wrist_up_base_r {parent=hand_r, loc=<-0.0002 -2.3007 -0.0001>, rot=<0 0 0 -1>}
 deform_wrist_up_r {parent=deform_wrist_up_base_r, loc=<0 0 0>, rot=<0 0 0 1>}
 lowerarm_twist_02_r {parent=lowerarm_r, loc=<-10.6707 0 0>, rot=<0 0 0 -1>}
 lowerarm_twist_01_r {parent=lowerarm_r, loc=<-18.316 0 0>, rot=<0 0 0 -1>}
 deform_elbow_in_base_r {parent=lowerarm_r, loc=<-0.1834 -3.6062 0.2572>, rot=<0 0 0 -1>}
 deform_elbow_in_r {parent=deform_elbow_in_base_r, loc=<0 0 0>, rot=<0 0 0 1>}
 deform_elbow_out_base_r {parent=lowerarm_r, loc=<0 4 -0.0001>, rot=<0 0 0 -1>}
 deform_elbow_out_r {parent=deform_elbow_out_base_r, loc=<0 0 0>, rot=<0 0 0 1>}
 upperarm_twist_02_r {parent=upperarm_r, loc=<-19.7998 0.1134 0>, rot=<0 0 0 -1>}
 base_sleevescloth_r {parent=upperarm_r, loc=<0.0002 0 -0.0003>, rot=<-0.1428 -0.0557 0.9878 0.028>}
 dyn_sleevescloth_r_02 {parent=base_sleevescloth_r, loc=<0 0 0>, rot=<0 0 0 1>}
 dyn_sleevescloth_r_03 {parent=dyn_sleevescloth_r_02, loc=<6.9162 0.1723 2.0465>, rot=<0 0 0 1>}
 dyn_sleevescloth_r_04 {parent=dyn_sleevescloth_r_03, loc=<7.5703 0 0>, rot=<0 0 0 1>}
 clavicle_r_null {parent=spine_05, type=Null, loc=<5.4831 -0.1053 3.2946>, rot=<0.7289 0.0465 -0.6829 0.009>}
 clavicle_l_null {parent=spine_05, type=Null, loc=<5.4831 -0.1053 -3.2946>, rot=<0.0465 -0.7289 -0.009 -0.6829>}
 neck_01_null {parent=spine_05, type=Null, loc=<11.3138 0.1529 0>, rot=<0 0 -0.2003 -0.9797>}
 dyn_zipper_01 {parent=spine_05, loc=<5.129 6.8209 0.0001>, rot=<-0.2073 -0.676 -0.676 -0.2073>}
 attach_backpack {parent=spine_05, loc=<-0.123 -13.473 0>, rot=<-0.0093 -0.707 0.0093 -0.707>}
 attach_cape {parent=spine_05, loc=<9.017 -11.714 0>, rot=<0.0616 -0.7044 -0.0616 -0.7044>}
 head_null_sp {parent=spine_05, type=Null, loc=<20.6573 3.5489 0>, rot=<0.4223 0.0152 -0.0326 -0.9057>}
 deform_pec_aim_base_l {parent=spine_04, loc=<3.2456 9.3309 -1.2237>, rot=<0.0603 -0.4033 0.3318 -0.8507>}
 deform_pec_aim_l {parent=deform_pec_aim_base_l, loc=<0 0 0>, rot=<0 0 0 1>}
 deform_pec_base_l {parent=deform_pec_aim_l, loc=<9.3665 0 0>, rot=<0 0 0 1>}
 deform_pec_l {parent=deform_pec_base_l, loc=<0 0 0>, rot=<0 0 0 1>}
 deform_pec_aim_base_r {parent=spine_04, loc=<3.2452 9.3309 1.2237>, rot=<-0.4033 -0.0603 0.8507 0.3318>}
 deform_pec_aim_r {parent=deform_pec_aim_base_r, loc=<0 0 0>, rot=<0 0 0 1>}
 deform_pec_base_r {parent=deform_pec_aim_r, loc=<-9.3667 -0.0003 -0.0004>, rot=<0 0 0 1>}
 deform_pec_r {parent=deform_pec_base_r, loc=<0 0 0>, rot=<0 0 0 1>}
 thigh_r {parent=pelvis, loc=<-1.0855 -0.2109 8.7053>, rot=<0.0355 -0.0009 0.9993 0.0155>}
 thigh_twist_01_r {parent=thigh_r, loc=<27.952 0 0>, rot=<0.0013 0 0 -1>}
 calf_r {parent=thigh_r, loc=<42.5284 0.0019 0.7297>, rot=<0.0006 -0.0161 -0.0391 -0.9991>}
 foot_r {parent=calf_r, loc=<46.7401 2.5014 0.592>, rot=<0.0427 -0.0238 0.0353 -0.9982>}
 ball_r {parent=foot_r, loc=<3.5658 -10.8066 0.1544>, rot=<0.0001 -0.0001 -0.7063 -0.7079>}
 calf_twist_02_r {parent=calf_r, loc=<19.1546 1.7943 0.2638>, rot=<0.0237 0.0095 -0.0035 -0.9997>}
 calf_twist_01_r {parent=calf_r, loc=<35.8483 1.9197 0.5794>, rot=<0.0237 0.0095 -0.0035 -0.9997>}
 deform_knee_in_base_r {parent=calf_r, loc=<-0.8031 5.0231 -0.2457>, rot=<0 0 0 -1>}
 deform_knee_in_r {parent=deform_knee_in_base_r, loc=<0 0 0>, rot=<0 0 0 1>}
 deform_knee_out_base_r {parent=calf_r, loc=<0 -4 0>, rot=<0 0 0 -1>}
 deform_knee_out_r {parent=deform_knee_out_base_r, loc=<0 0 0>, rot=<0 0 0 1>}
 deform_groin_base_r {parent=thigh_r, loc=<-0.4801 -5.6776 0.2486>, rot=<0 0 0 -1>}
 deform_groin_r {parent=deform_groin_base_r, loc=<0 0 0>, rot=<0 0 0 1>}
 thigh_l {parent=pelvis, loc=<-1.0855 -0.2109 -8.7053>, rot=<0.0009 0.0355 0.0155 -0.9993>}
 thigh_twist_01_l {parent=thigh_l, loc=<-27.952 0 0>, rot=<0.0013 0 0 -1>}
 calf_l {parent=thigh_l, loc=<-42.5284 -0.0019 -0.7297>, rot=<0.0006 -0.0161 -0.0391 -0.9991>}
 foot_l {parent=calf_l, loc=<-46.7401 -2.5014 -0.592>, rot=<0.0427 -0.0238 0.0353 -0.9982>}
 ball_l {parent=foot_l, loc=<-3.5658 10.8066 -0.1544>, rot=<0.0001 -0.0001 -0.7063 -0.7079>}
 calf_twist_02_l {parent=calf_l, loc=<-19.1546 -1.7943 -0.2638>, rot=<0.0237 0.0095 -0.0035 -0.9997>}
 calf_twist_01_l {parent=calf_l, loc=<-35.8483 -1.9197 -0.5794>, rot=<0.0237 0.0095 -0.0035 -0.9997>}
 deform_knee_in_base_l {parent=calf_l, loc=<0.8031 -5.0231 0.2456>, rot=<0 0 0 -1>}
 deform_knee_in_l {parent=deform_knee_in_base_l, loc=<0 0 0>, rot=<0 0 0 1>}
 deform_knee_out_base_l {parent=calf_l, loc=<0 4 0>, rot=<0 0 0 -1>}
 deform_knee_out_l {parent=deform_knee_out_base_l, loc=<0 0 0>, rot=<0 0 0 1>}
 deform_groin_base_l {parent=thigh_l, loc=<0.4801 5.6776 -0.2486>, rot=<0 0 0 -1>}
 deform_groin_l {parent=deform_groin_base_l, loc=<0 0 0>, rot=<0 0 0 1>}
 spine_02_null {parent=pelvis, type=Null, loc=<12.6908 -1.5909 0>, rot=<0 0 -0.0578 0.9983>}
 deform_glute_base_l {parent=pelvis, loc=<-7.3466 -9.0213 -9.1245>, rot=<0.0172 0.031 -0.4477 -0.8935>}
 deform_glute_l {parent=deform_glute_base_l, loc=<0 0 0>, rot=<0 0 0 1>}
 deform_glute_base_r {parent=pelvis, loc=<-7.3464 -9.0213 9.1245>, rot=<0.031 -0.0172 0.8935 -0.4477>}
 deform_glute_r {parent=deform_glute_base_r, loc=<0 0 0>, rot=<0 0 0 1>}
 base_skull_l {parent=pelvis, loc=<7.9531 5.3014 -12.9217>, rot=<-0.0055 -0.9687 -0.237 -0.0733>}
 dyn_skull_l {parent=base_skull_l, loc=<0 0 0>, rot=<0 0 0 1>}
 base_L_FrontSkirt {parent=pelvis, loc=<6.5729 5.2603 -8.9922>, rot=<-0.0662 -0.8148 -0.5698 -0.0839>}
 dyn_L_FrontSkirt_1 {parent=base_L_FrontSkirt, loc=<0 0 0>, rot=<0 0 0 1>}
 base_L_SideSkirt {parent=pelvis, loc=<7.0661 -2.4659 -11.9863>, rot=<-0.0175 -0.978 0.0037 -0.2077>}
 dyn_L_SideSkirt_1 {parent=base_L_SideSkirt, loc=<0 0 0>, rot=<0 0 0 1>}
 base_L_BackSkirt {parent=pelvis, loc=<6.0893 -7.7765 -8.3178>, rot=<0.0586 -0.8146 0.5706 -0.0864>}
 dyn_L_BackSkirt_1 {parent=base_L_BackSkirt, loc=<0 0 0>, rot=<0 0 0 1>}
 dyn_L_BackSkirt_2 {parent=dyn_L_BackSkirt_1, loc=<16.3581 0 0>, rot=<0 -0.1021 0 -0.9948>}
 dyn_L_BackSkirt_3 {parent=dyn_L_BackSkirt_2, loc=<16.5741 0 0>, rot=<0 0 0 1>}
 base_C_BackSkirt {parent=pelvis, loc=<6.0278 -9.2667 0.0001>, rot=<0.0598 -0.7046 0.7046 -0.0598>}
 dyn_C_BackSkirt_1 {parent=base_C_BackSkirt, loc=<0 0 0>, rot=<0 0 0 1>}
 dyn_C_BackSkirt_2 {parent=dyn_C_BackSkirt_1, loc=<16.3581 0 0>, rot=<0 -0.0917 0 -0.9958>}
 dyn_C_BackSkirt_3 {parent=dyn_C_BackSkirt_2, loc=<16.5916 0 0>, rot=<0 0 0 1>}
 dyn_C_BackSkirt_4 {parent=dyn_C_BackSkirt_3, loc=<18.6728 0 0>, rot=<0 0 0 1>}
 dyn_C_BackSkirt_5 {parent=dyn_C_BackSkirt_4, loc=<17.7199 0 0>, rot=<0 0 0 1>}
 base_R_BackSkirt {parent=pelvis, loc=<6.083 -7.4585 7.8062>, rot=<0.1033 -0.5725 0.8123 -0.0409>}
 dyn_R_BackSkirt_1 {parent=base_R_BackSkirt, loc=<0 0 0>, rot=<0 0 0 1>}
 dyn_R_BackSkirt_2 {parent=dyn_R_BackSkirt_1, loc=<16.3581 0 0>, rot=<0 -0.1039 0 -0.9946>}
 dyn_R_BackSkirt_3 {parent=dyn_R_BackSkirt_2, loc=<16.5916 0 0>, rot=<0 0 0 1>}
 dyn_R_BackSkirt_4 {parent=dyn_R_BackSkirt_3, loc=<18.6728 0 0>, rot=<0 0 0 1>}
 dyn_R_BackSkirt_5 {parent=dyn_R_BackSkirt_4, loc=<17.8444 0 0>, rot=<0 0 0 1>}
 base_R_BackSkirt_Charm {parent=dyn_R_BackSkirt_4, loc=<2.7205 -2.2327 0.3942>, rot=<-0.041 -0.0219 -0.019 -0.9987>}
 dyn_R_BackSkirt_Charm_1 {parent=base_R_BackSkirt_Charm, loc=<0 0 0>, rot=<0 0 0 1>}
 dyn_R_BackSkirt_Charm_2 {parent=dyn_R_BackSkirt_Charm_1, loc=<9.4295 0 0.8531>, rot=<0 0 0 1>}
 base_R_SideSkirt {parent=pelvis, loc=<6.0548 -2.4296 12.0019>, rot=<0.1395 -0.0025 0.9901 0.0177>}
 dyn_R_SideSkirt_1 {parent=base_R_SideSkirt, loc=<0 0 0>, rot=<0 0 0 1>}
 dyn_R_SideSkirt_2 {parent=dyn_R_SideSkirt_1, loc=<16.3581 0 0>, rot=<0 -0.1073 0 -0.9942>}
 dyn_R_SideSkirt_3 {parent=dyn_R_SideSkirt_2, loc=<16.5916 0 0>, rot=<0 0 0 1>}
 dyn_R_SideSkirt_4 {parent=dyn_R_SideSkirt_3, loc=<18.6727 0 0>, rot=<0 0 0 1>}
 dyn_R_SideSkirt_5 {parent=dyn_R_SideSkirt_4, loc=<18.8445 0 0>, rot=<0 0 0 1>}
 base_R_FrontSkirt {parent=pelvis, loc=<6.5814 3.6944 9.6019>, rot=<0.0765 0.5704 0.8155 0.061>}
 dyn_R_FrontSkirt_1 {parent=base_R_FrontSkirt, loc=<0 0 0>, rot=<0 0 0 1>}
 dyn_R_FrontSkirt_2 {parent=dyn_R_FrontSkirt_1, loc=<16.3581 0 0>, rot=<0 -0.0543 0 -0.9985>}
 dyn_R_FrontSkirt_3 {parent=dyn_R_FrontSkirt_2, loc=<15.9986 0 0>, rot=<0 -0.026 0 -0.9997>}
 base_C_FrontSkirt {parent=pelvis, loc=<6.739 8.0903 0.0001>, rot=<0.0127 0.707 0.707 0.0127>}
 dyn_C_FrontSkirt_1 {parent=base_C_FrontSkirt, loc=<0 0 0>, rot=<0 0 0 1>}
 dyn_C_FrontSkirt_2 {parent=dyn_C_FrontSkirt_1, loc=<16.3881 0 0>, rot=<0 0 0 1>}
 dyn_R_hair_7_ctrl_null {type=Null, loc=<-9.4953 -14.7292 96.1939>, rot=<0.5577 -0.594 -0.4227 -0.3968>}
 dyn_R_hair_7_ctrl {controlType=EULER_TRANSFORM, parent=dyn_R_hair_7_ctrl_null, type=Control, loc=<-10.701 0.3917 -1.0286>, rot=<0.0021 0.0478 0.0183 -0.9987>}
 dyn_R_hair_6_ctrl_null {type=Null, loc=<-9.2115 -12.1921 106.644>, rot=<0.5485 -0.5631 -0.4428 -0.4313>}
 dyn_R_hair_6_ctrl {controlType=EULER_TRANSFORM, parent=dyn_R_hair_6_ctrl_null, type=Control, loc=<-10.2908 -0.0318 -3.1536>, rot=<0.004 0.1481 -0.0009 -0.989>}
 dyn_R_hair_5_ctrl_null {type=Null, loc=<-8.9082 -12.8298 117.384>, rot=<0.4781 -0.4917 -0.5218 -0.5073>}
 dyn_R_hair_5_ctrl {controlType=EULER_TRANSFORM, parent=dyn_R_hair_5_ctrl_null, type=Control, loc=<-10.7688 0.2382 0.2483>, rot=<-0.0002 -0.0115 0.0111 -0.9999>}
 dyn_R_hair_4_ctrl_null {type=Null, loc=<-8.843 -13.2203 128.151>, rot=<0.4894 -0.4923 -0.5105 -0.5074>}
 dyn_R_hair_4_ctrl {
  controlType=EULER_TRANSFORM
  parent=dyn_R_hair_4_ctrl_null
  type=Control
  loc=<-10.547 -0.4151 1.5637>
  rot=<-0.0019 -0.0735 -0.0194 -0.9971>
 }
 dyn_R_hair_3_ctrl_null {type=Null, loc=<-8.3639 -12.0398 138.745>, rot=<-0.515 0.5387 0.482 0.4608>}
 dyn_R_hair_3_ctrl {controlType=EULER_TRANSFORM, parent=dyn_R_hair_3_ctrl_null, type=Control, loc=<-10.2483 0.0154 3.3103>, rot=<-0.0067 -0.1556 0.0018 -0.9878>}
 dyn_R_hair_2_ctrl_null {type=Null, loc=<-7.9192 -7.6159 148.554>, rot=<0.5816 -0.6061 -0.3915 -0.3757>}
 dyn_R_hair_2_ctrl {controlType=EULER_TRANSFORM, parent=dyn_R_hair_2_ctrl_null, type=Control, loc=<-10.6324 0.5008 -1.6641>, rot=<0.0014 0.0775 0.0234 -0.9967>}
 dyn_R_hair_1_ctrl_null {type=Null, loc=<-7.9806 -4.7569 158.941>, rot=<-0.564 0.5608 0.4273 0.4298>}
 dyn_R_hair_1_ctrl {controlType=EULER_TRANSFORM, parent=dyn_R_hair_1_ctrl_null, type=Control, loc=<152.0224 -7.1151 46.7668>, rot=<0.564 -0.5608 -0.4273 0.4298>}
 dyn_L_hair_7_ctrl_null {type=Null, loc=<9.4953 -14.7292 96.1939>, rot=<0.594 -0.5577 -0.3968 -0.4227>}
 dyn_L_hair_7_ctrl {
  controlType=EULER_TRANSFORM
  parent=dyn_L_hair_7_ctrl_null
  type=Control
  loc=<-10.7006 -0.3916 -1.0285>
  rot=<-0.0021 0.0478 -0.0183 -0.9987>
 }
 dyn_L_hair_6_ctrl_null {type=Null, loc=<9.2115 -12.1921 106.6437>, rot=<0.5631 -0.5485 -0.4313 -0.4428>}
 dyn_L_hair_6_ctrl {controlType=EULER_TRANSFORM, parent=dyn_L_hair_6_ctrl_null, type=Control, loc=<-10.2913 0.0318 -3.1537>, rot=<-0.004 0.1481 0.0009 -0.989>}
 dyn_L_hair_5_ctrl_null {type=Null, loc=<8.9082 -12.8298 117.3842>, rot=<0.4917 -0.4781 -0.5073 -0.5218>}
 dyn_L_hair_5_ctrl {
  controlType=EULER_TRANSFORM
  parent=dyn_L_hair_5_ctrl_null
  type=Control
  loc=<-10.7687 -0.2382 0.2483>
  rot=<0.0002 -0.0115 -0.0111 -0.9999>
 }
 dyn_L_hair_4_ctrl_null {type=Null, loc=<8.843 -13.2203 128.1512>, rot=<0.4923 -0.4894 -0.5074 -0.5105>}
 dyn_L_hair_4_ctrl {controlType=EULER_TRANSFORM, parent=dyn_L_hair_4_ctrl_null, type=Control, loc=<-10.5466 0.4151 1.5637>, rot=<0.0019 -0.0735 0.0194 -0.9971>}
 dyn_L_hair_3_ctrl_null {type=Null, loc=<8.364 -12.0398 138.7448>, rot=<-0.5387 0.515 0.4608 0.482>}
 dyn_L_hair_3_ctrl {
  controlType=EULER_TRANSFORM
  parent=dyn_L_hair_3_ctrl_null
  type=Control
  loc=<-10.2482 -0.0154 3.3103>
  rot=<0.0067 -0.1556 -0.0018 -0.9878>
 }
 dyn_L_hair_2_ctrl_null {type=Null, loc=<7.9192 -7.6159 148.5537>, rot=<0.6061 -0.5816 -0.3757 -0.3915>}
 dyn_L_hair_2_ctrl {
  controlType=EULER_TRANSFORM
  parent=dyn_L_hair_2_ctrl_null
  type=Control
  loc=<-10.6322 -0.5008 -1.664>
  rot=<-0.0014 0.0775 -0.0234 -0.9967>
 }
 dyn_L_hair_1_ctrl_null {type=Null, loc=<7.9806 -4.7569 158.9405>, rot=<0.5608 -0.564 -0.4298 -0.4273>}
 dyn_L_hair_1_ctrl {controlType=EULER_TRANSFORM, parent=dyn_L_hair_1_ctrl_null, type=Control, loc=<152.0219 7.1151 46.7667>, rot=<-0.5608 0.564 0.4298 -0.4273>}
 dyn_C_hair_7_ctrl_null {type=Null, loc=<0 -15.0214 96.4145>, rot=<0.5712 -0.5712 -0.4168 -0.4168>}
 dyn_C_hair_7_ctrl {controlType=EULER_TRANSFORM, parent=dyn_C_hair_7_ctrl_null, type=Control, loc=<-10.4268 0 -1.8123>, rot=<0 0.0859 0 -0.9963>}
 dyn_C_hair_6_ctrl_null {type=Null, loc=<0 -13.5663 106.8971>, rot=<0.5333 -0.5333 -0.4644 -0.4644>}
 dyn_C_hair_6_ctrl {controlType=EULER_TRANSFORM, parent=dyn_C_hair_6_ctrl_null, type=Control, loc=<-10.4243 0 -1.9059>, rot=<0 0.0903 0 -0.9959>}
 dyn_C_hair_5_ctrl_null {type=Null, loc=<0 -14.0207 117.4844>, rot=<0.4892 -0.4892 -0.5106 -0.5106>}
 dyn_C_hair_5_ctrl {controlType=EULER_TRANSFORM, parent=dyn_C_hair_5_ctrl_null, type=Control, loc=<-10.5971 0 0.0026>, rot=<0 -0.0001 0 -1>}
 dyn_C_hair_4_ctrl_null {type=Null, loc=<0 -14.4726 128.0719>, rot=<0.4892 -0.4892 -0.5105 -0.5105>}
 dyn_C_hair_4_ctrl {controlType=EULER_TRANSFORM, parent=dyn_C_hair_4_ctrl_null, type=Control, loc=<-10.4123 0 1.8926>, rot=<0 -0.0898 0 -0.996>}
 dyn_C_hair_3_ctrl_null {type=Null, loc=<0 -13.0257 138.5554>, rot=<-0.5331 0.5331 0.4646 0.4646>}
 dyn_C_hair_3_ctrl {controlType=EULER_TRANSFORM, parent=dyn_C_hair_3_ctrl_null, type=Control, loc=<-10.4687 0 1.6685>, rot=<0 -0.0789 0 -0.9969>}
 dyn_C_hair_2_ctrl_null {type=Null, loc=<0 -9.9416 148.6977>, rot=<0.5681 -0.5681 -0.421 -0.421>}
 dyn_C_hair_2_ctrl {controlType=EULER_TRANSFORM, parent=dyn_C_hair_2_ctrl_null, type=Control, loc=<-10.5953 0 -0.3639>, rot=<0 0.0172 0 -0.9999>}
 dyn_C_hair_1_ctrl_null {type=Null, loc=<0 -7.2072 158.9405>, rot=<-0.5608 0.5608 0.4307 0.4307>}
 dyn_C_hair_1_ctrl {controlType=EULER_TRANSFORM, parent=dyn_C_hair_1_ctrl_null, type=Control, loc=<151.7041 0 47.9572>, rot=<0.5608 -0.5608 -0.4307 0.4307>}
 spine_01_ctrl {controlType=EULER_TRANSFORM, type=Control, loc=<0 0 0>, rot=<0 0 0 1>}
 spine_02_ctrl {controlType=EULER_TRANSFORM, parent=spine_01_ctrl, type=Control, loc=<0 0 0>, rot=<0 0 0 1>}
 spine_03_ctrl {controlType=EULER_TRANSFORM, parent=spine_02_ctrl, type=Control, loc=<0 0 0>, rot=<0 0 0 1>}
 spine_04_ctrl {controlType=EULER_TRANSFORM, parent=spine_03_ctrl, type=Control, loc=<0 0 0>, rot=<0 0 0 1>}
 spine_05_ctrl {controlType=EULER_TRANSFORM, parent=spine_04_ctrl, type=Control, loc=<0 0 0>, rot=<0 0 0 1>}
 clavicle_r_ctrl {controlType=EULER_TRANSFORM, parent=spine_05_ctrl, type=Control, loc=<0 0 0>, rot=<0 0 0 1>}
 upperarm_r_ctrl {controlType=EULER_TRANSFORM, parent=clavicle_r_ctrl, type=Control, loc=<0 0 0>, rot=<0 0 0 1>}
 clavicle_l_ctrl {controlType=EULER_TRANSFORM, parent=spine_05_ctrl, type=Control, loc=<0 0 0>, rot=<0 0 0 1>}
 upperarm_l_ctrl {controlType=EULER_TRANSFORM, parent=clavicle_l_ctrl, type=Control, loc=<0 0 0>, rot=<0 0 0 1>}
 neck_01_ctrl {controlType=EULER_TRANSFORM, parent=spine_05_ctrl, type=Control, loc=<0 0 0>, rot=<0 0 0 1>}
 neck_02_ctrl {controlType=EULER_TRANSFORM, parent=neck_01_ctrl, type=Control, loc=<0 0 0>, rot=<0 0 0 1>}
 head_ctrl {controlType=EULER_TRANSFORM, parent=neck_02_ctrl, type=Control, loc=<0 0 0>, rot=<0 0 0 1>}
 pelvis_ctrl {controlType=EULER_TRANSFORM, type=Control, loc=<0 0 0>, rot=<0 0 0 1>}
 thigh_r_ctrl {controlType=EULER_TRANSFORM, parent=pelvis_ctrl, type=Control, loc=<0 0 0>, rot=<0 0 0 1>}
 thigh_l_ctrl {controlType=EULER_TRANSFORM, parent=pelvis_ctrl, type=Control, loc=<0 0 0>, rot=<0 0 0 1>}
]
variables [
 AimUp {type=double, default=0}
 Celebration Emote {type=bool, default=False}
 Crouch {type=bool, default=False}
 CrouchingTargeting {type=bool, default=False}
 GrounMovement {type=float, default=0}
 HeadUpCrouch {type=double, default=0}
 Parachuteopened {type=bool, default=False}
 Pickaxe {type=double, default=0}
 Skydive {type=bool, default=False}
 Sprinting {type=bool, default=False}
 c_dyn_hair_controls {type="TArray<FRigElementKey>"}
 c_dyn_hair_pose {type=FRigPose, default=(Elements=,HierarchyTopologyVersion=-1,PoseHash=-1)}
 l_dyn_hair_controls {type="TArray<FRigElementKey>"}
 l_dyn_hair_pose {type=FRigPose, default=(Elements=,HierarchyTopologyVersion=-1,PoseHash=-1)}
 pistolHead Down {type=double, default=0}
 pub aimPitch {type=double, default=0}
 pub disableUpperLimbsCorrection {type=bool, default=False}
 pub disablecontrolRig {type=bool, default=True}
 pub diveFloating {type=bool, default=False}
 pub isAssaultRifleEquipped {type=double, default=0}
 pub isTargeting {type=bool, default=False}
 pub isTargeting_lerp {type=double, default=0}
 r_dyn_hair_controls {type="TArray<FRigElementKey>"}
 r_dyn_hair_pose {type=FRigPose, default=(Elements=,HierarchyTopologyVersion=-1,PoseHash=-1)}
 wearingBackpack {type=bool, default=False}
]
