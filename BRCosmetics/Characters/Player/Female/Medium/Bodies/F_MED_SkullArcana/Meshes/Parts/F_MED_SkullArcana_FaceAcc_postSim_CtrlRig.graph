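# Post-simulation Control Rig for the SkullArcana female face accessory, correcting the three dynamic hair chains, an earring and HairALTSkin deform bones.
# A sequence layers constant and weighted bone offsets, mostly smoothed by AlphaInterp, from crouch, jump, zipline, backpack and front-end idle variables.
# Pelvis height shapes a crouch lift through a curve, and upper-arm spherical pose readers drive idle hair correctives.
# The sliding corrective is not wired into execution.
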ControlRigBlueprint /BRCosmetics/Characters/Player/Female/Medium/Bodies/F_MED_SkullArcana/Meshes/Parts/F_MED_SkullArcana_FaceAcc_postSim_CtrlRig
graphs [
 RigVMFunctionLibrary {links [], nodes []}
 RigVMFunction_Sequence_1_ContainedGraph {nodes [
   n1 = RigVMFunctionEntryNode()
   RigVMFunction_Sequence = &RigVMFunction_Sequence(ExecuteContext=@n1.ExecuteContext)
   n2 = &RigVMFunction_Sequence(ExecuteContext=@RigVMFunction_Sequence.B)
   n3 = &RigVMFunction_Sequence(ExecuteContext=@n2.B)
   n4 = &RigVMFunction_Sequence(ExecuteContext=@n3.B)
   n5 = &RigVMFunction_Sequence(ExecuteContext=@n4.B)
   RigVMFunctionReturnNode(A=@RigVMFunction_Sequence.A, B=@n2.A, C=@n3.A, D=@n4.A, E=@n5.A, F=@n5.B)
  ]}
 RigVMModel {nodes [
   RigUnit_BeginExecution = &RigUnit_BeginExecution()
   RigVMFunction_Sequence = RigVMAggregateNode(ExecuteContext=@RigUnit_BeginExecution.ExecutePin)
   AlphaInterp = &RigVMFunction_AlphaInterp(
    InterpSpeedDecreasing=15
    InterpSpeedIncreasing=15
    Value=1
    bClampResult=False
    Scale=&RigVMFunction_MathBoolToFloat(Value=RigVMVariableNode() {variable=Crouching}.Value).Result
   )
   &RigUnit_OffsetTransformForItem(
    Item=(Type=Bone,Name="dyn_L_hair_5")
    OffsetTransform=(Rotation=(X=0,Y=0.0436,Z=0,W=0.999),Translation=(X=0,Y=0,Z=0),Scale3D=(X=1,Y=1,Z=1))
    bPropagateToChildren=True
    ExecutePin=@RigVMFunction_Sequence.A
    Weight=@AlphaInterp.Result
   )
   > &RigUnit_OffsetTransformForItem(
    Item=(Type=Bone,Name="dyn_C_hair_5")
    OffsetTransform=(Rotation=(X=0,Y=0.0436,Z=0,W=0.999),Translation=(X=0,Y=0,Z=0),Scale3D=(X=1,Y=1,Z=1))
    bPropagateToChildren=True
    Weight=@AlphaInterp.Result
   )
   > &RigUnit_OffsetTransformForItem(
    Item=(Type=Bone,Name="dyn_R_hair_5")
    OffsetTransform=(Rotation=(X=0,Y=0.0436,Z=0,W=0.999),Translation=(X=0,Y=0,Z=0),Scale3D=(X=1,Y=1,Z=1))
    bPropagateToChildren=True
    Weight=@AlphaInterp.Result
   )
   n1 = &RigVMFunction_AnimEvalRichCurve(
    Curve="(EditorCurveData=(Keys=((InterpMode=RCIM_Cubic,TangentMode=RCTM_Break,TangentWeightMode=RCTWM_WeightedNone,Time=0,Value=0,ArriveTangent=0,ArriveTangentWeight=0,LeaveTangent=0,LeaveTangentWeight=0),(InterpMode=RCIM_Cubic,TangentMode=RCTM_Break,TangentWeight ...(truncated)"
    SourceMinimum=0.5
    TargetMinimum=0
    Value=&RigVMFunction_MathDoubleRemap(
     SourceMaximum=30
     SourceMinimum=60
     TargetMaximum=0
     TargetMinimum=1
     bClamp=False
     Value=&RigUnit_GetRelativeTransformForItem(Child=(Type=Bone,Name="pelvis"), Parent=(Type=Bone,Name="root"), bChildInitial=False, bParentInitial=False).RelativeTransform.Translation.Z
    ).Result
   )
   Multiply = &RigVMFunction_MathDoubleMul(A=@n1.Result, B=&RigVMFunction_MathBoolToFloat(Value=RigVMVariableNode() {variable=Crouching}.Value).Result)
   n2 = &RigVMFunction_AlphaInterp(InterpSpeedDecreasing=15, InterpSpeedIncreasing=15, Scale=1, bClampResult=False, Value=@Multiply.Result)
   > &RigUnit_OffsetTransformForItem(
    Item=(Type=Bone,Name="dyn_R_hair_3")
    OffsetTransform=(Rotation=(X=0,Y=0,Z=0,W=1),Translation=(X=0,Y=0,Z=2),Scale3D=(X=1,Y=1,Z=1))
    bPropagateToChildren=True
    Weight=@n2.Result
   )
   > &RigUnit_OffsetTransformForItem(
    Item=(Type=Bone,Name="dyn_C_hair_3")
    OffsetTransform=(Rotation=(X=0,Y=0,Z=0,W=1),Translation=(X=0,Y=0,Z=2),Scale3D=(X=1,Y=1,Z=1))
    bPropagateToChildren=True
    Weight=@n2.Result
   )
   > &RigUnit_OffsetTransformForItem(
    Item=(Type=Bone,Name="dyn_L_hair_3")
    OffsetTransform=(Rotation=(X=0,Y=0,Z=0,W=1),Translation=(X=0,Y=0,Z=2),Scale3D=(X=1,Y=1,Z=1))
    bPropagateToChildren=True
    Weight=@n2.Result
   )
   RigVMFunction_MathBoolToFloat = &RigVMFunction_MathBoolToFloat(Value=&RigVMFunction_MathBoolAnd(B=RigVMVariableNode() {variable=Jumping}.Value, A=RigVMVariableNode() {variable=wearing Backpack}.Value).Result)
   n3 = &RigVMFunction_AlphaInterp(InterpSpeedDecreasing=5, InterpSpeedIncreasing=2, Value=1, bClampResult=False, Scale=@RigVMFunction_MathBoolToFloat.Result)
   &RigUnit_OffsetTransformForItem(
    Item=(Type=Bone,Name="dyn_L_hair_6")
    OffsetTransform=(Rotation=(X=0,Y=0.1736,Z=0,W=0.9848),Translation=(X=0,Y=0,Z=0),Scale3D=(X=1,Y=1,Z=1))
    bPropagateToChildren=True
    ExecutePin=@RigVMFunction_Sequence.B
    Weight=@n3.Result
   )
   > &RigUnit_OffsetTransformForItem(
    Item=(Type=Bone,Name="dyn_C_hair_6")
    OffsetTransform=(Rotation=(X=0,Y=0.1736,Z=0,W=0.9848),Translation=(X=0,Y=0,Z=0),Scale3D=(X=1,Y=1,Z=1))
    bPropagateToChildren=True
    Weight=@n3.Result
   )
   > &RigUnit_OffsetTransformForItem(
    Item=(Type=Bone,Name="dyn_R_hair_6")
    OffsetTransform=(Rotation=(X=0,Y=0.1736,Z=0,W=0.9848),Translation=(X=0,Y=0,Z=0),Scale3D=(X=1,Y=1,Z=1))
    bPropagateToChildren=True
    Weight=@n3.Result
   )
   &RigUnit_OffsetTransformForItem(
    Item=(Type=Bone,Name="dyn_L_hair_4")
    OffsetTransform=(Rotation=(X=0,Y=-0.0436,Z=0,W=0.999),Translation=(X=0,Y=0,Z=0),Scale3D=(X=1,Y=1,Z=1))
    Weight=1
    bPropagateToChildren=True
    ExecutePin=@RigVMFunction_Sequence.C
   )
   > &RigUnit_OffsetTransformForItem(
    Item=(Type=Bone,Name="dyn_L_hair_6")
    OffsetTransform=(Rotation=(X=0,Y=0.0436,Z=0,W=0.999),Translation=(X=0,Y=0,Z=0),Scale3D=(X=1,Y=1,Z=1))
    Weight=1
    bPropagateToChildren=True
   )
   > &RigUnit_OffsetTransformForItem(
    Item=(Type=Bone,Name="dyn_R_hair_6")
    OffsetTransform=(Rotation=(X=0,Y=0.0436,Z=0,W=0.999),Translation=(X=0,Y=0,Z=0),Scale3D=(X=1,Y=1,Z=1))
    Weight=1
    bPropagateToChildren=True
   )
   > &RigUnit_OffsetTransformForItem(
    Item=(Type=Bone,Name="dyn_C_hair_6")
    OffsetTransform=(Rotation=(X=0,Y=0.0436,Z=0,W=0.999),Translation=(X=0,Y=0,Z=0),Scale3D=(X=1,Y=1,Z=1))
    Weight=1
    bPropagateToChildren=True
   )
   > &RigUnit_OffsetTransformForItem(
    Item=(Type=Bone,Name="dyn_R_hair_3")
    OffsetTransform=(Rotation=(X=0,Y=-0.0436,Z=0,W=0.999),Translation=(X=0,Y=0,Z=0),Scale3D=(X=1,Y=1,Z=1))
    Weight=1
    bPropagateToChildren=True
   )
   > &RigUnit_OffsetTransformForItem(
    Item=(Type=Bone,Name="dyn_L_hair_3")
    OffsetTransform=(Rotation=(X=0,Y=-0.0436,Z=0,W=0.999),Translation=(X=0,Y=0,Z=0),Scale3D=(X=1,Y=1,Z=1))
    Weight=1
    bPropagateToChildren=True
   )
   > &RigUnit_OffsetTransformForItem(
    Item=(Type=Bone,Name="dyn_C_hair_3")
    OffsetTransform=(Rotation=(X=0,Y=-0.0436,Z=0,W=0.999),Translation=(X=0,Y=0,Z=0),Scale3D=(X=1,Y=1,Z=1))
    Weight=1
    bPropagateToChildren=True
   )
   n4 = &RigVMFunction_AlphaInterp(InterpSpeedDecreasing=20, InterpSpeedIncreasing=20, Scale=1, bClampResult=False, Value=RigVMVariableNode() {variable=Crouch Walk}.Value)
   > &RigUnit_ModifyTransforms(
    ItemToModify="((Item=(Type=Bone,Name=\"dyn_R_hair_6\"),Transform=(Rotation=(X=0,Y=0.0436,Z=0,W=0.999),Translation=(X=0,Y=0,Z=0),Scale3D=(X=1,Y=1,Z=1))),(Item=(Type=Bone,Name=\"dyn_C_hair_6\"),Transform=(Rotation=(X=0,Y=0.0436,Z=0,W=0.999),Translation=(X=0,Y=0,Z=0),Scale3D=( ...(truncated)"
    WeightMinimum=0
    Weight=@n4.Result
   )
   > &RigUnit_ModifyTransforms(
    ItemToModify="((Item=(Type=Bone,Name=\"dyn_R_hair_7\"),Transform=(Rotation=(X=0,Y=0.0872,Z=0,W=0.9962),Translation=(X=0,Y=0,Z=0),Scale3D=(X=1,Y=1,Z=1))),(Item=(Type=Bone,Name=\"dyn_C_hair_7\"),Transform=(Rotation=(X=0,Y=0.0872,Z=0,W=0.9962),Translation=(X=0,Y=0,Z=0),Scale3D ...(truncated)"
    WeightMinimum=0
    Weight=@n4.Result
   )
   n5 = &RigVMFunction_AlphaInterp(
    InterpSpeedDecreasing=10
    InterpSpeedIncreasing=10
    Scale=1
    bClampResult=False
    Value=&RigVMFunction_MathBoolToFloat(Value=RigVMVariableNode() {variable=wearing Backpack}.Value).Result
   )
   &RigUnit_ModifyTransforms(
    ItemToModify="((Item=(Type=Bone,Name=\"deform_HairALTSkin_02\"),Transform=(Rotation=(X=0,Y=0,Z=0.4226,W=0.9063),Translation=(X=0,Y=0,Z=0),Scale3D=(X=1,Y=1,Z=1))),(Item=(Type=Bone,Name=\"deform_HairALTSkin_01\"),Transform=(Rotation=(X=0,Y=0,Z=0.4226,W=0.9063),Translation=(X= ...(truncated)"
    WeightMinimum=0
    Weight=@n5.Result
    ExecutePin=@RigVMFunction_Sequence.D
   )
   n6 = RigVMVariableNode() {variable=FrontEndIdle}
   n7 = &RigVMFunction_AlphaInterp(InterpSpeedDecreasing=10, InterpSpeedIncreasing=10, Scale=1, bClampResult=False, Value=@n6.Value)
   > &RigUnit_ModifyTransforms(
    ItemToModify="((Item=(Type=Bone,Name=\"dyn_L_hair_1\"),Transform=(Rotation=(X=-0.0019,Y=-0.0436,Z=-0.0436,W=0.9981),Translation=(X=0,Y=0,Z=0),Scale3D=(X=1,Y=1,Z=1))),(Item=(Type=Bone,Name=\"dyn_R_hair_3\"),Transform=(Rotation=(X=0,Y=0,Z=0.0436,W=0.999),Translation=(X=0,Y=0, ...(truncated)"
    WeightMinimum=0
    Weight=@n7.Result
   )
   > &RigUnit_ModifyTransforms(
    ItemToModify="((Item=(Type=Bone,Name=\"dyn_L_hair_4\"),Transform=(Rotation=(X=0,Y=0,Z=0.1305,W=0.9914),Translation=(X=0,Y=0,Z=0),Scale3D=(X=1,Y=1,Z=1))),(Item=(Type=Bone,Name=\"dyn_C_hair_4\"),Transform=(Rotation=(X=0,Y=0,Z=0,W=1),Translation=(X=0,Y=0,Z=0),Scale3D=(X=1,Y=1, ...(truncated)"
    WeightMinimum=0
    Weight=@n7.Result
   )
   > &RigUnit_ModifyTransforms(
    ItemToModify="((Item=(Type=Bone,Name=\"dyn_R_hair_3\"),Transform=(Rotation=(X=-0.0038,Y=0.0435,Z=0.0871,W=0.9952),Translation=(X=0,Y=0,Z=0),Scale3D=(X=1,Y=1,Z=1))),(Item=(Type=Bone,Name=\"dyn_L_hair_3\"),Transform=(Rotation=(X=0.0038,Y=0.0435,Z=-0.0871,W=0.9952),Translation ...(truncated)"
    WeightMinimum=0
    Weight=@n7.Result
   )
   > &RigUnit_OffsetTransformForItem(
    Item=(Type=Bone,Name="dyn_R_hair_2")
    OffsetTransform=(Rotation=(X=0.0038,Y=0.0871,Z=-0.0435,W=0.9952),Translation=(X=0,Y=1,Z=1.5),Scale3D=(X=1,Y=1,Z=1))
    bPropagateToChildren=True
    Weight=@n7.Result
   )
   > &RigUnit_OffsetTransformForItem(
    Item=(Type=Bone,Name="dyn_R_hair_1")
    OffsetTransform=(Rotation=(X=0,Y=-0.0872,Z=0,W=0.9962),Translation=(X=0,Y=0,Z=-2),Scale3D=(X=1,Y=1,Z=1))
    bPropagateToChildren=True
    Weight=@n7.Result
   )
   > &RigUnit_OffsetTransformForItem(
    Item=(Type=Bone,Name="dyn_R_hair_3")
    OffsetTransform=(Rotation=(X=0,Y=0,Z=0,W=1),Translation=(X=0,Y=-2,Z=0),Scale3D=(X=1,Y=1,Z=1))
    bPropagateToChildren=True
    Weight=@n7.Result
   )
   > &RigUnit_OffsetTransformForItem(
    Item=(Type=Bone,Name="dyn_L_hair_2")
    OffsetTransform=(Rotation=(X=0,Y=0,Z=0,W=1),Translation=(X=0,Y=1,Z=0),Scale3D=(X=1,Y=1,Z=1))
    bPropagateToChildren=True
    Weight=@n7.Result
   )
   > &RigUnit_OffsetTransformForItem(
    Item=(Type=Bone,Name="dyn_L_hair_5")
    OffsetTransform=(Rotation=(X=0,Y=0,Z=0,W=1),Translation=(X=0,Y=5,Z=0),Scale3D=(X=1,Y=1,Z=1))
    Weight=1
    bPropagateToChildren=True
   )
   > &RigUnit_OffsetTransformForItem(
    Item=(Type=Bone,Name="dyn_C_hair_3")
    OffsetTransform=(Rotation=(X=0,Y=0,Z=0.0872,W=0.9962),Translation=(X=0,Y=2,Z=0),Scale3D=(X=1,Y=1,Z=1))
    Weight=1
    bPropagateToChildren=True
   )
   > &RigUnit_OffsetTransformForItem(
    Item=(Type=Bone,Name="dyn_L_hair_3")
    OffsetTransform=(Rotation=(X=0,Y=0,Z=-0.0193,W=-0.9998),Translation=(X=-0.0583,Y=1.494,Z=-2.7437),Scale3D=(X=1,Y=1,Z=1))
    Weight=1
    bPropagateToChildren=True
   )
   > &RigUnit_OffsetTransformForItem(
    Item=(Type=Bone,Name="dyn_R_hair_5")
    OffsetTransform=(Rotation=(X=0,Y=0,Z=0,W=1),Translation=(X=0,Y=0,Z=0),Scale3D=(X=1,Y=1,Z=1))
    Weight=1
    bPropagateToChildren=True
   )
   > n8 = &RigUnit_SphericalPoseReader(
    ActiveRegionScaleFactors=(PositiveWidth=1,NegativeWidth=1,PositiveHeight=1,NegativeHeight=1)
    ActiveRegionSize=0.05
    Debug=(bDrawDebug=true,bDraw2D=False,bDrawLocalAxes=False,DebugScale=25,DebugSegments=20,DebugThickness=0.25)
    DriverAxis=(X=1,Y=0,Z=0)
    DriverItem=(Type=Bone,Name="upperarm_l")
    FalloffRegionScaleFactors=(PositiveWidth=1,NegativeWidth=1,PositiveHeight=1,NegativeHeight=1)
    FalloffSize=0.1
    FlipHeightScaling=False
    FlipWidthScaling=False
    OptionalParentItem=(Type=Bone,Name="None")
    RotationOffset=(X=-90,Y=50,Z=60)
   )
   > &RigUnit_ModifyTransforms(
    ItemToModify=((Item=(Type=Bone,Name="dyn_R_hair_2"),Transform=(Rotation=(X=0,Y=0,Z=-0.0436,W=0.999),Translation=(X=0,Y=0,Z=-3),Scale3D=(X=1,Y=1,Z=1))))
    WeightMinimum=0
    Weight=&RigVMFunction_MathDoubleMul(B=@n8.OutputParam, A=@n7.Result).Result
   )
   > n9 = &RigUnit_SphericalPoseReader(
    ActiveRegionScaleFactors=(PositiveWidth=1,NegativeWidth=1,PositiveHeight=1,NegativeHeight=1)
    ActiveRegionSize=0.05
    Debug=(bDrawDebug=true,DebugScale=25,DebugSegments=20,DebugThickness=0.25)
    DriverAxis=(X=-1,Y=0,Z=0)
    DriverItem=(Type=Bone,Name="upperarm_r")
    FalloffRegionScaleFactors=(PositiveWidth=1,NegativeWidth=1,PositiveHeight=1,NegativeHeight=1)
    FalloffSize=0.1
    FlipHeightScaling=false
    FlipWidthScaling=False
    OptionalParentItem=(Type=Bone,Name="None")
    RotationOffset=(X=80,Y=50,Z=60)
   )
   > &RigUnit_ModifyTransforms(
    ItemToModify=((Item=(Type=Bone,Name="dyn_L_hair_2"),Transform=(Rotation=(X=0,Y=0,Z=0,W=1),Translation=(X=0,Y=0,Z=-2),Scale3D=(X=1,Y=1,Z=1))))
    WeightMinimum=0
    Weight=&RigVMFunction_MathDoubleMul(A=@n9.OutputParam, B=@n7.Result).Result
   )
   n10 = RigVMVariableNode() {variable=wearing Backpack}
   n11 = &RigVMFunction_AlphaInterp(
    InterpSpeedDecreasing=10
    InterpSpeedIncreasing=10
    Scale=1
    bClampResult=False
    Value=&RigVMFunction_MathDoubleMul(
     A=RigVMDispatchNode(False=0, True=1, Condition=@n10.Value) {template="DISPATCH_RigVMDispatch_If(in Condition,in True,in False,out Result)"}.Result
     B=RigVMVariableNode() {variable=FrontEndIdle}.Value
    ).Result
   )
   > &RigUnit_ModifyTransforms(
    ItemToModify="((Item=(Type=Bone,Name=\"dyn_L_hair_5\"),Transform=(Rotation=(X=0,Y=0.0436,Z=0,W=0.999),Translation=(X=0,Y=4,Z=0),Scale3D=(X=1,Y=1,Z=1))),(Item=(Type=Bone,Name=\"dyn_R_hair_5\"),Transform=(Rotation=(X=0,Y=0.0436,Z=0,W=0.999),Translation=(X=0,Y=-4,Z=0),Scale3D= ...(truncated)"
    WeightMinimum=0
    Weight=@n11.Result
   )
   > &RigUnit_ModifyTransforms(
    ItemToModify="((Item=(Type=Bone,Name=\"dyn_L_hair_6\"),Transform=(Rotation=(X=0,Y=0.0872,Z=0,W=0.9962),Translation=(X=0,Y=0,Z=0),Scale3D=(X=1,Y=1,Z=1))),(Item=(Type=Bone,Name=\"dyn_R_hair_6\"),Transform=(Rotation=(X=0,Y=0.0872,Z=0,W=0.9962),Translation=(X=0,Y=0,Z=0),Scale3D ...(truncated)"
    WeightMinimum=0
    Weight=@n11.Result
   )
   > &RigUnit_OffsetTransformForItem(
    Item=(Type=Bone,Name="dyn_R_hair_1")
    OffsetTransform=(Rotation=(X=0,Y=0,Z=0,W=1),Translation=(X=0,Y=0,Z=3),Scale3D=(X=1,Y=1,Z=1))
    bPropagateToChildren=True
    Weight=@n11.Result
   )
   > &RigUnit_OffsetTransformForItem(
    Item=(Type=Bone,Name="dyn_R_hair_3")
    OffsetTransform=(Rotation=(X=0,Y=0,Z=0,W=1),Translation=(X=2,Y=0,Z=-2),Scale3D=(X=1,Y=1,Z=1))
    bPropagateToChildren=True
    Weight=@n11.Result
   )
   n12 = RigVMVariableNode() {variable=wearing Backpack}
   n13 = RigVMVariableNode() {variable=Crouching}
   n14 = &RigVMFunction_AlphaInterp(
    InterpSpeedDecreasing=10
    InterpSpeedIncreasing=10
    Scale=1
    bClampResult=False
    Value=&RigVMFunction_MathDoubleMul(
     A=&RigVMFunction_MathBoolToFloat(Value=@n12.Value).Result
     B=&RigVMFunction_MathBoolToFloat(Value=@n13.Value).Result
    ).Result
   )
   > &RigUnit_OffsetTransformForItem(
    Item=(Type=Bone,Name="dyn_R_hair_6")
    OffsetTransform=(Rotation=(X=0,Y=-0.0872,Z=0,W=0.9962),Translation=(X=0,Y=0,Z=0),Scale3D=(X=1,Y=1,Z=1))
    bPropagateToChildren=True
    Weight=@n14.Result
   )
   > &RigUnit_OffsetTransformForItem(
    Item=(Type=Bone,Name="dyn_C_hair_6")
    OffsetTransform=(Rotation=(X=0,Y=-0.0872,Z=0,W=0.9962),Translation=(X=0,Y=0,Z=0),Scale3D=(X=1,Y=1,Z=1))
    bPropagateToChildren=True
    Weight=@n14.Result
   )
   > &RigUnit_OffsetTransformForItem(
    Item=(Type=Bone,Name="dyn_L_hair_6")
    OffsetTransform=(Rotation=(X=0,Y=-0.0872,Z=0,W=0.9962),Translation=(X=0,Y=0,Z=0),Scale3D=(X=1,Y=1,Z=1))
    bPropagateToChildren=True
    Weight=@n14.Result
   )
   > &RigUnit_ModifyTransforms(
    ItemToModify=((Item=(Type=Bone,Name="dyn_earring_l"),Transform=(Rotation=(X=0.2586,Y=0.0113,Z=0.0421,W=0.965),Translation=(X=0,Y=0,Z=0),Scale3D=(X=1,Y=1,Z=1))))
    WeightMinimum=0
    Weight=RigVMVariableNode() {variable=FrontEndIdle}.Value
   )
   n15 = RigVMVariableNode() {variable=Zipline}
   n16 = RigVMVariableNode() {variable=wearing Backpack}
   n17 = &RigVMFunction_AlphaInterp(
    InterpSpeedDecreasing=2
    InterpSpeedIncreasing=10
    Scale=1
    bClampResult=False
    Value=&RigVMFunction_MathBoolToFloat(Value=&RigVMFunction_MathBoolAnd(B=@n15.Value, A=@n16.Value).Result).Result
   )
   &RigUnit_ModifyTransforms(
    ItemToModify="((Item=(Type=Bone,Name=\"dyn_C_hair_5\"),Transform=(Rotation=(X=0,Y=0.1736,Z=0,W=0.9848),Translation=(X=0,Y=0,Z=0),Scale3D=(X=1,Y=1,Z=1))),(Item=(Type=Bone,Name=\"dyn_L_hair_5\"),Transform=(Rotation=(X=0,Y=0.1736,Z=0,W=0.9848),Translation=(X=0,Y=0,Z=0),Scale3D ...(truncated)"
    WeightMinimum=0
    Weight=@n17.Result
    ExecutePin=@RigVMFunction_Sequence.F
   )
   n18 = &RigVMFunction_AlphaInterp(
    InterpSpeedDecreasing=10
    InterpSpeedIncreasing=10
    Scale=1
    bClampResult=False
    Value=&RigVMFunction_MathBoolToFloat(Value=RigVMVariableNode() {variable=Zipline}.Value).Result
   )
   > &RigUnit_ModifyTransforms(
    ItemToModify="((Item=(Type=Bone,Name=\"dyn_C_hair_3\"),Transform=(Rotation=(X=0,Y=0.1305,Z=0,W=0.9914),Translation=(X=0,Y=0,Z=0),Scale3D=(X=1,Y=1,Z=1))),(Item=(Type=Bone,Name=\"dyn_L_hair_3\"),Transform=(Rotation=(X=0,Y=0.1305,Z=0,W=0.9914),Translation=(X=0,Y=0,Z=0),Scale3D ...(truncated)"
    WeightMinimum=0
    Weight=@n18.Result
   )
   Add = &RigVMFunction_MathDoubleAdd(A=RigVMVariableNode() {variable=is knee sliding}.Value, B=RigVMVariableNode() {variable=is baseball sliding}.Value)
   &RigUnit_ModifyTransforms(
    ExecutePin=()
    ItemToModify="((Item=(Type=Bone,Name=\"dyn_R_hair_6\"),Transform=(Rotation=(X=0,Y=-0.0436,Z=0,W=0.999),Translation=(X=0,Y=0,Z=0),Scale3D=(X=1,Y=1,Z=1))),(Item=(Type=Bone,Name=\"dyn_C_hair_6\"),Transform=(Rotation=(X=0,Y=-0.0436,Z=0,W=0.999),Translation=(X=0,Y=0,Z=0),Scale3D ...(truncated)"
    WeightMinimum=0
    Weight=@Add.Result
   )
   &RigVMFunction_MathDoubleMul(
    A=RigVMDispatchNode(False=1, True=0, Condition=RigVMVariableNode() {variable=wearing Backpack}.Value) {template="DISPATCH_RigVMDispatch_If(in Condition,in True,in False,out Result)"}.Result
    B=@n6.Value
   )
   &RigUnit_SphericalPoseReader(
    ActiveRegionScaleFactors=(PositiveWidth=1,NegativeWidth=1,PositiveHeight=1,NegativeHeight=1)
    ActiveRegionSize=0.1
    Debug=(bDrawDebug=false,bDraw2D=False,bDrawLocalAxes=False,DebugScale=25,DebugSegments=20,DebugThickness=0.25)
    DriverAxis=(X=-1,Y=0,Z=0)
    DriverItem=(Type=Bone,Name="clavicle_r")
    ExecutePin=()
    FalloffRegionScaleFactors=(PositiveWidth=0.5,NegativeWidth=0.5,PositiveHeight=0.5,NegativeHeight=0.5)
    FalloffSize=0.3
    FlipHeightScaling=False
    FlipWidthScaling=False
    OptionalParentItem=(Type=Bone,Name="None")
    RotationOffset=(X=45,Y=174,Z=30)
   )
  ]}
]
hierarchy [
 root {loc=<0 0 0>, rot=<0 0 0 1>}
 pelvis {parent=root, loc=<0 -0.5301 98.0571>, rot=<0.0127 0.707 -0.0127 -0.707>}
 spine_01 {parent=pelvis, loc=<6.1208 -0.9328 0>, rot=<0 0 0.034 -0.9994>}
 spine_02 {parent=spine_01, loc=<6.5997 -0.2095 0>, rot=<0 0 0.0238 -0.9997>}
 spine_03 {parent=spine_02, loc=<6.4914 -0.1237 0>, rot=<0 0 0.0149 -0.9999>}
 spine_04 {parent=spine_03, loc=<6.9598 0.0794 0>, rot=<0 0 -0.0132 -0.9999>}
 spine_05 {parent=spine_04, loc=<6.9987 0 0>, rot=<0 0 -0.0015 -1>}
 neck_01 {parent=spine_05, loc=<11.3138 0.1529 0>, rot=<0 0 -0.2003 -0.9797>}
 neck_02 {parent=neck_01, loc=<4.5627 0 0>, rot=<0 0 0.0504 -0.9987>}
 head {parent=neck_02, loc=<5.3914 0 0>, rot=<0 0 0.115 -0.9934>}
 base_hat {parent=head, loc=<7.6428 0.3962 0.3568>, rot=<-0.0029 0.7071 0.7071 -0.0029>}
 dyn_hatCharms_d_01 {parent=base_hat, loc=<-1.1512 -14.906 -17.1704>, rot=<-0.8492 0 0 -0.528>}
 dyn_hatCharms_e_01 {parent=base_hat, loc=<1.6597 -8.3236 -20.8559>, rot=<0.9998 0 0 0.0186>}
 dyn_hatCharms_e_02 {parent=dyn_hatCharms_e_01, loc=<0.9396 -0.0207 -0.0387>, rot=<0.6938 0 0 -0.7201>}
 dyn_hatCharms_c_01 {parent=base_hat, loc=<-10.7912 -19.794 6.5926>, rot=<-0.708 0 0 -0.7062>}
 dyn_hatCharms_b_01 {parent=base_hat, loc=<-4.355 -10.1476 20.8766>, rot=<-0.3544 0 0 -0.9351>}
 dyn_hatCharms_b_02 {parent=dyn_hatCharms_b_01, loc=<1.8698 -0.0663 0.074>, rot=<0.0023 0 0 -1>}
 dyn_hatCharms_a_01 {parent=base_hat, loc=<7.0208 14.5282 17.2638>, rot=<-0.3724 0 0 -0.9281>}
 dyn_hatCharms_a_02 {parent=dyn_hatCharms_a_01, loc=<0.8609 -0.0437 0.007>, rot=<0.3724 0 0 -0.9281>}
 base_hat_side_r {parent=base_hat, loc=<1.8478 7.1684 -6.5786>, rot=<0.6533 0.3262 -0.5797 -0.3615>}
 dyn_hat_side_r {parent=base_hat_side_r, loc=<0 0 0>, rot=<0 0 0 1>}
 base_hat_SideFront_r {parent=base_hat, loc=<2.3114 8.6852 3.4102>, rot=<0.4906 0.4282 -0.4324 -0.6238>}
 dyn_hat_SideFront_r_1 {parent=base_hat_SideFront_r, loc=<0 0 0>, rot=<0 0 0 1>}
 base_hat_Side_l {parent=base_hat, loc=<-2.7631 -8.5483 -3.4995>, rot=<-0.571 0.5349 0.5873 -0.2072>}
 dyn_hat_Side_l_1 {parent=base_hat_Side_l, loc=<0 0 0>, rot=<0 0 0 1>}
 base_hat_front {parent=base_hat, loc=<0.8942 1.6565 10.2977>, rot=<0.2194 0.596 0.0607 -0.77>}
 dyn_hat_front_1 {parent=base_hat_front, loc=<0 0 0>, rot=<0 0 0 1>}
 base_hat_sidefront_l {parent=base_hat, loc=<-2.4806 -6.8502 6.7978>, rot=<-0.1234 0.6844 0.5071 -0.5091>}
 dyn_hat_sidefront_l_1 {parent=base_hat_sidefront_l, loc=<0 0 0>, rot=<0 0 0 1>}
 base_hat_rear1 {parent=base_hat, loc=<0.2496 -0.0922 -10.409>, rot=<0.8026 -0.0563 -0.5889 -0.0767>}
 dyn_hat_rear_1 {parent=base_hat_rear1, loc=<0 0 0>, rot=<0 0 0 1>}
 base_hatBowknot_01 {parent=base_hat, loc=<-1.6868 0.6759 -7.5295>, rot=<-0.2974 0.3685 0.7045 -0.5286>}
 dyn_hatBowknot_01 {parent=base_hatBowknot_01, loc=<0 0 0>, rot=<0 0 0 1>}
 base_hatBowknot_02 {parent=base_hat, loc=<-0.2072 1.3388 -8.4969>, rot=<-0.5462 -0.2899 0.55 -0.5613>}
 dyn_hatBowknot_02 {parent=base_hatBowknot_02, loc=<0 0 0>, rot=<0 0 0 1>}
 base_hatBowknot_03 {parent=base_hat, loc=<-1.3768 2.658 -7.3287>, rot=<0.734 0.4681 0.3899 0.3003>}
 dyn_hatBowknot_03 {parent=base_hatBowknot_03, loc=<0 0 0>, rot=<0 0 0 1>}
 dyn_earring_l {parent=head, loc=<0.4505 0.0271 -6.2969>, rot=<0.0034 -0.8155 -0.5787 0.0024>}
 base_L_hair_1 {parent=head, loc=<6.9314 -3.0124 -7.9806>, rot=<0.0956 -0.7006 -0.7001 0.0996>}
 dyn_L_hair_1 {parent=base_L_hair_1, loc=<0 0 0>, rot=<0 0 0 1>}
 dyn_L_hair_2 {parent=dyn_L_hair_1, loc=<10.7733 0 0>, rot=<0.0014 -0.0775 0.0234 -0.9967>}
 dyn_L_hair_3 {parent=dyn_L_hair_2, loc=<10.7695 0 0>, rot=<-0.0067 0.1556 0.0018 -0.9878>}
 dyn_L_hair_4 {parent=dyn_L_hair_3, loc=<10.67 0 0>, rot=<-0.0019 0.0735 -0.0194 -0.9971>}
 dyn_L_hair_5 {parent=dyn_L_hair_4, loc=<10.7742 0 0>, rot=<-0.0002 0.0115 0.0111 -0.9999>}
 dyn_L_hair_6 {parent=dyn_L_hair_5, loc=<10.7637 0 0>, rot=<0.004 -0.1481 -0.0009 -0.989>}
 dyn_L_hair_7 {parent=dyn_L_hair_6, loc=<10.7571 0 0>, rot=<0.0021 -0.0478 0.0183 -0.9987>}
 dyn_L_hair_7_end {parent=dyn_L_hair_7, loc=<10.7622 0 0>, rot=<0 -0.0983 0 -0.9952>}
 base_C_hair_1 {parent=head, loc=<6.9517 -5.4626 0>, rot=<-0.0949 0.7007 0.7007 -0.0949>}
 dyn_C_hair_1 {parent=base_C_hair_1, loc=<0 0 0>, rot=<0 0 0 1>}
 dyn_C_hair_2 {parent=dyn_C_hair_1, loc=<10.6015 0 0>, rot=<0 -0.0172 0 -0.9999>}
 dyn_C_hair_3 {parent=dyn_C_hair_2, loc=<10.6008 0 0>, rot=<0 0.0789 0 -0.9969>}
 dyn_C_hair_4 {parent=dyn_C_hair_3, loc=<10.5829 0 0>, rot=<0 0.0898 0 -0.996>}
 dyn_C_hair_5 {parent=dyn_C_hair_4, loc=<10.5971 0 0>, rot=<0 0.0001 0 -1>}
 dyn_C_hair_6 {parent=dyn_C_hair_5, loc=<10.5971 0 0>, rot=<0 -0.0903 0 -0.9959>}
 dyn_C_hair_7 {parent=dyn_C_hair_6, loc=<10.5831 0 0>, rot=<0 -0.0859 0 -0.9963>}
 dyn_C_hair_7_end {parent=dyn_C_hair_7, loc=<10.5979 0 0>, rot=<0 -0.0983 0 -0.9952>}
 base_R_hair_1 {parent=head, loc=<6.9319 -3.0124 7.9806>, rot=<-0.0996 0.7001 0.7006 -0.0956>}
 dyn_R_hair_1 {parent=base_R_hair_1, loc=<0 0 0>, rot=<0 0 0 1>}
 dyn_R_hair_2 {parent=dyn_R_hair_1, loc=<10.7735 0 0.0001>, rot=<-0.0014 -0.0775 -0.0234 -0.9967>}
 dyn_R_hair_3 {parent=dyn_R_hair_2, loc=<10.7696 0 0>, rot=<0.0067 0.1556 -0.0018 -0.9878>}
 dyn_R_hair_4 {parent=dyn_R_hair_3, loc=<10.6703 0 0.0001>, rot=<0.0019 0.0735 0.0194 -0.9971>}
 dyn_R_hair_5 {parent=dyn_R_hair_4, loc=<10.7743 0 0>, rot=<0.0002 0.0115 -0.0111 -0.9999>}
 dyn_R_hair_6 {parent=dyn_R_hair_5, loc=<10.7632 0 0>, rot=<-0.004 -0.1481 0.0009 -0.989>}
 dyn_R_hair_7 {parent=dyn_R_hair_6, loc=<10.7574 0 0>, rot=<-0.0021 -0.0478 -0.0183 -0.9987>}
 dyn_R_hair_7_end {parent=dyn_R_hair_7, loc=<10.7622 0 0.0001>, rot=<0 -0.0983 0 -0.9952>}
 clavicle_l {parent=spine_05, loc=<5.4831 -0.1053 -3.2946>, rot=<0.0465 -0.7289 -0.009 -0.6829>}
 upperarm_l {parent=clavicle_l, loc=<9.8658 -1.2136 -0.4105>, rot=<0.0164 -0.3245 -0.0269 -0.9454>}
 lowerarm_l {parent=upperarm_l, loc=<26.465 -0.8345 0.259>, rot=<0.0036 0.0161 -0.2298 -0.9731>}
 hand_l {parent=lowerarm_l, loc=<22.7355 -0.3203 -0.6105>, rot=<-0.6784 -0.0223 0.0735 -0.7307>}
 clavicle_r {parent=spine_05, loc=<5.4831 -0.1053 3.2946>, rot=<0.7289 0.0465 -0.6829 0.009>}
 upperarm_r {parent=clavicle_r, loc=<-9.8658 1.2136 0.4105>, rot=<0.0164 -0.3245 -0.0269 -0.9454>}
 lowerarm_r {parent=upperarm_r, loc=<-26.465 0.8345 -0.259>, rot=<0.0036 0.0161 -0.2298 -0.9731>}
 hand_r {parent=lowerarm_r, loc=<-22.7355 0.3203 0.6106>, rot=<-0.6784 -0.0223 0.0735 -0.7307>}
 attach_backpack {parent=spine_05, loc=<-0.123 -13.473 0>, rot=<-0.0093 -0.707 0.0093 -0.707>}
 attach_cape {parent=spine_05, loc=<9.017 -11.714 0>, rot=<0.0616 -0.7044 -0.0616 -0.7044>}
 deform_HairALTSkin_01 {parent=spine_05, loc=<2.7458 -13.8019 0>, rot=<0.0872 -0.9962 0 0>}
 deform_HairALTSkin_02 {parent=spine_05, loc=<-10.9853 -15.9066 0>, rot=<0.0561 -0.9984 0 0>}
 deform_HairALTSkin_03 {parent=spine_05, loc=<-22.1441 -17.0721 0>, rot=<0.0561 -0.9984 0 0>}
 thigh_l {parent=pelvis, loc=<-1.0855 -0.2109 -8.7053>, rot=<0.0009 0.0355 0.0155 -0.9993>}
 calf_l {parent=thigh_l, loc=<-42.5284 -0.0019 -0.7297>, rot=<0.0006 -0.0161 -0.0391 -0.9991>}
 thigh_r {parent=pelvis, loc=<-1.0855 -0.2109 8.7053>, rot=<0.0355 -0.0009 0.9993 0.0155>}
 calf_r {parent=thigh_r, loc=<42.5284 0.0019 0.7297>, rot=<0.0006 -0.0161 -0.0391 -0.9991>}
]
variables [
 AimUp {type=double, default=0}
 Crouch Walk {type=double, default=0}
 Crouching {type=bool, default=False}
 FrontEndIdle {type=double, default=0}
 Jumping {type=bool, default=False}
 Zipline {type=bool, default=False}
 is baseball sliding {type=double, default=0}
 is knee sliding {type=double, default=0}
 wearing Backpack {type=bool, default=False}
]
